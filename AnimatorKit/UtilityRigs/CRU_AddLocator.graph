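# Utility Control Rig that adds a configurable number of locator controls under a ComponentSpace null.
# The construction event imports bones from the skeleton, then loops over NumOfLocators to spawn transform controls named Locator.
# Exposed variables set the count, shape color and preferred rotation order.
# The forward solve event is empty, so the rig only builds hierarchy and drives no bones.
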ControlRigBlueprint /AnimatorKit/UtilityRigs/CRU_AddLocator
graphs [
 RigVMFunctionLibrary {links [], nodes []}
 RigVMModel {nodes [
   n1 = &RigUnit_PrepareForExecution()
   &RigUnit_HierarchyImportFromSkeleton(NameSpace=None, bIncludeCurves=False, bIncludeMeshSockets=False, bIncludeVirtualBones=True, ExecutePin=@n1.ExecutePin)
   > RigVMFunction_ForLoopCount = /Script/RigVM.RigVMFunction_ForLoopCount(Count=RigVMVariableNode() {variable=NumOfLocators}.Value) {is=struct RigVMUnitNode}
   > SpawnControl = /Script/ControlRig.RigUnit_HierarchyAddControlTransform(
    InitialValue=(Rotation=(X=0,Y=0,Z=0,W=1),Translation=(X=0,Y=0,Z=0),Scale3D=(X=1,Y=1,Z=1))
    Name=Locator
    OffsetSpace=LocalSpace
    OffsetTransform=(Rotation=(X=0,Y=0,Z=0,W=1),Translation=(X=0,Y=0,Z=0),Scale3D=(X=1,Y=1,Z=1))
    Parent=(Type=Null,Name="ComponentSpace")
    Settings="(InitialSpace=LocalSpace,bUsePreferredRotationOrder=true,PreferredRotationOrder=YZX,Limits=(LimitTranslationX=(bMinimum=False,bMaximum=False),LimitTranslationY=(bMinimum=False,bMaximum=False),LimitTranslationZ=(bMinimum=False,bMaximum=False),LimitPitch=(bM ...(truncated)"
    Settings.PreferredRotationOrder=RigVMVariableNode() {variable=PreferredRotationOrder}.Value
    Settings.Shape.Color=RigVMVariableNode() {variable=Color}.Value
   ) {
    is=struct RigVMUnitNode
    template="SpawnControl::Execute(in InitialValue,in Settings,in OffsetTransform,in OffsetSpace,in Parent,in Name,out Item)"
   }
   RigUnit_BeginExecution = &RigUnit_BeginExecution()
  ]}
]
hierarchy [ComponentSpace {type=Null, loc=<0 0 0>, rot=<0 0 0 1>}]
variables [
 Color {type=FLinearColor, default=(R=0,G=1,B=0,A=1)}
 NumOfLocators {type=int32, default=1}
 PreferredRotationOrder {type=EEulerRotationOrder, default=YZX}
]
