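# Editor utility widget providing Control Rig animation tools for Sequencer.
# Buttons drive locator creation, space-key moves, cycle extrapolation and embedded Python scripts, such as extra-control cleanup with Euler filtering.
# Functions also bake bindings to MR_Biped_v02, create constraints, cache selected controls' world transforms and insert offset actors between constraint parents and children.
# Locators come from AnimatorKit's CRU_AddLocator rig, recreated with overridden count and color settings.
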EditorUtilityWidgetBlueprint /AnimInEdTools/EpicAnimTools/EpicAnimTools
graphs [
 AddOffsetNode {nodes [
   n1 = Get Current Local Time!(execute, TimeUnit:EMovieSceneTimeUnit Enum=DisplayRate) -> (
    then
    ReturnValue_Frame_FrameNumber:Frame Number Structure
    ReturnValue_Frame_SubFrame:float=~
    ReturnValue_Time:float=~
    ReturnValue_MarkedFrame:String
    ReturnValue_Timecode:Timecode Structure
    ReturnValue_PositionType:EMovieScenePositionType Enum
    ReturnValue_UpdateMethod:EUpdatePositionMethod Enum
    ReturnValue_bHasJumped:bool=~
   ) {is=title K2Node_CallFunction}
   > n2 = EUW_MakeLocator!(execute, self, Locator Color:String=(R=0.1878,G=0.3419,B=0.0999,A=1)) -> (
    then
    Out_Loc_CR:Control Rig Object Reference
    Out_Loc_CtrlName:Name=None
    In_CR:Control Rig Object Reference
    In_Control:Name=None
    NewLocators:Array of STR Control Rig Proxy Object Name Tuple Structures
    ControlsRunOn:Array of STR Control Rig Proxy Object Name Tuple Structures
   )
   > n3 = GetFocusedLevelSequence!(execute) -> (then, ReturnValue:Level Sequence Object Reference)
   n4 = AddOffsetNode!() -> (then, InOutConstraint:Tickable Transform Constraint Object Reference) {is=title K2Node_FunctionEntry}
   > n5 = GetParentInfo!(execute, self, InConstraint:Tickable Transform Constraint Object Reference=@n4) -> (then, OutParentHandle:Transformable Handle Object Reference, OutWorldTrransform:Transform)
   n6 = SetControlRigWorldTransform!(
    execute=@n3.then
    LevelSequence:Level Sequence Object Reference=@n3
    ControlRig:Control Rig Object Reference=@n2
    ControlName:Name=@n2.Out_Loc_CtrlName
    Frame:Frame Number Structure=@n1
    WorldTransform:"Transform (by ref)"=@n5.OutWorldTrransform
    TimeUnit:EMovieSceneTimeUnit Enum=DisplayRate
    bSetKey:bool=true
   ) -> (then)
   n7 = GetEditorWorld(execute=[@n12.then, @n6.then], self=Get UnrealEditorSubsystem())
   n8 = SpawnActor Static Mesh Actor!(
    execute:Exec=@n5.then
    WorldContextObject:Object Reference
    Class:Actor Class Reference=&StaticMeshActor
    SpawnTransform:Transform=@n5.OutWorldTrransform
    CollisionHandlingOverride:ESpawnActorCollisionHandlingMethod Enum=Undefined
    TransformScaleMethod:ESpawnActorScaleMethod Enum=MultiplyWithRoot
    Owner:Actor Object Reference
    Instigator:Pawn Object Reference
   ) -> (then:Exec, ReturnValue:Static Mesh Actor Object Reference) {is=title K2Node_SpawnActorFromClass}
   n9 = CreateTransformableHandle!(
    execute=@n7.then
    InWorld:World Object Reference=@n7
    InObject:Object Reference=@n8
    InAttachmentName:"Name (by ref)"=Get DummySocket() -> (DummySocket:Name=None)
   ) -> (then, ReturnValue:Transformable Handle Object Reference)
   > Set ParentTRSHandle!(execute, ParentTRSHandle:Transformable Handle Object Reference=@n9, self:Tickable Transform Constraint Object Reference=@n4) -> (then, Output_Get:Transformable Handle Object Reference)
   > n10 = CreateTransformableHandle(InWorld=@n7, InObject=@n8, InAttachmentName=Get DummySocket() -> (DummySocket=None))
   > AddConstraint!(
    execute
    World:World Object Reference=@n7
    InType:ETransformConstraintType Enum=Parent
    InChild:Transformable Handle Object Reference=@n10
    InParent:Transformable Handle Object Reference=@n5
    bMaintainOffset:bool=true
   ) -> (then, ReturnValue:Tickable Constraint Object Reference)
   Reroute Node(InputPin:Control Rig Object Reference=@n2) -> (OutputPin:Control Rig Object Reference)
   Reroute Node(InputPin:Name=@n2.Out_Loc_CtrlName) -> (OutputPin:Name)
   n11 = GetFocusedLevelSequence(execute=@n13.then)
   > n12 = AddPossessable!(execute, Sequence:Movie Scene Sequence Object Reference=@n11, ObjectToPossess:Object Reference=@n8) -> (then, ReturnValue:Movie Scene Binding Proxy Structure)
   n13 = SetActorLabel(execute=@n8.then, self=@n8, NewActorLabel=OffsetNode, bMarkDirty=true)
  ]}
 Bake_to_MrBipedv02 {nodes [
   Bake_to_MrBipedv02() {is=title K2Node_FunctionEntry}
   > n1 = GetFocusedLevelSequence()
   > Set in_LevelSequence(in_LevelSequence:Level Sequence Object Reference=@n1) -> (Output_Get:Level Sequence Object Reference)
   > n2 = GetEditorWorld(self=Get UnrealEditorSubsystem())
   > n3 = Execute Python Script!(
    execute
    PythonScript="import unreal\r\n## Create animation sequence export options\r\nanim_seq_export_options = unreal.AnimSeqExportOption()\r\nanim_seq_export_options.export_transforms = True\r\nanim_seq_export_options.export_morph_targets = True"
   ) -> (then, ReturnValue, anim_seq_export_options:Anim Seq Export Option Object Reference) {outputs [anim_seq_export_options]}
   > n4 = For Each Loop(Array:Array of Movie Scene Binding Proxy Structures=GetSelectedBindings!() -> (ReturnValue:Array of Movie Scene Binding Proxy Structures)) -> (Array Element:Movie Scene Binding Proxy Structure)
   > BakeToControlRigFromReference!(
    execute
    World:World Object Reference=@n2
    LevelSequence:Level Sequence Object Reference=Get in_LevelSequence() -> (in_LevelSequence:Level Sequence Object Reference)
    InRef:"Control Rig Asset Strong Reference Structure (by ref)"=(BlueprintRigClass="/Script/ControlRig.ControlRigBlueprintGeneratedClass'/EpicControlRig/Templates/MR_Biped_v02.MR_Biped_v02_C'",ControlRigAsset=None)
    ExportOptions:Anim Seq Export Option Object Reference=@n3.anim_seq_export_options
    bReduceKeys:bool=false
    Tolerance:float=0
    Binding:"Movie Scene Binding Proxy Structure (by ref)"=@n4
    bResetControls:bool=true
   ) -> (then, ReturnValue:bool)
  ]}
 ControlExistCheck {nodes [
   n1 = ControlExistCheck!() -> (then, Control Rig:Control Rig Object Reference, ControlName_In:Name) {is=title K2Node_FunctionEntry}
   > n2 = Set Hierarchy(Hierarchy:Rig Hierarchy Object Reference=GetHierarchy!(self:Control Rig Object Reference=@n1) -> (ReturnValue:Rig Hierarchy Object Reference)) -> (Output_Get:Rig Hierarchy Object Reference)
   n3 = Get Control Settings!(self:Rig Hierarchy Object Reference=@n2, InKey_Type:ERigElementType Enum=Control, InKey_Name:Name=@n1.ControlName_In) -> (
    ReturnValue_AnimationType:ERigControlAnimationType Enum
    ReturnValue_ControlType:ERigControlType Enum
    ReturnValue_DisplayName:Name
    ReturnValue_PrimaryAxis:ERigControlAxis Enum
    ReturnValue_LimitEnabled:Array of Rig Control Limit Enabled Structures
    ReturnValue_bDrawLimits:bool=~
    ReturnValue_MinimumValue:Rig Control Value Structure
    ReturnValue_MaximumValue:Rig Control Value Structure
    ReturnValue_bShapeVisible:bool=~
    ReturnValue_ShapeVisibility:ERigControlVisibility Enum
    ReturnValue_ShapeName:Name
    ReturnValue_ShapeColor:Linear Color Structure
    ReturnValue_bIsTransientControl:bool=~
    ReturnValue_ControlEnum:Enum Object Reference
    ReturnValue_Customization:Rig Control Element Customization Structure
    ReturnValue_DrivenControls:Array of Rig Element Key Structures
    ReturnValue_bGroupWithParentControl:bool=~
    ReturnValue_bRestrictSpaceSwitching:bool=~
    ReturnValue_FilteredChannels:Array of ERigControlTransformChannel Enums
    ReturnValue_PreferredRotationOrder:EEulerRotationOrder Enum
    ReturnValue_bUsePreferredRotationOrder:bool=~
    ReturnValue_ShapeTransform:Transform
   ) {is=title K2Node_CallFunction}
   n4 = DEBUG!(
    execute
    self
    NewParam:String=Append!(
     A="Settings Preferred Rotation Order is: "
     B=Enum to String!(Enumerator:EEulerRotationOrder Enum=@n3.ReturnValue_PreferredRotationOrder) -> (ReturnValue:String)
     C:String=" ** use preferred rotation order == "
     D:String="To String (Boolean)"(InBool=@n3.ReturnValue_bUsePreferredRotationOrder)
     E:String="  ** preferred euler angles rotation order is: "
     F:String=Enum to String(Enumerator:ERigControlType Enum=@n3.ReturnValue_ControlType)
    ) -> (ReturnValue)
   ) -> (then)
   Return Node!(execute=[@n4.then, @n6.then], ControlExists:bool=true, ControlName_Out:Name=@n1.ControlName_In)
   Return Node(execute=@n6.else, ControlExists=false, ControlName_Out=None)
   n5 = "Equal Exactly (String)"(A=Animation Control, B=Enum to String(Enumerator:ERigControlAnimationType Enum=@n3))
   n6 = Branch(execute=@n2.then, Condition=@n5)
  ]}
 CopyWorldSpace_AllFrames {nodes [
   CopyWorldSpace_AllFrames() {is=title K2Node_FunctionEntry}
   > Clear!(
    execute
    TargetMap:Map of Control Rig Object References to Struct Name Transform Map Structures=Get Copy Paste Control Rig Proxy Bindings() -> (Copy Paste Control Rig Proxy Bindings:Map of Control Rig Object References to Struct Name Transform Map Structures)
   ) -> (then) {is=title K2Node_CallFunction}
   > Get Current Local Time(TimeUnit=DisplayRate) -> (ReturnValue_Frame_SubFrame=~, ReturnValue_Time=~, ReturnValue_bHasJumped=~)
   > n1 = GetFocusedLevelSequence()
   > n2 = GetControlRigs!(execute, LevelSequence:Level Sequence Object Reference=@n1) -> (then, ReturnValue:Array of Control Rig Sequencer Binding Proxy Structures)
   > n3 = For Each Loop(Array:Array of Control Rig Sequencer Binding Proxy Structures=@n2) -> (Array Element:Control Rig Sequencer Binding Proxy Structure)
   > Clear(TargetMap:Map of Names to Transforms=Get Controls Transform Map Local() -> (Controls Transform Map Local:Map of Names to Transforms))
   n4 = Break Control Rig Sequencer Binding Proxy!(ControlRigSequencerBindingProxy:"Control Rig Sequencer Binding Proxy Structure (by ref)"=@n3) -> (
    Proxy:Movie Scene Binding Proxy Structure
    ControlRig:Control Rig Object Reference
    Track:Movie Scene Control Rig Parameter Track Object Reference
   ) {is=title K2Node_BreakStruct}
   > n5 = For Each Loop!(
    Exec
    Array:Array of Rig Element Key Structures=Get Controls!(self:Rig Hierarchy Object Reference=GetHierarchy(self=@n4.ControlRig), bTraverse:bool=true) -> (ReturnValue:Array of Rig Element Key Structures) {is=title K2Node_CallFunction}
   ) -> (LoopBody, Array Element_Type:ERigElementType Enum, Array Element_Name:Name, Array Index, Completed)
   Add!(
    execute=@n5.Completed
    TargetMap:Map of Control Rig Object References to Struct Name Transform Map Structures=Get Copy Paste Control Rig Proxy Bindings()
    Key:"Control Rig Object Reference (by ref)"=@n4.ControlRig
    Value:"Struct Name Transform Map Structure (by ref)"=Make Struct Name Transform Map!(NameTransform_4_#1:Map of Names to Transforms=Get Controls Transform Map Local()) -> (Struct_NameTransformMap:Struct Name Transform Map Structure) {is=title K2Node_MakeStruct}
   ) -> (then) {is=title K2Node_CallFunction}
   Branch(
    execute=@n5.LoopBody
    Condition=IsControlSelected!(self:Control Rig Object Reference=@n4.ControlRig, InControlName:"Name (by ref)"=@n5.Array Element_Name) -> (ReturnValue:bool)
   )
   > n6 = GetPlaybackRange!(execute, Sequence:Movie Scene Sequence Object Reference=@n1) -> (then, ReturnValue_bHasStart:bool=~, ReturnValue_bHasEnd:bool=~, ReturnValue_InclusiveStart:int, ReturnValue_ExclusiveEnd:int)
   > n7 = For Loop(FirstIndex=@n6.ReturnValue_InclusiveStart)
   > n8 = GetControlRigWorldTransforms!(
    execute
    LevelSequence:Level Sequence Object Reference=@n1
    ControlRig:Control Rig Object Reference=@n4.ControlRig
    ControlName:Name=@n5.Array Element_Name
    Frames:Array of Frame Number Structures=Make Array("[0]":Frame Number Structure=Make Frame Number!(Value:int=@n7) -> (FrameNumber:Frame Number Structure) {is=title K2Node_MakeStruct}) -> (Array:Array of Frame Number Structures)
    TimeUnit:EMovieSceneTimeUnit Enum=DisplayRate
   ) -> (then, ReturnValue:Array of Transforms)
   > n9 = For Each Loop!(Exec, Array:Array of Transforms=@n8) -> (LoopBody, Array Element:Transform, Array Index, Completed)
   > Add(
    TargetMap:Map of Names to Transforms=Get Controls Transform Map Local()
    Key:"Name (by ref)"=@n5.Array Element_Name
    Value:"Transform (by ref)"=@n9
   )
  ]}
 CopyWorldSpace_CurrentFrame {nodes [
   CopyWorldSpace_CurrentFrame() {is=title K2Node_FunctionEntry}
   > Clear(TargetMap:Map of Control Rig Object References to Struct Name Transform Map Structures=Get Copy Paste Control Rig Proxy Bindings())
   > n1 = Get Current Local Time(TimeUnit=DisplayRate) -> (ReturnValue_Frame_SubFrame=~, ReturnValue_Time=~, ReturnValue_bHasJumped=~)
   > n2 = GetFocusedLevelSequence()
   > n3 = GetControlRigs(LevelSequence=@n2)
   > n4 = For Each Loop(Array:Array of Control Rig Sequencer Binding Proxy Structures=@n3) -> (Array Element:Control Rig Sequencer Binding Proxy Structure)
   > Clear(TargetMap:Map of Names to Transforms=Get Controls Transform Map Local())
   n5 = Break Control Rig Sequencer Binding Proxy(ControlRigSequencerBindingProxy=@n4)
   > n6 = For Each Loop!(Exec, Array:Array of Rig Element Key Structures=Get Controls(self=GetHierarchy(self=@n5.ControlRig), bTraverse=true)) -> (LoopBody, Array Element_Type:ERigElementType Enum, Array Element_Name:Name, Array Index, Completed)
   Add(
    execute=@n6.Completed
    TargetMap:Map of Control Rig Object References to Struct Name Transform Map Structures=Get Copy Paste Control Rig Proxy Bindings()
    Key:"Control Rig Object Reference (by ref)"=@n5.ControlRig
    Value:"Struct Name Transform Map Structure (by ref)"=Make Struct Name Transform Map(NameTransform_4_#1=Get Controls Transform Map Local())
   )
   Branch(execute=@n6.LoopBody, Condition=IsControlSelected(self=@n5.ControlRig, InControlName=@n6.Array Element_Name))
   > n7 = GetControlRigWorldTransform!(
    execute
    LevelSequence:Level Sequence Object Reference=@n2
    ControlRig:Control Rig Object Reference=@n5.ControlRig
    ControlName:Name=@n6.Array Element_Name
    Frame:Frame Number Structure=@n1
    TimeUnit:EMovieSceneTimeUnit Enum=DisplayRate
   ) -> (then, ReturnValue:Transform)
   > Add(
    TargetMap:Map of Names to Transforms=Get Controls Transform Map Local()
    Key:"Name (by ref)"=@n6.Array Element_Name
    Value:"Transform (by ref)"=@n7
   )
  ]}
 CreateConstraint {nodes [
   CreateConstraint() {is=title K2Node_FunctionEntry}
   > n1 = Sequence!(execute) -> (then_0, then_1, then_2:Exec, then_3:Exec)
   > n2 = GetEditorWorld(self=Get UnrealEditorSubsystem())
   n3 = GetFocusedLevelSequence(execute=@n1.then_1)
   > n4 = GetControlRigs(LevelSequence=@n3)
   > n5 = For Each Loop!(Exec, Array:Array of Control Rig Sequencer Binding Proxy Structures=@n4) -> (LoopBody, Array Element:Control Rig Sequencer Binding Proxy Structure, Array Index, Completed)
   n6 = Break Control Rig Sequencer Binding Proxy(ControlRigSequencerBindingProxy=@n5)
   n7 = "To String (Name)"(InName=Get ChildRig() -> (ChildRig:Name=None))
   > Branch(Condition="Equal Exactly (String)"(A=GetDisplayName(Object=@n6.ControlRig), B=@n7))
   > PrintString(InString=@n7)
   > n8 = CreateTransformableHandle(InWorld=@n2, InObject=@n6.ControlRig, InAttachmentName=Get ChildControl() -> (ChildControl:Name=None))
   n9 = GetFocusedLevelSequence(execute=@n1.then_2)
   > n10 = GetControlRigs(LevelSequence=@n9)
   > n11 = For Each Loop(Array=@n10)
   n12 = Break Control Rig Sequencer Binding Proxy(ControlRigSequencerBindingProxy=@n11)
   n13 = "To String (Name)"(InName=Get ParentRig() -> (ParentRig:Name=None))
   > Branch(Condition="Equal Exactly (String)"(A=GetDisplayName(Object=@n12.ControlRig), B=@n13))
   > PrintString(InString=@n13)
   > n14 = CreateTransformableHandle(InWorld=@n2, InObject=@n12.ControlRig, InAttachmentName=Get ParentControl() -> (ParentControl:Name=None))
   AddConstraint(
    execute=@n1.then_3
    World=@n2
    InType=Get ConstraintType() -> (ConstraintType:ETransformConstraintType Enum=Translation)
    InChild=@n8
    InParent=@n14
    bMaintainOffset=true
   )
  ]}
 DEBUG {nodes [
   n1 = DEBUG!() -> (then, NewParam:String) {is=title K2Node_FunctionEntry}
   > n2 = Branch(Condition=Get Debug_Check() -> (Debug_Check:bool))
   > n3 = PrintString(InString=@n1)
   Return Node!(execute=[@n3.then, @n2.else])
  ]}
 EUW_MakeLocator {nodes [
   n1 = DEBUG(NewParam="To String (Transform)"!(InTrans:"Transform (by ref)"=Get Local_WorldTransform() -> (Local_WorldTransform:Transform)) -> (ReturnValue:String)) {is=title K2Node_CallFunction}
   SetControlRigWorldTransform(
    execute=[@n33.then, @n1.then]
    LevelSequence=Get Local_LevelSequence() -> (Local_LevelSequence:Level Sequence Object Reference)
    ControlRig=Get Local_LocatorCRU_ControlRig() -> (Local_LocatorCRU_ControlRig:Control Rig Object Reference)
    ControlName=Get Local_LocatorCRU_ControlName() -> (Local_LocatorCRU_ControlName:Name=None)
    Frame=Get Local_Frame() -> (Local_Frame:Frame Number Structure)
    WorldTransform=Get Local_WorldTransform()
    TimeUnit=DisplayRate
    bSetKey=false
   )
   n2 = Get Local_LocatorCRU_ControlRig()
   n3 = Get Local_LocatorCRU_ControlName() -> (Local_LocatorCRU_ControlName=None)
   > Add!(
    execute
    TargetArray:Array of STR Control Rig Proxy Object Name Tuple Structures=Get STR_Out_Locator_locator() -> (STR_Out_Locator_locator:Array of STR Control Rig Proxy Object Name Tuple Structures)
    NewItem_CR_Proxy_11_#2:Movie Scene Binding Proxy Structure=Get Local_CRProxy() -> (Local_CRProxy:Movie Scene Binding Proxy Structure)
    NewItem_CR_Object_8_#3:Control Rig Object Reference=@n2
    NewItem_CR_ControlName_9_#4:Name=@n3
   ) -> (then, ReturnValue) {is=title K2Node_CallArrayFunction}
   n4 = Get Local_InControl() -> (Local_InControl:Name=None)
   n5 = Get Local_InCR() -> (Local_InCR:Control Rig Object Reference)
   > Add(
    TargetArray=Get STR_Out_Locator_sourceCtrl() -> (STR_Out_Locator_sourceCtrl:Array of STR Control Rig Proxy Object Name Tuple Structures)
    NewItem_CR_Proxy_11_#2=Get Local_CRProxy()
    NewItem_CR_Object_8_#3=@n5
    NewItem_CR_ControlName_9_#4=@n4
   )
   n6 = EUW_MakeLocator!() -> (then, Locator Color:String=(R=0.1878,G=0.3419,B=0.0999,A=1)) {is=title K2Node_FunctionEntry}
   > Set Local_Loc_Color(Local_Loc_Color:String=@n6) -> (Output_Get:String)
   > Set Local_NumLocators(Local_NumLocators:int=0) -> (Output_Get:int=0)
   > n7 = JJF_GetSelection_BindingsObjectsControlNames!(execute, self) -> (then, CurrentLevelSequence:Level Sequence Object Reference)
   > Set Local_LevelSequence(Local_LevelSequence:Level Sequence Object Reference=@n7) -> (Output_Get:Level Sequence Object Reference)
   n8 = Get STR_CR_ProxyObjName() -> (STR_CR_ProxyObjName:Array of STR Control Rig Proxy Object Name Tuple Structures)
   > n9 = Branch(Condition="integer > integer"(A=Length(TargetArray:Array of STR Control Rig Proxy Object Name Tuple Structures=@n8)))
   > n10 = For Each Loop!(Exec, Array:Array of STR Control Rig Proxy Object Name Tuple Structures=Get STR_CR_ProxyObjName()) -> (
    LoopBody
    Array Element_CR_Proxy_11_#2:Movie Scene Binding Proxy Structure
    Array Element_CR_Object_8_#3:Control Rig Object Reference
    Array Element_CR_ControlName_9_#4:Name
    Array Index
    Completed
   )
   > Set Local_CRProxy(Local_CRProxy:Movie Scene Binding Proxy Structure=@n10) -> (Output_Get:Movie Scene Binding Proxy Structure)
   > Set Local_InCR(Local_InCR:Control Rig Object Reference=@n10.Array Element_CR_Object_8_#3) -> (Output_Get:Control Rig Object Reference)
   > Set Local_InControl(Local_InControl:Name=@n10.Array Element_CR_ControlName_9_#4) -> (Output_Get:Name=None)
   > n11 = Sequence!(execute) -> (then_0, then_1, then_2, then_3, then_4:Exec, then_5:Exec)
   n12 = Get Local_LevelSequence()
   > n13 = Scan For CRU_AddLocatorRigs!(execute, self, Local_LevelSequence:Level Sequence Object Reference=@n12, Local_LocatorCRU_Track:Movie Scene Track Object Reference) -> (then, Found:bool, Out_ControlRig:Control Rig Object Reference)
   > n14 = Branch(Condition=@n13)
   n15 = GetHierarchy(self=@n13.Out_ControlRig)
   > Set Local_NumLocators(Local_NumLocators=Length(TargetArray:Array of Rig Element Key Structures=Get Controls(self=@n15, bTraverse=true))) -> (Output_Get=0)
   > n16 = Set Local_LocatorCRU_ControlRig(Local_LocatorCRU_ControlRig:Control Rig Object Reference=@n13.Out_ControlRig) -> (Output_Get:Control Rig Object Reference)
   n17 = FindOrCreateControlRigTrackFromReference!(
    execute=[@n22.then, @n16.then]
    World:World Object Reference
    LevelSequence:Level Sequence Object Reference=@n12
    InRef:"Control Rig Asset Strong Reference Structure (by ref)"=(BlueprintRigClass="&ControlRigBlueprintGeneratedClass'/AnimatorKit/UtilityRigs/CRU_AddLocator.CRU_AddLocator_C'",ControlRigAsset=None)
    InBinding:"Movie Scene Binding Proxy Structure (by ref)"=Get Local_CRProxy()
    bIsLayeredControlRig:bool=false
   ) -> (then, ReturnValue:Movie Scene Track Object Reference)
   > n18 = Set Local_LocatorCRU_Track(Local_LocatorCRU_Track:Movie Scene Track Object Reference=@n17) -> (Output_Get:Movie Scene Track Object Reference)
   > n19 = GetControlRigPriorityOrder!(execute, InTrack:Movie Scene Track Object Reference=@n18) -> (then, ReturnValue:int=0)
   n20 = Get Local_LocatorCRU_Track() -> (Local_LocatorCRU_Track:Movie Scene Track Object Reference)
   > SetControlRigPriorityOrder!(execute, InTrack:Movie Scene Track Object Reference=@n20, PriorityOrder:int=int + int(A=@n19, B=10)) -> (then)
   > ForceUpdate!(execute) -> (then)
   > n21 = Cast To MovieSceneControlRigParameterTrack!(execute, Object=@n20) -> (then, CastFailed, AsMovie Scene Control Rig Parameter Track:Movie Scene Control Rig Parameter Track Object Reference) {is=title K2Node_DynamicCast}
   > Add(
    TargetMap:Map of Names to Strings=Get STR_Recreate() -> (STR_Recreate:Map of Names to Strings)
    Key=Color
    Value:"String (by ref)"=Get Local_Loc_Color() -> (Local_Loc_Color:String)
   ) {is=title K2Node_CallFunction}
   > RecreateControlRigWithNewSettings!(
    execute
    InTrack:Movie Scene Control Rig Parameter Track Object Reference=@n21
    InSettingsToOverride:Map of Names to Strings=Get STR_Recreate()
   ) -> (then, ReturnValue:bool)
   n22 = Set Local_NumLocators(execute=@n14.else, Local_NumLocators=0) -> (Output_Get=0)
   n23 = Get Current Local Time(execute=@n11.then_1, TimeUnit=DisplayRate) -> (ReturnValue_Frame_SubFrame=~, ReturnValue_Time=~, ReturnValue_bHasJumped=~)
   > Set Local_Frame(Local_Frame:Frame Number Structure=@n23) -> (Output_Get:Frame Number Structure)
   > n24 = GetControlRigWorldTransform(
    LevelSequence=Get Local_LevelSequence()
    ControlRig=Get Local_InCR()
    ControlName=Get Local_InControl() -> (Local_InControl=None)
    Frame=@n23
    TimeUnit=DisplayRate
   )
   > Set Local_WorldTransform(Local_WorldTransform:Transform=@n24) -> (Output_Get:Transform)
   n25 = Get Local_NumLocators() -> (Local_NumLocators:int=0)
   DEBUG(execute=@n11.then_3, NewParam=Append!(A="initial num of locators: ", B="To String (Integer)"(InInt=@n25)) -> (ReturnValue))
   > n26 = Branch(Condition="integer > integer"(A=@n25))
   n27 = Get Local_NumLocators() -> (Local_NumLocators=0)
   > Add(TargetMap=Get STR_Recreate(), Key=NumOfLocators, Value="To String (Integer)"(InInt=int + int(A=@n27)))
   > n28 = Cast To MovieSceneControlRigParameterTrack(Object=Get Local_LocatorCRU_Track())
   > RecreateControlRigWithNewSettings(InTrack=@n28, InSettingsToOverride=Get STR_Recreate())
   Add(execute=@n26.else, TargetMap=Get STR_Recreate(), Key=NumOfLocators, Value="To String (Integer)"(InInt=1))
   > n29 = Cast To MovieSceneControlRigParameterTrack(Object=Get Local_LocatorCRU_Track())
   > RecreateControlRigWithNewSettings(InTrack=@n29, InSettingsToOverride=Get STR_Recreate())
   n30 = Scan For CRU_AddLocatorRigs(execute=@n11.then_4, Local_LevelSequence=Get Local_LevelSequence())
   > Set Local_LocatorCRU_ControlRig(Local_LocatorCRU_ControlRig=@n30.Out_ControlRig)
   n31 = Get Local_LocatorCRU_ControlRig()
   n32 = For Each Loop!(Exec=@n11.then_5, Array:Array of Rig Element Key Structures=Get Controls(self=GetHierarchy(self=@n31), bTraverse=true)) -> (LoopBody, Array Element_Type:ERigElementType Enum, Array Element_Name:Name, Array Index, Completed)
   SelectControl!(
    execute=@n32.Completed
    self:Control Rig Object Reference=Get Local_LocatorCRU_ControlRig()
    InControlName:"Name (by ref)"=Get Local_LocatorCRU_ControlName() -> (Local_LocatorCRU_ControlName=None)
    bSelect:bool=true
    bSetupUndo:bool=false
   ) -> (then)
   > SelectControl(self=Get Local_InCR(), InControlName=Get Local_InControl() -> (Local_InControl=None), bSelect=false, bSetupUndo=false)
   > n33 = ForceUpdate()
   Set Local_LocatorCRU_ControlName(execute=@n32.LoopBody, Local_LocatorCRU_ControlName:Name=@n32.Array Element_Name) -> (Output_Get:Name=None)
   Return Node!(
    execute=@n10.Completed
    Out_Loc_CR:Control Rig Object Reference=@n2
    Out_Loc_CtrlName:Name=@n3
    In_CR:Control Rig Object Reference=@n5
    In_Control:Name=@n4
    NewLocators:Array of STR Control Rig Proxy Object Name Tuple Structures=Get STR_Out_Locator_locator()
    ControlsRunOn:Array of STR Control Rig Proxy Object Name Tuple Structures=Get STR_Out_Locator_sourceCtrl()
   )
   PrintString(execute=@n9.else, InString="***   Nothing Selected   ***")
   > PrintString(InString=Hello)
   > n34 = AddSpawnableFromClass!(
    execute
    self:Level Sequence Editor Subsystem Object Reference=Get LevelSequenceEditorSubsystem() -> (ReturnValue:Level Sequence Editor Subsystem Object Reference) {is=title K2Node_GetEditorSubsystem}
    Sequence:Movie Scene Sequence Object Reference=Get Local_LevelSequence()
    ClassToSpawn:Object Class Reference=&CRU_AddLocator_C
   ) -> (then, ReturnValue:Movie Scene Binding Proxy Structure)
   > n35 = FindOrCreateControlRigTrackFromReference(
    LevelSequence=Get Local_LevelSequence()
    InRef=(BlueprintRigClass="&ControlRigBlueprintGeneratedClass'&CRU_AddLocator_C'",ControlRigAsset=None)
    InBinding=@n34
    bIsLayeredControlRig=false
   )
   > n36 = Cast To MovieSceneControlRigParameterTrack(Object=@n35)
   > Add(TargetMap=Get STR_Recreate(), Key=Color, Value=(R=1,G=0.3916,B=0.1119,A=1))
   > RecreateControlRigWithNewSettings(InTrack=@n36, InSettingsToOverride=Get STR_Recreate())
   n37 = For Each Loop!(Exec, Array:Array of STR Control Rig Proxy Object Name Tuple Structures=Get STR_CR_ProxyObjName()) -> (
    LoopBody
    Array Element_CR_Proxy_11_#2:Movie Scene Binding Proxy Structure
    Array Element_CR_Object_8_#3:Control Rig Object Reference
    Array Element_CR_ControlName_9_#4:Name
    Array Index
    Completed
   )
   > Set Local_InCR(Local_InCR=@n37.Array Element_CR_Object_8_#3)
   > Set Local_InControl(Local_InControl=@n37.Array Element_CR_ControlName_9_#4) -> (Output_Get=None)
   RecreateControlRigWithNewSettings()
   Scan For CRU_AddLocatorRigs(Local_LevelSequence=Get Local_LevelSequence())
  ]}
 EventGraph {nodes [
   n1 = GetRootBoneName!(execute=@n3.LoopBody, TargetMesh:Dynamic Mesh Object Reference) -> (then, BoneName:Name=None, ReturnValue:Dynamic Mesh Object Reference)
   n2 = GetSelectedLevelActors(self=Get EditorActorSubsystem())
   > n3 = For Each Loop!(Exec, Array:Array of Actor Object References=@n2) -> (LoopBody, Array Element:Actor Object Reference, Array Index, Completed)
   n4 = CreateTransformableHandle(InObject=@n3, InAttachmentName=@n1)
   > n5 = AddSpawnableFromClass(self=Get LevelSequenceEditorSubsystem())
   > n6 = AddTrack!(
    execute
    InBinding:"Movie Scene Binding Proxy Structure (by ref)"=@n5
    TrackType:Movie Scene Track Class Reference=/Script/MovieSceneTracks.MovieSceneTransformTrack
   ) -> (then, ReturnValue:Movie Scene Track Object Reference)
   > n7 = CreateTransformableHandle(InObject=@n6, InAttachmentName=None)
   > AddConstraint!(
    execute
    InWorld:World Object Reference
    InParentHandle:Transformable Handle Object Reference=@n4
    InChildHandle:Transformable Handle Object Reference=@n7
    InConstraint:Tickable Transform Constraint Object Reference
    bMaintainOffset=false
   ) -> (then, ReturnValue:bool)
   n8 = GetSelectedActors()
   > Branch(Condition="Equal (Integer)"(A=Length(TargetArray:Array of Actor Object References=@n8), B=1))
   > n9 = GetEditorWorld(self=Get UnrealEditorSubsystem())
   > n10 = For Each Loop(Array:Array of Tickable Transform Constraint Object References=Get Child Constraints!(Child:Actor Object Reference="Get (a copy)"(Array:Array of Actor Object References=@n8) -> (Output:Actor Object Reference)) -> (Constraints:Array of Tickable Transform Constraint Object References) {is=title K2Node_CallFunction}) -> (Array Element:Tickable Transform Constraint Object Reference)
   > n11 = Cast To TransformableComponentHandle!(
    execute
    Object=Get ParentTRSHandle!(self:Tickable Transform Constraint Object Reference=@n10) -> (ParentTRSHandle:Transformable Handle Object Reference)
   ) -> (then, CastFailed, AsTransformable Component Handle:Transformable Component Handle Object Reference) {is=title K2Node_DynamicCast}
   > n12 = SpawnActor Static Mesh Actor(
    WorldContextObject=@n9
    Class=&StaticMeshActor
    SpawnTransform=GetSocketTransform(
     self=Get Component!(self:Transformable Component Handle Object Reference=@n11) -> (Component:Scene Component Weak object)
     InSocketName=Get SocketName!(self:Transformable Component Handle Object Reference=@n11) -> (SocketName:Name=None)
     TransformSpace=RTS_World
    )
    CollisionHandlingOverride=Undefined
    TransformScaleMethod=MultiplyWithRoot
   )
   > SetActorLabel(self=@n12, NewActorLabel=OffsetNode, bMarkDirty=true)
   > n13 = GetFocusedLevelSequence()
   > AddPossessable(Sequence=@n13, ObjectToPossess=@n12)
   > n14 = CreateTransformableHandle(InWorld=@n9, InObject=@n12, InAttachmentName=Get In Attachment Name() -> (In Attachment Name:Name=None))
   > Set ParentTRSHandle(ParentTRSHandle=@n14, self=@n10)
   > n15 = CreateTransformableHandle(InWorld=@n9, InObject=@n12, InAttachmentName=Get In Attachment Name_0() -> (In Attachment Name_0:Name=None))
   > AddConstraint!(
    execute
    World:World Object Reference=@n9
    InType:ETransformConstraintType Enum=Parent
    InChild:Transformable Handle Object Reference=@n15
    InParent:Transformable Handle Object Reference=@n11
    bMaintainOffset=true
   ) -> (then, ReturnValue:Tickable Constraint Object Reference)
   Clear(TargetMap:Map of Control Rig Object References to Struct Name Transform Map Structures=Get Control Rig Proxy Binding() -> (Control Rig Proxy Binding:Map of Control Rig Object References to Struct Name Transform Map Structures))
   > n16 = Get Current Time!(execute, TimeUnit:EMovieSceneTimeUnit Enum=DisplayRate) -> (
    then
    ReturnValue_Frame_FrameNumber:Frame Number Structure
    ReturnValue_Frame_SubFrame:float=~
    ReturnValue_Time:float=~
    ReturnValue_MarkedFrame:String
    ReturnValue_Timecode:Timecode Structure
    ReturnValue_PositionType:EMovieScenePositionType Enum
    ReturnValue_UpdateMethod:EUpdatePositionMethod Enum
    ReturnValue_bHasJumped:bool=~
   ) {is=title K2Node_CallFunction}
   > n17 = GetFocusedLevelSequence()
   > n18 = Set Focused Level Sequence(Focused Level Sequence:Level Sequence Object Reference=@n17) -> (Output_Get:Level Sequence Object Reference)
   > n19 = GetControlRigs(LevelSequence=@n18)
   > n20 = For Each Loop(Array:Array of Control Rig Sequencer Binding Proxy Structures=@n19) -> (Array Element:Control Rig Sequencer Binding Proxy Structure)
   > Clear(TargetMap:Map of Names to Transforms=Get Bone Transform Map Local() -> (Bone Transform Map Local:Map of Names to Transforms))
   n21 = Break Control Rig Sequencer Binding Proxy(ControlRigSequencerBindingProxy=@n20)
   > n22 = For Each Loop!(Exec, Array:Array of Rig Element Key Structures=Get Controls(self=GetHierarchy(self=@n21.ControlRig), bTraverse=true)) -> (LoopBody, Array Element_Type:ERigElementType Enum, Array Element_Name:Name, Array Index, Completed)
   Add(
    execute=@n22.Completed
    TargetMap:Map of Control Rig Object References to Struct Name Transform Map Structures=Get Control Rig Proxy Binding()
    Key:"Control Rig Object Reference (by ref)"=@n21.ControlRig
    Value:"Struct Name Transform Map Structure (by ref)"=Make Struct Name Transform Map(NameTransform_4_#1=Get Bone Transform Map Local())
   )
   n23 = IsControlSelected(self=@n21.ControlRig, InControlName=@n22.Array Element_Name)
   Branch(execute=@n22.LoopBody, Condition=@n23)
   > n24 = GetControlRigWorldTransform(
    LevelSequence=Get Focused Level Sequence() -> (Focused Level Sequence:Level Sequence Object Reference)
    ControlRig=@n21.ControlRig
    ControlName=@n22.Array Element_Name
    Frame=@n16
    TimeUnit=DisplayRate
   )
   > Add(
    TargetMap:Map of Names to Transforms=Get Bone Transform Map Local()
    Key:"Name (by ref)"=@n22.Array Element_Name
    Value:"Transform (by ref)"=@n24
   )
   "On Clicked (BTN_MoveSpaceRight)"() {is=title K2Node_ComponentBoundEvent}
   > MoveSpaceKey_Right()
   "On Clicked (BTN_Cycle)"() {is=title K2Node_ComponentBoundEvent}
   > Pre/Post Curve Interp!(execute, self, InExtrapolation:ERichCurveExtrapolation Enum=RCCE_Cycle) -> (then)
   "On Clicked (BTN_Locator)"() {is=title K2Node_ComponentBoundEvent}
   > EUW_MakeLocator(Locator Color=(R=0.1878,G=0.3419,B=0.0999,A=1)) -> (Out_Loc_CtrlName=None, In_Control=None) {is=title K2Node_CallFunction}
   "On Clicked (Btn_CleanExtras)"() {is=title K2Node_ComponentBoundEvent}
   > BeginTransaction(Description="LOCTEXT(\"Clean Controls\")")
   > n25 = Sequence!(execute) -> (then_0, then_1)
   Execute Python Script!(
    execute=[@n29.then_2, @n28.then_2, @n33.then_1, @n25.then_0]
    PythonScript="import unreal\r\n\r\n## Sec FK controls — match any control in Spine or Neck modules named \"Sec FK *\"\r\nSEC_FK_PREFIXES = [\"Spine/Sec FK\", \"Neck/Sec FK\"]\r\n\r\n## Explicit twist/mid controls\r\nEXTRA_CONTROLS = [\r\n    \"Arm R/Mid\",\r\n    \"Arm Upper Twist R/Twist 1\", \"Arm Upper Twist R/Twist 2\",\r\n    \"Arm Lower Twist R/Twist 1\", \"Arm Lower Twist R/Twist 2\",\r\n    \"Arm L/Mid\",\r\n    \"Arm Upper Twist L/Twist 1\", \"Arm Upper Twist L/Twist 2\",\r\n    \"Arm Lower Twist L/Twist 1\", \"Arm Lower Twist L/Twist 2\",\r\n    \"Leg R/Mid\",\r\n    \"Leg Upper Twist R/Twist 1\", \"Leg Upper Twist R/Twist 2\",\r\n    \"Leg Lower Twist R/Twist 1\", \"Leg Lower Twist R/Twist 2\",\r\n    \"Leg L/Mid\",\r\n    \"Leg Upper Twist L/Twist 1\", \"Leg Upper Twist L/Twist 2\",\r\n    \"Leg Lower Twist L/Twist 1\", \"Leg Lower Twist L/Twist 2\",\r\n]\r\n\r\ndef get_cr_section(level_seq):\r\n    for binding in level_seq.get_bindings():\r\n        for track in binding.get_tracks():\r\n            for section in track.get_sections():\r\n                if \"ControlRig\" in section.get_class().get_name():\r\n                    return section\r\n    return None\r\n\r\ndef clear_and_key_control(level_seq, rig, ctrl, frame0, all_channels, identity_euler):\r\n    ## Clear all keys on this control\r\n    ctrl_channels = [ch for ch in all_channels\r\n                     if str(ch.channel_name) == ctrl or\r\n                        str(ch.channel_name).startswith(ctrl + \".\")]\r\n    total_keys = 0\r\n    for ch in ctrl_channels:\r\n        keys = ch.get_keys()\r\n        total_keys += len(keys)\r\n        for key in keys:\r\n            ch.remove_key(key)\r\n\r\n    ## Set identity key\r\n    try:\r\n        unreal.ControlRigSequencerLibrary.set_local_control_rig_euler_transform(\r\n            level_seq, rig, ctrl, frame0, identity_euler, set_key=True)\r\n        unreal.log(f\"[ExtrasClean] Cleared {total_keys} keys, keyed bind pose: {ctrl}\")\r\n    except Exception as e:\r\n        unreal.log_warning(f\"[ExtrasClean] Key failed for {ctrl}: {e}\")\r\n\r\ndef euler_filter_channel(ch):\r\n    keys = ch.get_keys()\r\n    if len(keys) < 2:\r\n        return 0\r\n    fixed = 0\r\n    prev_val = keys[0].get_value()\r\n    for key in keys[1:]:\r\n        val = key.get_value()\r\n        diff = val - prev_val\r\n        if abs(diff) > 180.0:\r\n            offset = round(diff / 360.0) * 360.0\r\n            val -= offset\r\n            key.set_value(val)\r\n            fixed += 1\r\n        prev_val = val\r\n    return fixed\r\n\r\n## -----------------------------------------------------------------------\r\n## Main\r\n## -----------------------------------------------------------------------\r\n\r\nlevel_seq = unreal.LevelSequenceEditorBlueprintLibrary.get_current_level_sequence()\r\nif not level_seq:\r\n    unreal.log_error(\"[ExtrasClean] No Level Sequence is open.\")\r\nelse:\r\n    proxies = unreal.ControlRigSequencerLibrary.get_control_rigs(level_seq)\r\n    if not proxies:\r\n        unreal.log_error(\"[ExtrasClean] No Control Rig found in the current Level Sequence.\")\r\n    else:\r\n        rig = proxies[0].control_rig\r\n        frame0 = unreal.FrameNumber(value=0)\r\n        identity_euler = unreal.EulerTransform(\r\n            location=unreal.Vector(0, 0, 0),\r\n            rotation=unreal.Rotator(0, 0, 0),\r\n            scale=unreal.Vector(1, 1, 1)\r\n        )\r\n\r\n        cr_section = get_cr_section(level_seq)\r\n        if not cr_section:\r\n            unreal.log_error(\"[ExtrasClean] Could not find ControlRig section.\")\r\n        else:\r\n            all_channels = cr_section.get_all_channels()\r\n\r\n            ## --- Part 1: Find all Sec FK controls in Spine and Neck modules ---\r\n            ## Collect unique control names from channels matching the prefixes\r\n            sec_fk_controls = []\r\n            seen = set()\r\n            for ch in all_channels:\r\n                ch_name = str(ch.channel_name)\r\n                for prefix in SEC_FK_PREFIXES:\r\n                    if ch_name.startswith(prefix):\r\n                        ## Strip sub-channel suffix to get the control name\r\n                        ctrl_name = ch_name.split(\".\")[0]\r\n                        if ctrl_name not in seen:\r\n                            seen.add(ctrl_name)\r\n                            sec_fk_controls.append(ctrl_name)\r\n                        break\r\n\r\n            unreal.log(f\"[ExtrasClean] Found {len(sec_fk_controls)} Sec FK controls: {sec_fk_controls}\")\r\n\r\n            for ctrl in sec_fk_controls:\r\n                clear_and_key_control(level_seq, rig, ctrl, frame0, all_channels, identity_euler)\r\n\r\n            ## --- Part 2: Explicit twist/mid controls ---\r\n            unreal.log(f\"[ExtrasClean] Processing {len(EXTRA_CONTROLS)} twist/mid controls...\")\r\n            for ctrl in EXTRA_CONTROLS:\r\n                clear_and_key_control(level_seq, rig, ctrl, frame0, all_channels, identity_euler)\r\n\r\n            ## --- Part 3: Euler filter on all processed controls ---\r\n            all_processed = sec_fk_controls + EXTRA_CONTROLS\r\n            all_channels_fresh = cr_section.get_all_channels()\r\n            total_fixed = 0\r\n            for ctrl in all_processed:\r\n                rot_channels = [ch for ch in all_channels_fresh\r\n                                if str(ch.channel_name).startswith(ctrl + \".Rotation.\")]\r\n                for ch in rot_channels:\r\n                    fixed = euler_filter_channel(ch)\r\n                    if fixed:\r\n                        unreal.log(f\"[ExtrasClean] Euler filter fixed {fixed} keys on {ch.channel_name}\")\r\n                    total_fixed += fixed\r\n            unreal.log(f\"[ExtrasClean] Euler filter complete — {total_fixed} keys adjusted.\")\r\n\r\n        unreal.log(\"[ExtrasClean] Done!\")\r\n"
   ) -> (then, ReturnValue)
   EndTransaction(execute=@n25.then_1)
   "On Clicked (Btn_MovingLocal)"() {is=title K2Node_ComponentBoundEvent}
   > BeginTransaction(Description="LOCTEXT(\"Treadmill\")")
   > Execute Python Script(PythonScript="import unreal\r\n\r\n## ── Config ──────────────────────────────────────────────────────────────────\r\nSPACE_CONTROLS = [\r\n    \"Leg L/IK\",\r\n    \"Leg R/IK\",\r\n    \"Arm L/IK\",\r\n    \"Arm R/IK\",\r\n    \"Body/Body\",\r\n]\r\n\r\nROOT_CTRL   = \"root/Root\"\r\nLOCAL_CTRL  = \"root/Local\"\r\nATTACH_CTRL = \"Attach/Attach\"\r\n\r\n## All controls whose rotation keys we euler-filter at the end\r\nEULER_FILTER_CONTROLS = SPACE_CONTROLS + [ROOT_CTRL, LOCAL_CTRL, ATTACH_CTRL]\r\n\r\n## ── Space key constants ──────────────────────────────────────────────────────\r\nWORLD_SPACE  = unreal.ControlRigSequencerLibrary.get_world_space_reference_key()\r\nPARENT_SPACE = unreal.ControlRigSequencerLibrary.get_default_parent_key()\r\nLOCAL_SPACE  = unreal.RigElementKey(\r\n    type=unreal.RigElementType.CONTROL,\r\n    name=LOCAL_CTRL\r\n)\r\n\r\nZERO_EULER = unreal.EulerTransform(\r\n    location=unreal.Vector(0, 0, 0),\r\n    rotation=unreal.Rotator(0, 0, 0),\r\n    scale=unreal.Vector(1, 1, 1)\r\n)\r\n\r\n## ── Helpers ──────────────────────────────────────────────────────────────────\r\ndef get_sequence_and_cr():\r\n    sequence = unreal.LevelSequenceEditorBlueprintLibrary.get_current_level_sequence()\r\n    if not sequence:\r\n        unreal.log_error(\"No Level Sequence is open in Sequencer.\")\r\n        return None, None\r\n    cr_list = unreal.ControlRigSequencerLibrary.get_control_rigs(sequence)\r\n    if not cr_list:\r\n        unreal.log_error(\"No Control Rig found in sequence.\")\r\n        return None, None\r\n    return sequence, cr_list[0].control_rig\r\n\r\ndef get_frame_range(sequence):\r\n    start_f = int(sequence.get_playback_start())\r\n    end_f   = int(sequence.get_playback_end())\r\n    unreal.log(\"Frame range: {} -> {}\".format(start_f, end_f))\r\n    return start_f, end_f\r\n\r\ndef snap_local_to_root(sequence, cr, start_f, end_f):\r\n    lib = unreal.ControlRigSequencerLibrary\r\n    for f in range(start_f, end_f):\r\n        fn  = unreal.FrameNumber(value=f)\r\n        val = lib.get_local_control_rig_euler_transform(sequence, cr, ROOT_CTRL, fn)\r\n        lib.set_local_control_rig_euler_transform(sequence, cr, LOCAL_CTRL, fn, val)\r\n\r\ndef zero_control(sequence, cr, ctrl, start_f, end_f):\r\n    lib = unreal.ControlRigSequencerLibrary\r\n    for f in range(start_f, end_f):\r\n        fn = unreal.FrameNumber(value=f)\r\n        lib.set_local_control_rig_euler_transform(sequence, cr, ctrl, fn, ZERO_EULER)\r\n\r\ndef unwrap(angle, prev):\r\n    \"\"\"Unwrap angle to be within 180 degrees of prev.\"\"\"\r\n    diff = angle - prev\r\n    while diff >  180.0: diff -= 360.0\r\n    while diff < -180.0: diff += 360.0\r\n    return prev + diff\r\n\r\ndef euler_filter(sequence, cr, ctrl, start_f, end_f):\r\n    \"\"\"Walk rotation keys frame by frame and unwrap to remove flips.\"\"\"\r\n    lib = unreal.ControlRigSequencerLibrary\r\n    frames = list(range(start_f, end_f))\r\n    if len(frames) < 2:\r\n        return\r\n\r\n    ## Read all values first\r\n    vals = []\r\n    for f in frames:\r\n        fn = unreal.FrameNumber(value=f)\r\n        vals.append(lib.get_local_control_rig_euler_transform(sequence, cr, ctrl, fn))\r\n\r\n    ## Unwrap each channel independently\r\n    prev_p = vals[0].rotation.pitch\r\n    prev_y = vals[0].rotation.yaw\r\n    prev_r = vals[0].rotation.roll\r\n\r\n    for i in range(1, len(vals)):\r\n        p = unwrap(vals[i].rotation.pitch, prev_p)\r\n        y = unwrap(vals[i].rotation.yaw,   prev_y)\r\n        r = unwrap(vals[i].rotation.roll,  prev_r)\r\n\r\n        if p != vals[i].rotation.pitch or y != vals[i].rotation.yaw or r != vals[i].rotation.roll:\r\n            fn  = unreal.FrameNumber(value=frames[i])\r\n            new_val = unreal.EulerTransform(\r\n                location=vals[i].location,\r\n                rotation=unreal.Rotator(pitch=p, yaw=y, roll=r),\r\n                scale=vals[i].scale\r\n            )\r\n            lib.set_local_control_rig_euler_transform(sequence, cr, ctrl, fn, new_val)\r\n            vals[i] = new_val\r\n\r\n        prev_p, prev_y, prev_r = p, y, r\r\n\r\n## ── Main ─────────────────────────────────────────────────────────────────────\r\nsequence, cr = get_sequence_and_cr()\r\nif sequence and cr:\r\n    start_f, end_f = get_frame_range(sequence)\r\n    start_fn = unreal.FrameNumber(value=start_f)\r\n\r\n    ## Step 1 – move playhead to start frame\r\n    unreal.LevelSequenceEditorBlueprintLibrary.set_current_time(start_f)\r\n    unreal.log(\"Step 1: Playhead -> frame {}.\".format(start_f))\r\n\r\n    ## Step 2 – set IK / Body controls to World space\r\n    for ctrl in SPACE_CONTROLS:\r\n        unreal.ControlRigSequencerLibrary.set_control_rig_space(\r\n            sequence, cr, ctrl, WORLD_SPACE, start_fn\r\n        )\r\n    unreal.ControlRigSequencerLibrary.set_control_rig_space(\r\n            sequence, cr, ROOT_CTRL, WORLD_SPACE, start_fn\r\n        )\r\n    unreal.log(\"Step 2: IK + Body controls -> World space.\")\r\n\r\n    ## Step 3 – copy root/Root euler transform onto root/Local, all frames\r\n    snap_local_to_root(sequence, cr, start_f, end_f)\r\n    unreal.log(\"Step 3: Copied root/Root -> root/Local ({} frames).\".format(end_f - start_f))\r\n\r\n    ## Step 4 – set IK / Body controls back to Default Parent space\r\n    for ctrl in SPACE_CONTROLS:\r\n        unreal.ControlRigSequencerLibrary.set_control_rig_space(\r\n            sequence, cr, ctrl, PARENT_SPACE, start_fn\r\n        )\r\n    unreal.log(\"Step 4: IK + Body controls -> Default Parent space.\")\r\n\r\n    ## Step 5 – set root/Root to its Local Control space (root/Local)\r\n    unreal.ControlRigSequencerLibrary.set_control_rig_space(\r\n        sequence, cr, ROOT_CTRL, LOCAL_SPACE, start_fn\r\n    )\r\n    unreal.log(\"Step 5: root/Root -> Local Control space (root/Local).\")\r\n\r\n    ## Step 6 – zero out Attach/Attach loc/rot across all frames\r\n    zero_control(sequence, cr, ATTACH_CTRL, start_f, end_f)\r\n    unreal.log(\"Step 6: Attach/Attach zeroed ({} frames).\".format(end_f - start_f))\r\n\r\n    ## Step 7 – euler filter all edited controls to remove rotation flips\r\n    for ctrl in EULER_FILTER_CONTROLS:\r\n        euler_filter(sequence, cr, ctrl, start_f, end_f)\r\n    unreal.log(\"Step 7: Euler filter applied to {} controls.\".format(len(EULER_FILTER_CONTROLS)))\r\n\r\n    unreal.log(\"Done!\")")
   > EndTransaction()
   "On Clicked (BTN_Locator_Bake)"() {is=title K2Node_ComponentBoundEvent}
   > n26 = EUW_MakeLocator(Locator Color=(R=0.1878,G=0.3419,B=0.0999,A=1)) -> (Out_Loc_CtrlName=None, In_Control=None)
   > Execute Python Script!(
    execute
    PythonScript="import unreal\r\n\r\n## inputs injected by Execute Python Script node (always lowercase)\r\n## locator_cr, locator_ctrl, actor_cr, actor_ctrl\r\n\r\nlocator_ctrl_name = str(locator_ctrl)\r\nactor_ctrl_name   = str(actor_ctrl)\r\n\r\n## ── get the current level sequence ───────────────────────────────\r\nls = unreal.LevelSequenceEditorBlueprintLibrary.get_focused_level_sequence()\r\nif not ls:\r\n    unreal.log_error(\"Snap Locator: No level sequence is open.\")\r\n    raise SystemExit\r\n\r\n## ── determine frame range from selected keys ──────────────────────\r\ndef get_selected_key_range():\r\n    \"\"\"Returns (min_frame, max_frame) int tuple if keys are selected, else None.\r\n    Checks the curve editor first, then falls back to the main Sequencer selection.\"\"\"\r\n    all_frames = []\r\n\r\n    ## --- curve editor (preferred - works when curve editor is open) ---\r\n    ses = unreal.get_editor_subsystem(unreal.LevelSequenceEditorSubsystem)\r\n    curve_editor = ses.get_curve_editor()\r\n    if curve_editor.is_curve_editor_open():\r\n        ce_channels = curve_editor.get_channels_with_selected_keys()\r\n        for ch_proxy in ce_channels:\r\n            selected_indices = curve_editor.get_selected_keys(ch_proxy)\r\n            if not selected_indices:\r\n                continue\r\n            channel_obj = ch_proxy.section.get_channel(ch_proxy.channel_name)\r\n            if not channel_obj:\r\n                continue\r\n            key_objs = channel_obj.get_keys_by_index(selected_indices)\r\n            for k in key_objs:\r\n                all_frames.append(k.get_time().frame_number.value)\r\n\r\n    ## --- main sequencer fallback ---\r\n    if not all_frames:\r\n        keyed_channels = unreal.LevelSequenceEditorBlueprintLibrary.get_channels_with_selected_keys()\r\n        for ch_proxy in keyed_channels:\r\n            selected_indices = unreal.LevelSequenceEditorBlueprintLibrary.get_selected_keys(ch_proxy)\r\n            if not selected_indices:\r\n                continue\r\n            channel_obj = ch_proxy.section.get_channel(ch_proxy.channel_name)\r\n            if not channel_obj:\r\n                continue\r\n            key_objs = channel_obj.get_keys_by_index(selected_indices)\r\n            for k in key_objs:\r\n                all_frames.append(k.get_time().frame_number.value)\r\n\r\n    if not all_frames:\r\n        return None\r\n    return (min(all_frames), max(all_frames))\r\n\r\nselected_range = get_selected_key_range()\r\n\r\n## ── pop-up: ask All Frames or Keys Only ──────────────────────────\r\nmsg_lines = [f\"Snap '{locator_ctrl_name}' to '{actor_ctrl_name}'.\", \"\"]\r\nif selected_range:\r\n    msg_lines.append(\r\n        f\"Selected keys detected: frames {selected_range[0]} to {selected_range[1]}.\"\r\n    )\r\n    msg_lines.append(\"Keys Only will snap only on those frames.\")\r\nelse:\r\n    pb_start = ls.get_playback_start()\r\n    pb_end   = ls.get_playback_end() - 1\r\n    msg_lines.append(\r\n        f\"No keys selected. Range: sequence playback {pb_start} to {pb_end}.\"\r\n    )\r\n    msg_lines.append(\"Keys Only will bake only on existing key frames in that range.\")\r\n\r\nmsg_lines += [\"\", \"YES = All Frames    NO = Keys Only    Cancel = Abort\"]\r\nmessage = \"\\n\".join(msg_lines)\r\n\r\nresult = unreal.EditorDialog.show_message(\r\n    \"Snap Locator to Actor\",\r\n    message,\r\n    unreal.AppMsgType.YES_NO_CANCEL,\r\n    unreal.AppReturnType.NO,\r\n)\r\n\r\nif result == unreal.AppReturnType.CANCEL:\r\n    unreal.log(\"Snap Locator: Cancelled by user.\")\r\n    raise SystemExit\r\n\r\nkeys_only = (result == unreal.AppReturnType.NO)\r\n\r\n## ── resolve frame range ───────────────────────────────────────────\r\nif selected_range:\r\n    start_frame = unreal.FrameNumber(selected_range[0])\r\n    end_frame   = unreal.FrameNumber(selected_range[1])\r\nelse:\r\n    start_frame = unreal.FrameNumber(ls.get_playback_start())\r\n    end_frame   = unreal.FrameNumber(ls.get_playback_end() - 1)\r\n\r\n## ── build snapper selections ──────────────────────────────────────\r\n## Parent (driver): actor_ctrl on actor_cr\r\nparent_cr_entry = unreal.ControlRigForWorldTransforms()\r\nparent_cr_entry.control_rig   = actor_cr\r\nparent_cr_entry.control_names = [actor_ctrl_name]\r\n\r\nparent_sel = unreal.ControlRigSnapperSelection()\r\nparent_sel.control_rigs = [parent_cr_entry]\r\n\r\n## Child (snapped): locator_ctrl on locator_cr\r\nchild_cr_entry = unreal.ControlRigForWorldTransforms()\r\nchild_cr_entry.control_rig   = locator_cr\r\nchild_cr_entry.control_names = [locator_ctrl_name]\r\n\r\nchild_sel = unreal.ControlRigSnapperSelection()\r\nchild_sel.control_rigs = [child_cr_entry]\r\n\r\n## ── snap settings ─────────────────────────────────────────────────\r\nsnap_settings = unreal.ControlRigSnapSettings()\r\nsnap_settings.keep_offset         = False\r\nsnap_settings.snap_position       = True\r\nsnap_settings.snap_rotation       = True\r\nsnap_settings.snap_scale          = False\r\nsnap_settings.baking_key_settings = (\r\n    unreal.BakingKeySettings.KEYS_ONLY\r\n    if keys_only\r\n    else unreal.BakingKeySettings.ALL_FRAMES\r\n)\r\n\r\n## ── execute snap ──────────────────────────────────────────────────\r\nunreal.ControlRigSequencerLibrary.snap_control_rig(\r\n    ls, start_frame, end_frame, child_sel, parent_sel, snap_settings\r\n)\r\n\r\nmode_str  = \"keys only\" if keys_only else \"all frames\"\r\nrange_str = f\"{start_frame.value} to {end_frame.value}\"\r\nunreal.log(f\"Snap Locator: snapped '{locator_ctrl_name}' to '{actor_ctrl_name}' [{range_str}] ({mode_str}).\")\r\n"
    Locator_CR:Control Rig Object Reference=@n26
    Locator_Ctrl:Name=@n26.Out_Loc_CtrlName
    Actor_CR:Control Rig Object Reference=@n26.In_CR
    Actor_Ctrl:Name=@n26.In_Control
   ) -> (then, ReturnValue) {inputs [Locator_CR,Locator_Ctrl,Actor_CR,Actor_Ctrl]}
   "On Clicked (BTN_Locator_Bake_Connect)"() {is=title K2Node_ComponentBoundEvent}
   > n27 = EUW_MakeLocator(Locator Color=(R=0.1878,G=0.3419,B=0.0999,A=1)) -> (Out_Loc_CtrlName=None, In_Control=None)
   > Execute Python Script(
    PythonScript="import unreal\r\n\r\n## inputs injected by Execute Python Script node (always lowercase)\r\n## locator_cr, locator_ctrl, actor_cr, actor_ctrl\r\n\r\nlocator_ctrl_name = str(locator_ctrl)\r\nactor_ctrl_name   = str(actor_ctrl)\r\n\r\n## ── get the current level sequence ───────────────────────────────\r\nls = unreal.LevelSequenceEditorBlueprintLibrary.get_focused_level_sequence()\r\nif not ls:\r\n    unreal.log_error(\"Snap Locator: No level sequence is open.\")\r\n    raise SystemExit\r\n\r\n## ── determine frame range from selected keys ──────────────────────\r\ndef get_selected_key_range():\r\n    \"\"\"Returns (min_frame, max_frame) int tuple if keys are selected, else None.\r\n    Checks the curve editor first, then falls back to the main Sequencer selection.\"\"\"\r\n    all_frames = []\r\n\r\n    ## --- curve editor (preferred - works when curve editor is open) ---\r\n    ses = unreal.get_editor_subsystem(unreal.LevelSequenceEditorSubsystem)\r\n    curve_editor = ses.get_curve_editor()\r\n    if curve_editor.is_curve_editor_open():\r\n        ce_channels = curve_editor.get_channels_with_selected_keys()\r\n        for ch_proxy in ce_channels:\r\n            selected_indices = curve_editor.get_selected_keys(ch_proxy)\r\n            if not selected_indices:\r\n                continue\r\n            channel_obj = ch_proxy.section.get_channel(ch_proxy.channel_name)\r\n            if not channel_obj:\r\n                continue\r\n            key_objs = channel_obj.get_keys_by_index(selected_indices)\r\n            for k in key_objs:\r\n                all_frames.append(k.get_time().frame_number.value)\r\n\r\n    ## --- main sequencer fallback ---\r\n    if not all_frames:\r\n        keyed_channels = unreal.LevelSequenceEditorBlueprintLibrary.get_channels_with_selected_keys()\r\n        for ch_proxy in keyed_channels:\r\n            selected_indices = unreal.LevelSequenceEditorBlueprintLibrary.get_selected_keys(ch_proxy)\r\n            if not selected_indices:\r\n                continue\r\n            channel_obj = ch_proxy.section.get_channel(ch_proxy.channel_name)\r\n            if not channel_obj:\r\n                continue\r\n            key_objs = channel_obj.get_keys_by_index(selected_indices)\r\n            for k in key_objs:\r\n                all_frames.append(k.get_time().frame_number.value)\r\n\r\n    if not all_frames:\r\n        return None\r\n    return (min(all_frames), max(all_frames))\r\n\r\nselected_range = get_selected_key_range()\r\n\r\n## ── pop-up: ask All Frames or Keys Only ──────────────────────────\r\nmsg_lines = [f\"Snap '{locator_ctrl_name}' to '{actor_ctrl_name}'.\", \"\"]\r\nif selected_range:\r\n    msg_lines.append(\r\n        f\"Selected keys detected: frames {selected_range[0]} to {selected_range[1]}.\"\r\n    )\r\n    msg_lines.append(\"Keys Only will snap only on those frames.\")\r\nelse:\r\n    pb_start = ls.get_playback_start()\r\n    pb_end   = ls.get_playback_end() - 1\r\n    msg_lines.append(\r\n        f\"No keys selected. Range: sequence playback {pb_start} to {pb_end}.\"\r\n    )\r\n    msg_lines.append(\"Keys Only will bake only on existing key frames in that range.\")\r\n\r\nmsg_lines += [\"\", \"YES = All Frames    NO = Keys Only    Cancel = Abort\"]\r\nmessage = \"\\n\".join(msg_lines)\r\n\r\nresult = unreal.EditorDialog.show_message(\r\n    \"Snap Locator to Actor\",\r\n    message,\r\n    unreal.AppMsgType.YES_NO_CANCEL,\r\n    unreal.AppReturnType.NO,\r\n)\r\n\r\nif result == unreal.AppReturnType.CANCEL:\r\n    unreal.log(\"Snap Locator: Cancelled by user.\")\r\n    raise SystemExit\r\n\r\nkeys_only = (result == unreal.AppReturnType.NO)\r\n\r\n## ── resolve frame range ───────────────────────────────────────────\r\nif selected_range:\r\n    start_frame = unreal.FrameNumber(selected_range[0])\r\n    end_frame   = unreal.FrameNumber(selected_range[1])\r\nelse:\r\n    start_frame = unreal.FrameNumber(ls.get_playback_start())\r\n    end_frame   = unreal.FrameNumber(ls.get_playback_end() - 1)\r\n\r\n## ── build snapper selections ──────────────────────────────────────\r\n## Parent (driver): actor_ctrl on actor_cr\r\nparent_cr_entry = unreal.ControlRigForWorldTransforms()\r\nparent_cr_entry.control_rig   = actor_cr\r\nparent_cr_entry.control_names = [actor_ctrl_name]\r\n\r\nparent_sel = unreal.ControlRigSnapperSelection()\r\nparent_sel.control_rigs = [parent_cr_entry]\r\n\r\n## Child (snapped): locator_ctrl on locator_cr\r\nchild_cr_entry = unreal.ControlRigForWorldTransforms()\r\nchild_cr_entry.control_rig   = locator_cr\r\nchild_cr_entry.control_names = [locator_ctrl_name]\r\n\r\nchild_sel = unreal.ControlRigSnapperSelection()\r\nchild_sel.control_rigs = [child_cr_entry]\r\n\r\n## ── snap settings ─────────────────────────────────────────────────\r\nsnap_settings = unreal.ControlRigSnapSettings()\r\nsnap_settings.keep_offset         = False\r\nsnap_settings.snap_position       = True\r\nsnap_settings.snap_rotation       = True\r\nsnap_settings.snap_scale          = False\r\nsnap_settings.baking_key_settings = (\r\n    unreal.BakingKeySettings.KEYS_ONLY\r\n    if keys_only\r\n    else unreal.BakingKeySettings.ALL_FRAMES\r\n)\r\n\r\n## ── execute snap ──────────────────────────────────────────────────\r\nunreal.ControlRigSequencerLibrary.snap_control_rig(\r\n    ls, start_frame, end_frame, child_sel, parent_sel, snap_settings\r\n)\r\n\r\nmode_str  = \"keys only\" if keys_only else \"all frames\"\r\nrange_str = f\"{start_frame.value} to {end_frame.value}\"\r\nunreal.log(f\"Snap Locator: snapped '{locator_ctrl_name}' to '{actor_ctrl_name}' [{range_str}] ({mode_str}).\")\r\n"
    Locator_CR=@n27
    Locator_Ctrl=@n27.Out_Loc_CtrlName
    Actor_CR=@n27.In_CR
    Actor_Ctrl=@n27.In_Control
   ) {inputs [Locator_CR,Locator_Ctrl,Actor_CR,Actor_Ctrl]}
   > Execute Python Script(
    PythonScript="import unreal\r\n\r\n## inputs injected by Execute Python Script node (always lowercase)\r\n## locator_cr, locator_ctrl, actor_cr, actor_ctrl\r\n\r\nlocator_ctrl_name = str(locator_ctrl)\r\nactor_ctrl_name   = str(actor_ctrl)\r\n\r\n## ── get level sequence and world ──────────────────────────────────\r\nls = unreal.LevelSequenceEditorBlueprintLibrary.get_focused_level_sequence()\r\nif not ls:\r\n    unreal.log_error(\"Constrain Actor to Locator: No level sequence is open.\")\r\n    raise SystemExit\r\n\r\neditor_system = unreal.get_editor_subsystem(unreal.UnrealEditorSubsystem)\r\nworld = editor_system.get_editor_world()\r\n\r\n## ── re-detect selected key range (same logic as snap script) ──────\r\ndef get_selected_key_range():\r\n    all_frames = []\r\n    ses = unreal.get_editor_subsystem(unreal.LevelSequenceEditorSubsystem)\r\n    curve_editor = ses.get_curve_editor()\r\n    if curve_editor.is_curve_editor_open():\r\n        for ch_proxy in curve_editor.get_channels_with_selected_keys():\r\n            indices = curve_editor.get_selected_keys(ch_proxy)\r\n            if not indices:\r\n                continue\r\n            channel_obj = ch_proxy.section.get_channel(ch_proxy.channel_name)\r\n            if not channel_obj:\r\n                continue\r\n            for k in channel_obj.get_keys_by_index(indices):\r\n                all_frames.append(k.get_time().frame_number.value)\r\n    if not all_frames:\r\n        for ch_proxy in unreal.LevelSequenceEditorBlueprintLibrary.get_channels_with_selected_keys():\r\n            indices = unreal.LevelSequenceEditorBlueprintLibrary.get_selected_keys(ch_proxy)\r\n            if not indices:\r\n                continue\r\n            channel_obj = ch_proxy.section.get_channel(ch_proxy.channel_name)\r\n            if not channel_obj:\r\n                continue\r\n            for k in channel_obj.get_keys_by_index(indices):\r\n                all_frames.append(k.get_time().frame_number.value)\r\n    if not all_frames:\r\n        return None\r\n    return (min(all_frames), max(all_frames))\r\n\r\nselected_range = get_selected_key_range()\r\nhas_range = selected_range is not None\r\n\r\n## ── resolve start/end frames ──────────────────────────────────────\r\nif has_range:\r\n    start_val = selected_range[0]\r\n    end_val   = selected_range[1]\r\nelse:\r\n    ## no key selection: use the min/max of existing keys on actor_ctrl\r\n    ## fall back to sequence playback range if no keys found\r\n    start_val = ls.get_playback_start()\r\n    end_val   = ls.get_playback_end() - 1\r\n\r\n    ## try to find tighter range from actual keys on actor_ctrl\r\n    rigs = unreal.ControlRigSequencerLibrary.get_control_rigs(ls)\r\n    for proxy in rigs:\r\n        if proxy.control_rig == actor_cr:\r\n            rig_track = proxy.track\r\n            section = rig_track.get_section_to_key()\r\n            if section:\r\n                ## sample one channel to get actual key range\r\n                channel_name = f\"{actor_ctrl_name}.Location.X\"\r\n                channel_obj = section.get_channel(channel_name)\r\n                if channel_obj:\r\n                    keys = channel_obj.get_keys()\r\n                    if keys:\r\n                        frames = [k.get_time().frame_number.value for k in keys]\r\n                        start_val = min(frames)\r\n                        end_val   = max(frames)\r\n            break\r\n\r\nstart_frame = unreal.FrameNumber(start_val)\r\nend_frame   = unreal.FrameNumber(end_val)\r\n\r\n## ── jump to start frame ───────────────────────────────────────────\r\nunreal.LevelSequenceEditorBlueprintLibrary.set_current_time(start_val)\r\n\r\n## ── build handles ─────────────────────────────────────────────────\r\n## child = actor_ctrl (the one being constrained)\r\n## parent = locator_ctrl (the driver)\r\nchild_handle  = actor_cr.create_transformable_control_handle(actor_ctrl_name)\r\nparent_handle = locator_cr.create_transformable_control_handle(locator_ctrl_name)\r\n\r\nif not child_handle or not parent_handle:\r\n    unreal.log_error(\"Constrain Actor to Locator: Failed to create control handles.\")\r\n    raise SystemExit\r\n\r\n## ── create parent constraint ──────────────────────────────────────\r\nconstraint = unreal.ControlRigSequencerLibrary.add_constraint(\r\n    world,\r\n    unreal.TransformConstraintType.PARENT,\r\n    child_handle,\r\n    parent_handle,\r\n    False  ## maintain_offset = False, snap to locator\r\n)\r\n\r\nif not constraint:\r\n    unreal.log_error(\"Constrain Actor to Locator: Failed to create constraint.\")\r\n    raise SystemExit\r\n\r\nunreal.log(f\"Constrain: created parent constraint '{actor_ctrl_name}' -> '{locator_ctrl_name}'\")\r\n\r\n## ── key the constraint active/inactive based on range mode ────────\r\nif has_range:\r\n    ## had a key selection: turn ON at start_frame, OFF at end_frame\r\n    unreal.ControlRigSequencerLibrary.set_constraint_active_key(\r\n        constraint, True, start_frame\r\n    )\r\n    unreal.ControlRigSequencerLibrary.set_constraint_active_key(\r\n        constraint, False, end_frame\r\n    )\r\n    unreal.log(\r\n        f\"Constrain: keyed ON at frame {start_val}, OFF at frame {end_val}.\"\r\n    )\r\nelse:\r\n    ## no key selection: constraint is just on from start, no off key\r\n    unreal.ControlRigSequencerLibrary.set_constraint_active_key(\r\n        constraint, True, start_frame\r\n    )\r\n    unreal.log(\r\n        f\"Constrain: keyed ON at frame {start_val} (no range, no off key).\"\r\n    )\r\n"
    Locator_CR=@n27
    Locator_Ctrl=@n27.Out_Loc_CtrlName
    Actor_CR=@n27.In_CR
    Actor_Ctrl=@n27.In_Control
   ) {inputs [Locator_CR,Locator_Ctrl,Actor_CR,Actor_Ctrl]}
   "On Clicked (Btn_IKSpine_BodyAim)"() {is=title K2Node_ComponentBoundEvent}
   > BeginTransaction(Description="LOCTEXT(\"IK Spine - BodyAim\")")
   > n28 = Sequence!(execute) -> (then_0, then_1, then_2:Exec, then_3:Exec)
   Execute Python Script!(
    execute=[@n29.then_0, @n28.then_0, @n33.then_0]
    PythonScript="import unreal\n\n## -----------------------------------------------------------------------\n## Rig profiles — detection is done via CR section channel names\n## -----------------------------------------------------------------------\nRIG_PROFILES = {\n    \"MR_Biped_FN_v01\": {\n        \"detect_control\": \"Spine/FK 1\",\n        \"ik_controls\":    [\"Spine/Start\",    \"Spine/Mid\",    \"Spine/End\"],\n        \"fk_controls\":    [\"Spine/FK 1\",     \"Spine/FK 2\",   \"Spine/FK 3\"],\n    },\n    \"MR_Biped_v02\": {\n        \"detect_control\": \"Spine/Start FK\",\n        \"ik_controls\":    [\"Spine/Start IK\", \"Spine/Mid IK\", \"Spine/End IK\"],\n        \"fk_controls\":    [\"Spine/Start FK\", \"Spine/Mid FK\", \"Spine/End FK\"],\n    },\n}\n\ndef detect_rig_profile(cr_section):\n    all_channels = cr_section.get_all_channels()\n    channel_names = {str(ch.channel_name) for ch in all_channels}\n    for profile_name, profile in RIG_PROFILES.items():\n        probe = profile[\"detect_control\"] + \".\"\n        if any(n.startswith(probe) for n in channel_names):\n            unreal.log(f\"[SpineCleaner] Detected rig profile: {profile_name}\")\n            return profile\n    return None\n\ndef get_cr_section(level_seq):\n    for binding in level_seq.get_bindings():\n        for track in binding.get_tracks():\n            for section in track.get_sections():\n                if \"ControlRig\" in section.get_class().get_name():\n                    return section\n    return None\n\ndef euler_filter_channel(ch):\n    keys = ch.get_keys()\n    if len(keys) < 2:\n        return 0\n    fixed = 0\n    prev_val = keys[0].get_value()\n    for key in keys[1:]:\n        val = key.get_value()\n        diff = val - prev_val\n        if abs(diff) > 180.0:\n            offset = round(diff / 360.0) * 360.0\n            val -= offset\n            key.set_value(val)\n            fixed += 1\n        prev_val = val\n    return fixed\n\n## -----------------------------------------------------------------------\n## Main\n## -----------------------------------------------------------------------\n\nlevel_seq = unreal.LevelSequenceEditorBlueprintLibrary.get_current_level_sequence()\nif not level_seq:\n    unreal.log_error(\"[SpineCleaner] No Level Sequence is open.\")\nelse:\n    proxies = unreal.ControlRigSequencerLibrary.get_control_rigs(level_seq)\n    if not proxies:\n        unreal.log_error(\"[SpineCleaner] No Control Rig found in the current Level Sequence.\")\n    else:\n        rig = proxies[0].control_rig\n\n        ## Force playhead to the start of the sequence\n        start_frame = level_seq.get_playback_start()\n        unreal.LevelSequenceEditorBlueprintLibrary.set_current_time(start_frame)\n        unreal.log(f\"[SpineCleaner] Seeked to start frame: {start_frame}\")\n        frame0 = unreal.FrameNumber(value=start_frame)\n\n        cr_section = get_cr_section(level_seq)\n        if not cr_section:\n            unreal.log_error(\"[SpineCleaner] Could not find ControlRig section.\")\n        else:\n            profile = detect_rig_profile(cr_section)\n            if not profile:\n                unreal.log_error(\"[SpineCleaner] Unknown rig — cannot determine spine control names. Aborting.\")\n            else:\n                IK_CONTROLS = profile[\"ik_controls\"]\n                FK_CONTROLS = profile[\"fk_controls\"]\n\n                ## Step 1 - Set IK spine controls to World Space\n                world_space = unreal.ControlRigSequencerLibrary.get_world_space_reference_key()\n                for ctrl in IK_CONTROLS:\n                    try:\n                        unreal.ControlRigSequencerLibrary.set_control_rig_space(\n                            level_seq, rig, ctrl, world_space, frame0)\n                        unreal.log(f\"[SpineCleaner] Set world space: {ctrl}\")\n                    except Exception as e:\n                        unreal.log_warning(f\"[SpineCleaner] World space failed for {ctrl}: {e}\")\n\n                ## Re-fetch channels after space switch bakes new IK keys\n                all_channels = cr_section.get_all_channels()\n\n                ## Step 2 - Delete all FK spine keys\n                for ctrl in FK_CONTROLS:\n                    ctrl_channels = [ch for ch in all_channels\n                                     if str(ch.channel_name) == ctrl or\n                                        str(ch.channel_name).startswith(ctrl + \".\")]\n                    total_keys = 0\n                    for ch in ctrl_channels:\n                        keys = ch.get_keys()\n                        total_keys += len(keys)\n                        for key in keys:\n                            ch.remove_key(key)\n                    unreal.log(f\"[SpineCleaner] Cleared {total_keys} keys across \"\n                               f\"{len(ctrl_channels)} channels for: {ctrl}\")\n\n                ## Step 3 - Restore IK spine controls to Parent Space\n                parent_space = unreal.ControlRigSequencerLibrary.get_default_parent_key()\n                for ctrl in IK_CONTROLS:\n                    try:\n                        unreal.ControlRigSequencerLibrary.set_control_rig_space(\n                            level_seq, rig, ctrl, parent_space, frame0)\n                        unreal.log(f\"[SpineCleaner] Restored parent space: {ctrl}\")\n                    except Exception as e:\n                        unreal.log_warning(f\"[SpineCleaner] Parent space restore failed for {ctrl}: {e}\")\n\n                ## Step 4 - Key FK controls at bind pose (identity)\n                identity_euler = unreal.EulerTransform(\n                    location=unreal.Vector(0, 0, 0),\n                    rotation=unreal.Rotator(0, 0, 0),\n                    scale=unreal.Vector(1, 1, 1)\n                )\n                for ctrl in FK_CONTROLS:\n                    try:\n                        unreal.ControlRigSequencerLibrary.set_local_control_rig_euler_transform(\n                            level_seq, rig, ctrl, frame0, identity_euler, set_key=True)\n                        unreal.log(f\"[SpineCleaner] Keyed bind pose: {ctrl}\")\n                    except Exception as e:\n                        unreal.log_warning(f\"[SpineCleaner] Bind pose key failed for {ctrl}: {e}\")\n\n                ## Step 5 - Euler filter on all spine controls\n                ALL_SPINE = IK_CONTROLS + FK_CONTROLS\n                all_channels = cr_section.get_all_channels()\n                total_fixed = 0\n                for ctrl in ALL_SPINE:\n                    rot_channels = [ch for ch in all_channels\n                                    if str(ch.channel_name).startswith(ctrl + \".Rotation.\")]\n                    for ch in rot_channels:\n                        fixed = euler_filter_channel(ch)\n                        if fixed:\n                            unreal.log(f\"[SpineCleaner] Euler filter fixed {fixed} keys on {ch.channel_name}\")\n                        total_fixed += fixed\n                unreal.log(f\"[SpineCleaner] Euler filter complete — {total_fixed} keys adjusted.\")\n\n                ## Step 6 - Clear FK Vis channel keys and set to False\n                all_channels = cr_section.get_all_channels()\n                fk_vis_channels = [ch for ch in all_channels\n                                   if \"Spine\" in str(ch.channel_name) and\n                                      \"FK Vis\" in str(ch.channel_name)]\n                if fk_vis_channels:\n                    for ch in fk_vis_channels:\n                        for key in ch.get_keys():\n                            ch.remove_key(key)\n                        try:\n                            ch.set_default(False)\n                            unreal.log(f\"[SpineCleaner] Cleared keys and set False: {ch.channel_name}\")\n                        except Exception as e:\n                            unreal.log_warning(f\"[SpineCleaner] set_default failed for {ch.channel_name}: {e}\")\n                else:\n                    unreal.log_warning(\"[SpineCleaner] No Spine FK Vis channel found — check channel name.\")\n\n                ## Step 7 - Clear IK Vis channel keys and set to True\n                ik_vis_channels = [ch for ch in all_channels\n                                   if str(ch.channel_name) == \"Spine/IK Vis\"]\n                if ik_vis_channels:\n                    for ch in ik_vis_channels:\n                        for key in ch.get_keys():\n                            ch.remove_key(key)\n                        try:\n                            ch.set_default(True)\n                            unreal.log(f\"[SpineCleaner] Cleared keys and set True: {ch.channel_name}\")\n                        except Exception as e:\n                            unreal.log_warning(f\"[SpineCleaner] set_default failed for {ch.channel_name}: {e}\")\n                else:\n                    unreal.log_warning(\"[SpineCleaner] No Spine/IK Vis channel found — check channel name.\")\n\n                unreal.log(\"[SpineCleaner] Done!\")\n"
   ) -> (then, ReturnValue)
   Execute Python Script(
    execute=@n28.then_1
    PythonScript="import unreal\r\n\r\n## -----------------------------------------------------------------------\r\n## Rig detection\r\n## -----------------------------------------------------------------------\r\ndef detect_rig_version(cr_section):\r\n    all_channels = cr_section.get_all_channels()\r\n    channel_names = {str(ch.channel_name) for ch in all_channels}\r\n    if any(n.startswith(\"Spine/Start FK.\") for n in channel_names):\r\n        return \"v02\"\r\n    if any(n.startswith(\"Spine/FK 1.\") for n in channel_names):\r\n        return \"v01\"\r\n    return None\r\n\r\ndef get_cr_section(level_seq):\r\n    for binding in level_seq.get_bindings():\r\n        for track in binding.get_tracks():\r\n            for section in track.get_sections():\r\n                if \"ControlRig\" in section.get_class().get_name():\r\n                    return section\r\n    return None\r\n\r\ndef get_channel(all_channels, name):\r\n    for ch in all_channels:\r\n        if str(ch.channel_name) == name:\r\n            return ch\r\n    return None\r\n\r\ndef snap_control(level_seq, rig, source_ctrl, target_ctrl, snap_pos, snap_rot, start_fn, end_fn):\r\n    \"\"\"Snap target_ctrl to source_ctrl across the full frame range.\"\"\"\r\n    source_sel = unreal.ControlRigSnapperSelection()\r\n    source_cr  = unreal.ControlRigForWorldTransforms()\r\n    source_cr.control_rig   = rig\r\n    source_cr.control_names = [source_ctrl]\r\n    source_sel.control_rigs = [source_cr]\r\n\r\n    target_sel = unreal.ControlRigSnapperSelection()\r\n    target_cr  = unreal.ControlRigForWorldTransforms()\r\n    target_cr.control_rig   = rig\r\n    target_cr.control_names = [target_ctrl]\r\n    target_sel.control_rigs = [target_cr]\r\n\r\n    settings = unreal.ControlRigSnapSettings()\r\n    settings.snap_position = snap_pos\r\n    settings.snap_rotation = snap_rot\r\n    settings.snap_scale    = False\r\n    settings.keep_offset   = False\r\n\r\n    unreal.ControlRigSequencerLibrary.snap_control_rig(\r\n        level_seq, start_fn, end_fn, target_sel, source_sel, settings)\r\n\r\n## -----------------------------------------------------------------------\r\n## Main\r\n## -----------------------------------------------------------------------\r\n\r\nlevel_seq = unreal.LevelSequenceEditorBlueprintLibrary.get_current_level_sequence()\r\nif not level_seq:\r\n    unreal.log_error(\"[SpineSetup01] No Level Sequence is open.\")\r\nelse:\r\n    proxies = unreal.ControlRigSequencerLibrary.get_control_rigs(level_seq)\r\n    if not proxies:\r\n        unreal.log_error(\"[SpineSetup01] No Control Rig found in the current Level Sequence.\")\r\n    else:\r\n        rig = proxies[0].control_rig\r\n        cr_section = get_cr_section(level_seq)\r\n        if not cr_section:\r\n            unreal.log_error(\"[SpineSetup01] Could not find ControlRig section.\")\r\n        else:\r\n            rig_ver = detect_rig_version(cr_section)\r\n            if rig_ver is None:\r\n                unreal.log_error(\"[SpineSetup01] Unknown rig — cannot determine version. Aborting.\")\r\n\r\n            elif rig_ver == \"v01\":\r\n                ## -------------------------------------------------------\r\n                ## MR_Biped_v01 — original behaviour, untouched\r\n                ## -------------------------------------------------------\r\n                start_frame = level_seq.get_playback_start()\r\n                frame0 = unreal.FrameNumber(value=start_frame)\r\n                local_orient_space = unreal.RigElementKey(\r\n                    type=unreal.RigElementType.NULL,\r\n                    name=\"Spine/End Local Orient Space\"\r\n                )\r\n                try:\r\n                    unreal.ControlRigSequencerLibrary.set_control_rig_space(\r\n                        level_seq, rig, \"Spine/IK End\", local_orient_space, frame0)\r\n                    unreal.log(\"[SpineSetup01] (v01) Set Spine/IK End to Spine/End Local Orient Space.\")\r\n                except Exception as e:\r\n                    unreal.log_error(f\"[SpineSetup01] Space switch failed: {e}\")\r\n\r\n            else:\r\n                ## -------------------------------------------------------\r\n                ## MR_Biped_v02 — full body aim setup\r\n                ## -------------------------------------------------------\r\n\r\n                SPINE_IK_CONTROLS = [\"Spine/Start IK\", \"Spine/Mid IK\", \"Spine/End IK\"]\r\n\r\n                ## Step 1 — Go to frame 0\r\n                start_frame = level_seq.get_playback_start()\r\n                frame0      = unreal.FrameNumber(value=start_frame)\r\n                unreal.LevelSequenceEditorBlueprintLibrary.set_current_time(start_frame)\r\n                unreal.log(f\"[SpineSetup01] Seeked to frame: {start_frame}\")\r\n\r\n                end_frame = level_seq.get_playback_end()\r\n                start_fn  = unreal.FrameNumber(value=start_frame)\r\n                end_fn    = unreal.FrameNumber(value=end_frame)\r\n\r\n                ## Step 2 — Move Spine IK controls to World Space\r\n                world_space  = unreal.ControlRigSequencerLibrary.get_world_space_reference_key()\r\n                parent_space = unreal.ControlRigSequencerLibrary.get_default_parent_key()\r\n\r\n                for ctrl in SPINE_IK_CONTROLS:\r\n                    try:\r\n                        unreal.ControlRigSequencerLibrary.set_control_rig_space(\r\n                            level_seq, rig, ctrl, world_space, frame0)\r\n                        unreal.log(f\"[SpineSetup01] Set World Space: {ctrl}\")\r\n                    except Exception as e:\r\n                        unreal.log_warning(f\"[SpineSetup01] World Space failed for {ctrl}: {e}\")\r\n\r\n                ## Re-fetch channels after space switch bakes new keys\r\n                all_channels = cr_section.get_all_channels()\r\n\r\n                ## Step 3 — Delete all keys on Body/Body Aim Vis and Body/Aim Weight,\r\n                ##           then set Body/Body Aim Vis = True, Body/Aim Weight = 1\r\n                body_aim_vis_ch = get_channel(all_channels, \"Body/Body Aim Vis\")\r\n                body_aim_wt_ch  = get_channel(all_channels, \"Body/Aim Weight\")\r\n\r\n                for ch, label in [(body_aim_vis_ch, \"Body/Body Aim Vis\"), (body_aim_wt_ch, \"Body/Aim Weight\")]:\r\n                    if ch:\r\n                        for key in ch.get_keys():\r\n                            ch.remove_key(key)\r\n                        unreal.log(f\"[SpineSetup01] Cleared keys on {label}\")\r\n                    else:\r\n                        unreal.log_warning(f\"[SpineSetup01] Channel not found: {label}\")\r\n\r\n                if body_aim_vis_ch:\r\n                    try:\r\n                        body_aim_vis_ch.set_default(True)\r\n                        unreal.log(\"[SpineSetup01] Body/Body Aim Vis set to True\")\r\n                    except Exception as e:\r\n                        unreal.log_warning(f\"[SpineSetup01] set_default failed for Body/Body Aim Vis: {e}\")\r\n\r\n                if body_aim_wt_ch:\r\n                    try:\r\n                        body_aim_wt_ch.set_default(1.0)\r\n                        unreal.log(\"[SpineSetup01] Body/Aim Weight set to 1.0\")\r\n                    except Exception as e:\r\n                        unreal.log_warning(f\"[SpineSetup01] set_default failed for Body/Aim Weight: {e}\")\r\n\r\n                ## Step 4 — Snap Body/Body Aim to Spine/End IK (Position only, no offset, all frames)\r\n                try:\r\n                    snap_control(level_seq, rig, \"Spine/End IK\", \"Body/Body Aim\",\r\n                                 snap_pos=True, snap_rot=False,\r\n                                 start_fn=start_fn, end_fn=end_fn)\r\n                    unreal.log(\"[SpineSetup01] Snapped Body/Body Aim → Spine/End IK (Position).\")\r\n                except Exception as e:\r\n                    unreal.log_warning(f\"[SpineSetup01] Snap (position) failed: {e}\")\r\n\r\n                ## Step 5 — Snap Body/Body Aim to Spine/Start IK (Rotation only, no offset, all frames)\r\n                try:\r\n                    snap_control(level_seq, rig, \"Spine/Start IK\", \"Body/Body Aim\",\r\n                                 snap_pos=False, snap_rot=True,\r\n                                 start_fn=start_fn, end_fn=end_fn)\r\n                    unreal.log(\"[SpineSetup01] Snapped Body/Body Aim → Spine/Start IK (Rotation).\")\r\n                except Exception as e:\r\n                    unreal.log_warning(f\"[SpineSetup01] Snap (rotation) failed: {e}\")\r\n\r\n                ## Step 6 — Set Body/Body control to World Space\r\n                try:\r\n                    unreal.ControlRigSequencerLibrary.set_control_rig_space(\r\n                        level_seq, rig, \"Body/Body\", world_space, frame0)\r\n                    unreal.log(\"[SpineSetup01] Set Body/Body to World Space.\")\r\n                except Exception as e:\r\n                    unreal.log_warning(f\"[SpineSetup01] World Space failed for Body/Body: {e}\")\r\n\r\n                ## Re-fetch channels after bake\r\n                all_channels = cr_section.get_all_channels()\r\n\r\n                ## Step 7 — Reset: Body/Body Aim Vis = False, Body/Aim Weight = 0, Body/Aim Twist = 0\r\n                body_aim_vis_ch = get_channel(all_channels, \"Body/Body Aim Vis\")\r\n                body_aim_wt_ch  = get_channel(all_channels, \"Body/Aim Weight\")\r\n                body_aim_tw_ch  = get_channel(all_channels, \"Body/Aim Twist\")\r\n\r\n                if body_aim_vis_ch:\r\n                    for key in body_aim_vis_ch.get_keys():\r\n                        body_aim_vis_ch.remove_key(key)\r\n                    try:\r\n                        body_aim_vis_ch.set_default(False)\r\n                        unreal.log(\"[SpineSetup01] Body/Body Aim Vis set to False\")\r\n                    except Exception as e:\r\n                        unreal.log_warning(f\"[SpineSetup01] set_default failed for Body/Body Aim Vis: {e}\")\r\n\r\n                if body_aim_wt_ch:\r\n                    for key in body_aim_wt_ch.get_keys():\r\n                        body_aim_wt_ch.remove_key(key)\r\n                    try:\r\n                        body_aim_wt_ch.set_default(0.0)\r\n                        unreal.log(\"[SpineSetup01] Body/Aim Weight set to 0.0\")\r\n                    except Exception as e:\r\n                        unreal.log_warning(f\"[SpineSetup01] set_default failed for Body/Aim Weight: {e}\")\r\n\r\n                if body_aim_tw_ch:\r\n                    for key in body_aim_tw_ch.get_keys():\r\n                        body_aim_tw_ch.remove_key(key)\r\n                    try:\r\n                        body_aim_tw_ch.set_default(0.0)\r\n                        unreal.log(\"[SpineSetup01] Body/Aim Twist set to 0.0\")\r\n                    except Exception as e:\r\n                        unreal.log_warning(f\"[SpineSetup01] set_default failed for Body/Aim Twist: {e}\")\r\n\r\n                ## Step 8 — Set Body/Body back to Parent Space\r\n                try:\r\n                    unreal.ControlRigSequencerLibrary.set_control_rig_space(\r\n                        level_seq, rig, \"Body/Body\", parent_space, frame0)\r\n                    unreal.log(\"[SpineSetup01] Set Body/Body to Parent Space.\")\r\n                except Exception as e:\r\n                    unreal.log_warning(f\"[SpineSetup01] Parent Space failed for Body/Body: {e}\")\r\n\r\n                ## Step 9 — Restore Spine IK controls:\r\n                ##   Start IK and Mid IK → Parent Space\r\n                ##   End IK → Spine/End IK Local Orient Space\r\n                for ctrl in [\"Spine/Start IK\", \"Spine/Mid IK\"]:\r\n                    try:\r\n                        unreal.ControlRigSequencerLibrary.set_control_rig_space(\r\n                            level_seq, rig, ctrl, parent_space, frame0)\r\n                        unreal.log(f\"[SpineSetup01] Restored Parent Space: {ctrl}\")\r\n                    except Exception as e:\r\n                        unreal.log_warning(f\"[SpineSetup01] Parent Space restore failed for {ctrl}: {e}\")\r\n\r\n                end_ik_local_orient = unreal.RigElementKey(\r\n                    type=unreal.RigElementType.NULL,\r\n                    name=\"Spine/End IK Local Orient Space\"\r\n                )\r\n                try:\r\n                    unreal.ControlRigSequencerLibrary.set_control_rig_space(\r\n                        level_seq, rig, \"Spine/End IK\", end_ik_local_orient, frame0)\r\n                    unreal.log(\"[SpineSetup01] Set Spine/End IK to Spine/End IK Local Orient Space.\")\r\n                except Exception as e:\r\n                    unreal.log_warning(f\"[SpineSetup01] Local Orient Space failed for Spine/End IK: {e}\")\r\n\r\n                unreal.log(\"[SpineSetup01] Done!\")\r\n"
   )
   EndTransaction(execute=@n28.then_3)
   "On Clicked (Btn_IKSpine_Yaw)"() {is=title K2Node_ComponentBoundEvent}
   > BeginTransaction(Description="LOCTEXT(\"IK Spine - BodyYaw\")")
   > n29 = Sequence()
   Execute Python Script(
    execute=@n29.then_1
    PythonScript="import unreal\n\n## -----------------------------------------------------------------------\n## Rig profiles — detection is done via CR section channel names\n## -----------------------------------------------------------------------\nRIG_PROFILES = {\n    \"MR_Biped_FN_v01\": {\n        \"detect_control\":    \"Spine/FK 1\",\n        \"spine_ik_controls\": [\"Spine/Start\",    \"Spine/Mid\",    \"Spine/End\"],\n        \"ik_controls\": [\n            \"Arm R/IK\", \"Arm L/IK\", \"Leg R/IK\", \"Leg L/IK\",\n            \"Arm R/PV\", \"Arm L/PV\", \"Leg R/PV\", \"Leg L/PV\",\n        ],\n        \"body_zero_channels\": [\"Body/Body.Rotation.X\", \"Body/Body.Rotation.Y\"],\n    },\n    \"MR_Biped_v02\": {\n        \"detect_control\":    \"Spine/Start FK\",\n        \"spine_ik_controls\": [\"Spine/Start IK\", \"Spine/Mid IK\", \"Spine/End IK\"],\n        \"ik_controls\": [\n            \"Arm R/IK\", \"Arm L/IK\", \"Leg R/IK\", \"Leg L/IK\",\n            \"Arm R/PV\", \"Arm L/PV\", \"Leg R/PV\", \"Leg L/PV\",\n        ],\n        \"body_zero_channels\": [\"Body/Body.Rotation.X\", \"Body/Body.Rotation.Y\"],\n    },\n}\n\ndef detect_rig_profile(cr_section):\n    all_channels = cr_section.get_all_channels()\n    channel_names = {str(ch.channel_name) for ch in all_channels}\n    for profile_name, profile in RIG_PROFILES.items():\n        probe = profile[\"detect_control\"] + \".\"\n        if any(n.startswith(probe) for n in channel_names):\n            unreal.log(f\"[SpineSetup02] Detected rig profile: {profile_name}\")\n            return profile\n    return None\n\ndef get_cr_section(level_seq):\n    for binding in level_seq.get_bindings():\n        for track in binding.get_tracks():\n            for section in track.get_sections():\n                if \"ControlRig\" in section.get_class().get_name():\n                    return section\n    return None\n\ndef channels_for_controls(all_channels, control_list):\n    results = []\n    for ch in all_channels:\n        name = str(ch.channel_name)\n        ctrl = name.split(\".\")[0]\n        if ctrl in control_list:\n            results.append(ch)\n    return results\n\ndef get_broadest_key_range(all_channels, control_list):\n    min_frame = None\n    max_frame = None\n    for ch in all_channels:\n        name = str(ch.channel_name)\n        ctrl = name.split(\".\")[0]\n        if ctrl not in control_list:\n            continue\n        for key in ch.get_keys():\n            t = key.get_time().frame_number.value\n            if min_frame is None or t < min_frame:\n                min_frame = t\n            if max_frame is None or t > max_frame:\n                max_frame = t\n    return min_frame, max_frame\n\ndef set_auto_smart_tangents(channels):\n    total = 0\n    for ch in channels:\n        for key in ch.get_keys():\n            try:\n                key.set_interpolation_mode(unreal.RichCurveInterpMode.RCIM_CUBIC)\n                key.set_tangent_mode(unreal.RichCurveTangentMode.RCTM_AUTO)\n                total += 1\n            except Exception as e:\n                unreal.log_warning(f\"[SpineSetup02] Tangent failed on {ch.channel_name}: {e}\")\n    unreal.log(f\"[SpineSetup02] Set Auto Smart tangents on {total} keys.\")\n\ndef euler_filter_channels(channels):\n    total_fixed = 0\n    for ch in channels:\n        name = str(ch.channel_name)\n        if \".Rotation.\" not in name:\n            continue\n        keys = ch.get_keys()\n        if len(keys) < 2:\n            continue\n        prev_val = keys[0].get_value()\n        for key in keys[1:]:\n            val = key.get_value()\n            diff = val - prev_val\n            if abs(diff) > 180.0:\n                offset = round(diff / 360.0) * 360.0\n                val -= offset\n                key.set_value(val)\n                total_fixed += 1\n            prev_val = val\n    unreal.log(f\"[SpineSetup02] Euler filter complete — {total_fixed} keys adjusted.\")\n\n## -----------------------------------------------------------------------\n## Main\n## -----------------------------------------------------------------------\n\nlevel_seq = unreal.LevelSequenceEditorBlueprintLibrary.get_current_level_sequence()\nif not level_seq:\n    unreal.log_error(\"[SpineSetup02] No Level Sequence is open.\")\nelse:\n    proxies = unreal.ControlRigSequencerLibrary.get_control_rigs(level_seq)\n    if not proxies:\n        unreal.log_error(\"[SpineSetup02] No Control Rig found in the current Level Sequence.\")\n    else:\n        rig = proxies[0].control_rig\n\n        ## Force playhead to the start of the sequence\n        start_frame = level_seq.get_playback_start()\n        unreal.LevelSequenceEditorBlueprintLibrary.set_current_time(start_frame)\n        unreal.log(f\"[SpineSetup02] Seeked to start frame: {start_frame}\")\n        frame0 = unreal.FrameNumber(value=start_frame)\n\n        cr_section = get_cr_section(level_seq)\n        if not cr_section:\n            unreal.log_error(\"[SpineSetup02] Could not find ControlRig section.\")\n        else:\n            profile = detect_rig_profile(cr_section)\n            if not profile:\n                unreal.log_error(\"[SpineSetup02] Unknown rig — cannot determine spine control names. Aborting.\")\n            else:\n                SPINE_IK_CONTROLS = profile[\"spine_ik_controls\"]\n                IK_CONTROLS       = profile[\"ik_controls\"]\n                BODY_ZERO_CHANNELS = profile[\"body_zero_channels\"]\n                ALL_CONTROLS      = SPINE_IK_CONTROLS + IK_CONTROLS + [\"Body/Body\"]\n\n                all_channels = cr_section.get_all_channels()\n\n                ## Step 1 - Set Spine IK controls to World Space\n                world_space = unreal.ControlRigSequencerLibrary.get_world_space_reference_key()\n                for ctrl in SPINE_IK_CONTROLS:\n                    try:\n                        unreal.ControlRigSequencerLibrary.set_control_rig_space(\n                            level_seq, rig, ctrl, world_space, frame0)\n                        unreal.log(f\"[SpineSetup02] Set World Space: {ctrl}\")\n                    except Exception as e:\n                        unreal.log_warning(f\"[SpineSetup02] World Space failed for {ctrl}: {e}\")\n\n                ## Re-fetch after space switch bakes new keys\n                all_channels = cr_section.get_all_channels()\n\n                ## Step 2 - Find broadest key range across all touched controls\n                min_frame, max_frame = get_broadest_key_range(all_channels, ALL_CONTROLS)\n                unreal.log(f\"[SpineSetup02] Key range: frame {min_frame} to {max_frame}\")\n\n                ## Step 3 - Delete all keys on Body/Body Rotation.X and .Y,\n                ##          then insert 0.0 keys at start and end of range\n                for ch in all_channels:\n                    name = str(ch.channel_name)\n                    if name not in BODY_ZERO_CHANNELS:\n                        continue\n                    for key in ch.get_keys():\n                        ch.remove_key(key)\n                    for frame_val in [min_frame, max_frame]:\n                        fn = unreal.FrameNumber(value=frame_val)\n                        ch.add_key(fn, 0.0)\n                    unreal.log(f\"[SpineSetup02] Rebuilt {name} with 0.0 keys at frames {min_frame} and {max_frame}\")\n\n                ## Step 4 - Move IK controls back to Parent Space\n                parent_space = unreal.ControlRigSequencerLibrary.get_default_parent_key()\n                for ctrl in IK_CONTROLS:\n                    try:\n                        unreal.ControlRigSequencerLibrary.set_control_rig_space(\n                            level_seq, rig, ctrl, parent_space, frame0)\n                        unreal.log(f\"[SpineSetup02] Restored Parent Space: {ctrl}\")\n                    except Exception as e:\n                        unreal.log_warning(f\"[SpineSetup02] Parent Space failed for {ctrl}: {e}\")\n\n                ## Step 5 - Move Spine IK controls back to Parent Space\n                for ctrl in SPINE_IK_CONTROLS:\n                    try:\n                        unreal.ControlRigSequencerLibrary.set_control_rig_space(\n                            level_seq, rig, ctrl, parent_space, frame0)\n                        unreal.log(f\"[SpineSetup02] Restored Parent Space: {ctrl}\")\n                    except Exception as e:\n                        unreal.log_warning(f\"[SpineSetup02] Parent Space restore failed for {ctrl}: {e}\")\n\n                ## Re-fetch after all space switches are done\n                all_channels = cr_section.get_all_channels()\n                touched_channels = channels_for_controls(all_channels, ALL_CONTROLS)\n\n                ## Step 6 - Auto Smart tangents on all touched controls\n                set_auto_smart_tangents(touched_channels)\n\n                ## Step 7 - Euler filter on all touched rotation channels\n                euler_filter_channels(touched_channels)\n\n                unreal.log(\"[SpineSetup02] Done!\")\n"
   )
   EndTransaction(execute=@n29.then_3)
   "On Clicked (BTN_Locator_Bake_to_Aim)"() {is=title K2Node_ComponentBoundEvent}
   n30 = Get EditorUtilitySubsystem()
   > n31 = FindUtilityWidgetFromBlueprint!(
    execute
    self:Editor Utility Subsystem Object Reference=@n30
    InBlueprint:Editor Utility Widget Blueprint Object Reference=&/EAT_AimLoc
   ) -> (then, ReturnValue:Editor Utility Widget Object Reference)
   > n32 = Is Valid(InputObject=@n31)
   SpawnAndRegisterTab(execute=@n32.Is Not Valid, self=@n30, InBlueprint=&EAT_AimLoc)
   ReleaseInstanceOfAsset!(execute=@n32.Is Valid, self:Editor Utility Subsystem Object Reference=@n30, Asset:Object Reference=&EAT_AimLoc) -> (then)
   "On Clicked (Btn_MovingLocal_Undo)"() {is=title K2Node_ComponentBoundEvent}
   > BeginTransaction(Description="LOCTEXT(\"Un-Treadmill\")")
   > Execute Python Script(PythonScript="import unreal\r\n\r\n## ── Config ──────────────────────────────────────────────────────────────────\r\nSPACE_CONTROLS = [\r\n    \"Leg L/IK\",\r\n    \"Leg R/IK\",\r\n    \"Arm L/IK\",\r\n    \"Arm R/IK\",\r\n    \"Body/Body\",\r\n]\r\n\r\nROOT_CTRL   = \"root/Root\"\r\nLOCAL_CTRL  = \"root/Local\"\r\nATTACH_CTRL = \"Attach/Attach\"\r\n\r\n## All controls whose rotation keys we euler-filter at the end.\r\n## root/Local is intentionally excluded — it is zeroed on every frame\r\n## (rotation is flat 0, nothing to unwrap).\r\nEULER_FILTER_CONTROLS = SPACE_CONTROLS + [ROOT_CTRL, ATTACH_CTRL]\r\n\r\n## ── Space key constants ──────────────────────────────────────────────────────\r\nWORLD_SPACE  = unreal.ControlRigSequencerLibrary.get_world_space_reference_key()\r\nPARENT_SPACE = unreal.ControlRigSequencerLibrary.get_default_parent_key()\r\nLOCAL_SPACE  = unreal.RigElementKey(\r\n    type=unreal.RigElementType.CONTROL,\r\n    name=LOCAL_CTRL\r\n)\r\n\r\nZERO_EULER = unreal.EulerTransform(\r\n    location=unreal.Vector(0, 0, 0),\r\n    rotation=unreal.Rotator(0, 0, 0),\r\n    scale=unreal.Vector(1, 1, 1)\r\n)\r\n\r\n## ── Helpers ──────────────────────────────────────────────────────────────────\r\ndef get_sequence_and_cr():\r\n    sequence = unreal.LevelSequenceEditorBlueprintLibrary.get_current_level_sequence()\r\n    if not sequence:\r\n        unreal.log_error(\"No Level Sequence is open in Sequencer.\")\r\n        return None, None\r\n    cr_list = unreal.ControlRigSequencerLibrary.get_control_rigs(sequence)\r\n    if not cr_list:\r\n        unreal.log_error(\"No Control Rig found in sequence.\")\r\n        return None, None\r\n    return sequence, cr_list[0].control_rig\r\n\r\ndef get_frame_range(sequence):\r\n    start_f = int(sequence.get_playback_start())\r\n    end_f   = int(sequence.get_playback_end())\r\n    unreal.log(\"Frame range: {} -> {}\".format(start_f, end_f))\r\n    return start_f, end_f\r\n\r\ndef zero_control(sequence, cr, ctrl, start_f, end_f):\r\n    lib = unreal.ControlRigSequencerLibrary\r\n    for f in range(start_f, end_f):\r\n        fn = unreal.FrameNumber(value=f)\r\n        lib.set_local_control_rig_euler_transform(sequence, cr, ctrl, fn, ZERO_EULER)\r\n\r\ndef unwrap(angle, prev):\r\n    \"\"\"Unwrap angle to be within 180 degrees of prev.\"\"\"\r\n    diff = angle - prev\r\n    while diff >  180.0: diff -= 360.0\r\n    while diff < -180.0: diff += 360.0\r\n    return prev + diff\r\n\r\ndef euler_filter(sequence, cr, ctrl, start_f, end_f):\r\n    \"\"\"Walk rotation keys frame by frame and unwrap to remove flips.\"\"\"\r\n    lib = unreal.ControlRigSequencerLibrary\r\n    frames = list(range(start_f, end_f))\r\n    if len(frames) < 2:\r\n        return\r\n\r\n    ## Read all values first\r\n    vals = []\r\n    for f in frames:\r\n        fn = unreal.FrameNumber(value=f)\r\n        vals.append(lib.get_local_control_rig_euler_transform(sequence, cr, ctrl, fn))\r\n\r\n    ## Unwrap each channel independently\r\n    prev_p = vals[0].rotation.pitch\r\n    prev_y = vals[0].rotation.yaw\r\n    prev_r = vals[0].rotation.roll\r\n\r\n    for i in range(1, len(vals)):\r\n        p = unwrap(vals[i].rotation.pitch, prev_p)\r\n        y = unwrap(vals[i].rotation.yaw,   prev_y)\r\n        r = unwrap(vals[i].rotation.roll,  prev_r)\r\n\r\n        if p != vals[i].rotation.pitch or y != vals[i].rotation.yaw or r != vals[i].rotation.roll:\r\n            fn  = unreal.FrameNumber(value=frames[i])\r\n            new_val = unreal.EulerTransform(\r\n                location=vals[i].location,\r\n                rotation=unreal.Rotator(pitch=p, yaw=y, roll=r),\r\n                scale=vals[i].scale\r\n            )\r\n            lib.set_local_control_rig_euler_transform(sequence, cr, ctrl, fn, new_val)\r\n            vals[i] = new_val\r\n\r\n        prev_p, prev_y, prev_r = p, y, r\r\n\r\n## ── Main ─────────────────────────────────────────────────────────────────────\r\nsequence, cr = get_sequence_and_cr()\r\nif sequence and cr:\r\n    start_f, end_f = get_frame_range(sequence)\r\n    start_fn = unreal.FrameNumber(value=start_f)\r\n\r\n    ## Step 1 – move playhead to start frame\r\n    unreal.LevelSequenceEditorBlueprintLibrary.set_current_time(start_f)\r\n    unreal.log(\"Step 1: Playhead -> frame {}.\".format(start_f))\r\n\r\n    ## Step 2 – set IK / Body controls AND root/Root to World space.\r\n    ## root/Root rides along so its world motion is held independently while\r\n    ## root/Local is zeroed below; it then comes back to Local space (Step 5)\r\n    ## with that motion preserved.\r\n    for ctrl in SPACE_CONTROLS + [ROOT_CTRL]:\r\n        unreal.ControlRigSequencerLibrary.set_control_rig_space(\r\n            sequence, cr, ctrl, WORLD_SPACE, start_fn\r\n        )\r\n    unreal.log(\"Step 2: IK + Body + root/Root -> World space.\")\r\n\r\n    ## Step 3 – zero out root/Local loc/rot across all frames\r\n    zero_control(sequence, cr, LOCAL_CTRL, start_f, end_f)\r\n    unreal.log(\"Step 3: root/Local zeroed ({} frames).\".format(end_f - start_f))\r\n\r\n    ## Step 4 – set IK / Body controls back to Default Parent space\r\n    for ctrl in SPACE_CONTROLS:\r\n        unreal.ControlRigSequencerLibrary.set_control_rig_space(\r\n            sequence, cr, ctrl, PARENT_SPACE, start_fn\r\n        )\r\n    unreal.log(\"Step 4: IK + Body controls -> Default Parent space.\")\r\n\r\n    ## Step 5 – set root/Root to its Local Control space (root/Local)\r\n    unreal.ControlRigSequencerLibrary.set_control_rig_space(\r\n        sequence, cr, ROOT_CTRL, LOCAL_SPACE, start_fn\r\n    )\r\n    unreal.log(\"Step 5: root/Root -> Local Control space (root/Local).\")\r\n\r\n    ## Step 6 – zero out Attach/Attach loc/rot across all frames\r\n    zero_control(sequence, cr, ATTACH_CTRL, start_f, end_f)\r\n    unreal.log(\"Step 6: Attach/Attach zeroed ({} frames).\".format(end_f - start_f))\r\n\r\n    ## Step 7 – euler filter all edited controls to remove rotation flips\r\n    for ctrl in EULER_FILTER_CONTROLS:\r\n        euler_filter(sequence, cr, ctrl, start_f, end_f)\r\n    unreal.log(\"Step 7: Euler filter applied to {} controls.\".format(len(EULER_FILTER_CONTROLS)))\r\n\r\n    unreal.log(\"Done!\")\r\n")
   > EndTransaction()
   "On Clicked (BTN_Copy_KeyTiming)"() {is=title K2Node_ComponentBoundEvent}
   > JJF_CopyKeyTimings()
   > JJF_PasteKeyTimings!(execute, self, RemoveNonTiming:bool=false) -> (then)
   "On Clicked (BTN_Constant)"() {is=title K2Node_ComponentBoundEvent}
   > Pre/Post Curve Interp(InExtrapolation=RCCE_Constant)
   "On Clicked (Btn_Bake_MRB2)"() {is=title K2Node_ComponentBoundEvent}
   > Bake_to_MrBipedv02() {is=title K2Node_CallFunction}
   "On Clicked (BTN_MoveSpaceLeft)"() {is=title K2Node_ComponentBoundEvent}
   > MoveSpaceKey_Left()
   "On Clicked (Btn_IKSpine)"() {is=title K2Node_ComponentBoundEvent}
   > BeginTransaction(Description="LOCTEXT(\"IK Spine - Simple\")")
   > n33 = Sequence!(execute) -> (then_0, then_1, then_2)
   EndTransaction(execute=@n33.then_2)
   "On Clicked (BTN_PIE_MultiCam)"() {is=title K2Node_ComponentBoundEvent}
   > n34 = GetAllActorsOfClass!(execute, WorldContextObject:Object Reference, ActorClass:Actor Class Reference) -> (then, OutActors:Array of Actor Object References)
   > n35 = Branch(Condition="integer > integer"(A=Length(TargetArray:Array of Gameplay Statics Object References=@n34)))
   > PrintString(InString=Already imported into the scene, Duration=5)
   > n36 = For Each Loop!(Exec, Array:Array of Object References=@n34) -> (LoopBody, Array Element:Object Reference, Array Index, Completed)
   > Destroy Actor(self=@n36)
   > Set Text(Text="LOCTEXT(\"PIE MultiCam - add\")", self=Get TB_TabName_PIE() -> (TB_TabName_PIE:Text Object Reference))
   > SetColorAndOpacity(
    self:Button Object Reference=Get BTN_PIE_MultiCam() -> (BTN_PIE_MultiCam:Editor Utility Button Object Reference)
    InColorAndOpacity:Linear Color Structure=(R=0.8714,G=0.855,B=0.1559,A=1)
   )
   PrintText(execute=@n35.else, InText="INVTEXT(\"Importing Pie Multiecam Setup\")", Duration=5)
   > SpawnActor NONE!(
    execute:Exec
    WorldContextObject:Object Reference
    Class:Actor Class Reference
    SpawnTransform:Transform=MakeTransform(Location="0,0,-100000", Rotation="0, 0, 0")
    CollisionHandlingOverride:ESpawnActorCollisionHandlingMethod Enum=Undefined
    TransformScaleMethod:ESpawnActorScaleMethod Enum=MultiplyWithRoot
    Owner:Actor Object Reference
    Instigator:Pawn Object Reference
   ) -> (then:Exec, ReturnValue:Actor Object Reference) {is=title K2Node_SpawnActorFromClass}
   > Set Text(Text="LOCTEXT(\"PIE MultiCam - remove\")", self=Get TB_TabName_PIE())
   > SetColorAndOpacity(self=Get BTN_PIE_MultiCam(), InColorAndOpacity=(R=0.8714,G=0.1559,B=0.1559,A=1))
   "On Clicked (BTN_Offset)"() {is=title K2Node_ComponentBoundEvent}
   > n37 = GetSelectedChild!(execute, self) -> (then, OutObject:Object Reference, OutSocket:Name=None)
   n38 = GetConstraints!(InChild:Object Reference=@n37, InSocket:Name=@n37.OutSocket) -> (Constraints:Array of Tickable Transform Constraint Object References)
   > Branch(Condition="Equal (Integer)"(A=Length(TargetArray:Array of Tickable Transform Constraint Object References=@n38), B=1))
   > AddOffsetNode!(
    execute
    self
    InOutConstraint:Tickable Transform Constraint Object Reference="Get (a copy)"(Array:Array of Tickable Transform Constraint Object References=@n38) -> (Output:Tickable Transform Constraint Object Reference)
   ) -> (then) {is=title K2Node_CallFunction}
   "On Clicked (BTN_Copy_WorldspaceXform)"() {is=title K2Node_ComponentBoundEvent}
   > WorldspaceXform_Copy()
   "On Clicked (BTN_Paste_WorldspaceXform)"() {is=title K2Node_ComponentBoundEvent}
   > WorldsaceXform_Paste()
   Event Construct()
   > n39 = GetAllActorsOfClass()
   > n40 = Branch(Condition="integer > integer"(A=Length(TargetArray:Array of Actor Object References=@n39)))
   > Set Text(Text="LOCTEXT(\"PIE MultiCam - remove\")", self=Get TB_TabName_PIE())
   > SetColorAndOpacity(self=Get BTN_PIE_MultiCam(), InColorAndOpacity=(R=0.8714,G=0.1559,B=0.1559,A=1))
   Set Text(execute=@n40.else, Text="LOCTEXT(\"PIE MultiCam - add\")", self=Get TB_TabName_PIE())
   > SetColorAndOpacity(self=Get BTN_PIE_MultiCam(), InColorAndOpacity=(R=0.8714,G=0.855,B=0.1559,A=1))
   n41 = Sequence!(execute) -> (then_0, then_1)
   > Clear(TargetMap=Get Controls Transform Map Local())
   > n42 = JJF_GetSelection_BindingsObjectsControlNames()
   > n43 = Get Current Local Time(TimeUnit=DisplayRate) -> (ReturnValue_Frame_SubFrame=~, ReturnValue_Time=~, ReturnValue_bHasJumped=~)
   n44 = Get STR_CR_ProxyObjName()
   n45 = "Get (a copy)"!(
    Array:Array of STR Control Rig Proxy Object Name Tuple Structures=@n44
    Dimension 1=Last Index!(TargetArray:Array of STR Control Rig Proxy Object Name Tuple Structures=@n44) -> (ReturnValue=0)
   ) -> (
    Output_CR_Proxy_11_#2:Movie Scene Binding Proxy Structure
    Output_CR_Object_8_#3:Control Rig Object Reference
    Output_CR_ControlName_9_#4:Name
   )
   > DEBUG(NewParam=Append(A="Last Selected Control is: ", B="To String (Name)"(InName=@n45.Output_CR_ControlName_9_#4)))
   > n46 = GetControlRigWorldTransform(
    LevelSequence=@n42
    ControlRig=@n45.Output_CR_Object_8_#3
    ControlName=@n45.Output_CR_ControlName_9_#4
    Frame=@n43
    TimeUnit=DisplayRate
   )
   > Add(TargetMap=Get Controls Transform Map Local(), Key=@n45.Output_CR_ControlName_9_#4, Value=@n46)
   n47 = Get STR_CR_ProxyObjName()
   n48 = For Each Loop!(Exec=@n41.then_1, Array:Array of STR Control Rig Proxy Object Name Tuple Structures=@n47) -> (
    LoopBody
    Array Element_CR_Proxy_11_#2:Movie Scene Binding Proxy Structure
    Array Element_CR_Object_8_#3:Control Rig Object Reference
    Array Element_CR_ControlName_9_#4:Name
    Array Index
    Completed
   )
   > Branch(Condition="Not Equal (Integer)"(A=Last Index(TargetArray=@n47) -> (ReturnValue=0), B=@n48.Array Index))
   > SetControlRigWorldTransform(
    LevelSequence=@n42
    ControlRig=@n48.Array Element_CR_Object_8_#3
    ControlName=@n48.Array Element_CR_ControlName_9_#4
    Frame=@n43
    WorldTransform=Get Key Value By Index!(TargetMap:Map of Names to Transforms=Get Controls Transform Map Local(), Index:int=0) -> (Key:Name, Value:Transform) {is=title K2Node_CallFunction}.Value
    TimeUnit=DisplayRate
    bSetKey=true
   )
   n49 = GetCurveEditor!(self:Level Sequence Editor Subsystem Object Reference=Get LevelSequenceEditorSubsystem()) -> (ReturnValue:Sequencer Curve Editor Object Object Reference)
   n50 = Construct Object from Class(Class=/Script/CurveEditor.CurveEditorEulerFilter, self=@n49) -> (ReturnValue:Euler Object Reference)
   > ApplyFilter!(execute, self:Sequencer Curve Editor Object Object Reference=@n49, Filter:Curve Editor Filter Base Object Reference=@n50) -> (then)
   PasteWorldSpace_AllFrames()
   CopyWorldSpace_AllFrames() {is=title K2Node_CallFunction}
   PrintString(execute=@n1.then, InString=Hello, Key=@n1)
   AND Boolean(A=EndsWith(SourceString="To String (Name)"(InName=@n22.Array Element_Name), InSuffix=_ctrl), B=@n23)
   Event PreConstruct()
   Event Tick()
   Execute Python Script!(
    execute
    PythonScript="import unreal\r\nimport tkinter as tk\r\nfrom tkinter import ttk\r\nimport ctypes\r\nfrom ctypes import wintypes\r\n\r\n## inputs injected by Execute Python Script node (always lowercase)\r\n## locator_cr, locator_ctrl, actor_cr, actor_ctrl\r\n\r\nlocator_ctrl_name = str(locator_ctrl)\r\nactor_ctrl_name   = str(actor_ctrl)\r\n\r\nLSEBL = unreal.LevelSequenceEditorBlueprintLibrary\r\n\r\n## ── Win32 helpers ─────────────────────────────────────────────────\r\nclass _POINT(ctypes.Structure):\r\n    _fields_ = [(\"x\", ctypes.c_long), (\"y\", ctypes.c_long)]\r\nclass _RECT(ctypes.Structure):\r\n    _fields_ = [(\"left\", ctypes.c_long), (\"top\", ctypes.c_long),\r\n                (\"right\", ctypes.c_long), (\"bottom\", ctypes.c_long)]\r\nclass _MONITORINFO(ctypes.Structure):\r\n    _fields_ = [(\"cbSize\", wintypes.DWORD),\r\n                (\"rcMonitor\", _RECT), (\"rcWork\", _RECT),\r\n                (\"dwFlags\", wintypes.DWORD)]\r\n\r\ndef is_windows_dark_mode():\r\n    try:\r\n        import winreg\r\n        path = r\"Software\\Microsoft\\Windows\\CurrentVersion\\Themes\\Personalize\"\r\n        with winreg.OpenKey(winreg.HKEY_CURRENT_USER, path) as k:\r\n            v, _ = winreg.QueryValueEx(k, \"AppsUseLightTheme\")\r\n            return v == 0\r\n    except Exception:\r\n        return False\r\n\r\ndef cursor_monitor_workarea():\r\n    try:\r\n        user32 = ctypes.windll.user32\r\n        pt = _POINT()\r\n        user32.GetCursorPos(ctypes.byref(pt))\r\n        hmon = user32.MonitorFromPoint(pt, 2)\r\n        mi = _MONITORINFO()\r\n        mi.cbSize = ctypes.sizeof(_MONITORINFO)\r\n        user32.GetMonitorInfoW(hmon, ctypes.byref(mi))\r\n        r = mi.rcWork\r\n        return (r.left, r.top, r.right - r.left, r.bottom - r.top)\r\n    except Exception:\r\n        return None\r\n\r\n## ── get frame range from selected keys ───────────────────────────\r\nls = LSEBL.get_focused_level_sequence()\r\nif not ls:\r\n    unreal.log_error(\"Offset Snap: No level sequence is open.\")\r\n    raise SystemExit\r\n\r\ndef get_selected_key_range():\r\n    all_frames = []\r\n    ses = unreal.get_editor_subsystem(unreal.LevelSequenceEditorSubsystem)\r\n    curve_editor = ses.get_curve_editor()\r\n    if curve_editor.is_curve_editor_open():\r\n        for ch_proxy in curve_editor.get_channels_with_selected_keys():\r\n            indices = curve_editor.get_selected_keys(ch_proxy)\r\n            if not indices:\r\n                continue\r\n            channel_obj = ch_proxy.section.get_channel(ch_proxy.channel_name)\r\n            if not channel_obj:\r\n                continue\r\n            for k in channel_obj.get_keys_by_index(indices):\r\n                all_frames.append(k.get_time().frame_number.value)\r\n    if not all_frames:\r\n        for ch_proxy in LSEBL.get_channels_with_selected_keys():\r\n            indices = LSEBL.get_selected_keys(ch_proxy)\r\n            if not indices:\r\n                continue\r\n            channel_obj = ch_proxy.section.get_channel(ch_proxy.channel_name)\r\n            if not channel_obj:\r\n                continue\r\n            for k in channel_obj.get_keys_by_index(indices):\r\n                all_frames.append(k.get_time().frame_number.value)\r\n    return (min(all_frames), max(all_frames)) if all_frames else None\r\n\r\nselected_range = get_selected_key_range()\r\nif selected_range:\r\n    start_val, end_val = selected_range\r\nelse:\r\n    start_val = ls.get_playback_start()\r\n    end_val   = ls.get_playback_end() - 1\r\n\r\nstart_frame = unreal.FrameNumber(start_val)\r\nend_frame   = unreal.FrameNumber(end_val)\r\nrange_str   = f\"{start_val} to {end_val}\"\r\nrange_src   = \"selected keys\" if selected_range else \"playback range\"\r\n\r\n## ── tkinter dialog ────────────────────────────────────────────────\r\nclass OffsetSnapDialog:\r\n    def __init__(self):\r\n        self.result = None\r\n        self.root = tk.Tk()\r\n        self.root.title(\"Offset Snap Locator\")\r\n        self.root.attributes('-topmost', True)\r\n        self.root.resizable(False, False)\r\n        self.dark = is_windows_dark_mode()\r\n        self._apply_theme()\r\n        pad = {'padx': 8, 'pady': 4}\r\n\r\n        ttk.Label(self.root,\r\n                  text=f\"Snap '{locator_ctrl_name}' to '{actor_ctrl_name}'\\n\"\r\n                       f\"Range: {range_str}  ({range_src})\").pack(**pad)\r\n        ttk.Separator(self.root, orient='horizontal').pack(fill='x', padx=8, pady=2)\r\n\r\n        axis_frame = ttk.LabelFrame(self.root, text=\"Local Offset Axis\")\r\n        axis_frame.pack(fill='x', **pad)\r\n        self.axis_var = tk.StringVar(value='Y')\r\n        for a in ('X', 'Y', 'Z'):\r\n            ttk.Radiobutton(axis_frame, text=a, variable=self.axis_var,\r\n                            value=a).pack(side='left', padx=12, pady=4)\r\n\r\n        dist_frame = ttk.Frame(self.root)\r\n        dist_frame.pack(fill='x', **pad)\r\n        ttk.Label(dist_frame, text=\"Distance (cm):\").pack(side='left')\r\n        self.dist_var = tk.StringVar(value='50.0')\r\n        ttk.Entry(dist_frame, textvariable=self.dist_var, width=10).pack(side='left', padx=6)\r\n\r\n        mode_frame = ttk.LabelFrame(self.root, text=\"Bake Mode\")\r\n        mode_frame.pack(fill='x', **pad)\r\n        self.mode_var = tk.StringVar(value='all')\r\n        ttk.Radiobutton(mode_frame, text=\"All Frames\", variable=self.mode_var,\r\n                        value='all').pack(side='left', padx=12, pady=4)\r\n        ttk.Radiobutton(mode_frame, text=\"Keys Only\", variable=self.mode_var,\r\n                        value='keys').pack(side='left', padx=12, pady=4)\r\n\r\n        btn_frame = ttk.Frame(self.root)\r\n        btn_frame.pack(fill='x', padx=8, pady=8)\r\n        ttk.Button(btn_frame, text=\"Cancel\", command=self._cancel).pack(side='right', padx=4)\r\n        ttk.Button(btn_frame, text=\"OK\",     command=self._ok).pack(side='right', padx=4)\r\n\r\n        self.root.protocol(\"WM_DELETE_WINDOW\", self._cancel)\r\n        self.root.bind('<Return>', lambda _: self._ok())\r\n        self.root.bind('<Escape>', lambda _: self._cancel())\r\n\r\n    def _apply_theme(self):\r\n        style = ttk.Style(self.root)\r\n        if self.dark:\r\n            bg, bg_alt, fg = '##2b2b2b', '##3c3c3c', '##e6e6e6'\r\n            self.root.configure(bg=bg)\r\n            try: style.theme_use('clam')\r\n            except tk.TclError: pass\r\n            style.configure('.', background=bg, foreground=fg, fieldbackground=bg_alt)\r\n            style.configure('TFrame', background=bg)\r\n            style.configure('TLabel', background=bg, foreground=fg)\r\n            style.configure('TLabelframe', background=bg, foreground=fg)\r\n            style.configure('TLabelframe.Label', background=bg, foreground=fg)\r\n            style.configure('TRadiobutton', background=bg, foreground=fg)\r\n            style.configure('TEntry', fieldbackground=bg_alt, foreground=fg)\r\n            style.configure('TButton', background=bg_alt, foreground=fg)\r\n            style.map('TRadiobutton', background=[('active', bg)], foreground=[('active', fg)])\r\n            style.map('TButton', background=[('active', '##505050')])\r\n\r\n    def _ok(self):\r\n        try:    distance = float(self.dist_var.get())\r\n        except: distance = 0.0\r\n        self.result = {'axis': self.axis_var.get(), 'distance': distance,\r\n                       'keys_only': self.mode_var.get() == 'keys'}\r\n        self.root.destroy()\r\n\r\n    def _cancel(self):\r\n        self.result = None\r\n        self.root.destroy()\r\n\r\n    def show(self):\r\n        self.root.update_idletasks()\r\n        w, h = self.root.winfo_reqwidth(), self.root.winfo_reqheight()\r\n        rect = cursor_monitor_workarea()\r\n        if rect:\r\n            mx, my, mw, mh = rect\r\n            x, y = mx + (mw - w) // 2, my + (mh - h) // 2\r\n        else:\r\n            x = (self.root.winfo_screenwidth()  - w) // 2\r\n            y = (self.root.winfo_screenheight() - h) // 2\r\n        self.root.geometry(f\"+{x}+{y}\")\r\n        self.root.mainloop()\r\n        return self.result\r\n\r\nparams = OffsetSnapDialog().show()\r\nif params is None:\r\n    unreal.log(\"Offset Snap: Cancelled by user.\")\r\n    raise SystemExit\r\n\r\noffset_axis     = params['axis']\r\noffset_distance = params['distance']\r\nkeys_only       = params['keys_only']\r\n\r\n## ── channel collection helper (called twice - before and after bake) ──\r\nchannel_types = [t for t in (\r\n    getattr(unreal, 'MovieSceneScriptingDoubleChannel', None),\r\n    getattr(unreal, 'MovieSceneScriptingFloatChannel',  None),\r\n) if t is not None]\r\n\r\ndef collect_channels():\r\n    loc_chs  = {}\r\n    act_chs  = {}\r\n    loc_prefix = f\"{locator_ctrl_name}.\"\r\n    act_prefix = f\"{actor_ctrl_name}.\"\r\n    for binding in ls.get_bindings():\r\n        for track in binding.get_tracks():\r\n            if not isinstance(track, unreal.MovieSceneControlRigParameterTrack):\r\n                continue\r\n            for section in track.get_sections():\r\n                for ct in channel_types:\r\n                    try:\r\n                        channels = section.get_channels_by_type(ct)\r\n                    except Exception:\r\n                        continue\r\n                    for ch in channels:\r\n                        full_name = str(getattr(ch, 'channel_name', ''))\r\n                        if full_name.startswith(loc_prefix):\r\n                            short = full_name[len(loc_prefix):]\r\n                            if short.startswith('Location.') or short.startswith('Rotation.'):\r\n                                loc_chs[short] = ch\r\n                        if full_name.startswith(act_prefix):\r\n                            act_chs[full_name] = ch\r\n    return loc_chs, act_chs\r\n\r\n## ── find the constraint (keep index for deletion) ─────────────────\r\neditor_system = unreal.get_editor_subsystem(unreal.UnrealEditorSubsystem)\r\nworld = editor_system.get_editor_world()\r\n\r\nconstraints = unreal.ConstraintsScriptingLibrary.get_constraints_array(world)\r\ntarget_constraint = None\r\ntarget_index      = -1\r\nfor i, c in enumerate(constraints):\r\n    if not isinstance(c, unreal.TickableParentConstraint):\r\n        continue\r\n    ch = c.child_trs_handle\r\n    if ch and ch.control_name == locator_ctrl_name:\r\n        target_constraint = c\r\n        target_index      = i\r\n        break\r\n\r\nif target_constraint is None:\r\n    unreal.log_error(\r\n        f\"Offset Snap: Could not find parent constraint with child='{locator_ctrl_name}'.\"\r\n    )\r\n    raise SystemExit\r\n\r\n## ── step 1: turn on maintain offset ──────────────────────────────\r\ntarget_constraint.modify(True)\r\ntarget_constraint.set_editor_property('bMaintainOffset', True)\r\n\r\n## ── step 2: write offset to chosen axis, zero all others ─────────\r\nlocator_channels, actor_channels = collect_channels()\r\n\r\noffset_channel_name = f\"Location.{offset_axis}\"\r\nfor short_name, ch in locator_channels.items():\r\n    value = offset_distance if short_name == offset_channel_name else 0.0\r\n    for k in ch.get_keys():\r\n        k.set_value(value)\r\n\r\nunreal.log(\r\n    f\"Offset Snap: '{locator_ctrl_name}' offset {offset_axis}={offset_distance}cm, \"\r\n    f\"all other Location/Rotation channels zeroed.\"\r\n)\r\n\r\n## ── step 3: bake the constraint ───────────────────────────────────\r\nif keys_only:\r\n    key_frames = set()\r\n    for ch in actor_channels.values():\r\n        for k in ch.get_keys():\r\n            f = k.get_time().frame_number.value\r\n            if start_val <= f <= end_val:\r\n                key_frames.add(f)\r\n    if not key_frames:\r\n        unreal.log_warning(\r\n            \"Offset Snap: Keys Only selected but no keys found on actor_ctrl - \"\r\n            \"falling back to All Frames.\"\r\n        )\r\n        bake_frames = []\r\n    else:\r\n        bake_frames = [unreal.FrameNumber(f) for f in sorted(key_frames)]\r\n        unreal.log(f\"Offset Snap: baking at {len(bake_frames)} key frames from actor_ctrl.\")\r\nelse:\r\n    bake_frames = []\r\n\r\nresult = unreal.ControlRigSequencerLibrary.bake_constraint(\r\n    world,\r\n    target_constraint,\r\n    bake_frames,\r\n    unreal.MovieSceneTimeUnit.DISPLAY_RATE\r\n)\r\n\r\nmode_str = \"keys only\" if (keys_only and bake_frames) else \"all frames\"\r\nunreal.log(\r\n    f\"Offset Snap: baked constraint on '{locator_ctrl_name}' \"\r\n    f\"[{start_val} to {end_val}] ({mode_str}). Success: {result}\"\r\n)\r\n\r\n## ── step 4: delete the parent constraint ─────────────────────────\r\nunreal.ConstraintsScriptingLibrary.remove_constraint(world, target_index)\r\nunreal.log(f\"Offset Snap: parent constraint deleted (index {target_index}).\")\r\n\r\n## ── step 5: read locator key range from fresh channels ───────────\r\nlocator_channels_fresh, _ = collect_channels()\r\n\r\nlocator_key_frames = set()\r\nfor ch in locator_channels_fresh.values():\r\n    try:\r\n        for k in ch.get_keys():\r\n            locator_key_frames.add(k.get_time().frame_number.value)\r\n    except Exception:\r\n        continue\r\n\r\nif locator_key_frames:\r\n    lookat_start = min(locator_key_frames)\r\n    lookat_end   = max(locator_key_frames)\r\nelse:\r\n    lookat_start = start_val\r\n    lookat_end   = end_val\r\n\r\nseq_last_frame = ls.get_playback_end() - 1\r\n\r\n## ── step 6: add Look At constraint actor_ctrl -> locator_ctrl ─────\r\nchild_handle  = actor_cr.create_transformable_control_handle(actor_ctrl_name)\r\nparent_handle = locator_cr.create_transformable_control_handle(locator_ctrl_name)\r\n\r\nif not child_handle or not parent_handle:\r\n    unreal.log_error(\"Offset Snap: Failed to create handles for Look At constraint.\")\r\n    raise SystemExit\r\n\r\nLSEBL.set_current_time(lookat_start)\r\n\r\nlookat_constraint = unreal.ControlRigSequencerLibrary.add_constraint(\r\n    world,\r\n    unreal.TransformConstraintType.LOOK_AT,\r\n    child_handle,\r\n    parent_handle,\r\n    True  ## maintain_offset\r\n)\r\n\r\nif not lookat_constraint:\r\n    unreal.log_error(\"Offset Snap: Failed to create Look At constraint.\")\r\n    raise SystemExit\r\n\r\nunreal.log(\r\n    f\"Offset Snap: Look At constraint created '{actor_ctrl_name}' -> '{locator_ctrl_name}'.\"\r\n)\r\n\r\n## key ON at lookat_start\r\nunreal.ControlRigSequencerLibrary.set_constraint_active_key(\r\n    lookat_constraint, True, unreal.FrameNumber(lookat_start)\r\n)\r\n\r\n## key OFF at lookat_end only if it doesn't reach the last frame of the sequence\r\nif lookat_end < seq_last_frame:\r\n    unreal.ControlRigSequencerLibrary.set_constraint_active_key(\r\n        lookat_constraint, False, unreal.FrameNumber(lookat_end)\r\n    )\r\n    unreal.log(\r\n        f\"Offset Snap: Look At keyed ON at {lookat_start}, OFF at {lookat_end} \"\r\n        f\"(sequence ends at {seq_last_frame}).\"\r\n    )\r\nelse:\r\n    unreal.log(\r\n        f\"Offset Snap: Look At keyed ON at {lookat_start}, no OFF key \"\r\n        f\"(lookat_end {lookat_end} == sequence end {seq_last_frame}).\"\r\n    )\r\n"
    Locator_CR:Wildcard
    Locator_Ctrl:Wildcard
    Actor_CR:Wildcard
    Actor_Ctrl:Wildcard
   ) -> (then, ReturnValue) {inputs [Locator_CR,Locator_Ctrl,Actor_CR,Actor_Ctrl]}
  ]}
 Get Child Constraints {nodes [
   n1 = Get Child Constraints!() -> (then, Child:Actor Object Reference) {is=title K2Node_FunctionEntry}
   > n2 = GetEditorWorld(self=Get UnrealEditorSubsystem())
   > n3 = GetConstraintsArray!(execute, InWorld:World Object Reference=@n2) -> (then, ReturnValue:Array of Tickable Constraint Object References)
   > n4 = For Each Loop!(Exec, Array:Array of Tickable Constraint Object References=@n3) -> (LoopBody, Array Element:Tickable Constraint Object Reference, Array Index, Completed)
   Return Node!(
    execute=@n4.Completed
    Constraints:Array of Tickable Transform Constraint Object References=Get NewLocalVar() -> (NewLocalVar:Array of Tickable Transform Constraint Object References)
   )
   > n5 = Cast To TickableTransformConstraint!(execute, Object=@n4) -> (then, CastFailed, AsTickable Transform Constraint:Tickable Transform Constraint Object Reference) {is=title K2Node_DynamicCast}
   > n6 = Cast To TransformableComponentHandle(Object=Get ChildTRSHandle!(self:Tickable Transform Constraint Object Reference=@n5) -> (ChildTRSHandle:Transformable Handle Object Reference))
   > Branch(Condition="Equal (Object)"(A=Get Component(self=@n6), B=Get RootComponent!(self:Actor Object Reference=@n1) -> (RootComponent:Scene Component Object Reference)))
   > Add!(
    execute
    TargetArray:Array of Tickable Transform Constraint Object References=Get NewLocalVar()
    NewItem:"Tickable Transform Constraint Object Reference (by ref)"=@n5
   ) -> (then, ReturnValue) {is=title K2Node_CallArrayFunction}
  ]}
 GetConstraints {nodes [
   n1 = GetConstraints!() -> (then, InChild:Object Reference, InSocket:Name) {is=title K2Node_FunctionEntry}
   > n2 = GetEditorWorld(self=Get UnrealEditorSubsystem())
   > n3 = GetConstraintsArray(InWorld=@n2)
   > n4 = For Each Loop(Array=@n3)
   Return Node(execute=@n4.Completed, Constraints=Get NewLocalVar())
   > n5 = Cast To TickableTransformConstraint(Object=@n4)
   n6 = Get ChildTRSHandle(self=@n5)
   > n7 = Cast To TransformableComponentHandle(Object=@n6)
   n8 = Get Component(self=@n7)
   n9 = Get SocketName(self=@n7) -> (SocketName=None)
   > Branch(Condition=AND Boolean(A="Equal (Object)"(A=@n8, B=@n1), B="Equal (Name)"(A=@n9, B=@n1.InSocket)))
   > Add(TargetArray=Get NewLocalVar(), NewItem=@n5)
   n10 = Cast To TransformableControlHandle!(execute=@n7.CastFailed, Object=@n6) -> (then, CastFailed, AsTransformable Control Handle:Transformable Control Handle Object Reference) {is=title K2Node_DynamicCast}
   n11 = Get ControlName!(self:Transformable Control Handle Object Reference=@n10) -> (ControlName:Name=None)
   n12 = Get ControlRig!(self:Transformable Control Handle Object Reference=@n10) -> (ControlRig:Control Rig Soft Object Reference)
   > Branch(Condition=AND Boolean(A="Equal (Object)"(A:Control Rig Soft Object Reference=@n12, B=@n1), B="Equal (Name)"(A=@n11, B=@n1.InSocket)))
   > Add(TargetArray=Get NewLocalVar(), NewItem=@n5)
  ]}
 GetParentInfo {nodes [
   n1 = GetParentInfo!() -> (then, InConstraint:Tickable Transform Constraint Object Reference) {is=title K2Node_FunctionEntry}
   n2 = Get ParentTRSHandle(self=@n1)
   > n3 = Cast To TransformableComponentHandle(Object=@n2)
   > Return Node!(
    execute
    OutParentHandle:Transformable Handle Object Reference=@n3
    OutWorldTrransform:Transform=GetSocketTransform(self=Get Component(self=@n3), InSocketName=Get SocketName(self=@n3) -> (SocketName=None), TransformSpace=RTS_World)
   )
   n4 = Cast To TransformableControlHandle(execute=@n3.CastFailed, Object=@n2)
   n5 = Resolve Soft Reference!(Input:Control Rig Soft Object Reference=Get ControlRig(self=@n4)) -> (Output:Control Rig Object Reference)
   n6 = GetHierarchy(self=@n5)
   n7 = Get ControlName(self=@n4) -> (ControlName=None)
   n8 = Get RootComponent(self=GetHostingActor!(self:Control Rig Object Reference=@n5) -> (ReturnValue:Actor Object Reference))
   > Return Node(
    OutParentHandle=@n4
    OutWorldTrransform=ComposeTransforms!(
     A:"Transform (by ref)"=Get World Transform!(self:Scene Component Object Reference=@n8) -> (ReturnValue:Transform) {is=title K2Node_CallFunction}
     B:"Transform (by ref)"=GetGlobalTransform!(self:Rig Hierarchy Object Reference=@n6, InKey_Type:ERigElementType Enum=Control, InKey_Name:Name=@n7, bInitial:bool=false) -> (ReturnValue:Transform)
    ) -> (ReturnValue:Transform)
   )
  ]}
 GetSelectedChild {nodes [
   GetSelectedChild() {is=title K2Node_FunctionEntry}
   > n1 = GetSelectedActors()
   > n2 = Branch(Condition="Equal (Integer)"(A=Length(TargetArray=@n1), B=1))
   > Return Node!(
    execute
    OutObject:Object Reference=Get RootComponent(self="Get (a copy)"!(Array:Array of Actor Object References=@n1, Dimension 1) -> (Output:Actor Object Reference))
    OutSocket:Name=None
   )
   n3 = GetFocusedLevelSequence(execute=@n2.else)
   > n4 = GetControlRigs(LevelSequence=@n3)
   > Branch(Condition="integer > integer"(A=Length(TargetArray:Array of Control Rig Sequencer Binding Proxy Structures=@n4)))
   > n5 = For Each Loop!(Exec, Array:Array of Control Rig Sequencer Binding Proxy Structures=@n4) -> (
    LoopBody
    Array Element_Proxy_BindingID:Guid Structure
    Array Element_Proxy_Sequence:Movie Scene Sequence Object Reference
    Array Element_ControlRig:Control Rig Object Reference
    Array Element_Track:Movie Scene Control Rig Parameter Track Object Reference
    Array Index
    Completed
   )
   n6 = CurrentControlSelection!(self:Control Rig Object Reference=@n5.Array Element_ControlRig) -> (ReturnValue:Array of Names)
   > Branch(Condition="integer > integer"(A=Length(TargetArray:Array of Names=@n6)))
   > Return Node(OutObject=@n5.Array Element_ControlRig, OutSocket="Get (a copy)"(Array:Array of Names=@n6) -> (Output:Name))
  ]}
 GetText {nodes [
   GetText() {is=title K2Node_FunctionEntry}
   > Return Node!(execute, ReturnValue:Text)
  ]}
 JJF_CopyKeyTimings {nodes [
   JJF_CopyKeyTimings() {is=title K2Node_FunctionEntry}
   > JJF_GetSelection_BindingsObjectsControlNames()
   > Execute Python Script!(
    execute
    PythonScript="import unreal\n\nbp  = unreal.load_asset('/AnimInEdTools/AnimationToolbox/EUW_AnimMisc')\neuw = unreal.get_editor_subsystem(unreal.EditorUtilitySubsystem).find_utility_widget_from_blueprint(bp)\nlib = unreal.ControlRigSequencerLibrary\nseq = unreal.LevelSequenceEditorBlueprintLibrary.get_focused_level_sequence()\n\nproxies = lib.get_control_rigs(seq)\nrig_to_track = {}\nfor proxy in proxies:\n    if proxy.track is not None:\n        rig_to_track[proxy.control_rig] = proxy.track\n\nselection = euw.get_editor_property('STR_CR_ProxyObjName')\nframe_set = set()\nfor entry in selection:\n    rig       = entry.get_editor_property('CR_Object_8_#3')\n    ctrl_name = str(entry.get_editor_property('CR_ControlName_9_#4'))\n    track     = rig_to_track.get(rig)\n    if track is None:\n        continue\n    stk = track.get_section_to_key()\n    sections = [stk] if stk is not None else track.get_sections()\n    for sec in sections:\n        for ch in sec.get_all_channels():\n            ch_name = str(ch.channel_name)\n            if not (ch_name == ctrl_name or ch_name.startswith(ctrl_name + '.')):\n                continue\n            for key in ch.get_keys():\n                frame_set.add(key.get_time().frame_number.value)\n\nsorted_frames = sorted(frame_set)\neuw.set_editor_property('CopiedKeyTimings', sorted_frames)\nunreal.log(f'Key Timings Copied: {len(sorted_frames)} frame(s) -> {sorted_frames}')\n"
   ) -> (then, ReturnValue)
  ]}
 JJF_GetSelection_BindingsObjectsControlNames {nodes [
   JJF_GetSelection_BindingsObjectsControlNames() {is=title K2Node_FunctionEntry}
   > Clear(TargetArray:Array of STR Control Rig Proxy Object Name Tuple Structures=Get STR_CR_ProxyObjName()) {is=title K2Node_CallArrayFunction}
   > n1 = GetFocusedLevelSequence()
   > n2 = GetControlRigs(LevelSequence=@n1)
   > n3 = For Each Loop!(Exec, Array=@n2) -> (LoopBody, Array Element:Control Rig Sequencer Binding Proxy Structure, Array Index, Completed)
   Return Node!(execute=@n3.Completed, CurrentLevelSequence:Level Sequence Object Reference=@n1)
   n4 = Break Control Rig Sequencer Binding Proxy(ControlRigSequencerBindingProxy=@n3)
   > n5 = For Each Loop(Array:Array of Names=CurrentControlSelection(self=@n4.ControlRig)) -> (Array Element:Name)
   > n6 = ControlExistCheck!(execute, self, Control Rig:Control Rig Object Reference=@n4.ControlRig, ControlName_In:Name=@n5) -> (then, ControlExists:bool, ControlName_Out:Name=None) {is=title K2Node_CallFunction}
   > Branch(Condition=@n6)
   > DEBUG(NewParam=Append(A="CtrlSelected: ", B="To String (Name)"(InName=@n5)))
   > Add!(
    execute
    TargetArray:Array of STR Control Rig Proxy Object Name Tuple Structures=Get STR_CR_ProxyObjName()
    NewItem_CR_Proxy_11_#2:Movie Scene Binding Proxy Structure=@n4
    NewItem_CR_Object_8_#3:Control Rig Object Reference=@n4.ControlRig
    NewItem_CR_ControlName_9_#4:Name=@n5
   ) -> (then, ReturnValue)
  ]}
 JJF_PasteKeyTimings {nodes [
   JJF_GetSelection_BindingsObjectsControlNames() {is=title K2Node_CallFunction}
   > n1 = Execute Python Script(PythonScript="import unreal\nimport tkinter as tk\nimport time\nfrom collections import defaultdict\n\nbp  = unreal.load_asset('&EUW_AnimMisc')\neuw = unreal.get_editor_subsystem(unreal.EditorUtilitySubsystem).find_utility_widget_from_blueprint(bp)\nlib     = unreal.ControlRigSequencerLibrary\nseq     = unreal.LevelSequenceEditorBlueprintLibrary.get_focused_level_sequence()\nproxies = lib.get_control_rigs(seq)\n\nrig_to_track = {}\nfor proxy in proxies:\n    if proxy.track is not None:\n        rig_to_track[proxy.control_rig] = proxy.track\n\nanim_layers    = lib.get_anim_layers(False)\nactive_sec_idx = 0\ncancelled      = False\n\nif len(anim_layers) > 1:\n    layer_names   = [str(al.get_fname()) for al in anim_layers]\n    layer_indices = [lib.get_anim_layer_index(al) for al in anim_layers]\n    chosen        = [None]\n\n    time.sleep(0.2)\n\n    dark = False\n    try:\n        import winreg\n        key = winreg.OpenKey(winreg.HKEY_CURRENT_USER,\n                             r'Software\\Microsoft\\Windows\\CurrentVersion\\Themes\\Personalize')\n        val, _ = winreg.QueryValueEx(key, 'AppsUseLightTheme')\n        dark = (val == 0)\n    except Exception:\n        pass\n\n    bg    = '##2b2b2b' if dark else '##f0f0f0'\n    fg    = '##e0e0e0' if dark else '##1a1a1a'\n    selbg = '##0078d4'\n    selfg = '##ffffff'\n    btnbg = '##3c3c3c' if dark else '##e1e1e1'\n\n    root = tk.Tk()\n    root.withdraw()\n\n    popup = tk.Toplevel(root)\n    popup.title('Paste Key Timings')\n    popup.configure(bg=bg)\n    popup.resizable(False, False)\n    popup.attributes('-topmost', True)\n\n    tk.Label(popup, text='Paste timings into which layer?',\n             bg=bg, fg=fg, font=('Segoe UI', 10), padx=12, pady=8).pack(anchor='w')\n\n    listbox = tk.Listbox(popup, selectmode=tk.SINGLE,\n                         bg=bg, fg=fg,\n                         selectbackground=selbg, selectforeground=selfg,\n                         font=('Segoe UI', 10), relief='flat',\n                         highlightthickness=1, highlightcolor=selbg,\n                         borderwidth=0, activestyle='none',\n                         width=32, height=len(layer_names))\n    for name in layer_names:\n        listbox.insert(tk.END, name)\n    listbox.selection_set(0)\n    listbox.pack(padx=12, pady=4)\n\n    def confirm(event=None):\n        sel = listbox.curselection()\n        if sel:\n            chosen[0] = layer_indices[sel[0]]\n        popup.destroy()\n        root.quit()\n\n    def cancel(event=None):\n        popup.destroy()\n        root.quit()\n\n    btn_frame = tk.Frame(popup, bg=bg)\n    btn_frame.pack(padx=12, pady=8, fill='x')\n    tk.Button(btn_frame, text='OK', bg=btnbg, fg=fg, relief='flat',\n              font=('Segoe UI', 10), padx=16, command=confirm).pack(side='right', padx=(4, 0))\n    tk.Button(btn_frame, text='Cancel', bg=btnbg, fg=fg, relief='flat',\n              font=('Segoe UI', 10), padx=8, command=cancel).pack(side='right')\n\n    popup.bind('<Return>', confirm)\n    popup.bind('<Escape>', cancel)\n    listbox.bind('<Double-Button-1>', confirm)\n    popup.protocol('WM_DELETE_WINDOW', cancel)\n\n    mx = root.winfo_pointerx()\n    my = root.winfo_pointery()\n    popup.geometry(f'+{mx}+{my}')\n    popup.deiconify()\n    popup.focus_force()\n    listbox.focus_set()\n    root.mainloop()\n    root.destroy()\n\n    if chosen[0] is None:\n        cancelled = True\n        unreal.log('Paste Key Timings: cancelled.')\n    else:\n        active_sec_idx = chosen[0]\n\nif not cancelled:\n    frame_ints = list(euw.get_editor_property('CopiedKeyTimings'))\n    if not frame_ints:\n        unreal.log_warning('Paste Key Timings: nothing copied yet.')\n    else:\n        frame_numbers = [unreal.FrameNumber(value=f) for f in frame_ints]\n        selection     = euw.get_editor_property('STR_CR_ProxyObjName')\n\n        track_to_controls = defaultdict(list)\n        for entry in selection:\n            rig       = entry.get_editor_property('CR_Object_8_#3')\n            ctrl_name = str(entry.get_editor_property('CR_ControlName_9_#4'))\n            track     = rig_to_track.get(rig)\n            if track is not None:\n                track_to_controls[track].append(ctrl_name)\n\n        for track, ctrl_names in track_to_controls.items():\n            all_sections = track.get_sections()\n            if active_sec_idx < len(all_sections):\n                sections = [all_sections[active_sec_idx]]\n            else:\n                sections = [all_sections[0]] if all_sections else []\n            for sec in sections:\n                lib.key_controls_at_frames(sec, ctrl_names, frame_numbers)\n\n        unreal.log(f'Key Timings Pasted: {len(frame_ints)} frame(s) -> {frame_ints}')\n")
   n2 = JJF_PasteKeyTimings!() -> (then, RemoveNonTiming:bool=~) {is=title K2Node_FunctionEntry}
   n3 = Branch(execute=@n1.then, Condition=@n2)
   > n4 = Execute Python Script(PythonScript="import unreal\nimport tkinter as tk\nimport time\nfrom collections import defaultdict\n\nbp  = unreal.load_asset('&EUW_AnimMisc')\neuw = unreal.get_editor_subsystem(unreal.EditorUtilitySubsystem).find_utility_widget_from_blueprint(bp)\nlib     = unreal.ControlRigSequencerLibrary\nseq     = unreal.LevelSequenceEditorBlueprintLibrary.get_focused_level_sequence()\nproxies = lib.get_control_rigs(seq)\n\nrig_to_track = {}\nfor proxy in proxies:\n    if proxy.track is not None:\n        rig_to_track[proxy.control_rig] = proxy.track\n\nanim_layers    = lib.get_anim_layers(False)\nactive_sec_idx = 0\ncancelled      = False\n\nif len(anim_layers) > 1:\n    layer_names   = [str(al.get_fname()) for al in anim_layers]\n    layer_indices = [lib.get_anim_layer_index(al) for al in anim_layers]\n    chosen        = [None]\n\n    time.sleep(0.2)\n\n    dark = False\n    try:\n        import winreg\n        key = winreg.OpenKey(winreg.HKEY_CURRENT_USER,\n                             r'Software\\Microsoft\\Windows\\CurrentVersion\\Themes\\Personalize')\n        val, _ = winreg.QueryValueEx(key, 'AppsUseLightTheme')\n        dark = (val == 0)\n    except Exception:\n        pass\n\n    bg    = '##2b2b2b' if dark else '##f0f0f0'\n    fg    = '##e0e0e0' if dark else '##1a1a1a'\n    selbg = '##0078d4'\n    selfg = '##ffffff'\n    btnbg = '##3c3c3c' if dark else '##e1e1e1'\n\n    root = tk.Tk()\n    root.withdraw()\n\n    popup = tk.Toplevel(root)\n    popup.title('Retime — Pick Layer')\n    popup.configure(bg=bg)\n    popup.resizable(False, False)\n    popup.attributes('-topmost', True)\n\n    tk.Label(popup, text='Remove non-timing keys from which layer?',\n             bg=bg, fg=fg, font=('Segoe UI', 10), padx=12, pady=8).pack(anchor='w')\n\n    listbox = tk.Listbox(popup, selectmode=tk.SINGLE,\n                         bg=bg, fg=fg,\n                         selectbackground=selbg, selectforeground=selfg,\n                         font=('Segoe UI', 10), relief='flat',\n                         highlightthickness=1, highlightcolor=selbg,\n                         borderwidth=0, activestyle='none',\n                         width=32, height=len(layer_names))\n    for name in layer_names:\n        listbox.insert(tk.END, name)\n    listbox.selection_set(0)\n    listbox.pack(padx=12, pady=4)\n\n    def confirm(event=None):\n        sel = listbox.curselection()\n        if sel:\n            chosen[0] = layer_indices[sel[0]]\n        popup.destroy()\n        root.quit()\n\n    def cancel(event=None):\n        popup.destroy()\n        root.quit()\n\n    btn_frame = tk.Frame(popup, bg=bg)\n    btn_frame.pack(padx=12, pady=8, fill='x')\n    tk.Button(btn_frame, text='OK', bg=btnbg, fg=fg, relief='flat',\n              font=('Segoe UI', 10), padx=16, command=confirm).pack(side='right', padx=(4, 0))\n    tk.Button(btn_frame, text='Cancel', bg=btnbg, fg=fg, relief='flat',\n              font=('Segoe UI', 10), padx=8, command=cancel).pack(side='right')\n\n    popup.bind('<Return>', confirm)\n    popup.bind('<Escape>', cancel)\n    listbox.bind('<Double-Button-1>', confirm)\n    popup.protocol('WM_DELETE_WINDOW', cancel)\n\n    mx = root.winfo_pointerx()\n    my = root.winfo_pointery()\n    popup.geometry(f'+{mx}+{my}')\n    popup.deiconify()\n    popup.focus_force()\n    listbox.focus_set()\n    root.mainloop()\n    root.destroy()\n\n    if chosen[0] is None:\n        cancelled = True\n        unreal.log('Retime: cancelled.')\n    else:\n        active_sec_idx = chosen[0]\n\nif not cancelled:\n    kept_frames = set(euw.get_editor_property('CopiedKeyTimings'))\n    selection   = euw.get_editor_property('STR_CR_ProxyObjName')\n\n    track_to_controls = defaultdict(list)\n    for entry in selection:\n        rig       = entry.get_editor_property('CR_Object_8_#3')\n        ctrl_name = str(entry.get_editor_property('CR_ControlName_9_#4'))\n        track     = rig_to_track.get(rig)\n        if track is not None:\n            track_to_controls[track].append(ctrl_name)\n\n    removed_total = 0\n    for track, ctrl_names in track_to_controls.items():\n        ctrl_name_set = set(ctrl_names)\n        all_sections  = track.get_sections()\n        if active_sec_idx < len(all_sections):\n            sections = [all_sections[active_sec_idx]]\n        else:\n            sections = [all_sections[0]] if all_sections else []\n        for sec in sections:\n            for ch in sec.get_all_channels():\n                ctrl = str(ch.channel_name).split('.')[0]\n                if ctrl not in ctrl_name_set:\n                    continue\n                for key in ch.get_keys():\n                    if key.get_time().frame_number.value not in kept_frames:\n                        ch.remove_key(key)\n                        removed_total += 1\n\n    unreal.log(f'Retime: removed {removed_total} out-of-timing keys')\n")
   n5 = "To String (Integer)"(InInt=Length(TargetArray:Array of Integers=Get CopiedKeyTimings() -> (CopiedKeyTimings:Array of Integers)))
   PrintString(execute=[@n4.then, @n3.else], InString=Append(A="Key Timings Pasted - ", B=Append(A=@n5, B=" frame(s)")), Duration=3)
   JJF_GetSelection_BindingsObjectsControlNames()
   > Execute Python Script(PythonScript="import unreal\nfrom collections import defaultdict\n\nbp  = unreal.load_asset('&EUW_AnimMisc')\neuw = unreal.get_editor_subsystem(unreal.EditorUtilitySubsystem).find_utility_widget_from_blueprint(bp)\n\nlib     = unreal.ControlRigSequencerLibrary\nseq     = unreal.LevelSequenceEditorBlueprintLibrary.get_focused_level_sequence()\nproxies = lib.get_control_rigs(seq)\n\nrig_to_track = {}\nfor proxy in proxies:\n    if proxy.track is not None:\n        rig_to_track[proxy.control_rig] = proxy.track\n\nanim_layers = lib.get_anim_layers(False)\nif len(anim_layers) > 1:\n    active_sec_idx = lib.get_anim_layer_index(anim_layers[-1])\nelse:\n    active_sec_idx = None\n\nframe_ints = list(euw.get_editor_property('CopiedKeyTimings'))\nif not frame_ints:\n    unreal.log_warning('Paste Key Timings: nothing copied yet.')\nelse:\n    frame_numbers = [unreal.FrameNumber(value=f) for f in frame_ints]\n    selection     = euw.get_editor_property('STR_CR_ProxyObjName')\n\n    track_to_controls = defaultdict(list)\n    for entry in selection:\n        rig       = entry.get_editor_property('CR_Object_8_#3')\n        ctrl_name = str(entry.get_editor_property('CR_ControlName_9_#4'))\n        track     = rig_to_track.get(rig)\n        if track is not None:\n            track_to_controls[track].append(ctrl_name)\n\n    for track, ctrl_names in track_to_controls.items():\n        all_sections = track.get_sections()\n        if active_sec_idx is not None and active_sec_idx < len(all_sections):\n            sections = [all_sections[active_sec_idx]]\n        else:\n            sections = [all_sections[0]] if all_sections else []\n        for sec in sections:\n            lib.key_controls_at_frames(sec, ctrl_names, frame_numbers)\n\n    unreal.log(f'Key Timings Pasted: {len(frame_ints)} frame(s) -> {frame_ints}')\n")
   > n6 = Branch(Condition=@n2)
   > n7 = Execute Python Script(PythonScript="import unreal\nfrom collections import defaultdict\n\nbp  = unreal.load_asset('&EUW_AnimMisc')\neuw = unreal.get_editor_subsystem(unreal.EditorUtilitySubsystem).find_utility_widget_from_blueprint(bp)\n\nlib     = unreal.ControlRigSequencerLibrary\nseq     = unreal.LevelSequenceEditorBlueprintLibrary.get_focused_level_sequence()\nproxies = lib.get_control_rigs(seq)\n\nrig_to_track = {}\nfor proxy in proxies:\n    if proxy.track is not None:\n        rig_to_track[proxy.control_rig] = proxy.track\n\nanim_layers = lib.get_anim_layers(False)\nif len(anim_layers) > 1:\n    active_sec_idx = lib.get_anim_layer_index(anim_layers[-1])\nelse:\n    active_sec_idx = None\n\nkept_frames = set(euw.get_editor_property('CopiedKeyTimings'))\nselection   = euw.get_editor_property('STR_CR_ProxyObjName')\n\ntrack_to_controls = defaultdict(list)\nfor entry in selection:\n    rig       = entry.get_editor_property('CR_Object_8_#3')\n    ctrl_name = str(entry.get_editor_property('CR_ControlName_9_#4'))\n    track     = rig_to_track.get(rig)\n    if track is not None:\n        track_to_controls[track].append(ctrl_name)\n\nremoved_total = 0\nfor track, ctrl_names in track_to_controls.items():\n    ctrl_name_set = set(ctrl_names)\n    all_sections  = track.get_sections()\n    if active_sec_idx is not None and active_sec_idx < len(all_sections):\n        sections = [all_sections[active_sec_idx]]\n    else:\n        sections = [all_sections[0]] if all_sections else []\n    for sec in sections:\n        for ch in sec.get_all_channels():\n            ctrl = str(ch.channel_name).split('.')[0]\n            if ctrl not in ctrl_name_set:\n                continue\n            for key in ch.get_keys():\n                if key.get_time().frame_number.value not in kept_frames:\n                    ch.remove_key(key)\n                    removed_total += 1\n\nunreal.log(f'Retime: removed {removed_total} out-of-timing keys')\n")
   n8 = "To String (Integer)"(InInt=Length(TargetArray=Get CopiedKeyTimings()))
   PrintString(execute=[@n7.then, @n6.else], InString=Append(A="Key Timings Pasted - ", B=Append(A=@n8, B=" frame(s)")), Duration=3)
   n9 = "To String (Integer)"(InInt=Length(TargetArray=Get CopiedKeyTimings()))
   PrintString(execute=[@n11.then, @n10.else], InString=Append(A="Key Timings Pasted - ", B=Append(A=@n9, B=" frame(s)")), Duration=3)
   JJF_GetSelection_BindingsObjectsControlNames(execute=@n2.then)
   > Execute Python Script(PythonScript="import unreal\nfrom collections import defaultdict\n\nbp  = unreal.load_asset('&EUW_AnimMisc')\neuw = unreal.get_editor_subsystem(unreal.EditorUtilitySubsystem).find_utility_widget_from_blueprint(bp)\nlib     = unreal.ControlRigSequencerLibrary\nseq     = unreal.LevelSequenceEditorBlueprintLibrary.get_focused_level_sequence()\nproxies = lib.get_control_rigs(seq)\n\nrig_to_track = {}\nfor proxy in proxies:\n    if proxy.track is not None:\n        rig_to_track[proxy.control_rig] = proxy.track\n\nframe_ints = list(euw.get_editor_property('CopiedKeyTimings'))\nif not frame_ints:\n    unreal.log_warning('Paste Key Timings: nothing copied yet.')\nelse:\n    frame_numbers = [unreal.FrameNumber(value=f) for f in frame_ints]\n    selection     = euw.get_editor_property('STR_CR_ProxyObjName')\n\n    track_to_controls = defaultdict(list)\n    for entry in selection:\n        rig       = entry.get_editor_property('CR_Object_8_#3')\n        ctrl_name = str(entry.get_editor_property('CR_ControlName_9_#4'))\n        track     = rig_to_track.get(rig)\n        if track is not None:\n            track_to_controls[track].append(ctrl_name)\n\n    for track, ctrl_names in track_to_controls.items():\n        stk = track.get_section_to_key()\n        sections = [stk] if stk is not None else track.get_sections()\n        for sec in sections:\n            lib.key_controls_at_frames(sec, ctrl_names, frame_numbers)\n\n    unreal.log(f'Key Timings Pasted: {len(frame_ints)} frame(s) -> {frame_ints}')\n")
   > n10 = Branch(Condition=@n2)
   > n11 = Execute Python Script(PythonScript="import unreal\nfrom collections import defaultdict\n\nbp  = unreal.load_asset('&EUW_AnimMisc')\neuw = unreal.get_editor_subsystem(unreal.EditorUtilitySubsystem).find_utility_widget_from_blueprint(bp)\nlib     = unreal.ControlRigSequencerLibrary\nseq     = unreal.LevelSequenceEditorBlueprintLibrary.get_focused_level_sequence()\nproxies = lib.get_control_rigs(seq)\n\nrig_to_track = {}\nfor proxy in proxies:\n    if proxy.track is not None:\n        rig_to_track[proxy.control_rig] = proxy.track\n\nkept_frames = set(euw.get_editor_property('CopiedKeyTimings'))\nselection   = euw.get_editor_property('STR_CR_ProxyObjName')\n\ntrack_to_controls = defaultdict(list)\nfor entry in selection:\n    rig       = entry.get_editor_property('CR_Object_8_#3')\n    ctrl_name = str(entry.get_editor_property('CR_ControlName_9_#4'))\n    track     = rig_to_track.get(rig)\n    if track is not None:\n        track_to_controls[track].append(ctrl_name)\n\nremoved_total = 0\nfor track, ctrl_names in track_to_controls.items():\n    ctrl_name_set = set(ctrl_names)\n    stk = track.get_section_to_key()\n    sections = [stk] if stk is not None else track.get_sections()\n    for sec in sections:\n        for ch in sec.get_all_channels():\n            ctrl = str(ch.channel_name).split('.')[0]\n            if ctrl not in ctrl_name_set:\n                continue\n            for key in ch.get_keys():\n                if key.get_time().frame_number.value not in kept_frames:\n                    ch.remove_key(key)\n                    removed_total += 1\n\nunreal.log(f'Retime: removed {removed_total} out-of-timing keys')\n")
  ]}
 Locator_Bake_Connect_wOffset {nodes [
   n1 = Locator_Bake_Connect_wOffset!() -> (then, Constrain:bool=~) {is=title K2Node_FunctionEntry}
   > n2 = Set Locator_Constrain(Locator_Constrain:bool=@n1) -> (Output_Get:bool)
   > n3 = Branch(Condition=@n2)
   > n4 = Set locatorColor(locatorColor:String=(R=0.7157,G=0.0648,B=0.3916,A=1)) -> (Output_Get:String)
   n5 = EUW_MakeLocator(execute=[@n10.then, @n4.then], Locator Color=Get locatorColor() -> (locatorColor:String)) -> (Out_Loc_CtrlName=None, In_Control=None)
   > Set Locator_Loc_CR(Locator_Loc_CR:Control Rig Object Reference=@n5) -> (Output_Get:Control Rig Object Reference)
   > Set Locator_Loc_Ctrl(Locator_Loc_Ctrl:Name=@n5.Out_Loc_CtrlName) -> (Output_Get:Name=None)
   > Set Locator_In_CR(Locator_In_CR:Control Rig Object Reference=@n5.In_CR) -> (Output_Get:Control Rig Object Reference)
   > Set Locator_In_Ctrl(Locator_In_Ctrl:Name=@n5.In_Control) -> (Output_Get:Name=None)
   > n6 = Execute Python Script!(
    execute
    PythonScript="import unreal\r\nLSEBL = unreal.LevelSequenceEditorBlueprintLibrary\r\nls = LSEBL.get_focused_level_sequence()\r\nif not ls:\r\n    unreal.log_error(\"Offset Snap A: No level sequence is open.\")\r\n    raise SystemExit\r\ndef get_selected_key_range():\r\n    all_frames = []\r\n    ses = unreal.get_editor_subsystem(unreal.LevelSequenceEditorSubsystem)\r\n    curve_editor = ses.get_curve_editor()\r\n    if curve_editor.is_curve_editor_open():\r\n        for ch_proxy in curve_editor.get_channels_with_selected_keys():\r\n            indices = curve_editor.get_selected_keys(ch_proxy)\r\n            if not indices:\r\n                continue\r\n            channel_obj = ch_proxy.section.get_channel(ch_proxy.channel_name)\r\n            if not channel_obj:\r\n                continue\r\n            for k in channel_obj.get_keys_by_index(indices):\r\n                all_frames.append(k.get_time().frame_number.value)\r\n    if not all_frames:\r\n        for ch_proxy in LSEBL.get_channels_with_selected_keys():\r\n            indices = LSEBL.get_selected_keys(ch_proxy)\r\n            if not indices:\r\n                continue\r\n            channel_obj = ch_proxy.section.get_channel(ch_proxy.channel_name)\r\n            if not channel_obj:\r\n                continue\r\n            for k in channel_obj.get_keys_by_index(indices):\r\n                all_frames.append(k.get_time().frame_number.value)\r\n    return (min(all_frames), max(all_frames)) if all_frames else None\r\n\r\nselected_range = get_selected_key_range()\r\nif selected_range:\r\n    start_val, end_val = selected_range\r\nelse:\r\n    start_val = ls.get_playback_start()\r\n    end_val   = ls.get_playback_end() - 1"
   ) -> (then, ReturnValue, start_val:int, end_val:int) {outputs [start_val,end_val]}
   > Set Range Start(Range Start:int=@n6.start_val) -> (Output_Get:int=0)
   > Set Range End(Range End:int=@n6.end_val) -> (Output_Get:int=0)
   n7 = Get EditorUtilitySubsystem()
   > n8 = FindUtilityWidgetFromBlueprint(self=@n7, InBlueprint=&/EAT_BakeOffset)
   > n9 = Is Valid(InputObject=@n8)
   SpawnAndRegisterTab(execute=@n9.Is Not Valid, self=@n7, InBlueprint=&EAT_BakeOffset)
   ReleaseInstanceOfAsset(execute=@n9.Is Valid, self=@n7, Asset=&EAT_BakeOffset)
   n10 = Set locatorColor(execute=@n3.else, locatorColor=(R=0.1878,G=0.3419,B=0.0999,A=1))
  ]}
 MoveSpaceKey_Left {nodes [
   MoveSpaceKey_Left() {is=title K2Node_FunctionEntry}
   > n1 = Sequence!(execute) -> (then_0, then_1, then_2:Exec)
   > n2 = GetFocusedLevelSequence()
   > Set in_sequence(in_sequence:Level Sequence Object Reference=@n2) -> (Output_Get:Level Sequence Object Reference)
   > n3 = GetControlRigs(LevelSequence=@n2)
   > n4 = For Each Loop(Array:Array of Control Rig Sequencer Binding Proxy Structures=@n3) -> (Array Element:Control Rig Sequencer Binding Proxy Structure)
   > Set in_controlRig(in_controlRig:Control Rig Object Reference=Break Control Rig Sequencer Binding Proxy(ControlRigSequencerBindingProxy=@n4).ControlRig) -> (Output_Get:Control Rig Object Reference)
   > n5 = Get Current Time(TimeUnit=DisplayRate) -> (ReturnValue_Frame_SubFrame=~, ReturnValue_Time=~, ReturnValue_bHasJumped=~)
   > n6 = Set in_frame(in_frame:Frame Number Structure=@n5) -> (Output_Get:Frame Number Structure)
   n7 = FrameNumber to Integer!(InFrameNumber:"Frame Number Structure (by ref)"=@n6) -> (ReturnValue:int=0)
   > n8 = Set next_Frame(next_Frame:Frame Number Structure=Make Frame Number(Value=int + int(A=@n7, B=-1))) -> (Output_Get:Frame Number Structure)
   Set Current Time!(
    execute=[@n8.then, @n10.then]
    PlaybackParams_Frame_FrameNumber:Frame Number Structure=Get next_Frame() -> (next_Frame:Frame Number Structure)
    PlaybackParams_Frame_SubFrame:float=0
    PlaybackParams_Time:float=0
    PlaybackParams_MarkedFrame:String
    PlaybackParams_Timecode:Timecode Structure=(Hours=0,Minutes=0,Seconds=0,Frames=0,Subframe=0,bDropFrameFormat=False)
    PlaybackParams_PositionType:EMovieScenePositionType Enum=Frame
    PlaybackParams_UpdateMethod:EUpdatePositionMethod Enum=Play
    PlaybackParams_bHasJumped:bool=False
    TimeUnit:EMovieSceneTimeUnit Enum=DisplayRate
   ) -> (then) {is=title K2Node_CallFunction}
   > WorldspaceXform_Copy()
   > Set Current Time(
    PlaybackParams_Frame_FrameNumber=Get in_frame() -> (in_frame:Frame Number Structure)
    PlaybackParams_Frame_SubFrame=0
    PlaybackParams_Time=0
    PlaybackParams_Timecode=(Hours=0,Minutes=0,Seconds=0,Frames=0,Subframe=0,bDropFrameFormat=False)
    PlaybackParams_PositionType=Frame
    PlaybackParams_UpdateMethod=Play
    PlaybackParams_bHasJumped=False
    TimeUnit=DisplayRate
   )
   Execute Python Script!(
    execute=@n1.then_1
    PythonScript="import unreal\n\nlib       = unreal.ControlRigSequencerLibrary\nsequence  = unreal.LevelSequenceEditorBlueprintLibrary.get_focused_level_sequence()\ncur_frame = unreal.LevelSequenceEditorBlueprintLibrary.get_current_time()\nsrc_frame = int(cur_frame)\ndst_frame = src_frame - 1\n\nproxies = lib.get_control_rigs(sequence)\nif not proxies:\n    raise RuntimeError('No Control Rig in sequence')\nproxy    = proxies[0]\nrig      = proxy.control_rig\nselected = rig.current_control_selection()\nif not selected:\n    raise RuntimeError('No control selected')\n\nPARENT = unreal.MovieSceneControlRigSpaceType.PARENT\nWORLD  = unreal.MovieSceneControlRigSpaceType.WORLD\n\ndef resolve_rig_element_key(base_key):\n    if base_key.space_type == PARENT:\n        return lib.get_default_parent_key()\n    elif base_key.space_type == WORLD:\n        return lib.get_world_space_reference_key()\n    else:\n        return base_key.control_rig_element\n\nfor control_name in selected:\n    control_name = str(control_name)\n    channel = lib.get_space_channel(sequence, rig, control_name)\n\n    ## find the active space at src_frame (step/constant -- last key at-or-before)\n    rig_element_key = lib.get_default_parent_key()\n    if channel is not None:\n        for key in channel.get_keys():\n            if key.get_time().frame_number.value <= src_frame:\n                rig_element_key = resolve_rig_element_key(key.get_value())\n            else:\n                break\n\n    unreal.LevelSequenceEditorBlueprintLibrary.set_current_time(dst_frame)\n    lib.set_control_rig_space(\n        sequence, rig, control_name, rig_element_key, unreal.FrameNumber(value=dst_frame)\n    )\n"
   ) -> (then, ReturnValue)
   WorldsaceXform_Paste(execute=@n1.then_2)
   n9 = Get in_controlRig() -> (in_controlRig:Control Rig Object Reference)
   n10 = Set in_controlName(in_controlName:Name="Get (a copy)"(Array=CurrentControlSelection(self=@n9))) -> (Output_Get:Name=None)
  ]}
 MoveSpaceKey_Right {nodes [
   MoveSpaceKey_Right() {is=title K2Node_FunctionEntry}
   > n1 = GetFocusedLevelSequence()
   > n2 = GetControlRigs(LevelSequence=@n1)
   > n3 = For Each Loop(Array=@n2)
   > n4 = Get Current Time(TimeUnit=DisplayRate) -> (ReturnValue_Frame_SubFrame=~, ReturnValue_Time=~, ReturnValue_bHasJumped=~)
   > CopyWorldSpace_CurrentFrame() {is=title K2Node_CallFunction}
   n5 = FrameNumber to Integer(InFrameNumber=@n4) -> (ReturnValue=0)
   n6 = Break Control Rig Sequencer Binding Proxy(ControlRigSequencerBindingProxy=@n3)
   > MoveControlRigSpace!(
    execute
    InSequence:Level Sequence Object Reference=@n1
    InControlRig:Control Rig Object Reference=@n6.ControlRig
    InControlName:Name=For Each Loop(Array:Array of Names=CurrentControlSelection(self=@n6.ControlRig)) -> (Array Element:Name)
    InTime:Frame Number Structure=@n4
    InNewTime:Frame Number Structure=Make Frame Number(Value=int + int(A=@n5))
    TimeUnit:EMovieSceneTimeUnit Enum=DisplayRate
   ) -> (then, ReturnValue:bool)
   > PasteWorldSpace_CurrentFrame()
   > n7 = Get Current Time(TimeUnit=DisplayRate) -> (ReturnValue_Frame_SubFrame=~, ReturnValue_Time=~, ReturnValue_bHasJumped=~)
   n8 = FrameNumber to Integer(InFrameNumber=@n7) -> (ReturnValue=0)
   > Set Current Time(
    PlaybackParams_Frame_FrameNumber=Make Frame Number(Value=int + int(A=@n8))
    PlaybackParams_Frame_SubFrame=0
    PlaybackParams_Time=0
    PlaybackParams_Timecode=(Hours=0,Minutes=0,Seconds=0,Frames=0,Subframe=0,bDropFrameFormat=False)
    PlaybackParams_PositionType=Frame
    PlaybackParams_UpdateMethod=Play
    PlaybackParams_bHasJumped=False
    TimeUnit=DisplayRate
   )
  ]}
 OnMouseButtonDown {nodes [
   n1 = OnMouseButtonDown!() -> (then, MyGeometry:Geometry Structure, MouseEvent:"Pointer Event Structure (by ref)") {is=title K2Node_FunctionEntry}
   > Set MouseEvent_Var(MouseEvent_Var:Pointer Event Structure=@n1.MouseEvent) -> (Output_Get:Pointer Event Structure)
   > n2 = Sequence!(execute) -> (then_0, then_1)
   > n3 = Branch(Condition=Is Mouse Button Down!(
     Input:"Pointer Event Structure (by ref)"=Get MouseEvent_Var() -> (MouseEvent_Var:Pointer Event Structure)
     MouseButton:Key Structure=RightMouseButton
    ) -> (ReturnValue:bool))
   > n4 = Branch(Condition=IsHovered(self=Get BTN_Locator_Bake_Connect() -> (BTN_Locator_Bake_Connect:Editor Utility Button Object Reference)))
   > Locator_Bake_Connect_wOffset!(execute, self, Constrain:bool=true) -> (then) {is=title K2Node_CallFunction}
   n5 = Branch(execute=@n4.else, Condition=IsHovered(self=Get BTN_Locator_Bake() -> (BTN_Locator_Bake:Editor Utility Button Object Reference)))
   > Locator_Bake_Connect_wOffset(Constrain=false)
   Branch(execute=@n5.else, Condition=IsHovered(self=Get BTN_Copy_KeyTiming() -> (BTN_Copy_KeyTiming:Button Object Reference)))
   > JJF_CopyKeyTimings() {is=title K2Node_CallFunction}
   Branch(execute=@n3.else, Condition=Is Mouse Button Down(Input=Get MouseEvent_Var(), MouseButton=MiddleMouseButton))
   n6 = Get BTN_Copy_KeyTiming()
   n7 = GetModifierKeysState!() -> (ReturnValue:Modifier Keys State Structure)
   > n8 = Branch(Condition=AND Boolean(A=IsHovered(self=@n6), B=Is Control Down!(KeysState:"Modifier Keys State Structure (by ref)"=@n7) -> (ReturnValue:bool)))
   > JJF_PasteKeyTimings(RemoveNonTiming=true) {is=title K2Node_CallFunction}
   Branch(execute=@n8.else, Condition=IsHovered(self=Get BTN_Copy_KeyTiming()))
   > JJF_PasteKeyTimings(RemoveNonTiming=false)
   Return Node!(execute=@n2.then_1, ReturnValue:Event Reply Structure=Handled())
  ]}
 PasteWorldSpace_AllFrames {nodes [
   PasteWorldSpace_AllFrames() {is=title K2Node_FunctionEntry}
   > n1 = Get Current Local Time(TimeUnit=DisplayRate) -> (ReturnValue_Frame_SubFrame=~, ReturnValue_Time=~, ReturnValue_bHasJumped=~)
   > n2 = GetFocusedLevelSequence()
   > n3 = Keys!(execute, TargetMap:Map of Control Rig Object References to Struct Name Transform Map Structures=Get Copy Paste Control Rig Proxy Bindings()) -> (then, Keys:Array of Control Rig Object References)
   > n4 = For Each Loop(Array:Array of Control Rig Object References=@n3) -> (Array Element:Control Rig Object Reference)
   > Is Valid(InputObject=@n4)
   n5 = Find!(
    TargetMap:Map of Control Rig Object References to Struct Name Transform Map Structures=Get Copy Paste Control Rig Proxy Bindings()
    Key:"Control Rig Object Reference (by ref)"=@n4
   ) -> (Value_NameTransform_4_#1:Map of Names to Transforms, ReturnValue)
   > Branch(Condition=@n5.ReturnValue)
   > n6 = Keys(TargetMap:Map of Names to Transforms=@n5) -> (Keys:Array of Names)
   > n7 = For Each Loop(Array:Array of Names=@n6) -> (Array Element:Name)
   > SetControlRigWorldTransform(
    LevelSequence=@n2
    ControlRig=@n4
    ControlName=@n7
    Frame=@n1
    WorldTransform=Find!(TargetMap:Map of Names to Transforms=@n5, Key:"Name (by ref)"=@n7) -> (Value:Transform, ReturnValue)
    TimeUnit=DisplayRate
    bSetKey=true
   )
  ]}
 PasteWorldSpace_CurrentFrame {nodes [
   PasteWorldSpace_CurrentFrame() {is=title K2Node_FunctionEntry}
   > n1 = Get Current Local Time(TimeUnit=DisplayRate) -> (ReturnValue_Frame_SubFrame=~, ReturnValue_Time=~, ReturnValue_bHasJumped=~)
   > n2 = GetFocusedLevelSequence()
   > n3 = Keys(TargetMap:Map of Control Rig Object References to Struct Name Transform Map Structures=Get Copy Paste Control Rig Proxy Bindings()) -> (Keys:Array of Control Rig Object References)
   > n4 = For Each Loop(Array:Array of Control Rig Object References=@n3) -> (Array Element:Control Rig Object Reference)
   > Is Valid(InputObject=@n4)
   n5 = Find!(
    TargetMap:Map of Control Rig Object References to Struct Name Transform Map Structures=Get Copy Paste Control Rig Proxy Bindings()
    Key:"Control Rig Object Reference (by ref)"=@n4
   ) -> (Value_NameTransform_4_#1:Map of Names to Transforms, ReturnValue)
   > Branch(Condition=@n5.ReturnValue)
   > n6 = Keys(TargetMap:Map of Names to Transforms=@n5) -> (Keys:Array of Names)
   > n7 = For Each Loop(Array:Array of Names=@n6) -> (Array Element:Name)
   > SetControlRigWorldTransform(
    LevelSequence=@n2
    ControlRig=@n4
    ControlName=@n7
    Frame=@n1
    WorldTransform=Find!(TargetMap:Map of Names to Transforms=@n5, Key:"Name (by ref)"=@n7) -> (Value:Transform, ReturnValue)
    TimeUnit=DisplayRate
    bSetKey=true
   )
  ]}
 Pre/Post Curve Interp {nodes [
   n1 = Get Current Time!(execute, TimeUnit=DisplayRate) -> (then, ReturnValue:Movie Scene Sequence Playback Params Structure)
   n2 = Break Movie Scene Sequence Playback Params!(MovieSceneSequencePlaybackParams:"Movie Scene Sequence Playback Params Structure (by ref)"=@n1) -> (
    Frame_FrameNumber_Value:int
    Frame_SubFrame:float=~
    Time:float=~
    MarkedFrame:String
    Timecode:Timecode Structure
    PositionType:EMovieScenePositionType Enum
    UpdateMethod:EUpdatePositionMethod Enum
    bHasJumped:bool=~
   ) {is=title K2Node_BreakStruct}
   > Set Current Time!(
    execute
    PlaybackParams:Movie Scene Sequence Playback Params Structure=Make Movie Scene Sequence Playback Params!(
     Frame_FrameNumber_Value:int=int + int(A=@n2)
     Frame_SubFrame:float=0
     Time:float=0
     MarkedFrame:String
     Timecode:Timecode Structure=(Hours=0,Minutes=0,Seconds=0,Frames=0,Subframe=0,bDropFrameFormat=False)
     PositionType:EMovieScenePositionType Enum=Frame
     UpdateMethod:EUpdatePositionMethod Enum=Play
     bHasJumped:bool=False
    ) -> (MovieSceneSequencePlaybackParams:Movie Scene Sequence Playback Params Structure) {is=title K2Node_MakeStruct}
    TimeUnit=DisplayRate
   ) -> (then)
   > Set Current Time(PlaybackParams=@n1, TimeUnit=DisplayRate)
   n3 = Pre/Post Curve Interp!() -> (then, InExtrapolation:ERichCurveExtrapolation Enum=RCCE_Cycle) {is=title K2Node_FunctionEntry}
   > n4 = For Each Loop(Array:Array of Sequencer Channel Proxy Structures=GetSelectedChannels!() -> (ReturnValue:Array of Sequencer Channel Proxy Structures)) -> (Array Element:Sequencer Channel Proxy Structure)
   n5 = Break Sequencer Channel Proxy!(SequencerChannelProxy:"Sequencer Channel Proxy Structure (by ref)"=@n4) -> (ChannelName:Name, Section:Movie Scene Section Object Reference) {is=title K2Node_BreakStruct}
   > n6 = GetChannel!(execute, Section:Movie Scene Section Object Reference=@n5.Section, ChannelName:"Name (by ref)"=@n5) -> (then, ReturnValue:Movie Scene Scripting Channel Object Reference)
   > n7 = Cast To MovieSceneScriptingFloatChannel!(execute, Object=@n6) -> (then, CastFailed, AsMovie Scene Scripting Float Channel:Movie Scene Scripting Float Channel Object Reference) {is=title K2Node_DynamicCast}
   > SetPreInfinityExtrapolation!(execute, self:Movie Scene Scripting Float Channel Object Reference=@n7, InExtrapolation:ERichCurveExtrapolation Enum=@n3) -> (then)
   > SetPostInfinityExtrapolation!(execute, self:Movie Scene Scripting Float Channel Object Reference=@n7, InExtrapolation:ERichCurveExtrapolation Enum=@n3) -> (then)
   > "Set Default (Float)"!(execute, self:Movie Scene Scripting Float Channel Object Reference=@n7, InDefaultValue:float=0) -> (then) {is=title K2Node_CallFunction}
  ]}
 Scan For CRU_AddLocatorRigs {nodes [
   Scan For CRU_AddLocatorRigs!() -> (then, Local_LevelSequence:Level Sequence Object Reference, Local_LocatorCRU_Track:Movie Scene Track Object Reference) {is=title K2Node_FunctionEntry}
   > n1 = GetFocusedLevelSequence()
   > n2 = GetControlRigs(LevelSequence=@n1)
   > n3 = For Each Loop(Array:Array of Control Rig Sequencer Binding Proxy Structures=@n2) -> (Array Element:Control Rig Sequencer Binding Proxy Structure)
   Return Node!(
    execute=@n3.Completed
    Found:bool=Get rigExistsCheck() -> (rigExistsCheck:bool)
    Out_ControlRig:Control Rig Object Reference=Get locControlRigFound() -> (locControlRigFound:Control Rig Object Reference)
   )
   n4 = Break Control Rig Sequencer Binding Proxy(ControlRigSequencerBindingProxy=@n3)
   n5 = "Equal (Class)"(A=Get Class(Object=@n4.ControlRig) -> (ReturnValue:Control Rig Class Reference), B=&CRU_AddLocator_C)
   > Branch(Condition=@n5)
   > Set rigExistsCheck(rigExistsCheck:bool=true) -> (Output_Get:bool)
   > Set locControlRigFound(locControlRigFound:Control Rig Object Reference=@n4.ControlRig) -> (Output_Get:Control Rig Object Reference)
   AND Boolean(A=@n5, B="Equal (Object)"(A:Object Reference=@n4.Track, B=Get Local_LocatorCRU_Track()))
  ]}
 WorldsaceXform_Paste {nodes [
   WorldsaceXform_Paste() {is=title K2Node_FunctionEntry}
   > n1 = Get Current Time!(execute, TimeUnit=DisplayRate) -> (
    then
    ReturnValue_Frame_FrameNumber_Value:int
    ReturnValue_Frame_SubFrame:float=~
    ReturnValue_Time:float=~
    ReturnValue_MarkedFrame:String
    ReturnValue_Timecode:Timecode Structure
    ReturnValue_PositionType:EMovieScenePositionType Enum
    ReturnValue_UpdateMethod:EUpdatePositionMethod Enum
    ReturnValue_bHasJumped:bool=~
   )
   > n2 = JJF_GetSelection_BindingsObjectsControlNames()
   > n3 = For Each Loop!(Exec, Array:Array of STR Control Rig Proxy Object Name Tuple Structures=Get STR_CR_ProxyObjName()) -> (
    LoopBody
    Array Element_CR_Proxy_11_#2:Movie Scene Binding Proxy Structure
    Array Element_CR_Object_8_#3:Control Rig Object Reference
    Array Element_CR_ControlName_9_#4:Name
    Array Index
    Completed
   )
   > SetControlRigWorldTransform!(
    execute
    LevelSequence=@n2
    ControlRig=@n3.Array Element_CR_Object_8_#3
    ControlName=@n3.Array Element_CR_ControlName_9_#4
    Frame_Value:int=@n1
    WorldTransform=Get WorldspaceXform() -> (WorldspaceXform:Transform)
    TimeUnit=DisplayRate
    bSetKey=true
   ) -> (then)
  ]}
 WorldspaceXform_Copy {nodes [
   WorldspaceXform_Copy() {is=title K2Node_FunctionEntry}
   > n1 = Get Current Local Time(TimeUnit=DisplayRate) -> (ReturnValue_Frame_SubFrame=~, ReturnValue_Time=~, ReturnValue_bHasJumped=~)
   > n2 = JJF_GetSelection_BindingsObjectsControlNames()
   n3 = Get STR_CR_ProxyObjName()
   n4 = "Get (a copy)"!(Array:Array of STR Control Rig Proxy Object Name Tuple Structures=@n3, Dimension 1=Last Index(TargetArray=@n3) -> (ReturnValue=0)) -> (
    Output_CR_Proxy_11_#2:Movie Scene Binding Proxy Structure
    Output_CR_Object_8_#3:Control Rig Object Reference
    Output_CR_ControlName_9_#4:Name
   )
   > n5 = GetControlRigWorldTransform(
    LevelSequence=@n2
    ControlRig=@n4.Output_CR_Object_8_#3
    ControlName=@n4.Output_CR_ControlName_9_#4
    Frame=@n1
    TimeUnit=DisplayRate
   )
   > Set WorldspaceXform(WorldspaceXform:Transform=@n5) -> (Output_Get:Transform)
  ]}
]
variables [
 Bone Transform Map Local {type="map<Transform>", default {}}
 ChildControl {type=string, default=None}
 ChildRig {type=string, default=None}
 ConstraintType {type=ETransformConstraintType, default=Translation}
 Control Rig {type=ControlRig, default=null}
 Control Rig Proxy Binding {default {}}
 Controls Transform Map Local {type="map<Transform>", default {}}
 CopiedKeyTimings {type="array<integer>"}
 CopiedKeyTimings_Ctrls {type="array<string>", default []}
 Copy Paste Control Rig Proxy Bindings {default {}}
 CurrControl {type=string, default=None}
 currFrame {type=FrameNumber, default {value=0}}
 CurrSeq {type=LevelSequence, default=null}
 Debug_Check {type=boolean, default=true}
 Focused Level Sequence {type=LevelSequence, default=null}
 Hierarchy {type=RigHierarchy, default=null}
 In Attachment Name {type=string, default=None}
 In Attachment Name_0 {type=string, default=None}
 In Extrapolation {type=ERichCurveExtrapolation, default=RCCE_Cycle}
 In Extrapolation_0 {type=ERichCurveExtrapolation, default=RCCE_Cycle}
 JF_TEST {type=string, default=test test test}
 Locator_Constrain {type=boolean}
 Locator_In_CR {type=ControlRig, default=null}
 Locator_In_Ctrl {type=string, default=None}
 Locator_Loc_CR {type=ControlRig, default=null}
 Locator_Loc_Ctrl {type=string, default=None}
 NewVar {type=string, default=""}
 ParentControl {type=string, default=None}
 ParentRig {type=string, default=None}
 Range End {type=integer}
 Range Start {type=integer}
 RetimeMode {type=boolean}
 SPN Time {type=EditorUtilitySpinBox, default=null}
 STR_CR_ProxyObjName {type="array<STR_ControlRigProxyObjectNameTuple>", default []}
 STR_Out_Locator_locator {type="array<STR_ControlRigProxyObjectNameTuple>", default []}
 STR_Out_Locator_sourceCtrl {type="array<STR_ControlRigProxyObjectNameTuple>", default []}
 WorldspaceXform {type=Transform, default {location=<0 0 0>, rotation=<rot 0 0 0>, scale=<1 1 1>}}
]
widgets {
 info {rootWidgetClass=&CanvasPanel, widgetCount=110}
 widgets [
  CanvasPanel_446 {widgetClassPath=&CanvasPanel, widget=CanvasPanel_446}
  ScrollBox_Buttons {
   widgetClassPath=&ScrollBox
   bIsVariable=true
   parent=CanvasPanel_446
   widget=ScrollBox_Buttons
   properties {
    widgetBarStyle {
     thickness=8
     draggedThumbImage {
      bIsDynamicallyLoaded=false
      drawAs=RoundedBox
      imageType=NoImage
      mirroring=NoMirror
      resourceName=None
      resourceObject=null
      tiling=NoTile
      imageSize=<0 0>
      margin=<ltrb 0 0 0 0>
      outlineSettings {
       bUseBrushTransparency=false
       roundingType=FixedRadius
       width=0
       color {colorUseRule=UseColor_Specified, specifiedColor=<rgba 0 0 0 0>}
       cornerRadii=<4 4 4 4>
      }
      tintColor {colorUseRule=UseColor_Specified, specifiedColor=<rgba 0.2159 0.2159 0.2159 1>}
      uVRegion {bIsValid=false, max=<0 0>, min=<0 0>}
     }
     horizontalBackgroundImage {
      bIsDynamicallyLoaded=false
      drawAs=NoDrawType
      imageType=NoImage
      mirroring=NoMirror
      resourceName=None
      resourceObject=null
      tiling=NoTile
      imageSize=<0 0>
      margin=<ltrb 0 0 0 0>
      outlineSettings {
       bUseBrushTransparency=false
       roundingType=HalfHeightRadius
       width=0
       color {colorUseRule=UseColor_Specified, specifiedColor=<rgba 0 0 0 0>}
       cornerRadii=<0 0 0 0>
      }
      tintColor {colorUseRule=UseColor_Specified, specifiedColor=<rgba 1 1 1 1>}
      uVRegion {bIsValid=false, max=<0 0>, min=<0 0>}
     }
     horizontalBottomSlotImage {
      bIsDynamicallyLoaded=false
      drawAs=NoDrawType
      imageType=NoImage
      mirroring=NoMirror
      resourceName=None
      resourceObject=null
      tiling=NoTile
      imageSize=<0 0>
      margin=<ltrb 0 0 0 0>
      outlineSettings {
       bUseBrushTransparency=false
       roundingType=HalfHeightRadius
       width=0
       color {colorUseRule=UseColor_Specified, specifiedColor=<rgba 0 0 0 0>}
       cornerRadii=<0 0 0 0>
      }
      tintColor {colorUseRule=UseColor_Specified, specifiedColor=<rgba 1 1 1 1>}
      uVRegion {bIsValid=false, max=<0 0>, min=<0 0>}
     }
     horizontalTopSlotImage {
      bIsDynamicallyLoaded=false
      drawAs=NoDrawType
      imageType=NoImage
      mirroring=NoMirror
      resourceName=None
      resourceObject=null
      tiling=NoTile
      imageSize=<0 0>
      margin=<ltrb 0 0 0 0>
      outlineSettings {
       bUseBrushTransparency=false
       roundingType=HalfHeightRadius
       width=0
       color {colorUseRule=UseColor_Specified, specifiedColor=<rgba 0 0 0 0>}
       cornerRadii=<0 0 0 0>
      }
      tintColor {colorUseRule=UseColor_Specified, specifiedColor=<rgba 1 1 1 1>}
      uVRegion {bIsValid=false, max=<0 0>, min=<0 0>}
     }
     hoveredThumbImage {
      bIsDynamicallyLoaded=false
      drawAs=RoundedBox
      imageType=NoImage
      mirroring=NoMirror
      resourceName=None
      resourceObject=null
      tiling=NoTile
      imageSize=<0 0>
      margin=<ltrb 0 0 0 0>
      outlineSettings {
       bUseBrushTransparency=false
       roundingType=FixedRadius
       width=0
       color {colorUseRule=UseColor_Specified, specifiedColor=<rgba 0 0 0 0>}
       cornerRadii=<4 4 4 4>
      }
      tintColor {colorUseRule=UseColor_Specified, specifiedColor=<rgba 0.2159 0.2159 0.2159 1>}
      uVRegion {bIsValid=false, max=<0 0>, min=<0 0>}
     }
     normalThumbImage {
      bIsDynamicallyLoaded=false
      drawAs=RoundedBox
      imageType=NoImage
      mirroring=NoMirror
      resourceName=None
      resourceObject=null
      tiling=NoTile
      imageSize=<0 0>
      margin=<ltrb 0 0 0 0>
      outlineSettings {
       bUseBrushTransparency=false
       roundingType=FixedRadius
       width=0
       color {colorUseRule=UseColor_Specified, specifiedColor=<rgba 0 0 0 0>}
       cornerRadii=<4 4 4 4>
      }
      tintColor {colorUseRule=UseColor_Specified, specifiedColor=<rgba 0.0953 0.0953 0.0953 1>}
      uVRegion {bIsValid=false, max=<0 0>, min=<0 0>}
     }
     verticalBackgroundImage {
      bIsDynamicallyLoaded=false
      drawAs=NoDrawType
      imageType=NoImage
      mirroring=NoMirror
      resourceName=None
      resourceObject=null
      tiling=NoTile
      imageSize=<0 0>
      margin=<ltrb 0 0 0 0>
      outlineSettings {
       bUseBrushTransparency=false
       roundingType=HalfHeightRadius
       width=0
       color {colorUseRule=UseColor_Specified, specifiedColor=<rgba 0 0 0 0>}
       cornerRadii=<0 0 0 0>
      }
      tintColor {colorUseRule=UseColor_Specified, specifiedColor=<rgba 1 1 1 1>}
      uVRegion {bIsValid=false, max=<0 0>, min=<0 0>}
     }
     verticalBottomSlotImage {
      bIsDynamicallyLoaded=false
      drawAs=NoDrawType
      imageType=NoImage
      mirroring=NoMirror
      resourceName=None
      resourceObject=null
      tiling=NoTile
      imageSize=<0 0>
      margin=<ltrb 0 0 0 0>
      outlineSettings {
       bUseBrushTransparency=false
       roundingType=HalfHeightRadius
       width=0
       color {colorUseRule=UseColor_Specified, specifiedColor=<rgba 0 0 0 0>}
       cornerRadii=<0 0 0 0>
      }
      tintColor {colorUseRule=UseColor_Specified, specifiedColor=<rgba 1 1 1 1>}
      uVRegion {bIsValid=false, max=<0 0>, min=<0 0>}
     }
     verticalTopSlotImage {
      bIsDynamicallyLoaded=false
      drawAs=NoDrawType
      imageType=NoImage
      mirroring=NoMirror
      resourceName=None
      resourceObject=null
      tiling=NoTile
      imageSize=<0 0>
      margin=<ltrb 0 0 0 0>
      outlineSettings {
       bUseBrushTransparency=false
       roundingType=HalfHeightRadius
       width=0
       color {colorUseRule=UseColor_Specified, specifiedColor=<rgba 0 0 0 0>}
       cornerRadii=<0 0 0 0>
      }
      tintColor {colorUseRule=UseColor_Specified, specifiedColor=<rgba 1 1 1 1>}
      uVRegion {bIsValid=false, max=<0 0>, min=<0 0>}
     }
    }
    widgetStyle {
     barThickness=8
     bottomShadowBrush {
      bIsDynamicallyLoaded=false
      drawAs=Box
      imageType=FullColor
      mirroring=NoMirror
      resourceName=../../../Engine/Content/Slate/Common/ScrollBoxShadowBottom.png
      resourceObject=null
      tiling=NoTile
      imageSize=<16 8>
      margin=<ltrb 0.5 0 0.5 1>
      outlineSettings {
       bUseBrushTransparency=false
       roundingType=HalfHeightRadius
       width=0
       color {colorUseRule=UseColor_Specified, specifiedColor=<rgba 0 0 0 0>}
       cornerRadii=<0 0 0 0>
      }
      tintColor {colorUseRule=UseColor_Specified, specifiedColor=<rgba 1 1 1 1>}
      uVRegion {bIsValid=false, max=<0 0>, min=<0 0>}
     }
     horizontalScrolledContentPadding=<ltrb 0 0 1 0>
     leftShadowBrush {
      bIsDynamicallyLoaded=false
      drawAs=Box
      imageType=FullColor
      mirroring=NoMirror
      resourceName=../../../Engine/Content/Slate/Common/ScrollBoxShadowLeft.png
      resourceObject=null
      tiling=NoTile
      imageSize=<8 16>
      margin=<ltrb 1 0.5 0 0.5>
      outlineSettings {
       bUseBrushTransparency=false
       roundingType=HalfHeightRadius
       width=0
       color {colorUseRule=UseColor_Specified, specifiedColor=<rgba 0 0 0 0>}
       cornerRadii=<0 0 0 0>
      }
      tintColor {colorUseRule=UseColor_Specified, specifiedColor=<rgba 1 1 1 1>}
      uVRegion {bIsValid=false, max=<0 0>, min=<0 0>}
     }
     rightShadowBrush {
      bIsDynamicallyLoaded=false
      drawAs=Box
      imageType=FullColor
      mirroring=NoMirror
      resourceName=../../../Engine/Content/Slate/Common/ScrollBoxShadowRight.png
      resourceObject=null
      tiling=NoTile
      imageSize=<8 16>
      margin=<ltrb 0 0.5 1 0.5>
      outlineSettings {
       bUseBrushTransparency=false
       roundingType=HalfHeightRadius
       width=0
       color {colorUseRule=UseColor_Specified, specifiedColor=<rgba 0 0 0 0>}
       cornerRadii=<0 0 0 0>
      }
      tintColor {colorUseRule=UseColor_Specified, specifiedColor=<rgba 1 1 1 1>}
      uVRegion {bIsValid=false, max=<0 0>, min=<0 0>}
     }
     topShadowBrush {
      bIsDynamicallyLoaded=false
      drawAs=Box
      imageType=FullColor
      mirroring=NoMirror
      resourceName=../../../Engine/Content/Slate/Common/ScrollBoxShadowTop.png
      resourceObject=null
      tiling=NoTile
      imageSize=<16 8>
      margin=<ltrb 0.5 1 0.5 0>
      outlineSettings {
       bUseBrushTransparency=false
       roundingType=HalfHeightRadius
       width=0
       color {colorUseRule=UseColor_Specified, specifiedColor=<rgba 0 0 0 0>}
       cornerRadii=<0 0 0 0>
      }
      tintColor {colorUseRule=UseColor_Specified, specifiedColor=<rgba 1 1 1 1>}
      uVRegion {bIsValid=false, max=<0 0>, min=<0 0>}
     }
     verticalScrolledContentPadding=<ltrb 0 0 0 1>
    }
   }
   slot {class=CanvasPanelSlot, properties {layoutData {alignment=<0 0>, anchors {maximum=<1 1>, minimum=<0 0>}, offsets=<ltrb 0 0 0 0>}}}
  }
  VerticalBox_0 {widgetClassPath=&VerticalBox, parent=ScrollBox_Buttons, widget=VerticalBox_0, slot {class=ScrollBoxSlot}}
  TextBlock_88 {
   widgetClassPath=&TextBlock
   parent=VerticalBox_0
   widget=TextBlock_88
   properties {
    text=RigSetup
    colorAndOpacity {colorUseRule=UseColor_Specified, specifiedColor=<rgba 0.8963 0.8963 0.8963 0.5>}
    font {
     bForceMonospaced=false
     bMaterialIsStencil=false
     fontMaterial=null
     fontObject=&Roboto
     letterSpacing=0
     monospacedWidth=0
     size=6
     skewAmount=0
     typefaceFontName=Bold
     outlineSettings {
      bApplyOutlineToDropShadows=false
      bMiteredCorners=false
      bSeparateFillAlpha=false
      outlineMaterial=null
      outlineSize=0
      outlineColor=<rgba 0 0 0 1>
     }
    }
   }
   slot {class=VerticalBoxSlot, properties {horizontalAlignment=HAlign_Left, padding=<ltrb 5 10 0 2>}}
  }
  Btn_Bake_MRB2 {
   widgetClassPath=&Button
   parent=VerticalBox_0
   widget=Btn_Bake_MRB2
   properties {
    toolTipText="bake to MR_Biped_v02\r\nif you have a control selected, it will run on that asset, otherwise, it needs you to select the actor track"
    colorAndOpacity=<rgba 0.1195 0.521 0.4072 1>
    widgetStyle {
     clickedSlateSound {resourceObject=null}
     disabled {
      bIsDynamicallyLoaded=false
      drawAs=RoundedBox
      imageType=NoImage
      mirroring=NoMirror
      resourceName=None
      resourceObject=null
      tiling=NoTile
      imageSize=<0 0>
      margin=<ltrb 0 0 0 0>
      outlineSettings {
       bUseBrushTransparency=false
       roundingType=FixedRadius
       width=1
       color {colorUseRule=UseColor_ColorTable, specifiedColor=<rgba 0.027 0.027 0.031 1>}
       cornerRadii=<5 5 5 5>
      }
      tintColor {colorUseRule=UseColor_Specified, specifiedColor=<rgba 0.015 0.015 0.018 1>}
      uVRegion {bIsValid=false, max=<0 0>, min=<0 0>}
     }
     disabledForeground {colorUseRule=UseColor_ColorTable, specifiedColor=<rgba 0.4 0.4 0.4 1>}
     hovered {
      bIsDynamicallyLoaded=false
      drawAs=RoundedBox
      imageType=NoImage
      mirroring=NoMirror
      resourceName=None
      resourceObject=null
      tiling=NoTile
      imageSize=<0 0>
      margin=<ltrb 0 0 0 0>
      outlineSettings {
       bUseBrushTransparency=false
       roundingType=FixedRadius
       width=1
       color {colorUseRule=UseColor_ColorTable, specifiedColor=<rgba 0.027 0.027 0.031 1>}
       cornerRadii=<5 5 5 5>
      }
      tintColor {colorUseRule=UseColor_Specified, specifiedColor=<rgba 0.031 0.031 0.038 1>}
      uVRegion {bIsValid=false, max=<0 0>, min=<0 0>}
     }
     hoveredForeground {colorUseRule=UseColor_Foreground, specifiedColor=<rgba 1 1 1 1>}
     hoveredSlateSound {resourceObject=null}
     normal {
      bIsDynamicallyLoaded=false
      drawAs=RoundedBox
      imageType=NoImage
      mirroring=NoMirror
      resourceName=None
      resourceObject=null
      tiling=NoTile
      imageSize=<0 0>
      margin=<ltrb 0 0 0 0>
      outlineSettings {
       bUseBrushTransparency=false
       roundingType=FixedRadius
       width=1
       color {colorUseRule=UseColor_ColorTable, specifiedColor=<rgba 0.027 0.027 0.031 1>}
       cornerRadii=<5 5 5 5>
      }
      tintColor {colorUseRule=UseColor_Specified, specifiedColor=<rgba 0.0222 0.0222 0.0252 1>}
      uVRegion {bIsValid=false, max=<0 0>, min=<0 0>}
     }
     normalForeground {colorUseRule=UseColor_Foreground, specifiedColor=<rgba 1 1 1 1>}
     normalPadding=<ltrb 0 0 0 0>
     pressed {
      bIsDynamicallyLoaded=false
      drawAs=RoundedBox
      imageType=NoImage
      mirroring=NoMirror
      resourceName=None
      resourceObject=null
      tiling=NoTile
      imageSize=<0 0>
      margin=<ltrb 0 0 0 0>
      outlineSettings {
       bUseBrushTransparency=false
       roundingType=FixedRadius
       width=1
       color {colorUseRule=UseColor_ColorTable, specifiedColor=<rgba 0.027 0.027 0.031 1>}
       cornerRadii=<5 5 5 5>
      }
      tintColor {colorUseRule=UseColor_Specified, specifiedColor=<rgba 0.04 0.04 0.048 1>}
      uVRegion {bIsValid=false, max=<0 0>, min=<0 0>}
     }
     pressedForeground {colorUseRule=UseColor_Foreground, specifiedColor=<rgba 1 1 1 1>}
     pressedPadding=<ltrb 0 0 0 0>
     pressedSlateSound {resourceObject=null}
    }
   }
   slot {class=VerticalBoxSlot, properties {padding=<ltrb 4 1 4 3>}}
  }
  HBox_Content_15 {
   widgetClassPath=&HorizontalBox
   parent=Btn_Bake_MRB2
   widget=HBox_Content_15
   slot {class=ButtonSlot, properties {horizontalAlignment=HAlign_Fill, padding=<ltrb 1 4 4 4>}}
  }
  Overlay_Icon_15 {
   widgetClassPath=&Overlay
   parent=HBox_Content_15
   widget=Overlay_Icon_15
   slot {class=HorizontalBoxSlot, properties {horizontalAlignment=HAlign_Left, verticalAlignment=VAlign_Center, padding=<ltrb 0 0 10 0>}}
  }
  TB_TabName_14 {
   widgetClassPath=&TextBlock
   bIsVariable=true
   parent=Overlay_Icon_15
   widget=TB_TabName_14
   properties {
    text=MR
    textOverflowPolicy=Ellipsis
    font {
     bForceMonospaced=false
     bMaterialIsStencil=false
     fontMaterial=null
     fontObject=&Roboto
     letterSpacing=0
     monospacedWidth=0
     size=12
     skewAmount=0
     typefaceFontName=Bold
     outlineSettings {
      bApplyOutlineToDropShadows=false
      bMiteredCorners=false
      bSeparateFillAlpha=false
      outlineMaterial=null
      outlineSize=0
      outlineColor=<rgba 0 0 0 1>
     }
    }
    renderTransform {angle=0, scale=<1 1>, shear=<0 0>, translation=<-3.2134 0>}
   }
   slot {class=OverlaySlot, properties {horizontalAlignment=HAlign_Center, verticalAlignment=VAlign_Center, padding=<ltrb 6 2 0 2>}}
  }
  TB_TabName_12 {
   widgetClassPath=&TextBlock
   bIsVariable=true
   parent=HBox_Content_15
   widget=TB_TabName_12
   properties {
    text=Bake to MR_Biped_v02
    textOverflowPolicy=Ellipsis
    font {
     bForceMonospaced=false
     bMaterialIsStencil=false
     fontMaterial=null
     fontObject=&Roboto
     letterSpacing=0
     monospacedWidth=0
     size=10
     skewAmount=0
     typefaceFontName=Bold
     outlineSettings {
      bApplyOutlineToDropShadows=false
      bMiteredCorners=false
      bSeparateFillAlpha=false
      outlineMaterial=null
      outlineSize=0
      outlineColor=<rgba 0 0 0 1>
     }
    }
   }
   slot {class=HorizontalBoxSlot, properties {verticalAlignment=VAlign_Center, size {sizeRule=Fill, value=1}}}
  }
  Btn_IKSpine {
   widgetClassPath=&Button
   parent=VerticalBox_0
   widget=Btn_IKSpine
   properties {
    toolTipText="IK Spine - Simple\r\n\r\nClears FK spine keys and resets to bind pose."
    colorAndOpacity=<rgba 0.1195 0.521 0.4072 1>
    widgetStyle {
     clickedSlateSound {resourceObject=null}
     disabled {
      bIsDynamicallyLoaded=false
      drawAs=RoundedBox
      imageType=NoImage
      mirroring=NoMirror
      resourceName=None
      resourceObject=null
      tiling=NoTile
      imageSize=<0 0>
      margin=<ltrb 0 0 0 0>
      outlineSettings {
       bUseBrushTransparency=false
       roundingType=FixedRadius
       width=1
       color {colorUseRule=UseColor_ColorTable, specifiedColor=<rgba 0.027 0.027 0.031 1>}
       cornerRadii=<5 5 5 5>
      }
      tintColor {colorUseRule=UseColor_Specified, specifiedColor=<rgba 0.015 0.015 0.018 1>}
      uVRegion {bIsValid=false, max=<0 0>, min=<0 0>}
     }
     disabledForeground {colorUseRule=UseColor_ColorTable, specifiedColor=<rgba 0.4 0.4 0.4 1>}
     hovered {
      bIsDynamicallyLoaded=false
      drawAs=RoundedBox
      imageType=NoImage
      mirroring=NoMirror
      resourceName=None
      resourceObject=null
      tiling=NoTile
      imageSize=<0 0>
      margin=<ltrb 0 0 0 0>
      outlineSettings {
       bUseBrushTransparency=false
       roundingType=FixedRadius
       width=1
       color {colorUseRule=UseColor_ColorTable, specifiedColor=<rgba 0.027 0.027 0.031 1>}
       cornerRadii=<5 5 5 5>
      }
      tintColor {colorUseRule=UseColor_Specified, specifiedColor=<rgba 0.031 0.031 0.038 1>}
      uVRegion {bIsValid=false, max=<0 0>, min=<0 0>}
     }
     hoveredForeground {colorUseRule=UseColor_Foreground, specifiedColor=<rgba 1 1 1 1>}
     hoveredSlateSound {resourceObject=null}
     normal {
      bIsDynamicallyLoaded=false
      drawAs=RoundedBox
      imageType=NoImage
      mirroring=NoMirror
      resourceName=None
      resourceObject=null
      tiling=NoTile
      imageSize=<0 0>
      margin=<ltrb 0 0 0 0>
      outlineSettings {
       bUseBrushTransparency=false
       roundingType=FixedRadius
       width=1
       color {colorUseRule=UseColor_ColorTable, specifiedColor=<rgba 0.027 0.027 0.031 1>}
       cornerRadii=<5 5 5 5>
      }
      tintColor {colorUseRule=UseColor_Specified, specifiedColor=<rgba 0.0222 0.0222 0.0252 1>}
      uVRegion {bIsValid=false, max=<0 0>, min=<0 0>}
     }
     normalForeground {colorUseRule=UseColor_Foreground, specifiedColor=<rgba 1 1 1 1>}
     normalPadding=<ltrb 0 0 0 0>
     pressed {
      bIsDynamicallyLoaded=false
      drawAs=RoundedBox
      imageType=NoImage
      mirroring=NoMirror
      resourceName=None
      resourceObject=null
      tiling=NoTile
      imageSize=<0 0>
      margin=<ltrb 0 0 0 0>
      outlineSettings {
       bUseBrushTransparency=false
       roundingType=FixedRadius
       width=1
       color {colorUseRule=UseColor_ColorTable, specifiedColor=<rgba 0.027 0.027 0.031 1>}
       cornerRadii=<5 5 5 5>
      }
      tintColor {colorUseRule=UseColor_Specified, specifiedColor=<rgba 0.04 0.04 0.048 1>}
      uVRegion {bIsValid=false, max=<0 0>, min=<0 0>}
     }
     pressedForeground {colorUseRule=UseColor_Foreground, specifiedColor=<rgba 1 1 1 1>}
     pressedPadding=<ltrb 0 0 0 0>
     pressedSlateSound {resourceObject=null}
    }
   }
   slot {class=VerticalBoxSlot, properties {padding=<ltrb 4 1 4 3>}}
  }
  HBox_Content_8 {
   widgetClassPath=&HorizontalBox
   parent=Btn_IKSpine
   widget=HBox_Content_8
   slot {class=ButtonSlot, properties {horizontalAlignment=HAlign_Fill, padding=<ltrb 1 4 4 4>}}
  }
  Overlay_Icon_8 {
   widgetClassPath=&Overlay
   parent=HBox_Content_8
   widget=Overlay_Icon_8
   slot {class=HorizontalBoxSlot, properties {horizontalAlignment=HAlign_Left, verticalAlignment=VAlign_Center, padding=<ltrb 0 0 10 0>}}
  }
  Img_Icon_8 {
   widgetClassPath=&Image
   bIsVariable=true
   parent=Overlay_Icon_8
   widget=Img_Icon_8
   properties {brush {
     bIsDynamicallyLoaded=false
     drawAs=Image
     imageType=FullColor
     mirroring=NoMirror
     resourceName=None
     resourceObject=/AnimInEdTools/AnimationToolbox/Icons/AnimToolbox_UI_Icon_43
     tiling=NoTile
     imageSize=<24 24>
     margin=<ltrb 0 0 0 0>
     outlineSettings {
      bUseBrushTransparency=false
      roundingType=FixedRadius
      width=0
      color {colorUseRule=UseColor_Specified, specifiedColor=<rgba 0 0 0 1>}
      cornerRadii=<0 0 0 0>
     }
     tintColor {colorUseRule=UseColor_Specified, specifiedColor=<rgba 1 1 1 1>}
     uVRegion {bIsValid=false, max=<0 0>, min=<0 0>}
    }}
   slot {class=OverlaySlot, properties {verticalAlignment=VAlign_Center, padding=<ltrb 6 0 0 0>}}
  }
  TB_TabName_8 {
   widgetClassPath=&TextBlock
   bIsVariable=true
   parent=HBox_Content_8
   widget=TB_TabName_8
   properties {
    text=IK Spine - Simple
    textOverflowPolicy=Ellipsis
    font {
     bForceMonospaced=false
     bMaterialIsStencil=false
     fontMaterial=null
     fontObject=&Roboto
     letterSpacing=0
     monospacedWidth=0
     size=10
     skewAmount=0
     typefaceFontName=Bold
     outlineSettings {
      bApplyOutlineToDropShadows=false
      bMiteredCorners=false
      bSeparateFillAlpha=false
      outlineMaterial=null
      outlineSize=0
      outlineColor=<rgba 0 0 0 1>
     }
    }
   }
   slot {class=HorizontalBoxSlot, properties {verticalAlignment=VAlign_Center, size {sizeRule=Fill, value=1}}}
  }
  Btn_IKSpine_BodyAim {
   widgetClassPath=&Button
   parent=VerticalBox_0
   widget=Btn_IKSpine_BodyAim
   properties {
    toolTipText="IK Spine - Body Aim\r\n\r\nClears FK spine keys and makes it so that the body rotation is controlling most of the chest translation"
    colorAndOpacity=<rgba 0.1195 0.521 0.4072 1>
    widgetStyle {
     clickedSlateSound {resourceObject=null}
     disabled {
      bIsDynamicallyLoaded=false
      drawAs=RoundedBox
      imageType=NoImage
      mirroring=NoMirror
      resourceName=None
      resourceObject=null
      tiling=NoTile
      imageSize=<0 0>
      margin=<ltrb 0 0 0 0>
      outlineSettings {
       bUseBrushTransparency=false
       roundingType=FixedRadius
       width=1
       color {colorUseRule=UseColor_ColorTable, specifiedColor=<rgba 0.027 0.027 0.031 1>}
       cornerRadii=<5 5 5 5>
      }
      tintColor {colorUseRule=UseColor_Specified, specifiedColor=<rgba 0.015 0.015 0.018 1>}
      uVRegion {bIsValid=false, max=<0 0>, min=<0 0>}
     }
     disabledForeground {colorUseRule=UseColor_ColorTable, specifiedColor=<rgba 0.4 0.4 0.4 1>}
     hovered {
      bIsDynamicallyLoaded=false
      drawAs=RoundedBox
      imageType=NoImage
      mirroring=NoMirror
      resourceName=None
      resourceObject=null
      tiling=NoTile
      imageSize=<0 0>
      margin=<ltrb 0 0 0 0>
      outlineSettings {
       bUseBrushTransparency=false
       roundingType=FixedRadius
       width=1
       color {colorUseRule=UseColor_ColorTable, specifiedColor=<rgba 0.027 0.027 0.031 1>}
       cornerRadii=<5 5 5 5>
      }
      tintColor {colorUseRule=UseColor_Specified, specifiedColor=<rgba 0.031 0.031 0.038 1>}
      uVRegion {bIsValid=false, max=<0 0>, min=<0 0>}
     }
     hoveredForeground {colorUseRule=UseColor_Foreground, specifiedColor=<rgba 1 1 1 1>}
     hoveredSlateSound {resourceObject=null}
     normal {
      bIsDynamicallyLoaded=false
      drawAs=RoundedBox
      imageType=NoImage
      mirroring=NoMirror
      resourceName=None
      resourceObject=null
      tiling=NoTile
      imageSize=<0 0>
      margin=<ltrb 0 0 0 0>
      outlineSettings {
       bUseBrushTransparency=false
       roundingType=FixedRadius
       width=1
       color {colorUseRule=UseColor_ColorTable, specifiedColor=<rgba 0.027 0.027 0.031 1>}
       cornerRadii=<5 5 5 5>
      }
      tintColor {colorUseRule=UseColor_Specified, specifiedColor=<rgba 0.0222 0.0222 0.0252 1>}
      uVRegion {bIsValid=false, max=<0 0>, min=<0 0>}
     }
     normalForeground {colorUseRule=UseColor_Foreground, specifiedColor=<rgba 1 1 1 1>}
     normalPadding=<ltrb 0 0 0 0>
     pressed {
      bIsDynamicallyLoaded=false
      drawAs=RoundedBox
      imageType=NoImage
      mirroring=NoMirror
      resourceName=None
      resourceObject=null
      tiling=NoTile
      imageSize=<0 0>
      margin=<ltrb 0 0 0 0>
      outlineSettings {
       bUseBrushTransparency=false
       roundingType=FixedRadius
       width=1
       color {colorUseRule=UseColor_ColorTable, specifiedColor=<rgba 0.027 0.027 0.031 1>}
       cornerRadii=<5 5 5 5>
      }
      tintColor {colorUseRule=UseColor_Specified, specifiedColor=<rgba 0.04 0.04 0.048 1>}
      uVRegion {bIsValid=false, max=<0 0>, min=<0 0>}
     }
     pressedForeground {colorUseRule=UseColor_Foreground, specifiedColor=<rgba 1 1 1 1>}
     pressedPadding=<ltrb 0 0 0 0>
     pressedSlateSound {resourceObject=null}
    }
   }
   slot {class=VerticalBoxSlot, properties {padding=<ltrb 4 1 4 3>}}
  }
  HBox_Content_9 {
   widgetClassPath=&HorizontalBox
   parent=Btn_IKSpine_BodyAim
   widget=HBox_Content_9
   slot {class=ButtonSlot, properties {horizontalAlignment=HAlign_Fill, padding=<ltrb 1 4 4 4>}}
  }
  Overlay_Icon_9 {
   widgetClassPath=&Overlay
   parent=HBox_Content_9
   widget=Overlay_Icon_9
   slot {class=HorizontalBoxSlot, properties {horizontalAlignment=HAlign_Left, verticalAlignment=VAlign_Center, padding=<ltrb 0 0 10 0>}}
  }
  Img_Icon_9 {
   widgetClassPath=&Image
   bIsVariable=true
   parent=Overlay_Icon_9
   widget=Img_Icon_9
   properties {brush {
     bIsDynamicallyLoaded=false
     drawAs=Image
     imageType=FullColor
     mirroring=NoMirror
     resourceName=None
     resourceObject=/AnimInEdTools/AnimationToolbox/Icons/AnimToolbox_UI_Icon_44
     tiling=NoTile
     imageSize=<24 24>
     margin=<ltrb 0 0 0 0>
     outlineSettings {
      bUseBrushTransparency=false
      roundingType=FixedRadius
      width=0
      color {colorUseRule=UseColor_Specified, specifiedColor=<rgba 0 0 0 1>}
      cornerRadii=<0 0 0 0>
     }
     tintColor {colorUseRule=UseColor_Specified, specifiedColor=<rgba 1 1 1 1>}
     uVRegion {bIsValid=false, max=<0 0>, min=<0 0>}
    }}
   slot {class=OverlaySlot, properties {verticalAlignment=VAlign_Center, padding=<ltrb 6 0 0 0>}}
  }
  TB_TabName_9 {
   widgetClassPath=&TextBlock
   bIsVariable=true
   parent=HBox_Content_9
   widget=TB_TabName_9
   properties {
    text=IK Spine - Body Aim
    textOverflowPolicy=Ellipsis
    font {
     bForceMonospaced=false
     bMaterialIsStencil=false
     fontMaterial=null
     fontObject=&Roboto
     letterSpacing=0
     monospacedWidth=0
     size=10
     skewAmount=0
     typefaceFontName=Bold
     outlineSettings {
      bApplyOutlineToDropShadows=false
      bMiteredCorners=false
      bSeparateFillAlpha=false
      outlineMaterial=null
      outlineSize=0
      outlineColor=<rgba 0 0 0 1>
     }
    }
   }
   slot {class=HorizontalBoxSlot, properties {verticalAlignment=VAlign_Center, size {sizeRule=Fill, value=1}}}
  }
  Btn_IKSpine_Yaw {
   widgetClassPath=&Button
   parent=VerticalBox_0
   widget=Btn_IKSpine_Yaw
   properties {
    toolTipText="IK Spine - Body Yaw\r\n\r\nClears FK spine keys and sets it so the body control only has yaw values"
    colorAndOpacity=<rgba 0.1195 0.521 0.4072 1>
    widgetStyle {
     clickedSlateSound {resourceObject=null}
     disabled {
      bIsDynamicallyLoaded=false
      drawAs=RoundedBox
      imageType=NoImage
      mirroring=NoMirror
      resourceName=None
      resourceObject=null
      tiling=NoTile
      imageSize=<0 0>
      margin=<ltrb 0 0 0 0>
      outlineSettings {
       bUseBrushTransparency=false
       roundingType=FixedRadius
       width=1
       color {colorUseRule=UseColor_ColorTable, specifiedColor=<rgba 0.027 0.027 0.031 1>}
       cornerRadii=<5 5 5 5>
      }
      tintColor {colorUseRule=UseColor_Specified, specifiedColor=<rgba 0.015 0.015 0.018 1>}
      uVRegion {bIsValid=false, max=<0 0>, min=<0 0>}
     }
     disabledForeground {colorUseRule=UseColor_ColorTable, specifiedColor=<rgba 0.4 0.4 0.4 1>}
     hovered {
      bIsDynamicallyLoaded=false
      drawAs=RoundedBox
      imageType=NoImage
      mirroring=NoMirror
      resourceName=None
      resourceObject=null
      tiling=NoTile
      imageSize=<0 0>
      margin=<ltrb 0 0 0 0>
      outlineSettings {
       bUseBrushTransparency=false
       roundingType=FixedRadius
       width=1
       color {colorUseRule=UseColor_ColorTable, specifiedColor=<rgba 0.027 0.027 0.031 1>}
       cornerRadii=<5 5 5 5>
      }
      tintColor {colorUseRule=UseColor_Specified, specifiedColor=<rgba 0.031 0.031 0.038 1>}
      uVRegion {bIsValid=false, max=<0 0>, min=<0 0>}
     }
     hoveredForeground {colorUseRule=UseColor_Foreground, specifiedColor=<rgba 1 1 1 1>}
     hoveredSlateSound {resourceObject=null}
     normal {
      bIsDynamicallyLoaded=false
      drawAs=RoundedBox
      imageType=NoImage
      mirroring=NoMirror
      resourceName=None
      resourceObject=null
      tiling=NoTile
      imageSize=<0 0>
      margin=<ltrb 0 0 0 0>
      outlineSettings {
       bUseBrushTransparency=false
       roundingType=FixedRadius
       width=1
       color {colorUseRule=UseColor_ColorTable, specifiedColor=<rgba 0.027 0.027 0.031 1>}
       cornerRadii=<5 5 5 5>
      }
      tintColor {colorUseRule=UseColor_Specified, specifiedColor=<rgba 0.0222 0.0222 0.0252 1>}
      uVRegion {bIsValid=false, max=<0 0>, min=<0 0>}
     }
     normalForeground {colorUseRule=UseColor_Foreground, specifiedColor=<rgba 1 1 1 1>}
     normalPadding=<ltrb 0 0 0 0>
     pressed {
      bIsDynamicallyLoaded=false
      drawAs=RoundedBox
      imageType=NoImage
      mirroring=NoMirror
      resourceName=None
      resourceObject=null
      tiling=NoTile
      imageSize=<0 0>
      margin=<ltrb 0 0 0 0>
      outlineSettings {
       bUseBrushTransparency=false
       roundingType=FixedRadius
       width=1
       color {colorUseRule=UseColor_ColorTable, specifiedColor=<rgba 0.027 0.027 0.031 1>}
       cornerRadii=<5 5 5 5>
      }
      tintColor {colorUseRule=UseColor_Specified, specifiedColor=<rgba 0.04 0.04 0.048 1>}
      uVRegion {bIsValid=false, max=<0 0>, min=<0 0>}
     }
     pressedForeground {colorUseRule=UseColor_Foreground, specifiedColor=<rgba 1 1 1 1>}
     pressedPadding=<ltrb 0 0 0 0>
     pressedSlateSound {resourceObject=null}
    }
   }
   slot {class=VerticalBoxSlot, properties {padding=<ltrb 4 1 4 3>}}
  }
  HBox_Content_10 {
   widgetClassPath=&HorizontalBox
   parent=Btn_IKSpine_Yaw
   widget=HBox_Content_10
   slot {class=ButtonSlot, properties {horizontalAlignment=HAlign_Fill, padding=<ltrb 1 4 4 4>}}
  }
  Overlay_Icon_10 {
   widgetClassPath=&Overlay
   parent=HBox_Content_10
   widget=Overlay_Icon_10
   slot {class=HorizontalBoxSlot, properties {horizontalAlignment=HAlign_Left, verticalAlignment=VAlign_Center, padding=<ltrb 0 0 10 0>}}
  }
  Img_Icon_10 {
   widgetClassPath=&Image
   bIsVariable=true
   parent=Overlay_Icon_10
   widget=Img_Icon_10
   properties {brush {
     bIsDynamicallyLoaded=false
     drawAs=Image
     imageType=FullColor
     mirroring=NoMirror
     resourceName=None
     resourceObject=/AnimInEdTools/AnimationToolbox/Icons/AnimToolbox_UI_Icon_45
     tiling=NoTile
     imageSize=<24 24>
     margin=<ltrb 0 0 0 0>
     outlineSettings {
      bUseBrushTransparency=false
      roundingType=FixedRadius
      width=0
      color {colorUseRule=UseColor_Specified, specifiedColor=<rgba 0 0 0 1>}
      cornerRadii=<0 0 0 0>
     }
     tintColor {colorUseRule=UseColor_Specified, specifiedColor=<rgba 1 1 1 1>}
     uVRegion {bIsValid=false, max=<0 0>, min=<0 0>}
    }}
   slot {class=OverlaySlot, properties {verticalAlignment=VAlign_Center, padding=<ltrb 6 0 0 0>}}
  }
  TB_TabName_10 {
   widgetClassPath=&TextBlock
   bIsVariable=true
   parent=HBox_Content_10
   widget=TB_TabName_10
   properties {
    text=IK Spine - Body Yaw
    textOverflowPolicy=Ellipsis
    font {
     bForceMonospaced=false
     bMaterialIsStencil=false
     fontMaterial=null
     fontObject=&Roboto
     letterSpacing=0
     monospacedWidth=0
     size=10
     skewAmount=0
     typefaceFontName=Bold
     outlineSettings {
      bApplyOutlineToDropShadows=false
      bMiteredCorners=false
      bSeparateFillAlpha=false
      outlineMaterial=null
      outlineSize=0
      outlineColor=<rgba 0 0 0 1>
     }
    }
   }
   slot {class=HorizontalBoxSlot, properties {verticalAlignment=VAlign_Center, size {sizeRule=Fill, value=1}}}
  }
  Btn_CleanExtras {
   widgetClassPath=&Button
   parent=VerticalBox_0
   widget=Btn_CleanExtras
   properties {
    toolTipText="Clean Extras\r\nselect any control and this should remove any of the extra anim on the secondary controls (spine, neck, and twist)"
    colorAndOpacity=<rgba 0.1195 0.521 0.4072 1>
    widgetStyle {
     clickedSlateSound {resourceObject=null}
     disabled {
      bIsDynamicallyLoaded=false
      drawAs=RoundedBox
      imageType=NoImage
      mirroring=NoMirror
      resourceName=None
      resourceObject=null
      tiling=NoTile
      imageSize=<0 0>
      margin=<ltrb 0 0 0 0>
      outlineSettings {
       bUseBrushTransparency=false
       roundingType=FixedRadius
       width=1
       color {colorUseRule=UseColor_ColorTable, specifiedColor=<rgba 0.027 0.027 0.031 1>}
       cornerRadii=<5 5 5 5>
      }
      tintColor {colorUseRule=UseColor_Specified, specifiedColor=<rgba 0.015 0.015 0.018 1>}
      uVRegion {bIsValid=false, max=<0 0>, min=<0 0>}
     }
     disabledForeground {colorUseRule=UseColor_ColorTable, specifiedColor=<rgba 0.4 0.4 0.4 1>}
     hovered {
      bIsDynamicallyLoaded=false
      drawAs=RoundedBox
      imageType=NoImage
      mirroring=NoMirror
      resourceName=None
      resourceObject=null
      tiling=NoTile
      imageSize=<0 0>
      margin=<ltrb 0 0 0 0>
      outlineSettings {
       bUseBrushTransparency=false
       roundingType=FixedRadius
       width=1
       color {colorUseRule=UseColor_ColorTable, specifiedColor=<rgba 0.027 0.027 0.031 1>}
       cornerRadii=<5 5 5 5>
      }
      tintColor {colorUseRule=UseColor_Specified, specifiedColor=<rgba 0.031 0.031 0.038 1>}
      uVRegion {bIsValid=false, max=<0 0>, min=<0 0>}
     }
     hoveredForeground {colorUseRule=UseColor_Foreground, specifiedColor=<rgba 1 1 1 1>}
     hoveredSlateSound {resourceObject=null}
     normal {
      bIsDynamicallyLoaded=false
      drawAs=RoundedBox
      imageType=NoImage
      mirroring=NoMirror
      resourceName=None
      resourceObject=null
      tiling=NoTile
      imageSize=<0 0>
      margin=<ltrb 0 0 0 0>
      outlineSettings {
       bUseBrushTransparency=false
       roundingType=FixedRadius
       width=1
       color {colorUseRule=UseColor_ColorTable, specifiedColor=<rgba 0.027 0.027 0.031 1>}
       cornerRadii=<5 5 5 5>
      }
      tintColor {colorUseRule=UseColor_Specified, specifiedColor=<rgba 0.0222 0.0222 0.0252 1>}
      uVRegion {bIsValid=false, max=<0 0>, min=<0 0>}
     }
     normalForeground {colorUseRule=UseColor_Foreground, specifiedColor=<rgba 1 1 1 1>}
     normalPadding=<ltrb 0 0 0 0>
     pressed {
      bIsDynamicallyLoaded=false
      drawAs=RoundedBox
      imageType=NoImage
      mirroring=NoMirror
      resourceName=None
      resourceObject=null
      tiling=NoTile
      imageSize=<0 0>
      margin=<ltrb 0 0 0 0>
      outlineSettings {
       bUseBrushTransparency=false
       roundingType=FixedRadius
       width=1
       color {colorUseRule=UseColor_ColorTable, specifiedColor=<rgba 0.027 0.027 0.031 1>}
       cornerRadii=<5 5 5 5>
      }
      tintColor {colorUseRule=UseColor_Specified, specifiedColor=<rgba 0.04 0.04 0.048 1>}
      uVRegion {bIsValid=false, max=<0 0>, min=<0 0>}
     }
     pressedForeground {colorUseRule=UseColor_Foreground, specifiedColor=<rgba 1 1 1 1>}
     pressedPadding=<ltrb 0 0 0 0>
     pressedSlateSound {resourceObject=null}
    }
   }
   slot {class=VerticalBoxSlot, properties {padding=<ltrb 4 1 4 3>}}
  }
  HBox_Content_11 {
   widgetClassPath=&HorizontalBox
   parent=Btn_CleanExtras
   widget=HBox_Content_11
   slot {class=ButtonSlot, properties {horizontalAlignment=HAlign_Fill, padding=<ltrb 1 4 4 4>}}
  }
  Overlay_Icon_11 {
   widgetClassPath=&Overlay
   parent=HBox_Content_11
   widget=Overlay_Icon_11
   slot {class=HorizontalBoxSlot, properties {horizontalAlignment=HAlign_Left, verticalAlignment=VAlign_Center, padding=<ltrb 0 0 10 0>}}
  }
  Img_Icon_11 {
   widgetClassPath=&Image
   bIsVariable=true
   parent=Overlay_Icon_11
   widget=Img_Icon_11
   properties {brush {
     bIsDynamicallyLoaded=false
     drawAs=Image
     imageType=FullColor
     mirroring=NoMirror
     resourceName=None
     resourceObject=/AnimInEdTools/AnimationToolbox/Icons/AnimToolbox_UI_Icon_46
     tiling=NoTile
     imageSize=<24 24>
     margin=<ltrb 0 0 0 0>
     outlineSettings {
      bUseBrushTransparency=false
      roundingType=FixedRadius
      width=0
      color {colorUseRule=UseColor_Specified, specifiedColor=<rgba 0 0 0 1>}
      cornerRadii=<0 0 0 0>
     }
     tintColor {colorUseRule=UseColor_Specified, specifiedColor=<rgba 1 1 1 1>}
     uVRegion {bIsValid=false, max=<0 0>, min=<0 0>}
    }}
   slot {class=OverlaySlot, properties {verticalAlignment=VAlign_Center, padding=<ltrb 6 0 0 0>}}
  }
  TB_TabName_11 {
   widgetClassPath=&TextBlock
   bIsVariable=true
   parent=HBox_Content_11
   widget=TB_TabName_11
   properties {
    text=Clean Extras
    textOverflowPolicy=Ellipsis
    font {
     bForceMonospaced=false
     bMaterialIsStencil=false
     fontMaterial=null
     fontObject=&Roboto
     letterSpacing=0
     monospacedWidth=0
     size=10
     skewAmount=0
     typefaceFontName=Bold
     outlineSettings {
      bApplyOutlineToDropShadows=false
      bMiteredCorners=false
      bSeparateFillAlpha=false
      outlineMaterial=null
      outlineSize=0
      outlineColor=<rgba 0 0 0 1>
     }
    }
   }
   slot {class=HorizontalBoxSlot, properties {verticalAlignment=VAlign_Center, size {sizeRule=Fill, value=1}}}
  }
  Btn_MovingLocal {
   widgetClassPath=&Button
   parent=VerticalBox_0
   widget=Btn_MovingLocal
   properties {
    toolTipText=Uses space switching to move forward motion from the root controller to the local
    colorAndOpacity=<rgba 0.1195 0.521 0.4072 1>
    widgetStyle {
     clickedSlateSound {resourceObject=null}
     disabled {
      bIsDynamicallyLoaded=false
      drawAs=RoundedBox
      imageType=NoImage
      mirroring=NoMirror
      resourceName=None
      resourceObject=null
      tiling=NoTile
      imageSize=<0 0>
      margin=<ltrb 0 0 0 0>
      outlineSettings {
       bUseBrushTransparency=false
       roundingType=FixedRadius
       width=1
       color {colorUseRule=UseColor_ColorTable, specifiedColor=<rgba 0.027 0.027 0.031 1>}
       cornerRadii=<5 5 5 5>
      }
      tintColor {colorUseRule=UseColor_Specified, specifiedColor=<rgba 0.015 0.015 0.018 1>}
      uVRegion {bIsValid=false, max=<0 0>, min=<0 0>}
     }
     disabledForeground {colorUseRule=UseColor_ColorTable, specifiedColor=<rgba 0.4 0.4 0.4 1>}
     hovered {
      bIsDynamicallyLoaded=false
      drawAs=RoundedBox
      imageType=NoImage
      mirroring=NoMirror
      resourceName=None
      resourceObject=null
      tiling=NoTile
      imageSize=<0 0>
      margin=<ltrb 0 0 0 0>
      outlineSettings {
       bUseBrushTransparency=false
       roundingType=FixedRadius
       width=1
       color {colorUseRule=UseColor_ColorTable, specifiedColor=<rgba 0.027 0.027 0.031 1>}
       cornerRadii=<5 5 5 5>
      }
      tintColor {colorUseRule=UseColor_Specified, specifiedColor=<rgba 0.031 0.031 0.038 1>}
      uVRegion {bIsValid=false, max=<0 0>, min=<0 0>}
     }
     hoveredForeground {colorUseRule=UseColor_Foreground, specifiedColor=<rgba 1 1 1 1>}
     hoveredSlateSound {resourceObject=null}
     normal {
      bIsDynamicallyLoaded=false
      drawAs=RoundedBox
      imageType=NoImage
      mirroring=NoMirror
      resourceName=None
      resourceObject=null
      tiling=NoTile
      imageSize=<0 0>
      margin=<ltrb 0 0 0 0>
      outlineSettings {
       bUseBrushTransparency=false
       roundingType=FixedRadius
       width=1
       color {colorUseRule=UseColor_ColorTable, specifiedColor=<rgba 0.027 0.027 0.031 1>}
       cornerRadii=<5 5 5 5>
      }
      tintColor {colorUseRule=UseColor_Specified, specifiedColor=<rgba 0.0222 0.0222 0.0252 1>}
      uVRegion {bIsValid=false, max=<0 0>, min=<0 0>}
     }
     normalForeground {colorUseRule=UseColor_Foreground, specifiedColor=<rgba 1 1 1 1>}
     normalPadding=<ltrb 0 0 0 0>
     pressed {
      bIsDynamicallyLoaded=false
      drawAs=RoundedBox
      imageType=NoImage
      mirroring=NoMirror
      resourceName=None
      resourceObject=null
      tiling=NoTile
      imageSize=<0 0>
      margin=<ltrb 0 0 0 0>
      outlineSettings {
       bUseBrushTransparency=false
       roundingType=FixedRadius
       width=1
       color {colorUseRule=UseColor_ColorTable, specifiedColor=<rgba 0.027 0.027 0.031 1>}
       cornerRadii=<5 5 5 5>
      }
      tintColor {colorUseRule=UseColor_Specified, specifiedColor=<rgba 0.04 0.04 0.048 1>}
      uVRegion {bIsValid=false, max=<0 0>, min=<0 0>}
     }
     pressedForeground {colorUseRule=UseColor_Foreground, specifiedColor=<rgba 1 1 1 1>}
     pressedPadding=<ltrb 0 0 0 0>
     pressedSlateSound {resourceObject=null}
    }
   }
   slot {class=VerticalBoxSlot, properties {padding=<ltrb 4 1 4 3>}}
  }
  HBox_Content_16 {
   widgetClassPath=&HorizontalBox
   parent=Btn_MovingLocal
   widget=HBox_Content_16
   slot {class=ButtonSlot, properties {horizontalAlignment=HAlign_Fill, padding=<ltrb 1 4 4 4>}}
  }
  Overlay_Icon_16 {
   widgetClassPath=&Overlay
   parent=HBox_Content_16
   widget=Overlay_Icon_16
   slot {class=HorizontalBoxSlot, properties {horizontalAlignment=HAlign_Left, verticalAlignment=VAlign_Center, padding=<ltrb 0 0 10 0>}}
  }
  Img_Icon_16 {
   widgetClassPath=&Image
   bIsVariable=true
   parent=Overlay_Icon_16
   widget=Img_Icon_16
   properties {brush {
     bIsDynamicallyLoaded=false
     drawAs=Image
     imageType=FullColor
     mirroring=NoMirror
     resourceName=None
     resourceObject=/AnimInEdTools/AnimationToolbox/Icons/AnimToolbox_UI_Icon_56
     tiling=NoTile
     imageSize=<24 24>
     margin=<ltrb 0 0 0 0>
     outlineSettings {
      bUseBrushTransparency=false
      roundingType=FixedRadius
      width=0
      color {colorUseRule=UseColor_Specified, specifiedColor=<rgba 0 0 0 1>}
      cornerRadii=<0 0 0 0>
     }
     tintColor {colorUseRule=UseColor_Specified, specifiedColor=<rgba 1 1 1 1>}
     uVRegion {bIsValid=false, max=<0 0>, min=<0 0>}
    }}
   slot {class=OverlaySlot, properties {verticalAlignment=VAlign_Center, padding=<ltrb 6 0 0 0>}}
  }
  TB_TabName_13 {
   widgetClassPath=&TextBlock
   bIsVariable=true
   parent=HBox_Content_16
   widget=TB_TabName_13
   properties {
    text="Root → Moving Local"
    textOverflowPolicy=Ellipsis
    font {
     bForceMonospaced=false
     bMaterialIsStencil=false
     fontMaterial=null
     fontObject=&Roboto
     letterSpacing=0
     monospacedWidth=0
     size=10
     skewAmount=0
     typefaceFontName=Bold
     outlineSettings {
      bApplyOutlineToDropShadows=false
      bMiteredCorners=false
      bSeparateFillAlpha=false
      outlineMaterial=null
      outlineSize=0
      outlineColor=<rgba 0 0 0 1>
     }
    }
   }
   slot {class=HorizontalBoxSlot, properties {verticalAlignment=VAlign_Center, size {sizeRule=Fill, value=1}}}
  }
  Btn_MovingLocal_Undo {
   widgetClassPath=&Button
   parent=VerticalBox_0
   widget=Btn_MovingLocal_Undo
   properties {
    toolTipText=Uses space switching to move forward motion from a moving Local control to static Local
    colorAndOpacity=<rgba 0.1195 0.521 0.4072 1>
    widgetStyle {
     clickedSlateSound {resourceObject=null}
     disabled {
      bIsDynamicallyLoaded=false
      drawAs=RoundedBox
      imageType=NoImage
      mirroring=NoMirror
      resourceName=None
      resourceObject=null
      tiling=NoTile
      imageSize=<0 0>
      margin=<ltrb 0 0 0 0>
      outlineSettings {
       bUseBrushTransparency=false
       roundingType=FixedRadius
       width=1
       color {colorUseRule=UseColor_ColorTable, specifiedColor=<rgba 0.027 0.027 0.031 1>}
       cornerRadii=<5 5 5 5>
      }
      tintColor {colorUseRule=UseColor_Specified, specifiedColor=<rgba 0.015 0.015 0.018 1>}
      uVRegion {bIsValid=false, max=<0 0>, min=<0 0>}
     }
     disabledForeground {colorUseRule=UseColor_ColorTable, specifiedColor=<rgba 0.4 0.4 0.4 1>}
     hovered {
      bIsDynamicallyLoaded=false
      drawAs=RoundedBox
      imageType=NoImage
      mirroring=NoMirror
      resourceName=None
      resourceObject=null
      tiling=NoTile
      imageSize=<0 0>
      margin=<ltrb 0 0 0 0>
      outlineSettings {
       bUseBrushTransparency=false
       roundingType=FixedRadius
       width=1
       color {colorUseRule=UseColor_ColorTable, specifiedColor=<rgba 0.027 0.027 0.031 1>}
       cornerRadii=<5 5 5 5>
      }
      tintColor {colorUseRule=UseColor_Specified, specifiedColor=<rgba 0.031 0.031 0.038 1>}
      uVRegion {bIsValid=false, max=<0 0>, min=<0 0>}
     }
     hoveredForeground {colorUseRule=UseColor_Foreground, specifiedColor=<rgba 1 1 1 1>}
     hoveredSlateSound {resourceObject=null}
     normal {
      bIsDynamicallyLoaded=false
      drawAs=RoundedBox
      imageType=NoImage
      mirroring=NoMirror
      resourceName=None
      resourceObject=null
      tiling=NoTile
      imageSize=<0 0>
      margin=<ltrb 0 0 0 0>
      outlineSettings {
       bUseBrushTransparency=false
       roundingType=FixedRadius
       width=1
       color {colorUseRule=UseColor_ColorTable, specifiedColor=<rgba 0.027 0.027 0.031 1>}
       cornerRadii=<5 5 5 5>
      }
      tintColor {colorUseRule=UseColor_Specified, specifiedColor=<rgba 0.0222 0.0222 0.0252 1>}
      uVRegion {bIsValid=false, max=<0 0>, min=<0 0>}
     }
     normalForeground {colorUseRule=UseColor_Foreground, specifiedColor=<rgba 1 1 1 1>}
     normalPadding=<ltrb 0 0 0 0>
     pressed {
      bIsDynamicallyLoaded=false
      drawAs=RoundedBox
      imageType=NoImage
      mirroring=NoMirror
      resourceName=None
      resourceObject=null
      tiling=NoTile
      imageSize=<0 0>
      margin=<ltrb 0 0 0 0>
      outlineSettings {
       bUseBrushTransparency=false
       roundingType=FixedRadius
       width=1
       color {colorUseRule=UseColor_ColorTable, specifiedColor=<rgba 0.027 0.027 0.031 1>}
       cornerRadii=<5 5 5 5>
      }
      tintColor {colorUseRule=UseColor_Specified, specifiedColor=<rgba 0.04 0.04 0.048 1>}
      uVRegion {bIsValid=false, max=<0 0>, min=<0 0>}
     }
     pressedForeground {colorUseRule=UseColor_Foreground, specifiedColor=<rgba 1 1 1 1>}
     pressedPadding=<ltrb 0 0 0 0>
     pressedSlateSound {resourceObject=null}
    }
   }
   slot {class=VerticalBoxSlot, properties {padding=<ltrb 4 1 4 3>}}
  }
  HBox_Content_18 {
   widgetClassPath=&HorizontalBox
   parent=Btn_MovingLocal_Undo
   widget=HBox_Content_18
   slot {class=ButtonSlot, properties {horizontalAlignment=HAlign_Fill, padding=<ltrb 1 4 4 4>}}
  }
  Overlay_Icon_18 {
   widgetClassPath=&Overlay
   parent=HBox_Content_18
   widget=Overlay_Icon_18
   slot {class=HorizontalBoxSlot, properties {horizontalAlignment=HAlign_Left, verticalAlignment=VAlign_Center, padding=<ltrb 0 0 10 0>}}
  }
  Img_Icon_17 {
   widgetClassPath=&Image
   bIsVariable=true
   parent=Overlay_Icon_18
   widget=Img_Icon_17
   properties {brush {
     bIsDynamicallyLoaded=false
     drawAs=Image
     imageType=FullColor
     mirroring=NoMirror
     resourceName=None
     resourceObject=/AnimInEdTools/AnimationToolbox/Icons/AnimToolbox_UI_Icon_56_02
     tiling=NoTile
     imageSize=<24 24>
     margin=<ltrb 0 0 0 0>
     outlineSettings {
      bUseBrushTransparency=false
      roundingType=FixedRadius
      width=0
      color {colorUseRule=UseColor_Specified, specifiedColor=<rgba 0 0 0 1>}
      cornerRadii=<0 0 0 0>
     }
     tintColor {colorUseRule=UseColor_Specified, specifiedColor=<rgba 1 1 1 1>}
     uVRegion {bIsValid=false, max=<0 0>, min=<0 0>}
    }}
   slot {class=OverlaySlot, properties {verticalAlignment=VAlign_Center, padding=<ltrb 6 0 0 0>}}
  }
  TB_TabName_16 {
   widgetClassPath=&TextBlock
   bIsVariable=true
   parent=HBox_Content_18
   widget=TB_TabName_16
   properties {
    text="Root ← Moving Local"
    textOverflowPolicy=Ellipsis
    font {
     bForceMonospaced=false
     bMaterialIsStencil=false
     fontMaterial=null
     fontObject=&Roboto
     letterSpacing=0
     monospacedWidth=0
     size=10
     skewAmount=0
     typefaceFontName=Bold
     outlineSettings {
      bApplyOutlineToDropShadows=false
      bMiteredCorners=false
      bSeparateFillAlpha=false
      outlineMaterial=null
      outlineSize=0
      outlineColor=<rgba 0 0 0 1>
     }
    }
   }
   slot {class=HorizontalBoxSlot, properties {verticalAlignment=VAlign_Center, size {sizeRule=Fill, value=1}}}
  }
  TextBlock_90 {
   widgetClassPath=&TextBlock
   parent=VerticalBox_0
   widget=TextBlock_90
   properties {
    bOverride_Cursor=true
    text="Locators&&Constraints"
    colorAndOpacity {colorUseRule=UseColor_Specified, specifiedColor=<rgba 1 1 1 0.5>}
    font {
     bForceMonospaced=false
     bMaterialIsStencil=false
     fontMaterial=null
     fontObject=&Roboto
     letterSpacing=0
     monospacedWidth=0
     size=6
     skewAmount=0
     typefaceFontName=Bold
     outlineSettings {
      bApplyOutlineToDropShadows=false
      bMiteredCorners=false
      bSeparateFillAlpha=false
      outlineMaterial=null
      outlineSize=0
      outlineColor=<rgba 0 0 0 1>
     }
    }
   }
   slot {class=VerticalBoxSlot, properties {horizontalAlignment=HAlign_Left, padding=<ltrb 5 2 0 2>}}
  }
  BTN_Locator {
   widgetClassPath=&EditorUtilityButton
   parent=VerticalBox_0
   widget=BTN_Locator
   properties {
    toolTipText="Add Locator\r\n\r\nadds locator to your selection"
    colorAndOpacity=<rgba 0.1878 0.3419 0.0999 1>
    widgetStyle {
     clickedSlateSound {resourceObject=null}
     disabled {
      bIsDynamicallyLoaded=false
      drawAs=RoundedBox
      imageType=NoImage
      mirroring=NoMirror
      resourceName=None
      resourceObject=null
      tiling=NoTile
      imageSize=<0 0>
      margin=<ltrb 0 0 0 0>
      outlineSettings {
       bUseBrushTransparency=false
       roundingType=FixedRadius
       width=1
       color {colorUseRule=UseColor_ColorTable, specifiedColor=<rgba 0.027 0.027 0.031 1>}
       cornerRadii=<5 5 5 5>
      }
      tintColor {colorUseRule=UseColor_Specified, specifiedColor=<rgba 0.015 0.015 0.018 1>}
      uVRegion {bIsValid=false, max=<0 0>, min=<0 0>}
     }
     disabledForeground {colorUseRule=UseColor_ColorTable, specifiedColor=<rgba 0.4 0.4 0.4 1>}
     hovered {
      bIsDynamicallyLoaded=false
      drawAs=RoundedBox
      imageType=NoImage
      mirroring=NoMirror
      resourceName=None
      resourceObject=null
      tiling=NoTile
      imageSize=<0 0>
      margin=<ltrb 0 0 0 0>
      outlineSettings {
       bUseBrushTransparency=false
       roundingType=FixedRadius
       width=1
       color {colorUseRule=UseColor_ColorTable, specifiedColor=<rgba 0.027 0.027 0.031 1>}
       cornerRadii=<5 5 5 5>
      }
      tintColor {colorUseRule=UseColor_Specified, specifiedColor=<rgba 0.031 0.031 0.038 1>}
      uVRegion {bIsValid=false, max=<0 0>, min=<0 0>}
     }
     hoveredForeground {colorUseRule=UseColor_Foreground, specifiedColor=<rgba 1 1 1 1>}
     hoveredSlateSound {resourceObject=null}
     normal {
      bIsDynamicallyLoaded=false
      drawAs=RoundedBox
      imageType=NoImage
      mirroring=NoMirror
      resourceName=None
      resourceObject=null
      tiling=NoTile
      imageSize=<0 0>
      margin=<ltrb 0 0 0 0>
      outlineSettings {
       bUseBrushTransparency=false
       roundingType=FixedRadius
       width=1
       color {colorUseRule=UseColor_ColorTable, specifiedColor=<rgba 0.027 0.027 0.031 1>}
       cornerRadii=<5 5 5 5>
      }
      tintColor {colorUseRule=UseColor_Specified, specifiedColor=<rgba 0.0222 0.0222 0.0252 1>}
      uVRegion {bIsValid=false, max=<0 0>, min=<0 0>}
     }
     normalForeground {colorUseRule=UseColor_Foreground, specifiedColor=<rgba 1 1 1 1>}
     normalPadding=<ltrb 0 0 0 0>
     pressed {
      bIsDynamicallyLoaded=false
      drawAs=RoundedBox
      imageType=NoImage
      mirroring=NoMirror
      resourceName=None
      resourceObject=null
      tiling=NoTile
      imageSize=<0 0>
      margin=<ltrb 0 0 0 0>
      outlineSettings {
       bUseBrushTransparency=false
       roundingType=FixedRadius
       width=1
       color {colorUseRule=UseColor_ColorTable, specifiedColor=<rgba 0.027 0.027 0.031 1>}
       cornerRadii=<5 5 5 5>
      }
      tintColor {colorUseRule=UseColor_Specified, specifiedColor=<rgba 0.04 0.04 0.048 1>}
      uVRegion {bIsValid=false, max=<0 0>, min=<0 0>}
     }
     pressedForeground {colorUseRule=UseColor_Foreground, specifiedColor=<rgba 1 1 1 1>}
     pressedPadding=<ltrb 0 0 0 0>
     pressedSlateSound {resourceObject=null}
    }
   }
   slot {class=VerticalBoxSlot, properties {padding=<ltrb 4 1 4 3>, size {sizeRule=Fill, value=1}}}
  }
  HBox_Content_1 {
   widgetClassPath=&HorizontalBox
   parent=BTN_Locator
   widget=HBox_Content_1
   slot {class=ButtonSlot, properties {horizontalAlignment=HAlign_Fill, padding=<ltrb 1 4 4 4>}}
  }
  Overlay_Icon_1 {
   widgetClassPath=&Overlay
   parent=HBox_Content_1
   widget=Overlay_Icon_1
   slot {class=HorizontalBoxSlot, properties {horizontalAlignment=HAlign_Left, verticalAlignment=VAlign_Center, padding=<ltrb 0 0 10 0>}}
  }
  Img_Icon_1 {
   widgetClassPath=&Image
   bIsVariable=true
   parent=Overlay_Icon_1
   widget=Img_Icon_1
   properties {brush {
     bIsDynamicallyLoaded=false
     drawAs=Image
     imageType=FullColor
     mirroring=NoMirror
     resourceName=None
     resourceObject=/AnimInEdTools/AnimationToolbox/Icons/AnimToolbox_UI_Icon_19
     tiling=NoTile
     imageSize=<24 24>
     margin=<ltrb 0 0 0 0>
     outlineSettings {
      bUseBrushTransparency=false
      roundingType=FixedRadius
      width=0
      color {colorUseRule=UseColor_Specified, specifiedColor=<rgba 0 0 0 1>}
      cornerRadii=<0 0 0 0>
     }
     tintColor {colorUseRule=UseColor_Specified, specifiedColor=<rgba 1 1 1 1>}
     uVRegion {bIsValid=false, max=<0 0>, min=<0 0>}
    }}
   slot {class=OverlaySlot, properties {verticalAlignment=VAlign_Center, padding=<ltrb 6 0 0 0>}}
  }
  TB_TabName_1 {
   widgetClassPath=&TextBlock
   bIsVariable=true
   parent=HBox_Content_1
   widget=TB_TabName_1
   properties {
    text=Add Locator
    textOverflowPolicy=Ellipsis
    font {
     bForceMonospaced=false
     bMaterialIsStencil=false
     fontMaterial=null
     fontObject=&Roboto
     letterSpacing=0
     monospacedWidth=0
     size=10
     skewAmount=0
     typefaceFontName=Bold
     outlineSettings {
      bApplyOutlineToDropShadows=false
      bMiteredCorners=false
      bSeparateFillAlpha=false
      outlineMaterial=null
      outlineSize=0
      outlineColor=<rgba 0 0 0 1>
     }
    }
   }
   slot {class=HorizontalBoxSlot, properties {verticalAlignment=VAlign_Center, size {sizeRule=Fill, value=1}}}
  }
  BTN_Locator_Bake {
   widgetClassPath=&EditorUtilityButton
   parent=VerticalBox_0
   widget=BTN_Locator_Bake
   properties {
    toolTipText="Bake to Locator\r\n\r\ncopy a control's worldspace transform to a locator\r\n\r\n***\r\nRight Click to bake with Offset\r\n***"
    colorAndOpacity=<rgba 0.1878 0.3419 0.0999 1>
    widgetStyle {
     clickedSlateSound {resourceObject=null}
     disabled {
      bIsDynamicallyLoaded=false
      drawAs=RoundedBox
      imageType=NoImage
      mirroring=NoMirror
      resourceName=None
      resourceObject=null
      tiling=NoTile
      imageSize=<0 0>
      margin=<ltrb 0 0 0 0>
      outlineSettings {
       bUseBrushTransparency=false
       roundingType=FixedRadius
       width=1
       color {colorUseRule=UseColor_ColorTable, specifiedColor=<rgba 0.027 0.027 0.031 1>}
       cornerRadii=<5 5 5 5>
      }
      tintColor {colorUseRule=UseColor_Specified, specifiedColor=<rgba 0.015 0.015 0.018 1>}
      uVRegion {bIsValid=false, max=<0 0>, min=<0 0>}
     }
     disabledForeground {colorUseRule=UseColor_ColorTable, specifiedColor=<rgba 0.4 0.4 0.4 1>}
     hovered {
      bIsDynamicallyLoaded=false
      drawAs=RoundedBox
      imageType=NoImage
      mirroring=NoMirror
      resourceName=None
      resourceObject=null
      tiling=NoTile
      imageSize=<0 0>
      margin=<ltrb 0 0 0 0>
      outlineSettings {
       bUseBrushTransparency=false
       roundingType=FixedRadius
       width=1
       color {colorUseRule=UseColor_Specified, specifiedColor=<rgba 0.1878 0.3419 0.0999 1>}
       cornerRadii=<5 5 5 5>
      }
      tintColor {colorUseRule=UseColor_Specified, specifiedColor=<rgba 0.031 0.031 0.038 1>}
      uVRegion {bIsValid=false, max=<0 0>, min=<0 0>}
     }
     hoveredForeground {colorUseRule=UseColor_Foreground, specifiedColor=<rgba 1 1 1 1>}
     hoveredSlateSound {resourceObject=null}
     normal {
      bIsDynamicallyLoaded=false
      drawAs=RoundedBox
      imageType=NoImage
      mirroring=NoMirror
      resourceName=None
      resourceObject=null
      tiling=NoTile
      imageSize=<0 0>
      margin=<ltrb 0 0 0 0>
      outlineSettings {
       bUseBrushTransparency=false
       roundingType=FixedRadius
       width=1
       color {colorUseRule=UseColor_ColorTable, specifiedColor=<rgba 0.027 0.027 0.031 1>}
       cornerRadii=<5 5 5 5>
      }
      tintColor {colorUseRule=UseColor_Specified, specifiedColor=<rgba 0.0222 0.0222 0.0252 1>}
      uVRegion {bIsValid=false, max=<0 0>, min=<0 0>}
     }
     normalForeground {colorUseRule=UseColor_Foreground, specifiedColor=<rgba 1 1 1 1>}
     normalPadding=<ltrb 0 0 0 0>
     pressed {
      bIsDynamicallyLoaded=false
      drawAs=RoundedBox
      imageType=NoImage
      mirroring=NoMirror
      resourceName=None
      resourceObject=null
      tiling=NoTile
      imageSize=<0 0>
      margin=<ltrb 0 0 0 0>
      outlineSettings {
       bUseBrushTransparency=false
       roundingType=FixedRadius
       width=1
       color {colorUseRule=UseColor_ColorTable, specifiedColor=<rgba 0.027 0.027 0.031 1>}
       cornerRadii=<5 5 5 5>
      }
      tintColor {colorUseRule=UseColor_Specified, specifiedColor=<rgba 0.04 0.04 0.048 1>}
      uVRegion {bIsValid=false, max=<0 0>, min=<0 0>}
     }
     pressedForeground {colorUseRule=UseColor_Foreground, specifiedColor=<rgba 1 1 1 1>}
     pressedPadding=<ltrb 0 0 0 0>
     pressedSlateSound {resourceObject=null}
    }
   }
   slot {class=VerticalBoxSlot, properties {padding=<ltrb 4 1 4 3>}}
  }
  HBox_Content {
   widgetClassPath=&HorizontalBox
   parent=BTN_Locator_Bake
   widget=HBox_Content
   slot {class=ButtonSlot, properties {horizontalAlignment=HAlign_Fill, padding=<ltrb 1 4 4 4>}}
  }
  Overlay_Icon {
   widgetClassPath=&Overlay
   parent=HBox_Content
   widget=Overlay_Icon
   slot {class=HorizontalBoxSlot, properties {horizontalAlignment=HAlign_Left, verticalAlignment=VAlign_Center, padding=<ltrb 0 0 10 0>}}
  }
  Img_Icon {
   widgetClassPath=&Image
   bIsVariable=true
   parent=Overlay_Icon
   widget=Img_Icon
   properties {brush {
     bIsDynamicallyLoaded=false
     drawAs=Image
     imageType=FullColor
     mirroring=NoMirror
     resourceName=None
     resourceObject=/AnimInEdTools/AnimationToolbox/Icons/AnimToolbox_UI_Icon_38
     tiling=NoTile
     imageSize=<24 24>
     margin=<ltrb 0 0 0 0>
     outlineSettings {
      bUseBrushTransparency=false
      roundingType=FixedRadius
      width=0
      color {colorUseRule=UseColor_Specified, specifiedColor=<rgba 0 0 0 1>}
      cornerRadii=<0 0 0 0>
     }
     tintColor {colorUseRule=UseColor_Specified, specifiedColor=<rgba 1 1 1 1>}
     uVRegion {bIsValid=false, max=<0 0>, min=<0 0>}
    }}
   slot {class=OverlaySlot, properties {verticalAlignment=VAlign_Center, padding=<ltrb 6 0 0 0>}}
  }
  TB_TabName {
   widgetClassPath=&TextBlock
   bIsVariable=true
   parent=HBox_Content
   widget=TB_TabName
   properties {
    text=Bake to Locator
    textOverflowPolicy=Ellipsis
    font {
     bForceMonospaced=false
     bMaterialIsStencil=false
     fontMaterial=null
     fontObject=&Roboto
     letterSpacing=0
     monospacedWidth=0
     size=10
     skewAmount=0
     typefaceFontName=Bold
     outlineSettings {
      bApplyOutlineToDropShadows=false
      bMiteredCorners=false
      bSeparateFillAlpha=false
      outlineMaterial=null
      outlineSize=0
      outlineColor=<rgba 0 0 0 1>
     }
    }
   }
   slot {class=HorizontalBoxSlot, properties {verticalAlignment=VAlign_Center, size {sizeRule=Fill, value=1}}}
  }
  BTN_Locator_Bake_Connect {
   widgetClassPath=&EditorUtilityButton
   parent=VerticalBox_0
   widget=BTN_Locator_Bake_Connect
   properties {
    toolTipText="Bake && Constrain to Locator\r\n\r\nBake a locator to a controller, then constrain the controller to that locator\r\n***\r\nRight Click to bake connect with Offset\r\n***"
    colorAndOpacity=<rgba 0.7157 0.0648 0.3916 1>
    widgetStyle {
     clickedSlateSound {resourceObject=null}
     disabled {
      bIsDynamicallyLoaded=false
      drawAs=RoundedBox
      imageType=NoImage
      mirroring=NoMirror
      resourceName=None
      resourceObject=null
      tiling=NoTile
      imageSize=<0 0>
      margin=<ltrb 0 0 0 0>
      outlineSettings {
       bUseBrushTransparency=false
       roundingType=FixedRadius
       width=1
       color {colorUseRule=UseColor_ColorTable, specifiedColor=<rgba 0.027 0.027 0.031 1>}
       cornerRadii=<5 5 5 5>
      }
      tintColor {colorUseRule=UseColor_Specified, specifiedColor=<rgba 0.015 0.015 0.018 1>}
      uVRegion {bIsValid=false, max=<0 0>, min=<0 0>}
     }
     disabledForeground {colorUseRule=UseColor_ColorTable, specifiedColor=<rgba 0.4 0.4 0.4 1>}
     hovered {
      bIsDynamicallyLoaded=false
      drawAs=RoundedBox
      imageType=NoImage
      mirroring=NoMirror
      resourceName=None
      resourceObject=null
      tiling=NoTile
      imageSize=<0 0>
      margin=<ltrb 0 0 0 0>
      outlineSettings {
       bUseBrushTransparency=false
       roundingType=FixedRadius
       width=1
       color {colorUseRule=UseColor_Specified, specifiedColor=<rgba 0.7157 0.0648 0.3916 1>}
       cornerRadii=<5 5 5 5>
      }
      tintColor {colorUseRule=UseColor_Specified, specifiedColor=<rgba 0.031 0.031 0.038 1>}
      uVRegion {bIsValid=false, max=<0 0>, min=<0 0>}
     }
     hoveredForeground {colorUseRule=UseColor_Foreground, specifiedColor=<rgba 1 1 1 1>}
     hoveredSlateSound {resourceObject=null}
     normal {
      bIsDynamicallyLoaded=false
      drawAs=RoundedBox
      imageType=NoImage
      mirroring=NoMirror
      resourceName=None
      resourceObject=null
      tiling=NoTile
      imageSize=<0 0>
      margin=<ltrb 0 0 0 0>
      outlineSettings {
       bUseBrushTransparency=false
       roundingType=FixedRadius
       width=1
       color {colorUseRule=UseColor_ColorTable, specifiedColor=<rgba 0.027 0.027 0.031 1>}
       cornerRadii=<5 5 5 5>
      }
      tintColor {colorUseRule=UseColor_Specified, specifiedColor=<rgba 0.0222 0.0222 0.0252 1>}
      uVRegion {bIsValid=false, max=<0 0>, min=<0 0>}
     }
     normalForeground {colorUseRule=UseColor_Foreground, specifiedColor=<rgba 1 1 1 1>}
     normalPadding=<ltrb 0 0 0 0>
     pressed {
      bIsDynamicallyLoaded=false
      drawAs=RoundedBox
      imageType=NoImage
      mirroring=NoMirror
      resourceName=None
      resourceObject=null
      tiling=NoTile
      imageSize=<0 0>
      margin=<ltrb 0 0 0 0>
      outlineSettings {
       bUseBrushTransparency=false
       roundingType=FixedRadius
       width=1
       color {colorUseRule=UseColor_ColorTable, specifiedColor=<rgba 0.027 0.027 0.031 1>}
       cornerRadii=<5 5 5 5>
      }
      tintColor {colorUseRule=UseColor_Specified, specifiedColor=<rgba 0.04 0.04 0.048 1>}
      uVRegion {bIsValid=false, max=<0 0>, min=<0 0>}
     }
     pressedForeground {colorUseRule=UseColor_Foreground, specifiedColor=<rgba 1 1 1 1>}
     pressedPadding=<ltrb 0 0 0 0>
     pressedSlateSound {resourceObject=null}
    }
   }
   slot {class=VerticalBoxSlot, properties {padding=<ltrb 4 1 4 3>}}
  }
  HBox_Content_2 {
   widgetClassPath=&HorizontalBox
   parent=BTN_Locator_Bake_Connect
   widget=HBox_Content_2
   slot {class=ButtonSlot, properties {horizontalAlignment=HAlign_Fill, padding=<ltrb 1 4 4 4>}}
  }
  Overlay_Icon_2 {
   widgetClassPath=&Overlay
   parent=HBox_Content_2
   widget=Overlay_Icon_2
   slot {class=HorizontalBoxSlot, properties {horizontalAlignment=HAlign_Left, verticalAlignment=VAlign_Center, padding=<ltrb 0 0 10 0>}}
  }
  Img_Icon_2 {
   widgetClassPath=&Image
   bIsVariable=true
   parent=Overlay_Icon_2
   widget=Img_Icon_2
   properties {brush {
     bIsDynamicallyLoaded=false
     drawAs=Image
     imageType=FullColor
     mirroring=NoMirror
     resourceName=None
     resourceObject=/AnimInEdTools/AnimationToolbox/Icons/AnimToolbox_UI_Icon_39
     tiling=NoTile
     imageSize=<24 24>
     margin=<ltrb 0 0 0 0>
     outlineSettings {
      bUseBrushTransparency=false
      roundingType=FixedRadius
      width=0
      color {colorUseRule=UseColor_Specified, specifiedColor=<rgba 0 0 0 1>}
      cornerRadii=<0 0 0 0>
     }
     tintColor {colorUseRule=UseColor_Specified, specifiedColor=<rgba 1 1 1 1>}
     uVRegion {bIsValid=false, max=<0 0>, min=<0 0>}
    }}
   slot {class=OverlaySlot, properties {verticalAlignment=VAlign_Center, padding=<ltrb 6 0 0 0>}}
  }
  TB_TabName_2 {
   widgetClassPath=&TextBlock
   bIsVariable=true
   parent=HBox_Content_2
   widget=TB_TabName_2
   properties {
    text="Bake && Constrain to Locator"
    textOverflowPolicy=Ellipsis
    font {
     bForceMonospaced=false
     bMaterialIsStencil=false
     fontMaterial=null
     fontObject=&Roboto
     letterSpacing=0
     monospacedWidth=0
     size=10
     skewAmount=0
     typefaceFontName=Bold
     outlineSettings {
      bApplyOutlineToDropShadows=false
      bMiteredCorners=false
      bSeparateFillAlpha=false
      outlineMaterial=null
      outlineSize=0
      outlineColor=<rgba 0 0 0 1>
     }
    }
   }
   slot {class=HorizontalBoxSlot, properties {verticalAlignment=VAlign_Center, size {sizeRule=Fill, value=1}}}
  }
  BTN_Locator_Bake_to_Aim {
   widgetClassPath=&EditorUtilityButton
   parent=VerticalBox_0
   widget=BTN_Locator_Bake_to_Aim
   properties {
    toolTipText="Bake to Aim Locator\r\n\r\nBake a locator to a controller with an offset, then lookat constrain the controller to that locator"
    colorAndOpacity=<rgba 0.7157 0.0648 0.3916 1>
    widgetStyle {
     clickedSlateSound {resourceObject=null}
     disabled {
      bIsDynamicallyLoaded=false
      drawAs=RoundedBox
      imageType=NoImage
      mirroring=NoMirror
      resourceName=None
      resourceObject=null
      tiling=NoTile
      imageSize=<0 0>
      margin=<ltrb 0 0 0 0>
      outlineSettings {
       bUseBrushTransparency=false
       roundingType=FixedRadius
       width=1
       color {colorUseRule=UseColor_ColorTable, specifiedColor=<rgba 0.027 0.027 0.031 1>}
       cornerRadii=<5 5 5 5>
      }
      tintColor {colorUseRule=UseColor_Specified, specifiedColor=<rgba 0.015 0.015 0.018 1>}
      uVRegion {bIsValid=false, max=<0 0>, min=<0 0>}
     }
     disabledForeground {colorUseRule=UseColor_ColorTable, specifiedColor=<rgba 0.4 0.4 0.4 1>}
     hovered {
      bIsDynamicallyLoaded=false
      drawAs=RoundedBox
      imageType=NoImage
      mirroring=NoMirror
      resourceName=None
      resourceObject=null
      tiling=NoTile
      imageSize=<0 0>
      margin=<ltrb 0 0 0 0>
      outlineSettings {
       bUseBrushTransparency=false
       roundingType=FixedRadius
       width=1
       color {colorUseRule=UseColor_ColorTable, specifiedColor=<rgba 0.027 0.027 0.031 1>}
       cornerRadii=<5 5 5 5>
      }
      tintColor {colorUseRule=UseColor_Specified, specifiedColor=<rgba 0.031 0.031 0.038 1>}
      uVRegion {bIsValid=false, max=<0 0>, min=<0 0>}
     }
     hoveredForeground {colorUseRule=UseColor_Foreground, specifiedColor=<rgba 1 1 1 1>}
     hoveredSlateSound {resourceObject=null}
     normal {
      bIsDynamicallyLoaded=false
      drawAs=RoundedBox
      imageType=NoImage
      mirroring=NoMirror
      resourceName=None
      resourceObject=null
      tiling=NoTile
      imageSize=<0 0>
      margin=<ltrb 0 0 0 0>
      outlineSettings {
       bUseBrushTransparency=false
       roundingType=FixedRadius
       width=1
       color {colorUseRule=UseColor_ColorTable, specifiedColor=<rgba 0.027 0.027 0.031 1>}
       cornerRadii=<5 5 5 5>
      }
      tintColor {colorUseRule=UseColor_Specified, specifiedColor=<rgba 0.0222 0.0222 0.0252 1>}
      uVRegion {bIsValid=false, max=<0 0>, min=<0 0>}
     }
     normalForeground {colorUseRule=UseColor_Foreground, specifiedColor=<rgba 1 1 1 1>}
     normalPadding=<ltrb 0 0 0 0>
     pressed {
      bIsDynamicallyLoaded=false
      drawAs=RoundedBox
      imageType=NoImage
      mirroring=NoMirror
      resourceName=None
      resourceObject=null
      tiling=NoTile
      imageSize=<0 0>
      margin=<ltrb 0 0 0 0>
      outlineSettings {
       bUseBrushTransparency=false
       roundingType=FixedRadius
       width=1
       color {colorUseRule=UseColor_ColorTable, specifiedColor=<rgba 0.027 0.027 0.031 1>}
       cornerRadii=<5 5 5 5>
      }
      tintColor {colorUseRule=UseColor_Specified, specifiedColor=<rgba 0.04 0.04 0.048 1>}
      uVRegion {bIsValid=false, max=<0 0>, min=<0 0>}
     }
     pressedForeground {colorUseRule=UseColor_Foreground, specifiedColor=<rgba 1 1 1 1>}
     pressedPadding=<ltrb 0 0 0 0>
     pressedSlateSound {resourceObject=null}
    }
   }
   slot {class=VerticalBoxSlot, properties {padding=<ltrb 4 1 4 3>}}
  }
  HBox_Content_17 {
   widgetClassPath=&HorizontalBox
   parent=BTN_Locator_Bake_to_Aim
   widget=HBox_Content_17
   slot {class=ButtonSlot, properties {horizontalAlignment=HAlign_Fill, padding=<ltrb 1 4 4 4>}}
  }
  Overlay_Icon_17 {
   widgetClassPath=&Overlay
   parent=HBox_Content_17
   widget=Overlay_Icon_17
   slot {class=HorizontalBoxSlot, properties {horizontalAlignment=HAlign_Left, verticalAlignment=VAlign_Center, padding=<ltrb 0 0 10 0>}}
  }
  Img_Icon_15 {
   widgetClassPath=&Image
   bIsVariable=true
   parent=Overlay_Icon_17
   widget=Img_Icon_15
   properties {brush {
     bIsDynamicallyLoaded=false
     drawAs=Image
     imageType=FullColor
     mirroring=NoMirror
     resourceName=None
     resourceObject=/AnimInEdTools/AnimationToolbox/Icons/AnimToolbox_UI_Icon_02
     tiling=NoTile
     imageSize=<24 24>
     margin=<ltrb 0 0 0 0>
     outlineSettings {
      bUseBrushTransparency=false
      roundingType=FixedRadius
      width=0
      color {colorUseRule=UseColor_Specified, specifiedColor=<rgba 0 0 0 1>}
      cornerRadii=<0 0 0 0>
     }
     tintColor {colorUseRule=UseColor_Specified, specifiedColor=<rgba 1 1 1 1>}
     uVRegion {bIsValid=false, max=<0 0>, min=<0 0>}
    }}
   slot {class=OverlaySlot, properties {verticalAlignment=VAlign_Center, padding=<ltrb 6 0 0 0>}}
  }
  TB_TabName_15 {
   widgetClassPath=&TextBlock
   bIsVariable=true
   parent=HBox_Content_17
   widget=TB_TabName_15
   properties {
    text=Bake to Aim Locator
    textOverflowPolicy=Ellipsis
    font {
     bForceMonospaced=false
     bMaterialIsStencil=false
     fontMaterial=null
     fontObject=&Roboto
     letterSpacing=0
     monospacedWidth=0
     size=10
     skewAmount=0
     typefaceFontName=Bold
     outlineSettings {
      bApplyOutlineToDropShadows=false
      bMiteredCorners=false
      bSeparateFillAlpha=false
      outlineMaterial=null
      outlineSize=0
      outlineColor=<rgba 0 0 0 1>
     }
    }
   }
   slot {class=HorizontalBoxSlot, properties {verticalAlignment=VAlign_Center, size {sizeRule=Fill, value=1}}}
  }
  BTN_Offset {
   widgetClassPath=&EditorUtilityButton
   parent=VerticalBox_0
   widget=BTN_Offset
   properties {
    toolTipText="Constraint Offset\r\n\r\nInserts a node in between your constraint so that you can then animate from the constraint offset point"
    colorAndOpacity=<rgba 0.7157 0.0648 0.3916 1>
    widgetStyle {
     clickedSlateSound {resourceObject=null}
     disabled {
      bIsDynamicallyLoaded=false
      drawAs=RoundedBox
      imageType=NoImage
      mirroring=NoMirror
      resourceName=None
      resourceObject=null
      tiling=NoTile
      imageSize=<0 0>
      margin=<ltrb 0 0 0 0>
      outlineSettings {
       bUseBrushTransparency=false
       roundingType=FixedRadius
       width=1
       color {colorUseRule=UseColor_ColorTable, specifiedColor=<rgba 0.027 0.027 0.031 1>}
       cornerRadii=<5 5 5 5>
      }
      tintColor {colorUseRule=UseColor_Specified, specifiedColor=<rgba 0.015 0.015 0.018 1>}
      uVRegion {bIsValid=false, max=<0 0>, min=<0 0>}
     }
     disabledForeground {colorUseRule=UseColor_ColorTable, specifiedColor=<rgba 0.4 0.4 0.4 1>}
     hovered {
      bIsDynamicallyLoaded=false
      drawAs=RoundedBox
      imageType=NoImage
      mirroring=NoMirror
      resourceName=None
      resourceObject=null
      tiling=NoTile
      imageSize=<0 0>
      margin=<ltrb 0 0 0 0>
      outlineSettings {
       bUseBrushTransparency=false
       roundingType=FixedRadius
       width=1
       color {colorUseRule=UseColor_ColorTable, specifiedColor=<rgba 0.027 0.027 0.031 1>}
       cornerRadii=<5 5 5 5>
      }
      tintColor {colorUseRule=UseColor_Specified, specifiedColor=<rgba 0.031 0.031 0.038 1>}
      uVRegion {bIsValid=false, max=<0 0>, min=<0 0>}
     }
     hoveredForeground {colorUseRule=UseColor_Foreground, specifiedColor=<rgba 1 1 1 1>}
     hoveredSlateSound {resourceObject=null}
     normal {
      bIsDynamicallyLoaded=false
      drawAs=RoundedBox
      imageType=NoImage
      mirroring=NoMirror
      resourceName=None
      resourceObject=null
      tiling=NoTile
      imageSize=<0 0>
      margin=<ltrb 0 0 0 0>
      outlineSettings {
       bUseBrushTransparency=false
       roundingType=FixedRadius
       width=1
       color {colorUseRule=UseColor_ColorTable, specifiedColor=<rgba 0.027 0.027 0.031 1>}
       cornerRadii=<5 5 5 5>
      }
      tintColor {colorUseRule=UseColor_Specified, specifiedColor=<rgba 0.0222 0.0222 0.0252 1>}
      uVRegion {bIsValid=false, max=<0 0>, min=<0 0>}
     }
     normalForeground {colorUseRule=UseColor_Foreground, specifiedColor=<rgba 1 1 1 1>}
     normalPadding=<ltrb 0 0 0 0>
     pressed {
      bIsDynamicallyLoaded=false
      drawAs=RoundedBox
      imageType=NoImage
      mirroring=NoMirror
      resourceName=None
      resourceObject=null
      tiling=NoTile
      imageSize=<0 0>
      margin=<ltrb 0 0 0 0>
      outlineSettings {
       bUseBrushTransparency=false
       roundingType=FixedRadius
       width=1
       color {colorUseRule=UseColor_ColorTable, specifiedColor=<rgba 0.027 0.027 0.031 1>}
       cornerRadii=<5 5 5 5>
      }
      tintColor {colorUseRule=UseColor_Specified, specifiedColor=<rgba 0.04 0.04 0.048 1>}
      uVRegion {bIsValid=false, max=<0 0>, min=<0 0>}
     }
     pressedForeground {colorUseRule=UseColor_Foreground, specifiedColor=<rgba 1 1 1 1>}
     pressedPadding=<ltrb 0 0 0 0>
     pressedSlateSound {resourceObject=null}
    }
   }
   slot {class=VerticalBoxSlot, properties {padding=<ltrb 4 1 4 3>}}
  }
  HBox_Content_5 {
   widgetClassPath=&HorizontalBox
   parent=BTN_Offset
   widget=HBox_Content_5
   slot {class=ButtonSlot, properties {horizontalAlignment=HAlign_Fill, padding=<ltrb 1 4 4 4>}}
  }
  Overlay_Icon_5 {
   widgetClassPath=&Overlay
   parent=HBox_Content_5
   widget=Overlay_Icon_5
   slot {class=HorizontalBoxSlot, properties {horizontalAlignment=HAlign_Left, verticalAlignment=VAlign_Center, padding=<ltrb 0 0 10 0>}}
  }
  Img_Icon_5 {
   widgetClassPath=&Image
   bIsVariable=true
   parent=Overlay_Icon_5
   widget=Img_Icon_5
   properties {brush {
     bIsDynamicallyLoaded=false
     drawAs=Image
     imageType=FullColor
     mirroring=NoMirror
     resourceName=None
     resourceObject=/AnimInEdTools/AnimationToolbox/Icons/AnimToolbox_UI_Icon_07
     tiling=NoTile
     imageSize=<24 24>
     margin=<ltrb 0 0 0 0>
     outlineSettings {
      bUseBrushTransparency=false
      roundingType=FixedRadius
      width=0
      color {colorUseRule=UseColor_Specified, specifiedColor=<rgba 0 0 0 1>}
      cornerRadii=<0 0 0 0>
     }
     tintColor {colorUseRule=UseColor_Specified, specifiedColor=<rgba 1 1 1 1>}
     uVRegion {bIsValid=false, max=<0 0>, min=<0 0>}
    }}
   slot {class=OverlaySlot, properties {verticalAlignment=VAlign_Center, padding=<ltrb 6 0 0 0>}}
  }
  TB_TabName_5 {
   widgetClassPath=&TextBlock
   bIsVariable=true
   parent=HBox_Content_5
   widget=TB_TabName_5
   properties {
    text=Constraint Offset
    textOverflowPolicy=Ellipsis
    font {
     bForceMonospaced=false
     bMaterialIsStencil=false
     fontMaterial=null
     fontObject=&Roboto
     letterSpacing=0
     monospacedWidth=0
     size=10
     skewAmount=0
     typefaceFontName=Bold
     outlineSettings {
      bApplyOutlineToDropShadows=false
      bMiteredCorners=false
      bSeparateFillAlpha=false
      outlineMaterial=null
      outlineSize=0
      outlineColor=<rgba 0 0 0 1>
     }
    }
   }
   slot {class=HorizontalBoxSlot, properties {verticalAlignment=VAlign_Center, size {sizeRule=Fill, value=1}}}
  }
  TextBlock_91 {
   widgetClassPath=&TextBlock
   parent=VerticalBox_0
   widget=TextBlock_91
   properties {
    bOverride_Cursor=true
    text=Worldspace/Aligning
    colorAndOpacity {colorUseRule=UseColor_Specified, specifiedColor=<rgba 0.8963 0.8963 0.8963 0.5>}
    font {
     bForceMonospaced=false
     bMaterialIsStencil=false
     fontMaterial=null
     fontObject=&Roboto
     letterSpacing=0
     monospacedWidth=0
     size=6
     skewAmount=0
     typefaceFontName=Bold
     outlineSettings {
      bApplyOutlineToDropShadows=false
      bMiteredCorners=false
      bSeparateFillAlpha=false
      outlineMaterial=null
      outlineSize=0
      outlineColor=<rgba 0 0 0 1>
     }
    }
   }
   slot {class=VerticalBoxSlot, properties {horizontalAlignment=HAlign_Left, padding=<ltrb 5 2 0 2>}}
  }
  BTN_Copy_WorldspaceXform {
   widgetClassPath=&Button
   parent=VerticalBox_0
   widget=BTN_Copy_WorldspaceXform
   properties {
    toolTipText=Copy Worldspace Xform
    colorAndOpacity=<rgba 0.5906 0.1356 0.1356 1>
    widgetStyle {
     clickedSlateSound {resourceObject=null}
     disabled {
      bIsDynamicallyLoaded=false
      drawAs=RoundedBox
      imageType=NoImage
      mirroring=NoMirror
      resourceName=None
      resourceObject=null
      tiling=NoTile
      imageSize=<0 0>
      margin=<ltrb 0 0 0 0>
      outlineSettings {
       bUseBrushTransparency=false
       roundingType=FixedRadius
       width=1
       color {colorUseRule=UseColor_ColorTable, specifiedColor=<rgba 0.027 0.027 0.031 1>}
       cornerRadii=<5 5 5 5>
      }
      tintColor {colorUseRule=UseColor_Specified, specifiedColor=<rgba 0.015 0.015 0.018 1>}
      uVRegion {bIsValid=false, max=<0 0>, min=<0 0>}
     }
     disabledForeground {colorUseRule=UseColor_ColorTable, specifiedColor=<rgba 0.4 0.4 0.4 1>}
     hovered {
      bIsDynamicallyLoaded=false
      drawAs=RoundedBox
      imageType=NoImage
      mirroring=NoMirror
      resourceName=None
      resourceObject=null
      tiling=NoTile
      imageSize=<0 0>
      margin=<ltrb 0 0 0 0>
      outlineSettings {
       bUseBrushTransparency=false
       roundingType=FixedRadius
       width=1
       color {colorUseRule=UseColor_ColorTable, specifiedColor=<rgba 0.027 0.027 0.031 1>}
       cornerRadii=<5 5 5 5>
      }
      tintColor {colorUseRule=UseColor_Specified, specifiedColor=<rgba 0.031 0.031 0.038 1>}
      uVRegion {bIsValid=false, max=<0 0>, min=<0 0>}
     }
     hoveredForeground {colorUseRule=UseColor_Foreground, specifiedColor=<rgba 1 1 1 1>}
     hoveredSlateSound {resourceObject=null}
     normal {
      bIsDynamicallyLoaded=false
      drawAs=RoundedBox
      imageType=NoImage
      mirroring=NoMirror
      resourceName=None
      resourceObject=null
      tiling=NoTile
      imageSize=<0 0>
      margin=<ltrb 0 0 0 0>
      outlineSettings {
       bUseBrushTransparency=false
       roundingType=FixedRadius
       width=1
       color {colorUseRule=UseColor_ColorTable, specifiedColor=<rgba 0.027 0.027 0.031 1>}
       cornerRadii=<5 5 5 5>
      }
      tintColor {colorUseRule=UseColor_Specified, specifiedColor=<rgba 0.0222 0.0222 0.0252 1>}
      uVRegion {bIsValid=false, max=<0 0>, min=<0 0>}
     }
     normalForeground {colorUseRule=UseColor_Foreground, specifiedColor=<rgba 1 1 1 1>}
     normalPadding=<ltrb 0 0 0 0>
     pressed {
      bIsDynamicallyLoaded=false
      drawAs=RoundedBox
      imageType=NoImage
      mirroring=NoMirror
      resourceName=None
      resourceObject=null
      tiling=NoTile
      imageSize=<0 0>
      margin=<ltrb 0 0 0 0>
      outlineSettings {
       bUseBrushTransparency=false
       roundingType=FixedRadius
       width=1
       color {colorUseRule=UseColor_ColorTable, specifiedColor=<rgba 0.027 0.027 0.031 1>}
       cornerRadii=<5 5 5 5>
      }
      tintColor {colorUseRule=UseColor_Specified, specifiedColor=<rgba 0.04 0.04 0.048 1>}
      uVRegion {bIsValid=false, max=<0 0>, min=<0 0>}
     }
     pressedForeground {colorUseRule=UseColor_Foreground, specifiedColor=<rgba 1 1 1 1>}
     pressedPadding=<ltrb 0 0 0 0>
     pressedSlateSound {resourceObject=null}
    }
   }
   slot {class=VerticalBoxSlot, properties {verticalAlignment=VAlign_Top, padding=<ltrb 4 1 4 3>}}
  }
  HBox_Content_3 {
   widgetClassPath=&HorizontalBox
   parent=BTN_Copy_WorldspaceXform
   widget=HBox_Content_3
   slot {class=ButtonSlot, properties {horizontalAlignment=HAlign_Fill, padding=<ltrb 1 4 4 4>}}
  }
  Overlay_Icon_3 {
   widgetClassPath=&Overlay
   parent=HBox_Content_3
   widget=Overlay_Icon_3
   slot {class=HorizontalBoxSlot, properties {horizontalAlignment=HAlign_Left, verticalAlignment=VAlign_Center, padding=<ltrb 0 0 10 0>}}
  }
  Img_Icon_3 {
   widgetClassPath=&Image
   bIsVariable=true
   parent=Overlay_Icon_3
   widget=Img_Icon_3
   properties {brush {
     bIsDynamicallyLoaded=false
     drawAs=Image
     imageType=FullColor
     mirroring=NoMirror
     resourceName=None
     resourceObject=/AnimInEdTools/AnimationToolbox/Icons/AnimToolbox_UI_Icon_27
     tiling=NoTile
     imageSize=<24 24>
     margin=<ltrb 0 0 0 0>
     outlineSettings {
      bUseBrushTransparency=false
      roundingType=FixedRadius
      width=0
      color {colorUseRule=UseColor_Specified, specifiedColor=<rgba 0 0 0 1>}
      cornerRadii=<0 0 0 0>
     }
     tintColor {colorUseRule=UseColor_Specified, specifiedColor=<rgba 1 1 1 1>}
     uVRegion {bIsValid=false, max=<0 0>, min=<0 0>}
    }}
   slot {class=OverlaySlot, properties {verticalAlignment=VAlign_Center, padding=<ltrb 6 0 0 0>}}
  }
  TB_TabName_3 {
   widgetClassPath=&TextBlock
   bIsVariable=true
   parent=HBox_Content_3
   widget=TB_TabName_3
   properties {
    text=Copy WS XForm
    textOverflowPolicy=Ellipsis
    font {
     bForceMonospaced=false
     bMaterialIsStencil=false
     fontMaterial=null
     fontObject=&Roboto
     letterSpacing=0
     monospacedWidth=0
     size=10
     skewAmount=0
     typefaceFontName=Bold
     outlineSettings {
      bApplyOutlineToDropShadows=false
      bMiteredCorners=false
      bSeparateFillAlpha=false
      outlineMaterial=null
      outlineSize=0
      outlineColor=<rgba 0 0 0 1>
     }
    }
   }
   slot {class=HorizontalBoxSlot, properties {verticalAlignment=VAlign_Center, size {sizeRule=Fill, value=1}}}
  }
  BTN_Paste_WorldspaceXform {
   widgetClassPath=&Button
   parent=VerticalBox_0
   widget=BTN_Paste_WorldspaceXform
   properties {
    toolTipText=Paste Worldspace Xform
    colorAndOpacity=<rgba 0.5906 0.1356 0.1356 1>
    widgetStyle {
     clickedSlateSound {resourceObject=null}
     disabled {
      bIsDynamicallyLoaded=false
      drawAs=RoundedBox
      imageType=NoImage
      mirroring=NoMirror
      resourceName=None
      resourceObject=null
      tiling=NoTile
      imageSize=<0 0>
      margin=<ltrb 0 0 0 0>
      outlineSettings {
       bUseBrushTransparency=false
       roundingType=FixedRadius
       width=1
       color {colorUseRule=UseColor_ColorTable, specifiedColor=<rgba 0.027 0.027 0.031 1>}
       cornerRadii=<5 5 5 5>
      }
      tintColor {colorUseRule=UseColor_Specified, specifiedColor=<rgba 0.015 0.015 0.018 1>}
      uVRegion {bIsValid=false, max=<0 0>, min=<0 0>}
     }
     disabledForeground {colorUseRule=UseColor_ColorTable, specifiedColor=<rgba 0.4 0.4 0.4 1>}
     hovered {
      bIsDynamicallyLoaded=false
      drawAs=RoundedBox
      imageType=NoImage
      mirroring=NoMirror
      resourceName=None
      resourceObject=null
      tiling=NoTile
      imageSize=<0 0>
      margin=<ltrb 0 0 0 0>
      outlineSettings {
       bUseBrushTransparency=false
       roundingType=FixedRadius
       width=1
       color {colorUseRule=UseColor_ColorTable, specifiedColor=<rgba 0.027 0.027 0.031 1>}
       cornerRadii=<5 5 5 5>
      }
      tintColor {colorUseRule=UseColor_Specified, specifiedColor=<rgba 0.031 0.031 0.038 1>}
      uVRegion {bIsValid=false, max=<0 0>, min=<0 0>}
     }
     hoveredForeground {colorUseRule=UseColor_Foreground, specifiedColor=<rgba 1 1 1 1>}
     hoveredSlateSound {resourceObject=null}
     normal {
      bIsDynamicallyLoaded=false
      drawAs=RoundedBox
      imageType=NoImage
      mirroring=NoMirror
      resourceName=None
      resourceObject=null
      tiling=NoTile
      imageSize=<0 0>
      margin=<ltrb 0 0 0 0>
      outlineSettings {
       bUseBrushTransparency=false
       roundingType=FixedRadius
       width=1
       color {colorUseRule=UseColor_ColorTable, specifiedColor=<rgba 0.027 0.027 0.031 1>}
       cornerRadii=<5 5 5 5>
      }
      tintColor {colorUseRule=UseColor_Specified, specifiedColor=<rgba 0.0222 0.0222 0.0252 1>}
      uVRegion {bIsValid=false, max=<0 0>, min=<0 0>}
     }
     normalForeground {colorUseRule=UseColor_Foreground, specifiedColor=<rgba 1 1 1 1>}
     normalPadding=<ltrb 0 0 0 0>
     pressed {
      bIsDynamicallyLoaded=false
      drawAs=RoundedBox
      imageType=NoImage
      mirroring=NoMirror
      resourceName=None
      resourceObject=null
      tiling=NoTile
      imageSize=<0 0>
      margin=<ltrb 0 0 0 0>
      outlineSettings {
       bUseBrushTransparency=false
       roundingType=FixedRadius
       width=1
       color {colorUseRule=UseColor_ColorTable, specifiedColor=<rgba 0.027 0.027 0.031 1>}
       cornerRadii=<5 5 5 5>
      }
      tintColor {colorUseRule=UseColor_Specified, specifiedColor=<rgba 0.04 0.04 0.048 1>}
      uVRegion {bIsValid=false, max=<0 0>, min=<0 0>}
     }
     pressedForeground {colorUseRule=UseColor_Foreground, specifiedColor=<rgba 1 1 1 1>}
     pressedPadding=<ltrb 0 0 0 0>
     pressedSlateSound {resourceObject=null}
    }
   }
   slot {class=VerticalBoxSlot, properties {verticalAlignment=VAlign_Top, padding=<ltrb 4 1 4 3>}}
  }
  HBox_Content_4 {
   widgetClassPath=&HorizontalBox
   parent=BTN_Paste_WorldspaceXform
   widget=HBox_Content_4
   slot {class=ButtonSlot, properties {horizontalAlignment=HAlign_Fill, padding=<ltrb 1 4 4 4>}}
  }
  Overlay_Icon_4 {
   widgetClassPath=&Overlay
   parent=HBox_Content_4
   widget=Overlay_Icon_4
   slot {class=HorizontalBoxSlot, properties {horizontalAlignment=HAlign_Left, verticalAlignment=VAlign_Center, padding=<ltrb 0 0 10 0>}}
  }
  Img_Icon_4 {
   widgetClassPath=&Image
   bIsVariable=true
   parent=Overlay_Icon_4
   widget=Img_Icon_4
   properties {brush {
     bIsDynamicallyLoaded=false
     drawAs=Image
     imageType=FullColor
     mirroring=NoMirror
     resourceName=None
     resourceObject=/AnimInEdTools/AnimationToolbox/Icons/AnimToolbox_UI_Icon_28
     tiling=NoTile
     imageSize=<24 24>
     margin=<ltrb 0 0 0 0>
     outlineSettings {
      bUseBrushTransparency=false
      roundingType=FixedRadius
      width=0
      color {colorUseRule=UseColor_Specified, specifiedColor=<rgba 0 0 0 1>}
      cornerRadii=<0 0 0 0>
     }
     tintColor {colorUseRule=UseColor_Specified, specifiedColor=<rgba 1 1 1 1>}
     uVRegion {bIsValid=false, max=<0 0>, min=<0 0>}
    }}
   slot {class=OverlaySlot, properties {verticalAlignment=VAlign_Center, padding=<ltrb 6 0 0 0>}}
  }
  TB_TabName_4 {
   widgetClassPath=&TextBlock
   bIsVariable=true
   parent=HBox_Content_4
   widget=TB_TabName_4
   properties {
    text=Paste WS XForm
    textOverflowPolicy=Ellipsis
    font {
     bForceMonospaced=false
     bMaterialIsStencil=false
     fontMaterial=null
     fontObject=&Roboto
     letterSpacing=0
     monospacedWidth=0
     size=10
     skewAmount=0
     typefaceFontName=Bold
     outlineSettings {
      bApplyOutlineToDropShadows=false
      bMiteredCorners=false
      bSeparateFillAlpha=false
      outlineMaterial=null
      outlineSize=0
      outlineColor=<rgba 0 0 0 1>
     }
    }
   }
   slot {class=HorizontalBoxSlot, properties {verticalAlignment=VAlign_Center, size {sizeRule=Fill, value=1}}}
  }
  BTN_Copy_KeyTiming {
   widgetClassPath=&Button
   parent=VerticalBox_0
   widget=BTN_Copy_KeyTiming
   properties {
    toolTipText="Left Click - propogate key timing to all controls selected\r\nRight Click - copy key timing only\r\nMiddle Click - Paste key timing only\r\nCtrl + Middle - Paste JUST key timing (and remove keys on other frames)\r\n\r\n**no anim layer support (yet) **"
    colorAndOpacity=<rgba 0.5906 0.1356 0.1356 1>
    widgetStyle {
     clickedSlateSound {resourceObject=null}
     disabled {
      bIsDynamicallyLoaded=false
      drawAs=RoundedBox
      imageType=NoImage
      mirroring=NoMirror
      resourceName=None
      resourceObject=null
      tiling=NoTile
      imageSize=<0 0>
      margin=<ltrb 0 0 0 0>
      outlineSettings {
       bUseBrushTransparency=false
       roundingType=FixedRadius
       width=1
       color {colorUseRule=UseColor_ColorTable, specifiedColor=<rgba 0.027 0.027 0.031 1>}
       cornerRadii=<5 5 5 5>
      }
      tintColor {colorUseRule=UseColor_Specified, specifiedColor=<rgba 0.015 0.015 0.018 1>}
      uVRegion {bIsValid=false, max=<0 0>, min=<0 0>}
     }
     disabledForeground {colorUseRule=UseColor_ColorTable, specifiedColor=<rgba 0.4 0.4 0.4 1>}
     hovered {
      bIsDynamicallyLoaded=false
      drawAs=RoundedBox
      imageType=NoImage
      mirroring=NoMirror
      resourceName=None
      resourceObject=null
      tiling=NoTile
      imageSize=<0 0>
      margin=<ltrb 0 0 0 0>
      outlineSettings {
       bUseBrushTransparency=false
       roundingType=FixedRadius
       width=1
       color {colorUseRule=UseColor_Foreground, specifiedColor=<rgba 0.5906 0.1356 0.1356 1>}
       cornerRadii=<5 5 5 5>
      }
      tintColor {colorUseRule=UseColor_Specified, specifiedColor=<rgba 0.031 0.031 0.038 1>}
      uVRegion {bIsValid=false, max=<0 0>, min=<0 0>}
     }
     hoveredForeground {colorUseRule=UseColor_Foreground, specifiedColor=<rgba 1 1 1 1>}
     hoveredSlateSound {resourceObject=null}
     normal {
      bIsDynamicallyLoaded=false
      drawAs=RoundedBox
      imageType=NoImage
      mirroring=NoMirror
      resourceName=None
      resourceObject=null
      tiling=NoTile
      imageSize=<0 0>
      margin=<ltrb 0 0 0 0>
      outlineSettings {
       bUseBrushTransparency=false
       roundingType=FixedRadius
       width=1
       color {colorUseRule=UseColor_ColorTable, specifiedColor=<rgba 0.027 0.027 0.031 1>}
       cornerRadii=<5 5 5 5>
      }
      tintColor {colorUseRule=UseColor_Specified, specifiedColor=<rgba 0.0222 0.0222 0.0252 1>}
      uVRegion {bIsValid=false, max=<0 0>, min=<0 0>}
     }
     normalForeground {colorUseRule=UseColor_Foreground, specifiedColor=<rgba 1 1 1 1>}
     normalPadding=<ltrb 0 0 0 0>
     pressed {
      bIsDynamicallyLoaded=false
      drawAs=RoundedBox
      imageType=NoImage
      mirroring=NoMirror
      resourceName=None
      resourceObject=null
      tiling=NoTile
      imageSize=<0 0>
      margin=<ltrb 0 0 0 0>
      outlineSettings {
       bUseBrushTransparency=false
       roundingType=FixedRadius
       width=1
       color {colorUseRule=UseColor_ColorTable, specifiedColor=<rgba 0.027 0.027 0.031 1>}
       cornerRadii=<5 5 5 5>
      }
      tintColor {colorUseRule=UseColor_Specified, specifiedColor=<rgba 0.04 0.04 0.048 1>}
      uVRegion {bIsValid=false, max=<0 0>, min=<0 0>}
     }
     pressedForeground {colorUseRule=UseColor_Foreground, specifiedColor=<rgba 1 1 1 1>}
     pressedPadding=<ltrb 0 0 0 0>
     pressedSlateSound {resourceObject=null}
    }
   }
   slot {class=VerticalBoxSlot, properties {verticalAlignment=VAlign_Top, padding=<ltrb 4 1 4 3>}}
  }
  HBox_Content_19 {
   widgetClassPath=&HorizontalBox
   parent=BTN_Copy_KeyTiming
   widget=HBox_Content_19
   slot {class=ButtonSlot, properties {horizontalAlignment=HAlign_Fill, padding=<ltrb 1 4 4 4>}}
  }
  Overlay_Icon_19 {
   widgetClassPath=&Overlay
   parent=HBox_Content_19
   widget=Overlay_Icon_19
   slot {class=HorizontalBoxSlot, properties {horizontalAlignment=HAlign_Left, verticalAlignment=VAlign_Center, padding=<ltrb 0 0 10 0>}}
  }
  Img_Icon_18 {
   widgetClassPath=&Image
   bIsVariable=true
   parent=Overlay_Icon_19
   widget=Img_Icon_18
   properties {
    brush {
     bIsDynamicallyLoaded=false
     drawAs=Image
     imageType=FullColor
     mirroring=NoMirror
     resourceName=None
     resourceObject=/AnimInEdTools/AnimationToolbox/Icons/AnimToolbox_UI_Icon_57
     tiling=NoTile
     imageSize=<24 24>
     margin=<ltrb 0 0 0 0>
     outlineSettings {
      bUseBrushTransparency=false
      roundingType=FixedRadius
      width=0
      color {colorUseRule=UseColor_Specified, specifiedColor=<rgba 0 0 0 1>}
      cornerRadii=<0 0 0 0>
     }
     tintColor {colorUseRule=UseColor_Specified, specifiedColor=<rgba 1 1 1 1>}
     uVRegion {bIsValid=false, max=<0 0>, min=<0 0>}
    }
    renderTransform {angle=0, scale=<0.5 0.5>, shear=<0 0>, translation=<-2 -9>}
   }
   slot {class=OverlaySlot, properties {verticalAlignment=VAlign_Center, padding=<ltrb 6 0 0 0>}}
  }
  Img_Icon_22 {
   widgetClassPath=&Image
   bIsVariable=true
   parent=Overlay_Icon_19
   widget=Img_Icon_22
   properties {
    brush {
     bIsDynamicallyLoaded=false
     drawAs=Image
     imageType=FullColor
     mirroring=NoMirror
     resourceName=None
     resourceObject=&AnimToolbox_UI_Icon_57
     tiling=NoTile
     imageSize=<24 24>
     margin=<ltrb 0 0 0 0>
     outlineSettings {
      bUseBrushTransparency=false
      roundingType=FixedRadius
      width=0
      color {colorUseRule=UseColor_Specified, specifiedColor=<rgba 0 0 0 1>}
      cornerRadii=<0 0 0 0>
     }
     tintColor {colorUseRule=UseColor_Specified, specifiedColor=<rgba 1 1 1 1>}
     uVRegion {bIsValid=false, max=<0 0>, min=<0 0>}
    }
    renderTransform {angle=-16.32, scale=<0.5 0.5>, shear=<0 0>, translation=<3 -7>}
   }
   slot {class=OverlaySlot, properties {verticalAlignment=VAlign_Center, padding=<ltrb 6 0 0 0>}}
  }
  Img_Icon_21 {
   widgetClassPath=&Image
   bIsVariable=true
   parent=Overlay_Icon_19
   widget=Img_Icon_21
   properties {
    brush {
     bIsDynamicallyLoaded=false
     drawAs=Image
     imageType=FullColor
     mirroring=NoMirror
     resourceName=None
     resourceObject=/AnimInEdTools/AnimationToolbox/Icons/AnimToolbox_UI_Icon_16
     tiling=NoTile
     imageSize=<30 30>
     margin=<ltrb 0 0 0 0>
     outlineSettings {
      bUseBrushTransparency=false
      roundingType=FixedRadius
      width=0
      color {colorUseRule=UseColor_Specified, specifiedColor=<rgba 0 0 0 1>}
      cornerRadii=<0 0 0 0>
     }
     tintColor {colorUseRule=UseColor_Specified, specifiedColor=<rgba 1 1 1 1>}
     uVRegion {bIsValid=false, max=<0 0>, min=<0 0>}
    }
    renderTransform {angle=0, scale=<1 1>, shear=<0 0>, translation=<-2 7>}
   }
   slot {class=OverlaySlot, properties {verticalAlignment=VAlign_Center, padding=<ltrb 6 0 0 0>}}
  }
  TB_TabName_17 {
   widgetClassPath=&TextBlock
   bIsVariable=true
   parent=HBox_Content_19
   widget=TB_TabName_17
   properties {
    text=Copy Key Timing
    textOverflowPolicy=Ellipsis
    font {
     bForceMonospaced=false
     bMaterialIsStencil=false
     fontMaterial=null
     fontObject=&Roboto
     letterSpacing=0
     monospacedWidth=0
     size=10
     skewAmount=0
     typefaceFontName=Bold
     outlineSettings {
      bApplyOutlineToDropShadows=false
      bMiteredCorners=false
      bSeparateFillAlpha=false
      outlineMaterial=null
      outlineSize=0
      outlineColor=<rgba 0 0 0 1>
     }
    }
   }
   slot {class=HorizontalBoxSlot, properties {verticalAlignment=VAlign_Center, size {sizeRule=Fill, value=1}}}
  }
  TextBlock_92 {
   widgetClassPath=&TextBlock
   parent=VerticalBox_0
   widget=TextBlock_92
   properties {
    bOverride_Cursor=true
    text=CurveEditor
    colorAndOpacity {colorUseRule=UseColor_Specified, specifiedColor=<rgba 0.8963 0.8963 0.8963 0.5>}
    font {
     bForceMonospaced=false
     bMaterialIsStencil=false
     fontMaterial=null
     fontObject=&Roboto
     letterSpacing=0
     monospacedWidth=0
     size=6
     skewAmount=0
     typefaceFontName=Bold
     outlineSettings {
      bApplyOutlineToDropShadows=false
      bMiteredCorners=false
      bSeparateFillAlpha=false
      outlineMaterial=null
      outlineSize=0
      outlineColor=<rgba 0 0 0 1>
     }
    }
   }
   slot {class=VerticalBoxSlot, properties {horizontalAlignment=HAlign_Left, padding=<ltrb 5 2 0 2>}}
  }
  BTN_Cycle {
   widgetClassPath=&EditorUtilityButton
   parent=VerticalBox_0
   widget=BTN_Cycle
   properties {
    toolTipText="Pre/Post Cycle\r\n\r\nSet your curves pre and post infinity to Cycle"
    colorAndOpacity=<rgba 1 0.3916 0.1119 1>
    widgetStyle {
     clickedSlateSound {resourceObject=null}
     disabled {
      bIsDynamicallyLoaded=false
      drawAs=RoundedBox
      imageType=NoImage
      mirroring=NoMirror
      resourceName=None
      resourceObject=null
      tiling=NoTile
      imageSize=<0 0>
      margin=<ltrb 0 0 0 0>
      outlineSettings {
       bUseBrushTransparency=false
       roundingType=FixedRadius
       width=1
       color {colorUseRule=UseColor_ColorTable, specifiedColor=<rgba 0.027 0.027 0.031 1>}
       cornerRadii=<5 5 5 5>
      }
      tintColor {colorUseRule=UseColor_Specified, specifiedColor=<rgba 0.015 0.015 0.018 1>}
      uVRegion {bIsValid=false, max=<0 0>, min=<0 0>}
     }
     disabledForeground {colorUseRule=UseColor_ColorTable, specifiedColor=<rgba 0.4 0.4 0.4 1>}
     hovered {
      bIsDynamicallyLoaded=false
      drawAs=RoundedBox
      imageType=NoImage
      mirroring=NoMirror
      resourceName=None
      resourceObject=null
      tiling=NoTile
      imageSize=<0 0>
      margin=<ltrb 0 0 0 0>
      outlineSettings {
       bUseBrushTransparency=false
       roundingType=FixedRadius
       width=1
       color {colorUseRule=UseColor_ColorTable, specifiedColor=<rgba 0.027 0.027 0.031 1>}
       cornerRadii=<5 5 5 5>
      }
      tintColor {colorUseRule=UseColor_Specified, specifiedColor=<rgba 0.031 0.031 0.038 1>}
      uVRegion {bIsValid=false, max=<0 0>, min=<0 0>}
     }
     hoveredForeground {colorUseRule=UseColor_Foreground, specifiedColor=<rgba 1 1 1 1>}
     hoveredSlateSound {resourceObject=null}
     normal {
      bIsDynamicallyLoaded=false
      drawAs=RoundedBox
      imageType=NoImage
      mirroring=NoMirror
      resourceName=None
      resourceObject=null
      tiling=NoTile
      imageSize=<0 0>
      margin=<ltrb 0 0 0 0>
      outlineSettings {
       bUseBrushTransparency=false
       roundingType=FixedRadius
       width=1
       color {colorUseRule=UseColor_ColorTable, specifiedColor=<rgba 0.027 0.027 0.031 1>}
       cornerRadii=<5 5 5 5>
      }
      tintColor {colorUseRule=UseColor_Specified, specifiedColor=<rgba 0.0222 0.0222 0.0252 1>}
      uVRegion {bIsValid=false, max=<0 0>, min=<0 0>}
     }
     normalForeground {colorUseRule=UseColor_Foreground, specifiedColor=<rgba 1 1 1 1>}
     normalPadding=<ltrb 0 0 0 0>
     pressed {
      bIsDynamicallyLoaded=false
      drawAs=RoundedBox
      imageType=NoImage
      mirroring=NoMirror
      resourceName=None
      resourceObject=null
      tiling=NoTile
      imageSize=<0 0>
      margin=<ltrb 0 0 0 0>
      outlineSettings {
       bUseBrushTransparency=false
       roundingType=FixedRadius
       width=1
       color {colorUseRule=UseColor_ColorTable, specifiedColor=<rgba 0.027 0.027 0.031 1>}
       cornerRadii=<5 5 5 5>
      }
      tintColor {colorUseRule=UseColor_Specified, specifiedColor=<rgba 0.04 0.04 0.048 1>}
      uVRegion {bIsValid=false, max=<0 0>, min=<0 0>}
     }
     pressedForeground {colorUseRule=UseColor_Foreground, specifiedColor=<rgba 1 1 1 1>}
     pressedPadding=<ltrb 0 0 0 0>
     pressedSlateSound {resourceObject=null}
    }
   }
   slot {class=VerticalBoxSlot, properties {padding=<ltrb 4 1 4 3>}}
  }
  HBox_Content_6 {
   widgetClassPath=&HorizontalBox
   parent=BTN_Cycle
   widget=HBox_Content_6
   slot {class=ButtonSlot, properties {horizontalAlignment=HAlign_Fill, padding=<ltrb 1 4 4 4>}}
  }
  Overlay_Icon_6 {
   widgetClassPath=&Overlay
   parent=HBox_Content_6
   widget=Overlay_Icon_6
   slot {class=HorizontalBoxSlot, properties {horizontalAlignment=HAlign_Left, verticalAlignment=VAlign_Center, padding=<ltrb 0 0 10 0>}}
  }
  Img_Icon_6 {
   widgetClassPath=&Image
   bIsVariable=true
   parent=Overlay_Icon_6
   widget=Img_Icon_6
   properties {brush {
     bIsDynamicallyLoaded=false
     drawAs=Image
     imageType=FullColor
     mirroring=NoMirror
     resourceName=None
     resourceObject=/AnimInEdTools/AnimationToolbox/Icons/AnimToolbox_UI_Icon_13
     tiling=NoTile
     imageSize=<24 24>
     margin=<ltrb 0 0 0 0>
     outlineSettings {
      bUseBrushTransparency=false
      roundingType=FixedRadius
      width=0
      color {colorUseRule=UseColor_Specified, specifiedColor=<rgba 0 0 0 1>}
      cornerRadii=<0 0 0 0>
     }
     tintColor {colorUseRule=UseColor_Specified, specifiedColor=<rgba 1 1 1 1>}
     uVRegion {bIsValid=false, max=<0 0>, min=<0 0>}
    }}
   slot {class=OverlaySlot, properties {verticalAlignment=VAlign_Center, padding=<ltrb 6 0 0 0>}}
  }
  TB_TabName_6 {
   widgetClassPath=&TextBlock
   bIsVariable=true
   parent=HBox_Content_6
   widget=TB_TabName_6
   properties {
    text=Cycle Curves
    textOverflowPolicy=Ellipsis
    font {
     bForceMonospaced=false
     bMaterialIsStencil=false
     fontMaterial=null
     fontObject=&Roboto
     letterSpacing=0
     monospacedWidth=0
     size=10
     skewAmount=0
     typefaceFontName=Bold
     outlineSettings {
      bApplyOutlineToDropShadows=false
      bMiteredCorners=false
      bSeparateFillAlpha=false
      outlineMaterial=null
      outlineSize=0
      outlineColor=<rgba 0 0 0 1>
     }
    }
   }
   slot {class=HorizontalBoxSlot, properties {verticalAlignment=VAlign_Center, size {sizeRule=Fill, value=1}}}
  }
  BTN_Constant {
   widgetClassPath=&EditorUtilityButton
   parent=VerticalBox_0
   widget=BTN_Constant
   properties {
    toolTipText="Pre/Post Constant\r\n\r\nSet your curves pre and post infinity to Constant"
    colorAndOpacity=<rgba 1 0.3916 0.1119 1>
    widgetStyle {
     clickedSlateSound {resourceObject=null}
     disabled {
      bIsDynamicallyLoaded=false
      drawAs=RoundedBox
      imageType=NoImage
      mirroring=NoMirror
      resourceName=None
      resourceObject=null
      tiling=NoTile
      imageSize=<0 0>
      margin=<ltrb 0 0 0 0>
      outlineSettings {
       bUseBrushTransparency=false
       roundingType=FixedRadius
       width=1
       color {colorUseRule=UseColor_ColorTable, specifiedColor=<rgba 0.027 0.027 0.031 1>}
       cornerRadii=<5 5 5 5>
      }
      tintColor {colorUseRule=UseColor_Specified, specifiedColor=<rgba 0.015 0.015 0.018 1>}
      uVRegion {bIsValid=false, max=<0 0>, min=<0 0>}
     }
     disabledForeground {colorUseRule=UseColor_ColorTable, specifiedColor=<rgba 0.4 0.4 0.4 1>}
     hovered {
      bIsDynamicallyLoaded=false
      drawAs=RoundedBox
      imageType=NoImage
      mirroring=NoMirror
      resourceName=None
      resourceObject=null
      tiling=NoTile
      imageSize=<0 0>
      margin=<ltrb 0 0 0 0>
      outlineSettings {
       bUseBrushTransparency=false
       roundingType=FixedRadius
       width=1
       color {colorUseRule=UseColor_ColorTable, specifiedColor=<rgba 0.027 0.027 0.031 1>}
       cornerRadii=<5 5 5 5>
      }
      tintColor {colorUseRule=UseColor_Specified, specifiedColor=<rgba 0.031 0.031 0.038 1>}
      uVRegion {bIsValid=false, max=<0 0>, min=<0 0>}
     }
     hoveredForeground {colorUseRule=UseColor_Foreground, specifiedColor=<rgba 1 1 1 1>}
     hoveredSlateSound {resourceObject=null}
     normal {
      bIsDynamicallyLoaded=false
      drawAs=RoundedBox
      imageType=NoImage
      mirroring=NoMirror
      resourceName=None
      resourceObject=null
      tiling=NoTile
      imageSize=<0 0>
      margin=<ltrb 0 0 0 0>
      outlineSettings {
       bUseBrushTransparency=false
       roundingType=FixedRadius
       width=1
       color {colorUseRule=UseColor_ColorTable, specifiedColor=<rgba 0.027 0.027 0.031 1>}
       cornerRadii=<5 5 5 5>
      }
      tintColor {colorUseRule=UseColor_Specified, specifiedColor=<rgba 0.0222 0.0222 0.0252 1>}
      uVRegion {bIsValid=false, max=<0 0>, min=<0 0>}
     }
     normalForeground {colorUseRule=UseColor_Foreground, specifiedColor=<rgba 1 1 1 1>}
     normalPadding=<ltrb 0 0 0 0>
     pressed {
      bIsDynamicallyLoaded=false
      drawAs=RoundedBox
      imageType=NoImage
      mirroring=NoMirror
      resourceName=None
      resourceObject=null
      tiling=NoTile
      imageSize=<0 0>
      margin=<ltrb 0 0 0 0>
      outlineSettings {
       bUseBrushTransparency=false
       roundingType=FixedRadius
       width=1
       color {colorUseRule=UseColor_ColorTable, specifiedColor=<rgba 0.027 0.027 0.031 1>}
       cornerRadii=<5 5 5 5>
      }
      tintColor {colorUseRule=UseColor_Specified, specifiedColor=<rgba 0.04 0.04 0.048 1>}
      uVRegion {bIsValid=false, max=<0 0>, min=<0 0>}
     }
     pressedForeground {colorUseRule=UseColor_Foreground, specifiedColor=<rgba 1 1 1 1>}
     pressedPadding=<ltrb 0 0 0 0>
     pressedSlateSound {resourceObject=null}
    }
   }
   slot {class=VerticalBoxSlot, properties {padding=<ltrb 4 1 4 3>}}
  }
  HBox_Content_7 {
   widgetClassPath=&HorizontalBox
   parent=BTN_Constant
   widget=HBox_Content_7
   slot {class=ButtonSlot, properties {horizontalAlignment=HAlign_Fill, padding=<ltrb 1 4 4 4>}}
  }
  Overlay_Icon_7 {
   widgetClassPath=&Overlay
   parent=HBox_Content_7
   widget=Overlay_Icon_7
   slot {class=HorizontalBoxSlot, properties {horizontalAlignment=HAlign_Left, verticalAlignment=VAlign_Center, padding=<ltrb 0 0 10 0>}}
  }
  Img_Icon_7 {
   widgetClassPath=&Image
   bIsVariable=true
   parent=Overlay_Icon_7
   widget=Img_Icon_7
   properties {brush {
     bIsDynamicallyLoaded=false
     drawAs=Image
     imageType=FullColor
     mirroring=NoMirror
     resourceName=None
     resourceObject=&AnimToolbox_UI_Icon_13
     tiling=NoTile
     imageSize=<24 24>
     margin=<ltrb 0 0 0 0>
     outlineSettings {
      bUseBrushTransparency=false
      roundingType=FixedRadius
      width=0
      color {colorUseRule=UseColor_Specified, specifiedColor=<rgba 0 0 0 1>}
      cornerRadii=<0 0 0 0>
     }
     tintColor {colorUseRule=UseColor_Specified, specifiedColor=<rgba 1 1 1 1>}
     uVRegion {bIsValid=false, max=<0 0>, min=<0 0>}
    }}
   slot {class=OverlaySlot, properties {verticalAlignment=VAlign_Center, padding=<ltrb 6 0 0 0>}}
  }
  TB_TabName_7 {
   widgetClassPath=&TextBlock
   bIsVariable=true
   parent=HBox_Content_7
   widget=TB_TabName_7
   properties {
    text=Constant Curves
    textOverflowPolicy=Ellipsis
    font {
     bForceMonospaced=false
     bMaterialIsStencil=false
     fontMaterial=null
     fontObject=&Roboto
     letterSpacing=0
     monospacedWidth=0
     size=10
     skewAmount=0
     typefaceFontName=Bold
     outlineSettings {
      bApplyOutlineToDropShadows=false
      bMiteredCorners=false
      bSeparateFillAlpha=false
      outlineMaterial=null
      outlineSize=0
      outlineColor=<rgba 0 0 0 1>
     }
    }
   }
   slot {class=HorizontalBoxSlot, properties {verticalAlignment=VAlign_Center, size {sizeRule=Fill, value=1}}}
  }
  TextBlock_93 {
   widgetClassPath=&TextBlock
   parent=VerticalBox_0
   widget=TextBlock_93
   properties {
    bOverride_Cursor=true
    text=Experimental
    colorAndOpacity {colorUseRule=UseColor_Specified, specifiedColor=<rgba 0.8963 0.8963 0.8963 0.5>}
    font {
     bForceMonospaced=false
     bMaterialIsStencil=false
     fontMaterial=null
     fontObject=&Roboto
     letterSpacing=0
     monospacedWidth=0
     size=6
     skewAmount=0
     typefaceFontName=Bold
     outlineSettings {
      bApplyOutlineToDropShadows=false
      bMiteredCorners=false
      bSeparateFillAlpha=false
      outlineMaterial=null
      outlineSize=0
      outlineColor=<rgba 0 0 0 1>
     }
    }
   }
   slot {class=VerticalBoxSlot, properties {horizontalAlignment=HAlign_Left, padding=<ltrb 5 2 0 2>}}
  }
  BTN_PIE_MultiCam {
   widgetClassPath=&EditorUtilityButton
   parent=VerticalBox_0
   widget=BTN_PIE_MultiCam
   properties {
    toolTipText="PIE Multicam\r\nadds an asset to your level that makes it so when you run PIE you get a 4 up comp"
    colorAndOpacity=<rgba 0.3005 0.107 0.4678 1>
    widgetStyle {
     clickedSlateSound {resourceObject=null}
     disabled {
      bIsDynamicallyLoaded=false
      drawAs=RoundedBox
      imageType=NoImage
      mirroring=NoMirror
      resourceName=None
      resourceObject=null
      tiling=NoTile
      imageSize=<0 0>
      margin=<ltrb 0 0 0 0>
      outlineSettings {
       bUseBrushTransparency=false
       roundingType=FixedRadius
       width=1
       color {colorUseRule=UseColor_ColorTable, specifiedColor=<rgba 0.027 0.027 0.031 1>}
       cornerRadii=<5 5 5 5>
      }
      tintColor {colorUseRule=UseColor_Specified, specifiedColor=<rgba 0.015 0.015 0.018 1>}
      uVRegion {bIsValid=false, max=<0 0>, min=<0 0>}
     }
     disabledForeground {colorUseRule=UseColor_ColorTable, specifiedColor=<rgba 0.4 0.4 0.4 1>}
     hovered {
      bIsDynamicallyLoaded=false
      drawAs=RoundedBox
      imageType=NoImage
      mirroring=NoMirror
      resourceName=None
      resourceObject=null
      tiling=NoTile
      imageSize=<0 0>
      margin=<ltrb 0 0 0 0>
      outlineSettings {
       bUseBrushTransparency=false
       roundingType=FixedRadius
       width=1
       color {colorUseRule=UseColor_ColorTable, specifiedColor=<rgba 0.027 0.027 0.031 1>}
       cornerRadii=<5 5 5 5>
      }
      tintColor {colorUseRule=UseColor_Specified, specifiedColor=<rgba 0.031 0.031 0.038 1>}
      uVRegion {bIsValid=false, max=<0 0>, min=<0 0>}
     }
     hoveredForeground {colorUseRule=UseColor_Foreground, specifiedColor=<rgba 1 1 1 1>}
     hoveredSlateSound {resourceObject=null}
     normal {
      bIsDynamicallyLoaded=false
      drawAs=RoundedBox
      imageType=NoImage
      mirroring=NoMirror
      resourceName=None
      resourceObject=null
      tiling=NoTile
      imageSize=<0 0>
      margin=<ltrb 0 0 0 0>
      outlineSettings {
       bUseBrushTransparency=false
       roundingType=FixedRadius
       width=1
       color {colorUseRule=UseColor_ColorTable, specifiedColor=<rgba 0.027 0.027 0.031 1>}
       cornerRadii=<5 5 5 5>
      }
      tintColor {colorUseRule=UseColor_Specified, specifiedColor=<rgba 0.0222 0.0222 0.0252 1>}
      uVRegion {bIsValid=false, max=<0 0>, min=<0 0>}
     }
     normalForeground {colorUseRule=UseColor_Foreground, specifiedColor=<rgba 1 1 1 1>}
     normalPadding=<ltrb 0 0 0 0>
     pressed {
      bIsDynamicallyLoaded=false
      drawAs=RoundedBox
      imageType=NoImage
      mirroring=NoMirror
      resourceName=None
      resourceObject=null
      tiling=NoTile
      imageSize=<0 0>
      margin=<ltrb 0 0 0 0>
      outlineSettings {
       bUseBrushTransparency=false
       roundingType=FixedRadius
       width=1
       color {colorUseRule=UseColor_ColorTable, specifiedColor=<rgba 0.027 0.027 0.031 1>}
       cornerRadii=<5 5 5 5>
      }
      tintColor {colorUseRule=UseColor_Specified, specifiedColor=<rgba 0.04 0.04 0.048 1>}
      uVRegion {bIsValid=false, max=<0 0>, min=<0 0>}
     }
     pressedForeground {colorUseRule=UseColor_Foreground, specifiedColor=<rgba 1 1 1 1>}
     pressedPadding=<ltrb 0 0 0 0>
     pressedSlateSound {resourceObject=null}
    }
   }
   slot {class=VerticalBoxSlot, properties {padding=<ltrb 4 1 4 3>}}
  }
  HBox_Content_12 {
   widgetClassPath=&HorizontalBox
   parent=BTN_PIE_MultiCam
   widget=HBox_Content_12
   slot {class=ButtonSlot, properties {horizontalAlignment=HAlign_Fill, padding=<ltrb 1 4 4 4>}}
  }
  Overlay_Icon_12 {
   widgetClassPath=&Overlay
   parent=HBox_Content_12
   widget=Overlay_Icon_12
   slot {class=HorizontalBoxSlot, properties {horizontalAlignment=HAlign_Left, verticalAlignment=VAlign_Center, padding=<ltrb 0 0 10 0>}}
  }
  Img_Icon_12 {
   widgetClassPath=&Image
   bIsVariable=true
   parent=Overlay_Icon_12
   widget=Img_Icon_12
   properties {brush {
     bIsDynamicallyLoaded=false
     drawAs=Image
     imageType=FullColor
     mirroring=NoMirror
     resourceName=None
     resourceObject=/AnimInEdTools/AnimationToolbox/Icons/AnimToolbox_UI_Icon_48
     tiling=NoTile
     imageSize=<24 24>
     margin=<ltrb 0 0 0 0>
     outlineSettings {
      bUseBrushTransparency=false
      roundingType=FixedRadius
      width=0
      color {colorUseRule=UseColor_Specified, specifiedColor=<rgba 0 0 0 1>}
      cornerRadii=<0 0 0 0>
     }
     tintColor {colorUseRule=UseColor_Specified, specifiedColor=<rgba 1 1 1 1>}
     uVRegion {bIsValid=false, max=<0 0>, min=<0 0>}
    }}
   slot {class=OverlaySlot, properties {verticalAlignment=VAlign_Center, padding=<ltrb 6 0 0 0>}}
  }
  TB_TabName_PIE {
   widgetClassPath=&TextBlock
   bIsVariable=true
   parent=HBox_Content_12
   widget=TB_TabName_PIE
   properties {
    text=PIE MultiCam - Add
    textOverflowPolicy=Ellipsis
    font {
     bForceMonospaced=false
     bMaterialIsStencil=false
     fontMaterial=null
     fontObject=&Roboto
     letterSpacing=0
     monospacedWidth=0
     size=10
     skewAmount=0
     typefaceFontName=Bold
     outlineSettings {
      bApplyOutlineToDropShadows=false
      bMiteredCorners=false
      bSeparateFillAlpha=false
      outlineMaterial=null
      outlineSize=0
      outlineColor=<rgba 0 0 0 1>
     }
    }
   }
   slot {class=HorizontalBoxSlot, properties {verticalAlignment=VAlign_Center, size {sizeRule=Fill, value=1}}}
  }
  BTN_MoveSpaceLeft {
   widgetClassPath=&EditorUtilityButton
   parent=VerticalBox_0
   widget=BTN_MoveSpaceLeft
   properties {
    toolTipText=Nudge Space Key Left
    colorAndOpacity=<rgba 1 0.3916 0.1119 1>
    widgetStyle {
     clickedSlateSound {resourceObject=null}
     disabled {
      bIsDynamicallyLoaded=false
      drawAs=RoundedBox
      imageType=NoImage
      mirroring=NoMirror
      resourceName=None
      resourceObject=null
      tiling=NoTile
      imageSize=<0 0>
      margin=<ltrb 0 0 0 0>
      outlineSettings {
       bUseBrushTransparency=false
       roundingType=FixedRadius
       width=1
       color {colorUseRule=UseColor_ColorTable, specifiedColor=<rgba 0.027 0.027 0.031 1>}
       cornerRadii=<5 5 5 5>
      }
      tintColor {colorUseRule=UseColor_Specified, specifiedColor=<rgba 0.015 0.015 0.018 1>}
      uVRegion {bIsValid=false, max=<0 0>, min=<0 0>}
     }
     disabledForeground {colorUseRule=UseColor_ColorTable, specifiedColor=<rgba 0.4 0.4 0.4 1>}
     hovered {
      bIsDynamicallyLoaded=false
      drawAs=RoundedBox
      imageType=NoImage
      mirroring=NoMirror
      resourceName=None
      resourceObject=null
      tiling=NoTile
      imageSize=<0 0>
      margin=<ltrb 0 0 0 0>
      outlineSettings {
       bUseBrushTransparency=false
       roundingType=FixedRadius
       width=1
       color {colorUseRule=UseColor_ColorTable, specifiedColor=<rgba 0.027 0.027 0.031 1>}
       cornerRadii=<5 5 5 5>
      }
      tintColor {colorUseRule=UseColor_Specified, specifiedColor=<rgba 0.031 0.031 0.038 1>}
      uVRegion {bIsValid=false, max=<0 0>, min=<0 0>}
     }
     hoveredForeground {colorUseRule=UseColor_Foreground, specifiedColor=<rgba 1 1 1 1>}
     hoveredSlateSound {resourceObject=null}
     normal {
      bIsDynamicallyLoaded=false
      drawAs=RoundedBox
      imageType=NoImage
      mirroring=NoMirror
      resourceName=None
      resourceObject=null
      tiling=NoTile
      imageSize=<0 0>
      margin=<ltrb 0 0 0 0>
      outlineSettings {
       bUseBrushTransparency=false
       roundingType=FixedRadius
       width=1
       color {colorUseRule=UseColor_ColorTable, specifiedColor=<rgba 0.027 0.027 0.031 1>}
       cornerRadii=<5 5 5 5>
      }
      tintColor {colorUseRule=UseColor_Specified, specifiedColor=<rgba 0.0222 0.0222 0.0252 1>}
      uVRegion {bIsValid=false, max=<0 0>, min=<0 0>}
     }
     normalForeground {colorUseRule=UseColor_Foreground, specifiedColor=<rgba 1 1 1 1>}
     normalPadding=<ltrb 0 0 0 0>
     pressed {
      bIsDynamicallyLoaded=false
      drawAs=RoundedBox
      imageType=NoImage
      mirroring=NoMirror
      resourceName=None
      resourceObject=null
      tiling=NoTile
      imageSize=<0 0>
      margin=<ltrb 0 0 0 0>
      outlineSettings {
       bUseBrushTransparency=false
       roundingType=FixedRadius
       width=1
       color {colorUseRule=UseColor_ColorTable, specifiedColor=<rgba 0.027 0.027 0.031 1>}
       cornerRadii=<5 5 5 5>
      }
      tintColor {colorUseRule=UseColor_Specified, specifiedColor=<rgba 0.04 0.04 0.048 1>}
      uVRegion {bIsValid=false, max=<0 0>, min=<0 0>}
     }
     pressedForeground {colorUseRule=UseColor_Foreground, specifiedColor=<rgba 1 1 1 1>}
     pressedPadding=<ltrb 0 0 0 0>
     pressedSlateSound {resourceObject=null}
    }
   }
   slot {class=VerticalBoxSlot, properties {padding=<ltrb 4 1 4 3>}}
  }
  HBox_Content_13 {
   widgetClassPath=&HorizontalBox
   parent=BTN_MoveSpaceLeft
   widget=HBox_Content_13
   slot {class=ButtonSlot, properties {horizontalAlignment=HAlign_Fill, padding=<ltrb 1 4 4 4>}}
  }
  Overlay_Icon_13 {
   widgetClassPath=&Overlay
   parent=HBox_Content_13
   widget=Overlay_Icon_13
   slot {class=HorizontalBoxSlot, properties {horizontalAlignment=HAlign_Left, verticalAlignment=VAlign_Center, padding=<ltrb 0 0 10 0>}}
  }
  Img_Icon_13 {
   widgetClassPath=&Image
   bIsVariable=true
   parent=Overlay_Icon_13
   widget=Img_Icon_13
   properties {brush {
     bIsDynamicallyLoaded=false
     drawAs=Image
     imageType=FullColor
     mirroring=NoMirror
     resourceName=None
     resourceObject=/AnimInEdTools/AnimationToolbox/Icons/AnimToolbox_UI_Icon_50
     tiling=NoTile
     imageSize=<24 24>
     margin=<ltrb 0 0 0 0>
     outlineSettings {
      bUseBrushTransparency=false
      roundingType=FixedRadius
      width=0
      color {colorUseRule=UseColor_Specified, specifiedColor=<rgba 0 0 0 1>}
      cornerRadii=<0 0 0 0>
     }
     tintColor {colorUseRule=UseColor_Specified, specifiedColor=<rgba 1 1 1 1>}
     uVRegion {bIsValid=false, max=<0 0>, min=<0 0>}
    }}
   slot {class=OverlaySlot, properties {verticalAlignment=VAlign_Center, padding=<ltrb 6 0 0 0>}}
  }
  TB_TabName_PIE_1 {
   widgetClassPath=&TextBlock
   bIsVariable=true
   parent=HBox_Content_13
   widget=TB_TabName_PIE_1
   properties {
    text=Nudge Space Switch Key Left
    textOverflowPolicy=Ellipsis
    font {
     bForceMonospaced=false
     bMaterialIsStencil=false
     fontMaterial=null
     fontObject=&Roboto
     letterSpacing=0
     monospacedWidth=0
     size=10
     skewAmount=0
     typefaceFontName=Bold
     outlineSettings {
      bApplyOutlineToDropShadows=false
      bMiteredCorners=false
      bSeparateFillAlpha=false
      outlineMaterial=null
      outlineSize=0
      outlineColor=<rgba 0 0 0 1>
     }
    }
   }
   slot {class=HorizontalBoxSlot, properties {verticalAlignment=VAlign_Center, size {sizeRule=Fill, value=1}}}
  }
  BTN_MoveSpaceRight {
   widgetClassPath=&EditorUtilityButton
   parent=VerticalBox_0
   widget=BTN_MoveSpaceRight
   properties {
    toolTipText=Nudge Space Key Right
    colorAndOpacity=<rgba 1 0.3916 0.1119 1>
    widgetStyle {
     clickedSlateSound {resourceObject=null}
     disabled {
      bIsDynamicallyLoaded=false
      drawAs=RoundedBox
      imageType=NoImage
      mirroring=NoMirror
      resourceName=None
      resourceObject=null
      tiling=NoTile
      imageSize=<0 0>
      margin=<ltrb 0 0 0 0>
      outlineSettings {
       bUseBrushTransparency=false
       roundingType=FixedRadius
       width=1
       color {colorUseRule=UseColor_ColorTable, specifiedColor=<rgba 0.027 0.027 0.031 1>}
       cornerRadii=<5 5 5 5>
      }
      tintColor {colorUseRule=UseColor_Specified, specifiedColor=<rgba 0.015 0.015 0.018 1>}
      uVRegion {bIsValid=false, max=<0 0>, min=<0 0>}
     }
     disabledForeground {colorUseRule=UseColor_ColorTable, specifiedColor=<rgba 0.4 0.4 0.4 1>}
     hovered {
      bIsDynamicallyLoaded=false
      drawAs=RoundedBox
      imageType=NoImage
      mirroring=NoMirror
      resourceName=None
      resourceObject=null
      tiling=NoTile
      imageSize=<0 0>
      margin=<ltrb 0 0 0 0>
      outlineSettings {
       bUseBrushTransparency=false
       roundingType=FixedRadius
       width=1
       color {colorUseRule=UseColor_ColorTable, specifiedColor=<rgba 0.027 0.027 0.031 1>}
       cornerRadii=<5 5 5 5>
      }
      tintColor {colorUseRule=UseColor_Specified, specifiedColor=<rgba 0.031 0.031 0.038 1>}
      uVRegion {bIsValid=false, max=<0 0>, min=<0 0>}
     }
     hoveredForeground {colorUseRule=UseColor_Foreground, specifiedColor=<rgba 1 1 1 1>}
     hoveredSlateSound {resourceObject=null}
     normal {
      bIsDynamicallyLoaded=false
      drawAs=RoundedBox
      imageType=NoImage
      mirroring=NoMirror
      resourceName=None
      resourceObject=null
      tiling=NoTile
      imageSize=<0 0>
      margin=<ltrb 0 0 0 0>
      outlineSettings {
       bUseBrushTransparency=false
       roundingType=FixedRadius
       width=1
       color {colorUseRule=UseColor_ColorTable, specifiedColor=<rgba 0.027 0.027 0.031 1>}
       cornerRadii=<5 5 5 5>
      }
      tintColor {colorUseRule=UseColor_Specified, specifiedColor=<rgba 0.0222 0.0222 0.0252 1>}
      uVRegion {bIsValid=false, max=<0 0>, min=<0 0>}
     }
     normalForeground {colorUseRule=UseColor_Foreground, specifiedColor=<rgba 1 1 1 1>}
     normalPadding=<ltrb 0 0 0 0>
     pressed {
      bIsDynamicallyLoaded=false
      drawAs=RoundedBox
      imageType=NoImage
      mirroring=NoMirror
      resourceName=None
      resourceObject=null
      tiling=NoTile
      imageSize=<0 0>
      margin=<ltrb 0 0 0 0>
      outlineSettings {
       bUseBrushTransparency=false
       roundingType=FixedRadius
       width=1
       color {colorUseRule=UseColor_ColorTable, specifiedColor=<rgba 0.027 0.027 0.031 1>}
       cornerRadii=<5 5 5 5>
      }
      tintColor {colorUseRule=UseColor_Specified, specifiedColor=<rgba 0.04 0.04 0.048 1>}
      uVRegion {bIsValid=false, max=<0 0>, min=<0 0>}
     }
     pressedForeground {colorUseRule=UseColor_Foreground, specifiedColor=<rgba 1 1 1 1>}
     pressedPadding=<ltrb 0 0 0 0>
     pressedSlateSound {resourceObject=null}
    }
   }
   slot {class=VerticalBoxSlot, properties {verticalAlignment=VAlign_Top, padding=<ltrb 4 1 4 3>, size {sizeRule=Fill, value=1}}}
  }
  HBox_Content_14 {
   widgetClassPath=&HorizontalBox
   parent=BTN_MoveSpaceRight
   widget=HBox_Content_14
   slot {class=ButtonSlot, properties {horizontalAlignment=HAlign_Fill, padding=<ltrb 1 4 4 4>}}
  }
  Overlay_Icon_14 {
   widgetClassPath=&Overlay
   parent=HBox_Content_14
   widget=Overlay_Icon_14
   slot {class=HorizontalBoxSlot, properties {horizontalAlignment=HAlign_Left, verticalAlignment=VAlign_Center, padding=<ltrb 0 0 10 0>}}
  }
  Img_Icon_14 {
   widgetClassPath=&Image
   bIsVariable=true
   parent=Overlay_Icon_14
   widget=Img_Icon_14
   properties {brush {
     bIsDynamicallyLoaded=false
     drawAs=Image
     imageType=FullColor
     mirroring=NoMirror
     resourceName=None
     resourceObject=/AnimInEdTools/AnimationToolbox/Icons/AnimToolbox_UI_Icon_51
     tiling=NoTile
     imageSize=<24 24>
     margin=<ltrb 0 0 0 0>
     outlineSettings {
      bUseBrushTransparency=false
      roundingType=FixedRadius
      width=0
      color {colorUseRule=UseColor_Specified, specifiedColor=<rgba 0 0 0 1>}
      cornerRadii=<0 0 0 0>
     }
     tintColor {colorUseRule=UseColor_Specified, specifiedColor=<rgba 1 1 1 1>}
     uVRegion {bIsValid=false, max=<0 0>, min=<0 0>}
    }}
   slot {class=OverlaySlot, properties {verticalAlignment=VAlign_Center, padding=<ltrb 6 0 0 0>}}
  }
  TB_TabName_PIE_2 {
   widgetClassPath=&TextBlock
   bIsVariable=true
   parent=HBox_Content_14
   widget=TB_TabName_PIE_2
   properties {
    text=Nudge Space Switch Key Right
    textOverflowPolicy=Ellipsis
    font {
     bForceMonospaced=false
     bMaterialIsStencil=false
     fontMaterial=null
     fontObject=&Roboto
     letterSpacing=0
     monospacedWidth=0
     size=10
     skewAmount=0
     typefaceFontName=Bold
     outlineSettings {
      bApplyOutlineToDropShadows=false
      bMiteredCorners=false
      bSeparateFillAlpha=false
      outlineMaterial=null
      outlineSize=0
      outlineColor=<rgba 0 0 0 1>
     }
    }
   }
   slot {class=HorizontalBoxSlot, properties {verticalAlignment=VAlign_Center, size {sizeRule=Fill, value=1}}}
  }
 ]
}
