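# Editor utility widget for Sequencer animation tools that bakes a locator control onto an animated Control Rig control.
# Run reads controls and range from EpicAnimTools, then uses embedded Python to snap all frames or keys only, optionally parent-constraining the control back to the locator.
# A toggle flips that constraint option and recolors the locator rig.
# Further chains without visible triggers handle axis-offset LookAt setup and parent-constraint bake/delete.
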EditorUtilityWidgetBlueprint /AnimInEdTools/EpicAnimTools/EAT_BakeOffset
graphs [
 BakeDelete_ParentConstraint {nodes [
   n1 = BakeDelete_ParentConstraint!() -> (then, in_ctrl_name:Name) {is=title K2Node_FunctionEntry}
   > Set BD_Par_ctrl_name(BD_Par_ctrl_name:Name=@n1) -> (Output_Get:Name=None)
   > n2 = GetEditorWorld(self=Get UnrealEditorSubsystem())
   > n3 = GetConstraintsArray!(execute, InWorld:World Object Reference=@n2) -> (then, ReturnValue:Array of Tickable Constraint Object References)
   > n4 = For Each Loop(Array:Array of Tickable Constraint Object References=@n3) -> (Array Element:Tickable Constraint Object Reference)
   > n5 = Cast To TickableParentConstraint!(execute, Object=@n4) -> (then, CastFailed, AsTickable Parent Constraint:Tickable Parent Constraint Object Reference) {is=title K2Node_DynamicCast}
   BakeConstraint!(
    execute=@n4.Completed
    World:World Object Reference=@n2
    Constraint:Tickable Constraint Object Reference=@n5
    Frames:Array of Frame Number Structures=Make Array!() -> (Array:Array of Frame Number Structures)
    TimeUnit:EMovieSceneTimeUnit Enum=DisplayRate
   ) -> (then, ReturnValue:bool)
   > RefreshCurrentLevelSequence!(execute) -> (then)
   > Execute Python Script!(
    execute
    PythonScript="import unreal\n\nif in_world is None:\n    print('[DeleteConstraint] No world provided, skipping.')\nelif in_par_constraint is None:\n    print('[DeleteConstraint] No constraint provided, skipping.')\nelse:\n    success = unreal.ConstraintsScriptingLibrary.remove_constraint(in_world, in_idx)\n    if success:\n        print(f'[DeleteConstraint] Successfully deleted constraint: {in_par_constraint.get_name()}')\n    else:\n        print(f'[DeleteConstraint] Failed to delete constraint: {in_par_constraint.get_name()}')\n"
    in_par_constraint:Tickable Parent Constraint Object Reference=Get BD_Par_constraint() -> (BD_Par_constraint:Tickable Parent Constraint Object Reference)
    in_world:World Object Reference=@n2
    in_idx:int=Get BD_Par_constraint_idx() -> (BD_Par_constraint_idx:int=0)
   ) -> (then, ReturnValue) {inputs [in_par_constraint, in_world, in_idx]}
   Set BD_Par_constraint_idx(execute=@n7.then, BD_Par_constraint_idx:int=@n4.Array Index) -> (Output_Get:int=0)
   n6 = Cast To TransformableControlHandle!(
    execute=@n5.then
    Object=Get ChildTRSHandle!(self:Tickable Transform Constraint Object Reference=@n5) -> (ChildTRSHandle:Transformable Handle Object Reference)
   ) -> (then, CastFailed, AsTransformable Control Handle:Transformable Control Handle Object Reference) {is=title K2Node_DynamicCast}
   > Branch(Condition="Equal (Name)"(
     A=Get ControlName!(self:Transformable Control Handle Object Reference=@n6) -> (ControlName:Name=None)
     B=Get BD_Par_ctrl_name() -> (BD_Par_ctrl_name:Name=None)
    ))
   > n7 = Set BD_Par_constraint(BD_Par_constraint:Tickable Parent Constraint Object Reference=@n5) -> (Output_Get:Tickable Parent Constraint Object Reference)
  ]}
 EventGraph {nodes [
   n1 = GetEditorWorld(self=Get UnrealEditorSubsystem())
   n2 = GetEditorWorld(self=Get UnrealEditorSubsystem())
   > n3 = GetConstraintsArray(InWorld=@n2)
   > n4 = For Each Loop(Array=@n3)
   > n5 = Cast To TickableParentConstraint(Object=@n4)
   BakeConstraint(execute=[@n1.then, @n5.then], World=@n1, Constraint=@n5, Frames=Make Array(), TimeUnit=DisplayRate)
   RefreshCurrentLevelSequence()
   > Execute Python Script!(
    execute
    PythonScript="import unreal\r\n\r\n## inputs injected by Execute Python Script node (always lowercase)\r\n## locator_cr, locator_ctrl, actor_cr, actor_ctrl, axis_var, dist_var, mode_var\r\n\r\nlocator_ctrl_name = str(locator_ctrl)\r\nactor_ctrl_name   = str(actor_ctrl)\r\n\r\nLSEBL = unreal.LevelSequenceEditorBlueprintLibrary\r\nCRSL  = unreal.ControlRigSequencerLibrary\r\n\r\n## ── read UI values from pins ──────────────────────────────────────\r\noffset_axis     = str(axis_var).strip().upper()             ## 'X', 'Y', or 'Z'\r\noffset_distance = float(str(dist_var).strip()) if str(dist_var).strip() else 50.0\r\nkeys_only       = str(mode_var).strip().lower() == 'keys'\r\n\r\ndef axis_vector(a):\r\n    if a == 'X': return unreal.Vector(1.0, 0.0, 0.0)\r\n    if a == 'Y': return unreal.Vector(0.0, 1.0, 0.0)\r\n    return unreal.Vector(0.0, 0.0, 1.0)\r\n\r\naxis_vec = axis_vector(offset_axis)\r\n\r\n## ── get level sequence and world ──────────────────────────────────\r\nls = LSEBL.get_focused_level_sequence()\r\nif not ls:\r\n    unreal.log_error(\"Offset LookAt: No level sequence is open.\")\r\n    raise SystemExit\r\n\r\neditor_system = unreal.get_editor_subsystem(unreal.UnrealEditorSubsystem)\r\nworld = editor_system.get_editor_world()\r\n\r\n## ── control-rig transform accessors (work on ControlRig and ModularRig) ──\r\ndef _control_key(name):\r\n    return unreal.RigElementKey(type=unreal.RigElementType.CONTROL, name=unreal.Name(name))\r\n\r\ndef get_control_global_transform(cr, ctrl_name):\r\n    if hasattr(cr, 'get_control_global_transform'):\r\n        return cr.get_control_global_transform(ctrl_name)\r\n    return cr.get_hierarchy().get_global_transform(_control_key(ctrl_name))\r\n\r\ndef set_control_global_transform(cr, ctrl_name, xform):\r\n    if hasattr(cr, 'set_control_global_transform'):\r\n        cr.set_control_global_transform(ctrl_name, xform, True)\r\n        return\r\n    cr.get_hierarchy().set_global_transform(_control_key(ctrl_name), xform, False, True, False, False)\r\n\r\ndef get_control_local_transform(cr, ctrl_name):\r\n    if hasattr(cr, 'get_control_local_transform'):\r\n        return cr.get_control_local_transform(ctrl_name)\r\n    return cr.get_hierarchy().get_local_transform(_control_key(ctrl_name))\r\n\r\n## ── playhead setter that prefers the non-deprecated overload ──────\r\ndef set_seq_current_time(frame):\r\n    try:\r\n        params = unreal.MovieSceneSequencePlaybackParams()\r\n        if hasattr(params, 'frame'):\r\n            params.frame = unreal.FrameTime(unreal.FrameNumber(int(frame)))\r\n        elif hasattr(params, 'frame_number'):\r\n            params.frame_number = unreal.FrameNumber(int(frame))\r\n        LSEBL.set_current_time(params)\r\n    except (TypeError, AttributeError):\r\n        LSEBL.set_current_time(int(frame))\r\n\r\n## ── snapper helpers ───────────────────────────────────────────────\r\ndef make_snapper_selection(cr, ctrl_name):\r\n    sel = unreal.ControlRigSnapperSelection()\r\n    item = unreal.ControlRigForWorldTransforms()\r\n    item.control_rig = cr\r\n    item.control_names = [unreal.Name(ctrl_name)]\r\n    if hasattr(sel, 'children_or_parents'):\r\n        sel.children_or_parents = [item]\r\n    else:\r\n        sel.control_rigs = [item]\r\n    return sel\r\n\r\n## ── channel discovery ─────────────────────────────────────────────\r\ndef _candidate_channel_types():\r\n    out = []\r\n    for type_name in ('MovieSceneScriptingDoubleChannel', 'MovieSceneScriptingFloatChannel'):\r\n        t = getattr(unreal, type_name, None)\r\n        if t is not None:\r\n            out.append(t)\r\n    return out\r\n\r\ndef find_locator_channels():\r\n    \"\"\"All keyable channels whose name starts with '<locator_ctrl>.'.\r\n    Prefix-match only (no CR-ref filter) — matches what offset_lookat_step_02's\r\n    collect_channels did. The CR-ref filter from offset_follow_snap.py doesn't\r\n    survive freshly-added sections, where the Python pin ref doesn't `==` the\r\n    section's get_control_rig() result.\"\"\"\r\n    out = []\r\n    prefix = f\"{locator_ctrl_name}.\"\r\n    for binding in ls.get_bindings():\r\n        for track in binding.get_tracks():\r\n            if not isinstance(track, unreal.MovieSceneControlRigParameterTrack):\r\n                continue\r\n            for section in track.get_sections():\r\n                for ct in _candidate_channel_types():\r\n                    try:\r\n                        channels = section.get_channels_by_type(ct)\r\n                    except Exception:\r\n                        channels = []\r\n                    for ch in channels:\r\n                        name = str(ch.channel_name) if hasattr(ch, 'channel_name') else \"\"\r\n                        if name.startswith(prefix):\r\n                            out.append(ch)\r\n    return out\r\n\r\ndef get_actor_key_frames(start, end):\r\n    \"\"\"Sorted unique key frames on actor_ctrl within [start, end].\"\"\"\r\n    times = set()\r\n    prefix = f\"{actor_ctrl_name}.\"\r\n    for binding in ls.get_bindings():\r\n        for track in binding.get_tracks():\r\n            if not isinstance(track, unreal.MovieSceneControlRigParameterTrack):\r\n                continue\r\n            for section in track.get_sections():\r\n                for ct in _candidate_channel_types():\r\n                    try:\r\n                        channels = section.get_channels_by_type(ct)\r\n                    except Exception:\r\n                        channels = []\r\n                    for ch in channels:\r\n                        name = str(ch.channel_name) if hasattr(ch, 'channel_name') else \"\"\r\n                        if not name.startswith(prefix):\r\n                            continue\r\n                        for k in ch.get_keys():\r\n                            t = int(k.get_time().frame_number.value)\r\n                            if start <= t <= end:\r\n                                times.add(t)\r\n    return sorted(times)\r\n\r\ndef _channel_value_from_pose(channel_name, translation, rotator, scale):\r\n    \"\"\"Map a channel name (e.g. 'Locator.Location.X') to the matching pose component.\r\n    Skips rotation channels: extracting FRotator gives XYZ-Euler, which doesn't\r\n    round-trip into channels that may use a different preferred rotation order\r\n    (e.g. ZXY). The locator's make-locator step is expected to have keyed the\r\n    rotation at start_frame to match the actor's orientation; we don't want to\r\n    overwrite those keys with mis-ordered values.\"\"\"\r\n    lower = channel_name.lower()\r\n    axis = None\r\n    for letter in 'XYZ':\r\n        if channel_name.endswith(letter) or channel_name.endswith('.' + letter) or channel_name.endswith('_' + letter):\r\n            axis = letter\r\n            break\r\n    if axis is None:\r\n        return None\r\n    if any(t in lower for t in ('location', 'translation', 'position')):\r\n        return float({'X': translation.x, 'Y': translation.y, 'Z': translation.z}[axis])\r\n    if 'rotation' in lower:\r\n        return None\r\n    if 'scale' in lower:\r\n        return float({'X': scale.x, 'Y': scale.y, 'Z': scale.z}[axis])\r\n    return None\r\n\r\ndef key_locator_pose_at_frame(frame):\r\n    \"\"\"Lock the locator's current runtime pose into a Sequencer key at `frame`.\"\"\"\r\n    local_xform = get_control_local_transform(locator_cr, locator_ctrl_name)\r\n    trans   = local_xform.translation\r\n    rotator = local_xform.rotation.rotator()\r\n    scale   = local_xform.scale3d\r\n    frame_num = unreal.FrameNumber(int(frame))\r\n    keyed = 0\r\n    for ch in find_locator_channels():\r\n        name = str(ch.channel_name) if hasattr(ch, 'channel_name') else \"\"\r\n        value = _channel_value_from_pose(name, trans, rotator, scale)\r\n        if value is None:\r\n            continue\r\n        try:\r\n            ch.add_key(frame_num, value)\r\n            keyed += 1\r\n        except Exception as e:\r\n            unreal.log_warning(f\"add_key failed on {name}: {e}\")\r\n    return keyed\r\n\r\ndef delete_locator_keys_except(keep_frames, start, end):\r\n    keep = set(int(f) for f in keep_frames)\r\n    removed = 0\r\n    for ch in find_locator_channels():\r\n        for k in list(ch.get_keys()):\r\n            try:\r\n                t = int(k.get_time().frame_number.value)\r\n            except Exception:\r\n                continue\r\n            if start <= t <= end and t not in keep:\r\n                try:\r\n                    ch.remove_key(k)\r\n                    removed += 1\r\n                except Exception as e:\r\n                    unreal.log_warning(f\"remove_key failed at {t}: {e}\")\r\n    return removed\r\n\r\n## ── detect selected key range ─────────────────────────────────────\r\ndef get_selected_key_range():\r\n    all_frames = []\r\n    ses = unreal.get_editor_subsystem(unreal.LevelSequenceEditorSubsystem)\r\n    curve_editor = ses.get_curve_editor()\r\n    if curve_editor.is_curve_editor_open():\r\n        for ch_proxy in curve_editor.get_channels_with_selected_keys():\r\n            indices = curve_editor.get_selected_keys(ch_proxy)\r\n            if not indices:\r\n                continue\r\n            channel_obj = ch_proxy.section.get_channel(ch_proxy.channel_name)\r\n            if not channel_obj:\r\n                continue\r\n            for k in channel_obj.get_keys_by_index(indices):\r\n                all_frames.append(k.get_time().frame_number.value)\r\n    if not all_frames:\r\n        for ch_proxy in LSEBL.get_channels_with_selected_keys():\r\n            indices = LSEBL.get_selected_keys(ch_proxy)\r\n            if not indices:\r\n                continue\r\n            channel_obj = ch_proxy.section.get_channel(ch_proxy.channel_name)\r\n            if not channel_obj:\r\n                continue\r\n            for k in channel_obj.get_keys_by_index(indices):\r\n                all_frames.append(k.get_time().frame_number.value)\r\n    return (min(all_frames), max(all_frames)) if all_frames else None\r\n\r\nselected_range = get_selected_key_range()\r\nhas_range = selected_range is not None\r\n\r\nif has_range:\r\n    start_val, end_val = selected_range\r\nelse:\r\n    start_val = ls.get_playback_start()\r\n    end_val   = ls.get_playback_end() - 1\r\n\r\nstart_val = int(start_val)\r\nend_val   = int(end_val)\r\n\r\nunreal.log(\r\n    f\"Offset LookAt: axis={offset_axis}, distance={offset_distance}, \"\r\n    f\"keys_only={keys_only}, range={start_val} to {end_val} \"\r\n    f\"({'selected keys' if has_range else 'playback'}).\"\r\n)\r\n\r\n## ══════════════════════════════════════════════════════════════════\r\n## STEP 1: snap locator -> actor, offset along local axis, key the\r\n##         offset pose, then follow-snap with keep_offset over range.\r\n## (Replaces the parent-constraint approach to avoid the engine bug\r\n##  in FControlRigSequencerHelpers::GetFloatChannels.)\r\n## ══════════════════════════════════════════════════════════════════\r\n\r\n## park playhead on start frame so runtime reads/writes evaluate correctly\r\nset_seq_current_time(start_val)\r\n\r\n## snap locator to actor (runtime — visible in viewport)\r\nactor_xform = get_control_global_transform(actor_cr, actor_ctrl_name)\r\nset_control_global_transform(locator_cr, locator_ctrl_name, actor_xform)\r\n\r\n## offset locator along its own local axis (runtime).\r\n## now that the locator's rotation matches the actor's, rotating the local\r\n## axis by the locator's global rotation gives the offset in actor-local space.\r\nglobal_xform = get_control_global_transform(locator_cr, locator_ctrl_name)\r\nworld_offset = global_xform.rotation.rotate_vector(axis_vec) * offset_distance\r\nglobal_xform.translation = global_xform.translation + world_offset\r\nset_control_global_transform(locator_cr, locator_ctrl_name, global_xform)\r\n\r\n## CRITICAL: lock the offset pose into a Sequencer key at start_val.\r\n## Without this, snap_control_rig re-evaluates the sequence and the runtime\r\n## offset is lost, so keep_offset has nothing to anchor against.\r\nkeyed = key_locator_pose_at_frame(start_val)\r\nunreal.log(f\"Offset LookAt: wrote {keyed} keys on locator at frame {start_val}.\")\r\n\r\n## park playhead back at start so the snapper reads the offset key as its\r\n## baseline.\r\nset_seq_current_time(start_val)\r\n\r\n## follow-snap with offset over the range\r\nchildren = make_snapper_selection(locator_cr, locator_ctrl_name)\r\nparents  = make_snapper_selection(actor_cr,   actor_ctrl_name)\r\n\r\nsnap_settings = unreal.ControlRigSnapSettings()\r\nsnap_settings.keep_offset   = True\r\nsnap_settings.snap_position = True\r\nsnap_settings.snap_rotation = True\r\nsnap_settings.snap_scale    = False\r\n\r\nCRSL.snap_control_rig(\r\n    ls,\r\n    unreal.FrameNumber(start_val),\r\n    unreal.FrameNumber(end_val),\r\n    children, parents, snap_settings,\r\n)\r\nunreal.log(\r\n    f\"Offset LookAt: snapped locator -> actor over [{start_val}, {end_val}] with offset \"\r\n    f\"{offset_axis}={offset_distance}cm.\"\r\n)\r\n\r\n## keys-only post-process: drop locator keys at frames where actor has none\r\nif keys_only:\r\n    actor_keys = get_actor_key_frames(start_val, end_val)\r\n    if not actor_keys:\r\n        unreal.log_warning(\r\n            \"Offset LookAt: keys_only requested but actor_ctrl has no keys in range; \"\r\n            \"skipping filter.\"\r\n        )\r\n    else:\r\n        removed = delete_locator_keys_except(actor_keys, start_val, end_val)\r\n        unreal.log(f\"Offset LookAt: keys-only filter removed {removed} locator keys.\")\r\n\r\n## ══════════════════════════════════════════════════════════════════\r\n## STEP 2: LookAt constraint actor_ctrl -> locator_ctrl\r\n## ══════════════════════════════════════════════════════════════════\r\n\r\n## read locator's final key range from the snapped output\r\nlocator_key_frames = set()\r\nfor ch in find_locator_channels():\r\n    try:\r\n        for k in ch.get_keys():\r\n            locator_key_frames.add(k.get_time().frame_number.value)\r\n    except Exception:\r\n        continue\r\n\r\nif locator_key_frames:\r\n    lookat_start = min(locator_key_frames)\r\n    lookat_end   = max(locator_key_frames)\r\nelse:\r\n    lookat_start = start_val\r\n    lookat_end   = end_val\r\n\r\nseq_last_frame = ls.get_playback_end() - 1\r\n\r\nchild_handle  = actor_cr.create_transformable_control_handle(actor_ctrl_name)\r\nparent_handle = locator_cr.create_transformable_control_handle(locator_ctrl_name)\r\n\r\nif not child_handle or not parent_handle:\r\n    unreal.log_error(\"Offset LookAt: Failed to create handles for Look At constraint.\")\r\n    raise SystemExit\r\n\r\nset_seq_current_time(lookat_start)\r\n\r\nlookat_constraint = CRSL.add_constraint(\r\n    world,\r\n    unreal.TransformConstraintType.LOOK_AT,\r\n    child_handle,\r\n    parent_handle,\r\n    True  ## maintain_offset\r\n)\r\n\r\nif not lookat_constraint:\r\n    unreal.log_error(\"Offset LookAt: Failed to create Look At constraint.\")\r\n    raise SystemExit\r\n\r\nunreal.log(\r\n    f\"Offset LookAt: Look At constraint created '{actor_ctrl_name}' -> '{locator_ctrl_name}'.\"\r\n)\r\n\r\nCRSL.set_constraint_active_key(\r\n    lookat_constraint, True, unreal.FrameNumber(int(lookat_start))\r\n)\r\n\r\nif lookat_end < seq_last_frame:\r\n    CRSL.set_constraint_active_key(\r\n        lookat_constraint, False, unreal.FrameNumber(int(lookat_end))\r\n    )\r\n    unreal.log(\r\n        f\"Offset LookAt: Look At keyed ON at {lookat_start}, OFF at {lookat_end} \"\r\n        f\"(sequence ends at {seq_last_frame}).\"\r\n    )\r\nelse:\r\n    unreal.log(\r\n        f\"Offset LookAt: Look At keyed ON at {lookat_start}, no OFF key \"\r\n        f\"(lookat_end {lookat_end} == sequence end {seq_last_frame}).\"\r\n    )\r\n"
    Locator_CR:Control Rig Object Reference=Get Out Loc CR() -> (Out Loc CR:Control Rig Object Reference)
    Locator_Ctrl:Name=Get Out Loc Ctrl Name() -> (Out Loc Ctrl Name:Name=None)
    Actor_CR:Control Rig Object Reference=Get In CR() -> (In CR:Control Rig Object Reference)
    Actor_Ctrl:Name=Get In Control() -> (In Control:Name=None)
    axis_var:String=Get axis() -> (axis:String)
    dist_var:String=Get distance() -> (distance:String)
    mode_var:String=Get mode() -> (mode:String)
   ) -> (then, ReturnValue) {inputs [Locator_CR,Locator_Ctrl,Actor_CR,Actor_Ctrl,axis_var,dist_var,mode_var]}
   > RefreshCurrentLevelSequence()
   ClearControlSelection!(execute, self:Control Rig Object Reference=Get In CR(), bSetupUndo:bool=false) -> (then, ReturnValue:bool)
   > SelectControl!(
    execute
    self:Control Rig Object Reference=Get Out Loc CR()
    InControlName:"Name (by ref)"=Get Out Loc Ctrl Name() -> (Out Loc Ctrl Name=None)
    bSelect:bool=true
    bSetupUndo:bool=false
   ) -> (then)
   "On Clicked (BTN_Run)"() {is=title K2Node_ComponentBoundEvent}
   > BeginTransaction(Description="LOCTEXT(\"Bake To Locator\")")
   > n6 = FindUtilityWidgetFromBlueprint!(
    execute
    self:Editor Utility Subsystem Object Reference=Get EditorUtilitySubsystem()
    InBlueprint:Editor Utility Widget Blueprint Object Reference=&/EpicAnimTools
   ) -> (then, ReturnValue:Editor Utility Widget Object Reference)
   > n7 = Cast To EpicAnimTools!(execute, Object=@n6) -> (then, CastFailed, AsEpic Anim Tools:Epic Anim Tools Object Reference) {is=title K2Node_DynamicCast}
   > n8 = Sequence()
   n9 = Get Locator_Loc_Ctrl!(self:Epic Anim Tools Object Reference=@n7) -> (Locator_Loc_Ctrl:Name=None)
   n10 = Get Range Start!(self:Epic Anim Tools Object Reference=@n7) -> (Range Start:int=0)
   n11 = Get Range End!(self:Epic Anim Tools Object Reference=@n7) -> (Range End:int=0)
   n12 = Get Locator_In_CR!(self:Epic Anim Tools Object Reference=@n7) -> (Locator_In_CR:Control Rig Object Reference)
   n13 = Get Locator_In_Ctrl!(self:Epic Anim Tools Object Reference=@n7) -> (Locator_In_Ctrl:Name=None)
   n14 = Get Locator_Loc_CR!(self:Epic Anim Tools Object Reference=@n7) -> (Locator_Loc_CR:Control Rig Object Reference)
   > Execute Python Script!(
    execute
    PythonScript="import unreal\r\n\r\n## inputs injected by Execute Python Script node (always lowercase)\r\n## locator_cr, locator_ctrl, actor_cr, actor_ctrl, range_start, range_end\r\n\r\nlocator_ctrl_name = str(locator_ctrl)\r\nactor_ctrl_name   = str(actor_ctrl)\r\n\r\n## ── get the current level sequence ───────────────────────────────\r\nls = unreal.LevelSequenceEditorBlueprintLibrary.get_focused_level_sequence()\r\nif not ls:\r\n    unreal.log_error(\"Snap Locator: No level sequence is open.\")\r\n    raise SystemExit\r\n\r\n## ── resolve frame range from delivered int pins ───────────────────\r\nstart_frame = unreal.FrameNumber(int(range_start))\r\nend_frame   = unreal.FrameNumber(int(range_end))\r\n\r\n## ── pop-up: ask All Frames or Keys Only ──────────────────────────\r\nmessage = (\r\n    f\"Snap '{locator_ctrl_name}' to '{actor_ctrl_name}'.\\n\"\r\n    f\"\\nFrame range: {int(range_start)} to {int(range_end)}.\\n\"\r\n    \"\\nYES = All Frames    NO = Keys Only    Cancel = Abort\"\r\n)\r\n\r\nresult = unreal.EditorDialog.show_message(\r\n    \"Snap Locator to Actor\",\r\n    message,\r\n    unreal.AppMsgType.YES_NO_CANCEL,\r\n    unreal.AppReturnType.NO,\r\n)\r\n\r\nif result == unreal.AppReturnType.CANCEL:\r\n    unreal.log(\"Snap Locator: Cancelled by user.\")\r\n    raise SystemExit\r\n\r\nkeys_only = (result == unreal.AppReturnType.NO)\r\n\r\n## ── build snapper selections ──────────────────────────────────────\r\n## Parent (driver): actor_ctrl on actor_cr\r\nparent_cr_entry = unreal.ControlRigForWorldTransforms()\r\nparent_cr_entry.control_rig   = actor_cr\r\nparent_cr_entry.control_names = [actor_ctrl_name]\r\n\r\nparent_sel = unreal.ControlRigSnapperSelection()\r\nparent_sel.control_rigs = [parent_cr_entry]\r\n\r\n## Child (snapped): locator_ctrl on locator_cr\r\nchild_cr_entry = unreal.ControlRigForWorldTransforms()\r\nchild_cr_entry.control_rig   = locator_cr\r\nchild_cr_entry.control_names = [locator_ctrl_name]\r\n\r\nchild_sel = unreal.ControlRigSnapperSelection()\r\nchild_sel.control_rigs = [child_cr_entry]\r\n\r\n## ── snap settings ─────────────────────────────────────────────────\r\nsnap_settings = unreal.ControlRigSnapSettings()\r\nsnap_settings.keep_offset         = True\r\nsnap_settings.snap_position       = True\r\nsnap_settings.snap_rotation       = True\r\nsnap_settings.snap_scale          = False\r\nsnap_settings.baking_key_settings = (\r\n    unreal.BakingKeySettings.KEYS_ONLY\r\n    if keys_only\r\n    else unreal.BakingKeySettings.ALL_FRAMES\r\n)\r\n\r\n## ── execute snap ──────────────────────────────────────────────────\r\nunreal.ControlRigSequencerLibrary.snap_control_rig(\r\n    ls, start_frame, end_frame, child_sel, parent_sel, snap_settings\r\n)\r\n\r\nmode_str  = \"keys only\" if keys_only else \"all frames\"\r\nrange_str = f\"{start_frame.value} to {end_frame.value}\"\r\nunreal.log(f\"Snap Locator: snapped '{locator_ctrl_name}' to '{actor_ctrl_name}' [{range_str}] ({mode_str}).\")\r\n"
    Locator_CR=@n14
    Locator_Ctrl=@n9
    Actor_CR=@n12
    Actor_Ctrl=@n13
    range_start:int=@n10
    range_end:int=@n11
   ) -> (then, ReturnValue) {inputs [Locator_CR,Locator_Ctrl,Actor_CR,Actor_Ctrl,range_start,range_end]}
   > Branch(Condition=Get Locator Constrain() -> (Locator Constrain:bool))
   > Execute Python Script(
    PythonScript="import unreal\n\n## inputs injected by Execute Python Script node (always lowercase)\n## locator_cr, locator_ctrl, actor_cr, actor_ctrl, range_start, range_end\n\nlocator_ctrl_name = str(locator_ctrl)\nactor_ctrl_name   = str(actor_ctrl)\n\n## ── get level sequence and world ──────────────────────────────────\nls = unreal.LevelSequenceEditorBlueprintLibrary.get_focused_level_sequence()\nif not ls:\n    unreal.log_error(\"Constrain Actor to Locator: No level sequence is open.\")\n    raise SystemExit\n\neditor_system = unreal.get_editor_subsystem(unreal.UnrealEditorSubsystem)\nworld = editor_system.get_editor_world()\n\n## ── resolve frame range from delivered int pins ───────────────────\nstart_frame = unreal.FrameNumber(int(range_start))\nend_frame   = unreal.FrameNumber(int(range_end))\n\n## ── jump to start frame ───────────────────────────────────────────\nunreal.LevelSequenceEditorBlueprintLibrary.set_current_time(int(range_start))\n\n## ── build handles ─────────────────────────────────────────────────\n## child = actor_ctrl (the one being constrained)\n## parent = locator_ctrl (the driver)\nchild_handle  = actor_cr.create_transformable_control_handle(actor_ctrl_name)\nparent_handle = locator_cr.create_transformable_control_handle(locator_ctrl_name)\n\nif not child_handle or not parent_handle:\n    unreal.log_error(\"Constrain Actor to Locator: Failed to create control handles.\")\n    raise SystemExit\n\n## ── create parent constraint ──────────────────────────────────────\nconstraint = unreal.ControlRigSequencerLibrary.add_constraint(\n    world,\n    unreal.TransformConstraintType.PARENT,\n    child_handle,\n    parent_handle,\n    True  ## maintain_offset\n)\n\nif not constraint:\n    unreal.log_error(\"Constrain Actor to Locator: Failed to create constraint.\")\n    raise SystemExit\n\nunreal.log(f\"Constrain: created parent constraint '{actor_ctrl_name}' -> '{locator_ctrl_name}'\")\n\n## ── key the constraint ON at start_frame, OFF at end_frame ────────\nunreal.ControlRigSequencerLibrary.set_constraint_active_key(\n    constraint, True, start_frame\n)\nunreal.ControlRigSequencerLibrary.set_constraint_active_key(\n    constraint, False, end_frame\n)\nunreal.log(\n    f\"Constrain: keyed ON at frame {start_frame.value}, OFF at frame {end_frame.value}.\"\n)\n"
    Locator_CR=@n14
    Locator_Ctrl=@n9
    Actor_CR=@n12
    Actor_Ctrl=@n13
    range_start=@n10
    range_end=@n11
   ) {inputs [Locator_CR,Locator_Ctrl,Actor_CR,Actor_Ctrl,range_start,range_end]}
   n15 = EndTransaction(execute=@n8.then_1)
   Execute Python Script!(
    execute=[@n15.then, @n25.then]
    PythonScript="import unreal\r\nunreal.log(\"Close button clicked - attempting to close tab\")\r\neus = unreal.get_editor_subsystem(unreal.EditorUtilitySubsystem)\r\nblueprint = unreal.load_object(None, '/AnimInEdTools/AnimationToolbox/EUW_AnimMisc_BakeOffset')\r\nunreal.log(f\"Blueprint object: {blueprint}, type: {type(blueprint)}\")\r\ntab_id = eus.get_tab_id_from_blueprint(blueprint)\r\nunreal.log(f\"Tab ID: {tab_id}\")\r\neus.unregister_tab_by_id(tab_id)\r\n"
   ) -> (then, ReturnValue)
   "On Clicked (EditorUtilityButton_70)"() {is=title K2Node_ComponentBoundEvent}
   > n16 = Sequence()
   > n17 = Set Locator Constrain(Locator Constrain:bool=NOT Boolean(A=Get Locator Constrain())) -> (Output_Get:bool)
   > n18 = Branch(Condition=@n17)
   > "SetText (Text)"(self=Get TXT_ConstraintStatus() -> (TXT_ConstraintStatus:Text Object Reference), InText="LOCTEXT(\"ON\")")
   > SetForegroundColor!(
    execute
    self:User Widget Object Reference
    InForegroundColor_SpecifiedColor:Linear Color Structure=(R=0.7157,G=0.0648,B=0.3916,A=1)
    InForegroundColor_ColorUseRule:ESlateColorStylingMode Enum
   ) -> (then)
   "SetText (Text)"(execute=@n18.else, self=Get TXT_ConstraintStatus(), InText="LOCTEXT(\"OFF\")")
   > SetForegroundColor(InForegroundColor_SpecifiedColor=(R=0.1878,G=0.3419,B=0.0999,A=1))
   n19 = GetFocusedLevelSequence!(execute=@n16.then_1) -> (then, ReturnValue:Level Sequence Object Reference)
   > n20 = GetControlRigs!(execute, LevelSequence:Level Sequence Object Reference=@n19) -> (then, ReturnValue:Array of Control Rig Sequencer Binding Proxy Structures)
   > n21 = For Each Loop(Array:Array of Control Rig Sequencer Binding Proxy Structures=@n20) -> (Array Element:Control Rig Sequencer Binding Proxy Structure)
   n22 = FindOrCreateControlRigTrackFromReference!(
    execute=@n21.Completed
    World:World Object Reference
    LevelSequence:Level Sequence Object Reference=@n19
    InRef:"Control Rig Asset Strong Reference Structure (by ref)"=(BlueprintRigClass="/Script/ControlRig.ControlRigBlueprintGeneratedClass'/AnimatorKit/UtilityRigs/CRU_AddLocator.CRU_AddLocator_C'",ControlRigAsset=None)
    InBinding:"Movie Scene Binding Proxy Structure (by ref)"=Get Proxy() -> (Proxy:Movie Scene Binding Proxy Structure)
    bIsLayeredControlRig:bool=false
   ) -> (then, ReturnValue:Movie Scene Track Object Reference)
   > n23 = Cast To MovieSceneControlRigParameterTrack!(execute, Object=@n22) -> (then, CastFailed, AsMovie Scene Control Rig Parameter Track:Movie Scene Control Rig Parameter Track Object Reference) {is=title K2Node_DynamicCast}
   > Add!(
    execute
    TargetMap:Map of Names to Strings=Get STR_Recreate() -> (STR_Recreate:Map of Names to Strings)
    Key:"Name (by ref)"=Color
    Value:"String (by ref)"="To String (LinearColor)"!(InColor:Linear Color Structure=Get!() -> (ForegroundColor_SpecifiedColor:Linear Color Structure, ForegroundColor_ColorUseRule:ESlateColorStylingMode Enum)) -> (ReturnValue:String)
   ) -> (then) {is=title K2Node_CallFunction}
   > RecreateControlRigWithNewSettings!(
    execute
    InTrack:Movie Scene Control Rig Parameter Track Object Reference=@n23
    InSettingsToOverride:Map of Names to Strings=Get STR_Recreate()
   ) -> (then, ReturnValue:bool)
   Branch(execute=@n21.LoopBody)
   n24 = Break Control Rig Sequencer Binding Proxy!(ControlRigSequencerBindingProxy:"Control Rig Sequencer Binding Proxy Structure (by ref)"=@n21) -> (
    Proxy:Movie Scene Binding Proxy Structure
    ControlRig:Control Rig Object Reference
    Track:Movie Scene Control Rig Parameter Track Object Reference
   ) {is=title K2Node_BreakStruct}
   > Set Proxy(Proxy:Movie Scene Binding Proxy Structure=@n24) -> (Output_Get:Movie Scene Binding Proxy Structure)
   n25 = "On Clicked (BTN_Cancel)"() {is=title K2Node_ComponentBoundEvent}
   Event Construct()
   > n26 = FindUtilityWidgetFromBlueprint(self=Get EditorUtilitySubsystem(), InBlueprint=&EpicAnimTools)
   > n27 = Cast To EpicAnimTools(Object=@n26)
   > n28 = Set Locator Constrain(Locator Constrain=Get Locator_Constrain!(self:Epic Anim Tools Object Reference=@n27) -> (Locator_Constrain:bool))
   > Set Locator Loc CR(Locator Loc CR:Control Rig Object Reference=Get Locator_Loc_CR(self=@n27)) -> (Output_Get:Control Rig Object Reference)
   > n29 = Branch(Condition=@n28)
   > "SetText (Text)"(self=Get TXT_ConstraintStatus(), InText="LOCTEXT(\"ON\")")
   > SetForegroundColor(InForegroundColor_SpecifiedColor=(R=0.7157,G=0.0648,B=0.3916,A=1))
   "SetText (Text)"(execute=@n29.else, self=Get TXT_ConstraintStatus(), InText="LOCTEXT(\"OFF\")")
   > SetForegroundColor(InForegroundColor_SpecifiedColor=(R=0.1878,G=0.3419,B=0.0999,A=1))
   n30 = Execute Python Script!(
    execute
    PythonScript="import unreal\r\nLSEBL = unreal.LevelSequenceEditorBlueprintLibrary\r\nls = LSEBL.get_focused_level_sequence()\r\nif not ls:\r\n    unreal.log_error(\"Offset Snap A: No level sequence is open.\")\r\n    raise SystemExit\r\ndef get_selected_key_range():\r\n    all_frames = []\r\n    ses = unreal.get_editor_subsystem(unreal.LevelSequenceEditorSubsystem)\r\n    curve_editor = ses.get_curve_editor()\r\n    if curve_editor.is_curve_editor_open():\r\n        for ch_proxy in curve_editor.get_channels_with_selected_keys():\r\n            indices = curve_editor.get_selected_keys(ch_proxy)\r\n            if not indices:\r\n                continue\r\n            channel_obj = ch_proxy.section.get_channel(ch_proxy.channel_name)\r\n            if not channel_obj:\r\n                continue\r\n            for k in channel_obj.get_keys_by_index(indices):\r\n                all_frames.append(k.get_time().frame_number.value)\r\n    if not all_frames:\r\n        for ch_proxy in LSEBL.get_channels_with_selected_keys():\r\n            indices = LSEBL.get_selected_keys(ch_proxy)\r\n            if not indices:\r\n                continue\r\n            channel_obj = ch_proxy.section.get_channel(ch_proxy.channel_name)\r\n            if not channel_obj:\r\n                continue\r\n            for k in channel_obj.get_keys_by_index(indices):\r\n                all_frames.append(k.get_time().frame_number.value)\r\n    return (min(all_frames), max(all_frames)) if all_frames else None\r\n\r\nselected_range = get_selected_key_range()\r\nif selected_range:\r\n    start_val, end_val = selected_range\r\nelse:\r\n    start_val = ls.get_playback_start()\r\n    end_val   = ls.get_playback_end() - 1"
   ) -> (then, ReturnValue, start_val:int, end_val:int) {outputs [start_val,end_val]}
   > Set range_start(range_start:int=@n30.start_val) -> (Output_Get:int=0)
   > Set range_end(range_end:int=@n30.end_val) -> (Output_Get:int=0)
   > Set Current Time!(
    execute
    PlaybackParams_Frame_FrameNumber:Frame Number Structure=Make Frame Number!(Value:int=Get range_start() -> (range_start:int=0)) -> (FrameNumber:Frame Number Structure) {is=title K2Node_MakeStruct}
    PlaybackParams_Frame_SubFrame:float=0
    PlaybackParams_Time:float=0
    PlaybackParams_MarkedFrame:String
    PlaybackParams_Timecode:Timecode Structure=(Hours=0,Minutes=0,Seconds=0,Frames=0,Subframe=0,bDropFrameFormat=False)
    PlaybackParams_PositionType:EMovieScenePositionType Enum=Frame
    PlaybackParams_UpdateMethod:EUpdatePositionMethod Enum=Play
    PlaybackParams_bHasJumped:bool=False
    TimeUnit:EMovieSceneTimeUnit Enum=DisplayRate
   ) -> (then) {is=title K2Node_CallFunction}
   > n31 = Branch(Condition="Equal Exactly (String)"(A=Get axis(), B=X))
   > n32 = MakeControlValueFromTransform!(
    execute
    InValue_Location_X:double=String To Float(InString=Get distance())
    InValue_Location_Y:double=0
    InValue_Location_Z:double=0
    InValue_Rotation:Rotator="0, 0, 0"
    InValue_Scale:Vector="1,1,1"
   ) -> (then, ReturnValue:Rig Control Value Structure)
   > n33 = GetTransformFromControlValue!(execute, InValue:Rig Control Value Structure=@n32) -> (then, ReturnValue:Transform)
   > n34 = Set offsetXform(offsetXform:Transform=@n33) -> (Output_Get:Transform)
   n35 = GetCurrentLevelSequence!(execute=[@n34.then, @n42.then, @n39.then]) -> (then, ReturnValue:Level Sequence Object Reference)
   > SetLocalControlRigTransform!(
    execute
    LevelSequence:Level Sequence Object Reference=@n35
    ControlRig:Control Rig Object Reference=Get Out Loc CR()
    ControlName:Name=Get Out Loc Ctrl Name() -> (Out Loc Ctrl Name=None)
    Frame:Frame Number Structure=Make Frame Number(Value=Get range_start() -> (range_start=0))
    Value:Transform=Get offsetXform() -> (offsetXform:Transform)
    TimeUnit:EMovieSceneTimeUnit Enum=DisplayRate
    bSetKey:bool=true
   ) -> (then)
   n36 = Branch(execute=@n31.else, Condition="Equal Exactly (String)"(A=Get axis(), B=Y))
   > n37 = MakeControlValueFromTransform(
    InValue_Location_X=0
    InValue_Location_Y=String To Float(InString=Get distance())
    InValue_Location_Z=0
    InValue_Rotation="0, 0, 0"
    InValue_Scale="1,1,1"
   )
   > n38 = GetTransformFromControlValue(InValue=@n37)
   > n39 = Set offsetXform(offsetXform=@n38)
   n40 = Branch(execute=@n36.else, Condition="Equal Exactly (String)"(A=Get axis(), B=Z))
   > MakeControlValueFromTransform(
    InValue_Location_X=0
    InValue_Location_Y=0
    InValue_Location_Z=String To Float(InString=Get distance())
    InValue_Rotation="0, 0, 0"
    InValue_Scale="1,1,1"
   )
   > n41 = GetTransformFromControlValue()
   > n42 = Set offsetXform(offsetXform=@n41)
   PrintString(execute=@n40.else, InString=Something happened. i shouldn't have failed here...)
   n43 = Execute Python Script(PythonScript="import unreal\r\nLSEBL = unreal.LevelSequenceEditorBlueprintLibrary\r\nls = LSEBL.get_focused_level_sequence()\r\nif not ls:\r\n    unreal.log_error(\"Offset Snap A: No level sequence is open.\")\r\n    raise SystemExit\r\ndef get_selected_key_range():\r\n    all_frames = []\r\n    ses = unreal.get_editor_subsystem(unreal.LevelSequenceEditorSubsystem)\r\n    curve_editor = ses.get_curve_editor()\r\n    if curve_editor.is_curve_editor_open():\r\n        for ch_proxy in curve_editor.get_channels_with_selected_keys():\r\n            indices = curve_editor.get_selected_keys(ch_proxy)\r\n            if not indices:\r\n                continue\r\n            channel_obj = ch_proxy.section.get_channel(ch_proxy.channel_name)\r\n            if not channel_obj:\r\n                continue\r\n            for k in channel_obj.get_keys_by_index(indices):\r\n                all_frames.append(k.get_time().frame_number.value)\r\n    if not all_frames:\r\n        for ch_proxy in LSEBL.get_channels_with_selected_keys():\r\n            indices = LSEBL.get_selected_keys(ch_proxy)\r\n            if not indices:\r\n                continue\r\n            channel_obj = ch_proxy.section.get_channel(ch_proxy.channel_name)\r\n            if not channel_obj:\r\n                continue\r\n            for k in channel_obj.get_keys_by_index(indices):\r\n                all_frames.append(k.get_time().frame_number.value)\r\n    return (min(all_frames), max(all_frames)) if all_frames else None\r\n\r\nselected_range = get_selected_key_range()\r\nif selected_range:\r\n    start_val, end_val = selected_range\r\nelse:\r\n    start_val = ls.get_playback_start()\r\n    end_val   = ls.get_playback_end() - 1") {outputs [start_val,end_val]}
   > Set range_start(range_start=@n43.start_val) -> (Output_Get=0)
   > Set range_end(range_end=@n43.end_val) -> (Output_Get=0)
   > Set Current Time(
    PlaybackParams_Frame_FrameNumber=Make Frame Number(Value=Get range_start() -> (range_start=0))
    PlaybackParams_Frame_SubFrame=0
    PlaybackParams_Time=0
    PlaybackParams_Timecode=(Hours=0,Minutes=0,Seconds=0,Frames=0,Subframe=0,bDropFrameFormat=False)
    PlaybackParams_PositionType=Frame
    PlaybackParams_UpdateMethod=Play
    PlaybackParams_bHasJumped=False
    TimeUnit=DisplayRate
   )
   > RefreshCurrentLevelSequence()
   BakeDelete_ParentConstraint!(execute, self, in_ctrl_name:Name=Get Out Loc Ctrl Name() -> (Out Loc Ctrl Name=None)) -> (then) {is=title K2Node_CallFunction}
   Event PreConstruct()
   Event Tick()
   "Equal (Object)"(A=Get In CR(), B=@n24.ControlRig)
  ]}
]
variables [
 axis {type=string, default=""}
 distance {type=string, default=""}
 In Control {type=string, default=None}
 In CR {type=ControlRig, default=null}
 JF_TEST {type=string, default=""}
 Local_Loc_Color {type=string, default=(R=0.187821,G=0.341915,B=0.099899,A=1.000000)}
 Locator Constrain {type=boolean}
 Locator Loc CR {type=ControlRig, default=null}
 mode {type=string, default=""}
 offsetXform {type=Transform, default {location=<0 0 0>, rotation=<rot 0 0 0>, scale=<1 1 1>}}
 Out Loc CR {type=ControlRig, default=null}
 Out Loc Ctrl Name {type=string, default=None}
 Proxy {type=MovieSceneBindingProxy, default {bindingId=#1, sequence=null}}
 range_end {type=integer}
 range_start {type=integer}
 STR_Recreate {type="map<string>", default {}}
]
widgets {
 info {rootWidgetClass=&SizeBox, widgetCount=20}
 widgets [
  SizeBox_0 {
   widgetClassPath=&SizeBox
   widget=SizeBox_0
   properties {
    bOverride_HeightOverride=true
    bOverride_MaxDesiredHeight=true
    bOverride_MaxDesiredWidth=true
    bOverride_MinDesiredHeight=true
    bOverride_MinDesiredWidth=true
    bOverride_WidthOverride=true
    heightOverride=100
    maxDesiredHeight=100
    maxDesiredWidth=300
    minDesiredHeight=100
    minDesiredWidth=300
    widthOverride=300
   }
  }
  ScaleBox_1 {widgetClassPath=&ScaleBox, parent=SizeBox_0, widget=ScaleBox_1, slot {class=SizeBoxSlot}}
  VerticalBox_48 {widgetClassPath=&VerticalBox, parent=ScaleBox_1, widget=VerticalBox_48, slot {class=ScaleBoxSlot}}
  ScaleBox_0 {
   widgetClassPath=&ScaleBox
   parent=VerticalBox_48
   widget=ScaleBox_0
   slot {class=VerticalBoxSlot, properties {horizontalAlignment=HAlign_Center, padding=<ltrb 0 5 0 0>}}
  }
  TextBlock_115 {
   widgetClassPath=&TextBlock
   parent=ScaleBox_0
   widget=TextBlock_115
   properties {
    text=Bake to Locator With Offset
    colorAndOpacity {colorUseRule=UseColor_Foreground, specifiedColor=<rgba 0.7157 0.0648 0.3916 1>}
    font {
     bForceMonospaced=false
     bMaterialIsStencil=false
     fontMaterial=null
     fontObject=&Roboto
     letterSpacing=0
     monospacedWidth=1
     size=22.5
     skewAmount=0
     typefaceFontName=Bold
     outlineSettings {
      bApplyOutlineToDropShadows=false
      bMiteredCorners=false
      bSeparateFillAlpha=false
      outlineMaterial=null
      outlineSize=0
      outlineColor=<rgba 0 0 0 1>
     }
    }
   }
   slot {class=ScaleBoxSlot}
  }
  HorizontalBox_0 {
   widgetClassPath=&HorizontalBox
   parent=VerticalBox_48
   widget=HorizontalBox_0
   slot {class=VerticalBoxSlot, properties {verticalAlignment=VAlign_Top}}
  }
  VerticalBox_0 {
   widgetClassPath=&VerticalBox
   parent=HorizontalBox_0
   widget=VerticalBox_0
   slot {class=HorizontalBoxSlot, properties {verticalAlignment=VAlign_Center, size {sizeRule=Fill, value=1}}}
  }
  TextBlock {
   widgetClassPath=&TextBlock
   parent=VerticalBox_0
   widget=TextBlock
   properties {
    text="Constraint: "
    colorAndOpacity {colorUseRule=UseColor_Foreground, specifiedColor=<rgba 0.7157 0.0648 0.3916 1>}
    font {
     bForceMonospaced=false
     bMaterialIsStencil=false
     fontMaterial=null
     fontObject=&Roboto
     letterSpacing=0
     monospacedWidth=1
     size=7.5
     skewAmount=0
     typefaceFontName=Regular
     outlineSettings {
      bApplyOutlineToDropShadows=false
      bMiteredCorners=false
      bSeparateFillAlpha=false
      outlineMaterial=null
      outlineSize=0
      outlineColor=<rgba 0 0 0 1>
     }
    }
   }
   slot {class=VerticalBoxSlot, properties {horizontalAlignment=HAlign_Right, size {sizeRule=Fill, value=1}}}
  }
  EditorUtilityButton_70 {
   widgetClassPath=&EditorUtilityButton
   parent=HorizontalBox_0
   widget=EditorUtilityButton_70
   slot {class=HorizontalBoxSlot, properties {horizontalAlignment=HAlign_Left, verticalAlignment=VAlign_Center, size {sizeRule=Fill, value=1}}}
  }
  TXT_ConstraintStatus {
   widgetClassPath=&TextBlock
   bIsVariable=true
   parent=EditorUtilityButton_70
   widget=TXT_ConstraintStatus
   properties {
    text=debug
    colorAndOpacity {colorUseRule=UseColor_Foreground, specifiedColor=<rgba 0.7157 0.0648 0.3916 1>}
    font {
     bForceMonospaced=false
     bMaterialIsStencil=false
     fontMaterial=null
     fontObject=&Roboto
     letterSpacing=0
     monospacedWidth=1
     size=7.5
     skewAmount=0
     typefaceFontName=Regular
     outlineSettings {
      bApplyOutlineToDropShadows=false
      bMiteredCorners=false
      bSeparateFillAlpha=false
      outlineMaterial=null
      outlineSize=0
      outlineColor=<rgba 0 0 0 1>
     }
    }
   }
   slot {class=ButtonSlot, properties {verticalAlignment=VAlign_Fill, padding=<ltrb 0 0 0 0>}}
  }
  TextBlock_2 {
   widgetClassPath=&TextBlock
   parent=VerticalBox_48
   widget=TextBlock_2
   properties {
    text=Move the Locator to your desired offset and hit Run
    colorAndOpacity {colorUseRule=UseColor_Foreground, specifiedColor=<rgba 0.7157 0.0648 0.3916 1>}
    font {
     bForceMonospaced=false
     bMaterialIsStencil=false
     fontMaterial=null
     fontObject=&Roboto
     letterSpacing=0
     monospacedWidth=1
     size=9
     skewAmount=0
     typefaceFontName=Italic
     outlineSettings {
      bApplyOutlineToDropShadows=false
      bMiteredCorners=false
      bSeparateFillAlpha=false
      outlineMaterial=null
      outlineSize=0
      outlineColor=<rgba 0 0 0 1>
     }
    }
   }
   slot {class=VerticalBoxSlot, properties {horizontalAlignment=HAlign_Center, padding=<ltrb 0 5 0 0>}}
  }
  Border_100 {
   widgetClassPath=&Border
   parent=VerticalBox_48
   widget=Border_100
   properties {
    background {
     bIsDynamicallyLoaded=false
     drawAs=Image
     imageType=NoImage
     mirroring=NoMirror
     resourceName=None
     resourceObject=null
     tiling=NoTile
     imageSize=<32 32>
     margin=<ltrb 0 0 0 0>
     outlineSettings {
      bUseBrushTransparency=false
      roundingType=HalfHeightRadius
      width=0
      color {colorUseRule=UseColor_Specified, specifiedColor=<rgba 0 0 0 0>}
      cornerRadii=<0 0 0 0>
     }
     tintColor {colorUseRule=UseColor_Foreground, specifiedColor=<rgba 1 1 1 1>}
     uVRegion {bIsValid=false, max=<0 0>, min=<0 0>}
    }
    padding=<ltrb 4 0.5 4 0.5>
   }
   slot {class=VerticalBoxSlot, properties {verticalAlignment=VAlign_Bottom}}
  }
  VerticalBox_200 {
   widgetClassPath=&VerticalBox
   parent=VerticalBox_48
   widget=VerticalBox_200
   slot {class=VerticalBoxSlot, properties {verticalAlignment=VAlign_Top}}
  }
  HBox_Run {
   widgetClassPath=&HorizontalBox
   parent=VerticalBox_200
   widget=HBox_Run
   slot {class=VerticalBoxSlot, properties {verticalAlignment=VAlign_Top, padding=<ltrb 0 10 0 0>, size {sizeRule=Fill, value=1}}}
  }
  BTN_Run {
   widgetClassPath=&EditorUtilityButton
   parent=HBox_Run
   widget=BTN_Run
   properties {
    colorAndOpacity=<rgba 0.0284 0.0284 0.0284 1>
    widgetStyle {
     clickedSlateSound {resourceObject=null}
     disabled {
      bIsDynamicallyLoaded=false
      drawAs=RoundedBox
      imageType=NoImage
      mirroring=NoMirror
      resourceName=None
      resourceObject=null
      tiling=NoTile
      imageSize=<0 0>
      margin=<ltrb 0 0 0 0>
      outlineSettings {
       bUseBrushTransparency=false
       roundingType=FixedRadius
       width=1
       color {colorUseRule=UseColor_ColorTable, specifiedColor=<rgba 0 0 0 0>}
       cornerRadii=<4 4 4 4>
      }
      tintColor {colorUseRule=UseColor_Specified, specifiedColor=<rgba 0.0395 0.0395 0.0395 1>}
      uVRegion {bIsValid=false, max=<0 0>, min=<0 0>}
     }
     disabledForeground {colorUseRule=UseColor_ColorTable, specifiedColor=<rgba 0 0 0 0>}
     hovered {
      bIsDynamicallyLoaded=false
      drawAs=RoundedBox
      imageType=NoImage
      mirroring=NoMirror
      resourceName=None
      resourceObject=null
      tiling=NoTile
      imageSize=<32 32>
      margin=<ltrb 0 0 0 0>
      outlineSettings {
       bUseBrushTransparency=false
       roundingType=FixedRadius
       width=1
       color {colorUseRule=UseColor_ColorTable, specifiedColor=<rgba 0 0 0 0>}
       cornerRadii=<4 4 4 4>
      }
      tintColor {colorUseRule=UseColor_Specified, specifiedColor=<rgba 0.0953 0.0953 0.0953 1>}
      uVRegion {bIsValid=false, max=<0 0>, min=<0 0>}
     }
     hoveredForeground {colorUseRule=UseColor_ColorTable, specifiedColor=<rgba 0 0 0 0>}
     hoveredSlateSound {resourceObject=null}
     normal {
      bIsDynamicallyLoaded=false
      drawAs=RoundedBox
      imageType=NoImage
      mirroring=NoMirror
      resourceName=None
      resourceObject=null
      tiling=NoTile
      imageSize=<32 32>
      margin=<ltrb 0 0 0 0>
      outlineSettings {
       bUseBrushTransparency=false
       roundingType=FixedRadius
       width=1
       color {colorUseRule=UseColor_ColorTable, specifiedColor=<rgba 0 0 0 0>}
       cornerRadii=<4 4 4 4>
      }
      tintColor {colorUseRule=UseColor_Foreground, specifiedColor=<rgba 0.7005 0.7005 0.7005 1>}
      uVRegion {bIsValid=false, max=<0 0>, min=<0 0>}
     }
     normalForeground {colorUseRule=UseColor_ColorTable, specifiedColor=<rgba 0 0 0 0>}
     normalPadding=<ltrb 12 1.5 12 1.5>
     pressed {
      bIsDynamicallyLoaded=false
      drawAs=RoundedBox
      imageType=NoImage
      mirroring=NoMirror
      resourceName=None
      resourceObject=null
      tiling=NoTile
      imageSize=<32 32>
      margin=<ltrb 0 0 0 0>
      outlineSettings {
       bUseBrushTransparency=false
       roundingType=FixedRadius
       width=1
       color {colorUseRule=UseColor_ColorTable, specifiedColor=<rgba 0 0 0 0>}
       cornerRadii=<4 4 4 4>
      }
      tintColor {colorUseRule=UseColor_Specified, specifiedColor=<rgba 0.0284 0.0284 0.0284 1>}
      uVRegion {bIsValid=false, max=<0 0>, min=<0 0>}
     }
     pressedForeground {colorUseRule=UseColor_ColorTable, specifiedColor=<rgba 0 0 0 0>}
     pressedPadding=<ltrb 12 2.5 12 0.5>
     pressedSlateSound {resourceObject=null}
    }
   }
   slot {class=HorizontalBoxSlot, properties {padding=<ltrb 0 0 0 10>, size {sizeRule=Fill, value=1}}}
  }
  HorizontalBox_52 {
   widgetClassPath=&HorizontalBox
   parent=BTN_Run
   widget=HorizontalBox_52
   slot {class=ButtonSlot, properties {verticalAlignment=VAlign_Fill, padding=<ltrb 0 4 0 4>}}
  }
  TextBlock_59 {
   widgetClassPath=&TextBlock
   parent=HorizontalBox_52
   widget=TextBlock_59
   properties {
    text=RUN
    font {
     bForceMonospaced=false
     bMaterialIsStencil=false
     fontMaterial=null
     fontObject=&Roboto
     letterSpacing=0
     monospacedWidth=1
     size=12
     skewAmount=0
     typefaceFontName=Bold
     outlineSettings {
      bApplyOutlineToDropShadows=false
      bMiteredCorners=false
      bSeparateFillAlpha=false
      outlineMaterial=null
      outlineSize=0
      outlineColor=<rgba 0 0 0 1>
     }
    }
   }
   slot {class=HorizontalBoxSlot, properties {horizontalAlignment=HAlign_Center}}
  }
  BTN_Cancel {
   widgetClassPath=&EditorUtilityButton
   parent=HBox_Run
   widget=BTN_Cancel
   properties {
    colorAndOpacity=<rgba 0.0284 0.0284 0.0284 1>
    widgetStyle {
     clickedSlateSound {resourceObject=null}
     disabled {
      bIsDynamicallyLoaded=false
      drawAs=RoundedBox
      imageType=NoImage
      mirroring=NoMirror
      resourceName=None
      resourceObject=null
      tiling=NoTile
      imageSize=<0 0>
      margin=<ltrb 0 0 0 0>
      outlineSettings {
       bUseBrushTransparency=false
       roundingType=FixedRadius
       width=1
       color {colorUseRule=UseColor_ColorTable, specifiedColor=<rgba 0 0 0 0>}
       cornerRadii=<4 4 4 4>
      }
      tintColor {colorUseRule=UseColor_Specified, specifiedColor=<rgba 0.0395 0.0395 0.0395 1>}
      uVRegion {bIsValid=false, max=<0 0>, min=<0 0>}
     }
     disabledForeground {colorUseRule=UseColor_ColorTable, specifiedColor=<rgba 0 0 0 0>}
     hovered {
      bIsDynamicallyLoaded=false
      drawAs=RoundedBox
      imageType=NoImage
      mirroring=NoMirror
      resourceName=None
      resourceObject=null
      tiling=NoTile
      imageSize=<32 32>
      margin=<ltrb 0 0 0 0>
      outlineSettings {
       bUseBrushTransparency=false
       roundingType=FixedRadius
       width=1
       color {colorUseRule=UseColor_ColorTable, specifiedColor=<rgba 0 0 0 0>}
       cornerRadii=<4 4 4 4>
      }
      tintColor {colorUseRule=UseColor_Specified, specifiedColor=<rgba 0.0953 0.0953 0.0953 1>}
      uVRegion {bIsValid=false, max=<0 0>, min=<0 0>}
     }
     hoveredForeground {colorUseRule=UseColor_ColorTable, specifiedColor=<rgba 0 0 0 0>}
     hoveredSlateSound {resourceObject=null}
     normal {
      bIsDynamicallyLoaded=false
      drawAs=RoundedBox
      imageType=NoImage
      mirroring=NoMirror
      resourceName=None
      resourceObject=null
      tiling=NoTile
      imageSize=<32 32>
      margin=<ltrb 0 0 0 0>
      outlineSettings {
       bUseBrushTransparency=false
       roundingType=FixedRadius
       width=1
       color {colorUseRule=UseColor_ColorTable, specifiedColor=<rgba 0 0 0 0>}
       cornerRadii=<4 4 4 4>
      }
      tintColor {colorUseRule=UseColor_Foreground, specifiedColor=<rgba 0.7005 0.7005 0.7005 1>}
      uVRegion {bIsValid=false, max=<0 0>, min=<0 0>}
     }
     normalForeground {colorUseRule=UseColor_ColorTable, specifiedColor=<rgba 0 0 0 0>}
     normalPadding=<ltrb 12 1.5 12 1.5>
     pressed {
      bIsDynamicallyLoaded=false
      drawAs=RoundedBox
      imageType=NoImage
      mirroring=NoMirror
      resourceName=None
      resourceObject=null
      tiling=NoTile
      imageSize=<32 32>
      margin=<ltrb 0 0 0 0>
      outlineSettings {
       bUseBrushTransparency=false
       roundingType=FixedRadius
       width=1
       color {colorUseRule=UseColor_ColorTable, specifiedColor=<rgba 0 0 0 0>}
       cornerRadii=<4 4 4 4>
      }
      tintColor {colorUseRule=UseColor_Specified, specifiedColor=<rgba 0.0284 0.0284 0.0284 1>}
      uVRegion {bIsValid=false, max=<0 0>, min=<0 0>}
     }
     pressedForeground {colorUseRule=UseColor_ColorTable, specifiedColor=<rgba 0 0 0 0>}
     pressedPadding=<ltrb 12 2.5 12 0.5>
     pressedSlateSound {resourceObject=null}
    }
   }
   slot {class=HorizontalBoxSlot, properties {padding=<ltrb 0 0 0 10>, size {sizeRule=Fill, value=1}}}
  }
  HorizontalBox_2 {
   widgetClassPath=&HorizontalBox
   parent=BTN_Cancel
   widget=HorizontalBox_2
   slot {class=ButtonSlot, properties {verticalAlignment=VAlign_Fill, padding=<ltrb 0 4 0 4>}}
  }
  TextBlock_1 {
   widgetClassPath=&TextBlock
   parent=HorizontalBox_2
   widget=TextBlock_1
   properties {
    text=Cancel
    font {
     bForceMonospaced=false
     bMaterialIsStencil=false
     fontMaterial=null
     fontObject=&Roboto
     letterSpacing=0
     monospacedWidth=1
     size=12
     skewAmount=0
     typefaceFontName=Bold
     outlineSettings {
      bApplyOutlineToDropShadows=false
      bMiteredCorners=false
      bSeparateFillAlpha=false
      outlineMaterial=null
      outlineSize=0
      outlineColor=<rgba 0 0 0 1>
     }
    }
   }
   slot {class=HorizontalBoxSlot, properties {horizontalAlignment=HAlign_Center}}
  }
 ]
}
